---
id: technique-001
kind: technique
---

# Destructible stud props: poll `IsValid`, then `SpawnProp` the cached transform

**Sources:** action_adventure_sample, brick_extraction_sample
**Snippets:**
- `../snippets/verse/BRICK_destructible_prop_manager.verse`
- `../snippets/verse/BRICK_destructible_prop_definition.verse`
**Tutorial:** island Design-your-Island / demo boards. No dedicated chapter.

Scene Graph despawn stays on spawn-remove-prefabs.md. VFX Trigger→Cinematic stays on trigger-cinematic-quake.md.

## Problem

Breakable world props should spill studs and come back, without a Creative Damage Volume as the reward bus and without a destruction event API on `creative_prop`.

## Pattern (shipped)

`BRICK_destructible_prop_definition` caches `Prop.GetTransform()` and a `creative_prop_asset`. Manager `loop { Sleep(UpdateRateInSeconds) }` (default **0.5**) calls `RewardIfDestroyed`: if the prop is gone (`not Prop?.IsValid[]`) spawn studs at `Translation + Z 50` then `SpawnProp(PropAsset, PropTransform)`. `race { Sleep(RespawnTime); OnRespawned.Await() }` cancels a pending respawn if `Respawn()` already ran. `CanOnlyRespawnManually` + `Trigger_ManualRespawn` ignore AutoRespawn.

**4** placed managers on Action Adventure (Warrior / Blaster / Boss / demo). Billboard: “Uses prop and a prop assets to spawn studs when that props are destroyed.” Extraction places **1** manager (`BRICKModule StudRewards-BRICK destructible manager`) plus **32** `BP_Sub_ArcheryTarget` in `Map/Destructible Props`. Docs filename typo `BRICK_destuctible_prop_manager` vs shipped `BRICK_destructible_prop_manager`.

## Don't

- Don't wait for a `DestroyedEvent` on `creative_prop` — this island **polls**.
- Don't use Damage Volume as the stud grant (27 volumes are lasers / pits, `damage=10` Elimination family on Obstacle, not this reward).
