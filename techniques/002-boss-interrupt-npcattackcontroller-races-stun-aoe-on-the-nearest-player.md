---
id: technique-002
kind: technique
---

# Boss interrupt: `npc_attack_controller` races stun/AOE on the nearest player

**Sources:** action_adventure_sample
**Snippets:**
- `../snippets/verse/npc_attack_controller.verse`
- `../snippets/verse/npc_attack.verse`
**Official tutorial (outline only):** BRICK Action Adventure Template — Create a Verse-based Attack System.

Wander `npc_behavior` stays on npc-behavior-navigatable-focus.md. Quest-giver stasis stays on persistable-quest-line-verse-dialog.md. Arcade tracker arenas stay on tracker-gated-encounter.md.

## Problem

A Guard NPC should keep using its Character Definition combat, but **periodically interrupt** to stun the player or drop a following AOE explosion — without replacing the whole behavior script.

## Pattern (shipped)

`npc_attack_controller` is a `creative_device` with `@editable NPCSpawner`, optional `AwaitCinematicSequence`, min/max seconds, `[]npc_attack`. **2** placed (`Boss` + demo `Boss2`). `SpawnedEvent` starts `RunBehavior`; optional cine `StoppedEvent.Await()` first.

```verse
race:
    OnNPCEliminated.Await()
    loop:
        Sleep(GetRandomFloat(TimeBetweenAttacks_Min, TimeBetweenAttacks_Max))
        Attack.StartAttack(InNPCAgent, InNPCCharacter, GetNearestPlayerCharacter[InNPCCharacter])
```

**Stun:** HUD Message Show, `PutInStasis` (no turning), `SpawnProp` indicator, `Sleep(StunDuration)` (default **2**), Dispose, `ReleaseFromStasis`.

**Ranged AOE:** stasis the NPC (AllowTurning), VFX follows the player 11× `MoveTo` 0.2s, `ExplosiveDevice.TeleportTo` + `Sleep(0.5)` + `Explode`. Live **2** Explosive devices (`_Attack`, `_BossLevel`). Authored `BP_Prop_NPCAttack_AOE` / `_Stun` + Niagara.

Nearest player is `DistanceSquared` over `GetPlayspace().GetPlayers()`. Logger still says `boss_attack_controller`.

## Don't

- Don't subclass `npc_behavior` for the boss attacks — the Guard CD keeps fighting; this device **interrupts**.
- Don't emit docs-only HideThenBuild Assembly Verse as the boss.
- Don't copy Walker `Character Definition Type=Walker` — these are Guard / custom defs (`NPCDef_Boss`, `NPCDef_Enemy_*`, `NPCDef_Enemy_Blaster`, `NPCDef_Enemy_Warrior`).
