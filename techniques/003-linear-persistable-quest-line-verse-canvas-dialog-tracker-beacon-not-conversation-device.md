---
id: technique-003
kind: technique
---

# Linear persistable quest line: Verse canvas dialog + Tracker/Beacon, not Conversation Device

**Sources:** action_adventure_sample
**Snippets:**
- `../snippets/verse/BRICK_quest_giver_device.verse`
- `../snippets/verse/BRICK_quest.verse`
- `../snippets/verse/BRICK_quest_dialog.verse`
- `../snippets/verse/BRICK_persistent_quest_log.verse`
- `../snippets/verse/npc_behavior_quest_giver.verse`
**Official tutorial (outline only):** BRICK Action Adventure Template. Tutorial *Verse* is untrusted if it conflicts; **`snippets/` win**.

Conversation Device trees stay on conversation-device-verse-quest.md. Speedway persistable *stats* stay on persistable-stats-and-sorted-leaderboard.md. Music Concert visit-beacons stay on beacon-volume-popup-trail.md. Clickable `button`+`button_slot` stays on verse-clickable-button.md — this dialog uses `button_loud` / `button_quiet`.

## Problem

A class-path adventure needs a **linear** quest line per giver: talk, accept, complete objectives, return, reward studs — with progress that survives sessions — without Conversation Device banks.

## Pattern (shipped)

**World:** Button `InteractedWithEvent` → `BRICK_quest_giver_device`. Live **2** placed: `BRICK quest giver device_Tutorial` (`questGiverName="PlayerCharacter"`, `autoProgressDialog=true`) and `_demo` (`"Demo Quest Giver"`). NPC Spawner uses `npc_behavior_quest_giver` (stasis in place, AllowTurning).

**Persistable log** (`weak_map(player, BRICK_persistent_quest_log)`):

```verse
BRICK_persistent_quest_log<public> := class<persistable><final>:
    CompletedQuests : []string = array{}
    ActiveQuests : []string = array{}
```

Quest `Name` **must be unique** (tooltip). `GetQuestLineProgression` returns the **first** quest that `CanBeStarted`, else the active one (`CanBeCompleted` vs `IsActive`). That is **linear**, not any-order. `BRICK_quest_repeatable` is the only way a completed name can start again. `Trigger_ResetProgress` wipes both arrays.

**Dialog** is Verse `canvas` + `AddWidget` (`ZOrder := 3`, `ui_input_mode.All`). `OnDialogCompleted.Await()` then Start / Complete / Abort. Completing with `AutoProgressDialog` re-opens the giver for the next row.

**Devices per quest:** `tracker_device` (objective) + `[]beacon_device` (`AddToShowList` while active) + return beacon on objective complete + `BRICK_stud_spawner` reward at the pawn. Tracker titles are empty in the editor — `SetTitleText` / `SetDescriptionText` run in `Setup` / `StartQuest`. Live `assignOnGameStart=false`, `statToTrack=None`.

## Don't

- Don't train Conversation Device `InitiateConversation` / `customWidget` from this island — **0** Conversation / Persona.
- Don't emit docs “progress through in any order.” The giver walks the **array in order**.
- Don't emit docs `Stat to Track = Events`. Live is **None**; Increment bindings (Assemble `On Assembled`, Conditional Button `On Activated`) are the bus.
- Don't feed BrickAssembled Brick Editor meshes to these Assembly devices — BrickAssembly-atom-static-mesh.md (static mesh ≠ geometry collection).
- Don't wrap this chrome in `button`+`button_slot` — shipped is `text_button_base` loud/quiet.
- Don't expect `GetDeviceProperties` on `quests[]` to return names — subobject `refPath` trap.
