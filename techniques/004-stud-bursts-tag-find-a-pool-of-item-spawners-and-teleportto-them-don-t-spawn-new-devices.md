---
id: technique-004
kind: technique
---

# Stud bursts: tag-find a pool of Item Spawners and `TeleportTo` them, don't spawn new devices

**Sources:** action_adventure_sample, brick_extraction_sample
**Snippets:**
- `../snippets/verse/BRICK_stud_spawner_pool.verse`
- `../snippets/verse/BRICK_stud_spawner.verse`
**Official tutorial (outline only):** BRICK Action Adventure Template — Award BRICK Studs / Create an Object Pool.

Parkour sequential `SpawnItem` stays on sequential-collect-race.md. Obstacle any-order coins stay on collectible-tracker-endgame.md.

## Problem

Quests, NPC elims, destroyed props, and finished assemblies all need to dump **N** studs at a world location. Creating/destroying Item Spawners at runtime is expensive; one spawner only ejects **one** pickup.

## Pattern (shipped)

Place a pool of BRICK Item Spawners once (**129** live on Action Adventure; Extraction labels `Item Spawner_BRICKStud_Toss_*` / `_Gravity_*`, **40** BRICK spawners total). Tag them `BRICK_stud_itemspawner_toss_tag` or `_gravity_tag`. Place **one** `BRICK_stud_pool` device tagged `BRICK_stud_pool_tag` (singleton; `GetBRICKStudPool` finds it). Extraction labels: `BRICKModule StudRewards-BRICK stud pool` + `BRICKModule StudRewards-BRICK stud spawner device`.

`OnBegin`: `Sleep(0.5)`, `FindCreativeObjectsWithTag`, wrap each as `pool_object_pickup`. Burst:

```verse
ValidDevice.TeleportTo[InLocation + vector3{Z := -125.0}, IdentityRotation()]
Sleep(0.25)
ValidDevice.SpawnItem()
Sleep(0.25)
Release()  # TeleportTo origin, IsAvailable = true
```

`BRICK_stud_spawner` subclasses pick the event and the location: trigger, button, NPC `EliminatedEvent`, assembly `AssemblyFinishedEvent` / `DisassemblyFinishedEvent`, or quest `CompleteQuest` at the pawn. `EN_ItemSpawnerConfig.Toss | Gravity` must match the tag pool.

Live labels: `Item Spawner_BRICKStud_Toss_*` / `_Gravity_*`. Billboard_Demo_Studs_ObjectPool: “Keep an object pool of Item Spawner devices that are teleported to where the studs should spawn.”

## Don't

- Don't `SpawnProp` a new Item Spawner per stud.
- Don't emit tutorial `BRICK_stud_itemspawner_tag` as the only tag — shipped splits **toss** vs **gravity**.
- Don't emit `CP_Ingredient_GoldStud` as the live collectible on the ground pickups. World collectibles are `BRICKCollectibleObject=Coin` (see tagged-collectible-stud-granter.md). Pool item class is waived (spawn Item struct empty, same Home Builder trap).
- Don't copy Parkour's ordered `item_spawner_device` array.
