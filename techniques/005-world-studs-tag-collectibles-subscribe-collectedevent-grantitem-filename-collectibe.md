---
id: technique-005
kind: technique
---

# World studs: tag Collectibles, Subscribe `CollectedEvent`, GrantItem — filename `collectibe`

**Sources:** action_adventure_sample
**Snippets:** `../snippets/verse/BRICK_collectibe_stud_granter.verse` (shipped filename typo **collectibe**; class is `BRICK_collectible_stud_granter`)
**Official tutorial:** island page points at Collectible + pool; companion Collectible Devices lists **Gold/Blue/Purple/Silver Stud**. Live **Coin**.

Obstacle Course any-order Gold Coin → Tracker → End Game stays on collectible-tracker-endgame.md (**0** Item Granter there). Pool bursts stay on stud-itemspawner-teleport-pool.md.

## Problem

Dozens of world pickups should increment a HUD stud count without wiring each Collectible's Details graph to one Granter.

## Pattern (shipped)

Tag collectibles `BRICK_collectible_stud_tag`. One `BRICK_collectible_stud_granter` device:

```verse
for (FoundObject : Self.FindCreativeObjectsWithTag(BRICK_collectible_stud_tag{})):
    if (ValidCollectible := collectible_object_device[FoundObject]):
        ValidCollectible.CollectedEvent.Subscribe(OnCollectibleCollected)

OnCollectibleCollected(InAgent : agent): void=
    ItemGranter.GrantItem(InAgent)
```

Live: **86** `BRICK_Device_Collectible` (find `"Collectible"` also hits demo billboards). Sampled `BRICKCollectibleObject=Coin`, `collectibleObject=Coin`, `consumeIfCollectedBy=Self`, `score=1`, `displayScoreUpdateOnHUD=false`. Granter label `Item Granter_GoldForCollectibles`; Details bindings **[]**.

HUD count is **not** the Collectible score: `BRICK_stud_counter_device` polls `ConditionalButton.GetItemCount` every 0.15s and `HUDMessageDevice.SetText` / `Show`. Billboard: “Animates HUD elements that showcase how many studs the player has!”

## Don't

- Don't emit companion **Gold Stud** as this island's placed type — live **Coin**.
- Don't wire Collectible `On Collected` → Tracker Increment as the stud currency (that's the obby win). Here tags + Granter + Conditional Button count.
- Don't "fix" the filename; `BRICK_collectibe_stud_granter.verse` is shipped.
