---
id: technique-006
kind: technique
---

# Hide and freeze the player for a cinematic, then restore

**Sources:** animation_template
**Snippet:** `../snippets/verse/player_cinematic_mode_device.verse`
**Related:** stand_up chair-gate (different trigger); triad/prop-hunt flythrough (camera, not pawn hide).

## Problem

Play a Sequencer shot of the museum without the player standing in frame or walking through it. When the shot ends they must move again and be visible.

## Pattern

```verse
Cinematic_Sequence_Trigger_Switch.TurnedOffEvent.Subscribe(HandleCinematicStart)
Museum_Cinematic_Device.StoppedEvent.Subscribe(HandleCinematicEnd)

HandleCinematicStart(Agent:agent):void=
    FortCharacter := Agent.GetFortCharacter[]
    set PlayerRef = option{Agent}
    FortCharacter.PutInStasis(stasis_args{AllowTurning := false, AllowFalling := false, AllowEmotes := false})
    FortCharacter.Hide()

HandleCinematicEnd():void=
    FortCharacter := PlayerRef?.GetFortCharacter[]
    FortCharacter.Show()
    FortCharacter.ReleaseFromStasis()
```

Keep the instigator — `StoppedEvent` has no agent. Stasis args all false = no fidget, no turn, no fall-through.

This island’s **Switch** fires on **TurnedOff** to *start*. The official tutorial says **Button** devices — shipped is Switch. Don’t emit `button_device` for this island.

## Don't

- Don’t `Hide` without `ReleaseFromStasis` on stop — player stays stuck.
- Don’t use this as “how to author Control Rig” — that’s the mannequin assets, not this file.
- Don’t copy `using {/Fortnite.com/Characters }` (missing space) as a style lesson.
