---
id: technique-007
kind: technique
---

# Verse NPC wander: subclass `npc_behavior`, not a creative_device — Navigatable + Focus + `race`

**Sources:** characters_cameras_npc_template
**Tutorial:** none for this island (spreadsheet Docs=No). Verse file header links the generic Verse onboarding guide only.

## Problem

You need an NPC that walks a random offset from spawn, looks back at where it came from (walks backwards), and keeps doing that. A `creative_device` in the world cannot be the NPC. Guard/Creature spawners don't take this script.

## Pattern

Subclass **`npc_behavior`**. Assign it as **NPC Behavior Script Override** on an NPC Spawner / Character Definition (`CharDef_NPC_Verse` on this island).

```
OnBegin:
  Agent := GetAgent[]
  NPC  := Agent.GetFortCharacter[]
  Nav  := NPC.GetNavigatable[]
  Focus := NPC.GetFocusInterface[]
  Spawn := NPC.GetTransform().Translation

  loop:
    GoTo = Spawn + random XY in ±DistanceFromSpawnPtToMove
    Nav.NavigateTo(MakeNavigationTarget(GoTo), ?MovementType := movement_types.Walking)
    Sleep(MoveToWaitDuration) if Reached else log nav debug

    race:
      Nav.NavigateTo(MakeNavigationTarget(Spawn))   # this one is allowed to win
      Focus.MaintainFocus(GoTo)                     # never completes
      DrawDebugLookAt(...)                          # never completes
```

`race` is the lesson: only `NavigateTo` should complete; Focus + debug draw are infinite so the NPC looks at the old point **while** walking home.

Debug: `debug_draw{Channel := npc_behavior_usingverse_debug_draw}`. **Island Settings → Verse Debug Draw = ON** or boxes/arrows never appear. Source comment.

Do **not** put this class on a Guard Spawner (hall 03 is the counter-example). Patrol paths (`AIPatrolPath_Device` + `CharDef_NPC_Patrol`) are the no-Verse alternative.

## Don't

- Don't subclass `creative_device` and hope it possesses the NPC.
- Don't `NavigateTo` without checking `navigation_result.Reached` — failed nav is silent without the print.
- Don't treat Scene Graph camera-component docs as hall 07 — those are Perspective / Fixed Point **devices**.
- Don't skip `OnEnd` cleanup if you add state; shipped OnEnd only prints.
- Don't train boss stun/AOE interrupts from this wander script — that is npc-attack-stasis-aoe.md (`creative_device`, not `npc_behavior`).
