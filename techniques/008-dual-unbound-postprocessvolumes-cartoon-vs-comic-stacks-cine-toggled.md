---
id: technique-008
kind: technique
---

# Dual unbound PostProcessVolumes: cartoon vs comic stacks, cine-toggled

**Sources:** city_starter_template
**Tutorial:** TURTLE City Starter Visual Styles — **External**. Live values **Observed**.

This is not the lighting-template named-job rig (environment-lighting-rig.md) and not Mutator Volume material swaps (mutator-volume-material-swap.md). City Mutator Zones only **Activate** teleporters.

## Problem

Two whole-island looks (cartoon / comic-noir) must share one city kit. Artists need to A/B in the editor; players need to pick a look in the sewer lair before they hit the streets. No Verse.

## Pattern

Two **unbound** `PostProcessVolume`s, same tiny box near the HUD rack, **priority 0**:

| Volume | `bEnabled` | `blendWeight` (default) | Weighted blendables (weight 1 each) |
|---|---|---|---|
| `PostProcessVolume_Cartoon` | true | **1** (on) | `PI_OutlinesCartoon`, `PP_DepthFog_Inst`, `PP_CelShaderNeutralEV` |
| `PostProcessVolume_Comic` | true | **0** (off) | `PP_ComicCelShader`, `PP_Posterize`, `PI_OutlinesComic`, `PP_Frame`, `PP_DesaturateLast` |

Unbound = the stack is the island, not a boxed room. Default cartoon = comic weight 0, not a disabled actor.

**Editor A/B** (docs): hide Cartoon (outliner eye), set Comic `blendWeight` to 1. Docs search string is the typo `"postproccess"`.

**Gameplay A/B** (live, not Switch devices):

| Button `interaction Text` | Cine | Sequence |
|---|---|---|
| New Cartoon Shader | Cinematic Sequence Device6 | `LS_PP_CelShader33` |
| New Comic Shader | Cinematic Sequence Device8 | `LS_PP_ComicShader33` |

`On Interact` → `PlayFunction`. `bAutoPlay=false`, `bLoopPlayback=false`, GameplayOnly, HUD stays Show. `timesCanTrigger=0` (unlimited).

Do not cite `PP_CartoonCelShader` / `LS_PP_CelShaderActive` as the live path — those assets have **0** referencers. Live cartoon cel is `PP_CelShaderNeutralEV`; live cines are the `*33` sequences.

Point lights that must not sparkle under cel/outlines are labeled `PointLight_NoSpec*` (docs: specular off). That is the lighting half of the same look, not a second PPV.

## Don't

- Don't put each style in a **bound** volume over a district. Both volumes are unbound; 05 vs 06 folders are not two looks (`06_Environment` is empty).
- Don't use Mutator Volume to swap the filter. City's two mutators fire teleporters.
- Don't train Arcade `Device_Switch_V2` or sidescroller cameras as this island's style picker. Shipped pickers are **Buttons + Sequencer**.
- Don't route two looks through one Popup (Party/Scary) — that is popup-theme-cinematic.md. City is one Button per cine.
- Don't assume `get_referencers` on a PP material proves it is unused — OFPA volumes often report 0. Read `settings.weightedBlendables` on the placed PPV.
