---
id: technique-009
kind: technique
---

# Fake interiors: cubemap windows + mask + sheen, Day Sequence kills stock emissives

**Sources:** city_starter_template
**Tutorial:** TURTLE City Starter Visual Styles — Windows / Custom Lighting — **External**. Live values **Observed**.

Skybox mesh+MI is the same pair as NightEnvironment / alternate dimension — do not duplicate; see digest. Depth fog lives on the cartoon PPV (cartoon-comic-unbound-ppv.md), not on these window materials.

## Problem

A skyline has to read as occupied rooms at street and roof distance without modeling interiors, and stock building emissives fight the TURTLE night look.

## Pattern

Authored window kit, then kill the kit lights you did not author:

| Asset | Job (docs + Content/) |
|---|---|
| `M_Windows_Interior` | cubemap “room” in buildings the player never enters |
| `M_WindowMask` / `M_WindowMask_Inst` | randomize which panes are lit from world position |
| `M_Windows_Channel6` | diagonal cartoon sheen (Channel 6 News) |
| `SM_BGBuildingA/B_WindowCorner` + `WindowMid` | the panes those materials sit on |

`Day Sequence Device2`: **priority 2000**, `use Volume=false`, `overrideSky=true`, **`enableSkyboxComponent=true`** (alternate dimension keeps this **false**), `enableSunComponent=false`, fog **0.01**, sunlight intensity **10**, stock `MI_DaySequenceSkydome`. Docs: DSD turns **off** background building material lights so the custom windows can own the night.

Foreground / mid / background simplify as you leave the playable strip (`05_Environment`, 8622 actors). That is layout, not a second PPV.

## Don't

- Don't keep Fortnite building emissives *and* these window materials. The DSD pass exists to switch the kit lights off.
- Don't look for Verse or a Mutator Volume doing the window lottery. It is world-position in the material.
- Don't treat `06_Environment` as a second city. Live count **0**.
