---
id: technique-010
kind: technique
---

# Override island time-of-day with stacked Day Sequences + PPV priority

**Sources:** day_sequence_template, dimension_x_template, prison_sample
**Tutorial:** ch.1 Device Information · ch.2 North

## Problem

One island, four biomes that must not share Chapter 4 Fortnite’s default sun. You cannot “delete the island Day Sequence.” A second device at default priority **mixes** with the first (2000+2000 = blend). Unchanged knobs on the second device keep the first’s values.

## Pattern

1. **Center device** = shared base. Few knobs: time ~4–6am (long shadows), Exposure 0, CA 0.3, Vignette 0.6, Saturation 0.8, Grain 0.2. Priority **2000** (default).
2. **Biome device**: Trigger Volume **On**, Priority **> 2000** (docs example 3000). Only then it 100% wins. Blend Distance or Time if you want a walk-in transition.
3. **Inherit vs override:** if center has CA and the biome device doesn’t touch CA, CA stays. Change the field on the biome device to replace it.
4. **Polish PPV** in the biome folder, Priority **> 10** (Lumen Exposure Manager’s PPV is 10). North uses 15 “because it’s higher than 10.”
5. **Lumen Exposure Manager** only if low-end must match Lumen. Otherwise skip — extra PP stack.

Read islands with **Show Only Modified Properties**.

**dimension_x_template** uses the same 2000-vs-higher rule for *local fog*, not biomes: island `Day Sequence Device_Main` priority **2000** / fog 0.2 / `use Volume=false`; four pool devices priority **2020** / fog **10** / green / `use Volume=true`. Write-up: ooze-pool-stack.md.

**prison_sample** uses it for an *interior*: `Day Sequence Device (Outside)` **2000** / fog **0.01** / sun **4** / `use Volume=false`; `Day Sequence Device (Underground)` **2050** / fog **1** / sun **1.5** / `use Volume=true`. Write-up: interior-volume-day-sequence.md. **Confirmed (3).**

North (extreme): sun intensity 10→**2**, fog 0.01→**1.2**, blue fog / inscattering. South: 4.2am, Z rot −10, fog 0.3, 80% clouds. East: sun **15**, size 3. West: 3.8am dawn, fog 0.2.

## Don't

- Don’t add a second Day Sequence at Priority 0 / 2000 and expect a hard cut.
- Don’t add Lumen Exposure Manager for a tiny grade — tutorial says it only pays when looks are drastic across scalabilities.
- Don’t copy `CC_ToDM_POC` — leftover umap, not the island.
- Don’t copy space_station_sample as a fourth 2000-vs-higher source. One DSD, priority **2000**, FixedTime **1:00**, `use Volume=false`, `overrideSky=true`.
