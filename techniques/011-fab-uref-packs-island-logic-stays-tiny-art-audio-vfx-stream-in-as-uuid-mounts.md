---
id: technique-011
kind: technique
---

# Fab `.uref` packs: island logic stays tiny; art/audio/VFX stream in as UUID mounts

**Sources:** domination_sample
**Corpus unique:** only packet with a `References/` folder or any `.uref`. 26 other templates: Fab count **0**.

## Problem

A domination island has ~6000 placed actors (4973 static meshes, 498 decals, 91 audio players) but only **30 authored `.uasset`** in `Content/`. Copying that art into the project would bloat it and fight Fab versioning. Deserted splits the look into three Epic Fab packs and keeps Verse + landscape + sequences local.

## Pattern

`References/` holds three 106-byte stubs (not the assets):

```
References/DesertedAudio_f39271642089687d9d80da5723ee2c9b.uref
References/DesertedProps_2959d4a89338333a14cae73f56c2581f.uref
References/DesertedVFX_f98ac6f313628ed8ee46a721f2639982.uref
```

Each file:

```json
{
  "fileVersion": 2,
  "id": "/EpicGames@Fab.com/DesertedAudio",
  "projectId": "",
  "bEnabled": true
}
```

(`DesertedProps` / `DesertedVFX` same shape.) Filename `{Pack}_{contenthash}.uref` — the hash is **not** the runtime mount UUID.

At cook/ref time those packs explode into **61 GUID mounts**. High-count examples (from `_raw/.../reference-mounts.txt`):

| Mount UUID | What ships | Pack (inferred) |
|---|---|---|
| `e46329d3-…` | walls / `SM_Base_Wall` / brick | Props |
| `65f759f2-…` | girders / `SM_MetalBeam07` | Props |
| `e32a5500-…` | `SM_Arid_Weed_*` | Props |
| `d81c9082-…` | barrels | Props |
| `18b7ebd6-…` | railings / stairs | Props |
| `467c3b2b-…` | pipe kits | Props |
| `f7ecabe1-…` | cement rubble | Props |
| `75c0fe65-…` | dirt/puddle decals | Props |
| `fc14bf95-…` | fence / barbed wire | Props |
| `0b5a55fe-…` | lockers / office | Props |
| `60e907a5-…` | `BFLN_VO_ANNC_*` announcer | **Audio** |
| `e5d3bf01-…` | `NS_AA_Fire`, jet dust, oil fire, landscape storm | **VFX** |
| `550a6647-…` | `fx_veh_jet_passby_*_Cue` | Audio |
| `ab315071-…` | **this project** (landscape, Verse, SEQ_*) | local |

Actors in the level hard-ref `/<uuid>/SM_…`. They do **not** copy into `Content/`. Duplicate the island without the three `.uref` files and the static meshes go missing.

Also `/Deserted1x` (5 refs) — leftover module name, not a fourth pack.

Local remainder (what *should* stay in the project): 6 Verse files, landscape layer infos + `M_Deserted_Landscape`, `SEQ_Jet_*` / cine lobby, `CapLighting_*`, `Deserted_Door`.

## Don't

- Don't vendor Fab meshes into `Content/` “to be safe.” That's what the 5998 OFPA actors already reference by GUID.
- Don't treat the `.uref` contenthash as a mount id.
- Don't expect `get_folders` to list Fab assets — they aren't Outliner folders; they're mounts.
- Don't copy this as “every island needs Fab.” 26/27 templates don't. Use it when art volume dwarfs gameplay code (this island: 30 vs 5998).
