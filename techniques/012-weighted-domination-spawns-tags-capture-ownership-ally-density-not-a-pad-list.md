---
id: technique-012
kind: technique
---

# Weighted domination spawns: tags + capture ownership + ally density — not a pad list

**Sources:** domination_sample
**Tutorial:** Deserted: Domination Template — spawn-pad / Capture Area / Player Counter chapters. Shipped extra: lighting cines, VO queue, sequencer randomizer.

## Problem

69 spawn pads. If you `@editable` every one, the device is unmaintainable. Players must spawn on **their** side of a 3-point domination map, then *near owned points and allies*, then still have a fallback if they own nothing.

## Pattern

**1. Tag the pads, don't list them.**

```
DominationSpawnTags:
  Tag_SpawnGroup_Initial
  Tag_SpawnGroup_GroupA / GroupB / GroupC
  Tag_TeamOwnership_Team1 / Team2
```

`FindCreativeObjectsWithTag` at `OnBegin`. Initial group: 16 pads (8 per team), used once then `Sleep(5)` + `Disable()`.

**2. Weight per capture point.**

```
CaptureWeight = +1 if Team1 owns, −1 if Team2 owns
PlayerWeight  += 0.5 per Team1 in LargeZone, −0.5 per Team2
Weight        = CaptureWeight + Floor(PlayerWeight)

Team1 pads Enable iff Weight > 0
Team2 pads Enable iff Weight < 0
```

`LargeZones` are `player_counter_device` with Use Zone = true, covering the neighborhood of each point — not the capture radius. Subscribe `CountedEvent` / `RemovedEvent`.

**3. Fallback pads stay always-on** (tutorial: Priority 20, Team Any, Enabled During Phase Always, no spawn-group tag). Priority 10 on tagged pads so they win when enabled.

**4. Per-point event handler class** (one instance per A/B/C) holds the two weights and forwards to `CalculateWeightForPoint`. Cache the handler array so it isn't GC'd (`CachedListeners` comment in VO controller: “In the future you won't need to do this!”).

## Don't

- Don't `@editable` 69 spawners.
- Don't leave initial pads enabled (spawn-camp the round-start line).
- Don't skip fallback pads — a 3-cap wipe would then fail `Only Allow Respawn if Spawn Pads Found`.
- Don't play “losing control” VO on still-neutral points (`HasBeenCaptured` gate in `DominationVOController`).
- Don't leave the pregame cine running (`cinematic_lobby_controller`: Stop lobby, Play cleanup).
