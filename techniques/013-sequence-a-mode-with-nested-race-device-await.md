---
id: technique-013
kind: technique
---

# Sequence a mode with nested `race` + device `Await`

**Sources:** detonation_template
**Snippet:** `../snippets/verse/detonation_device.verse`

## Problem

Two bomb sites. Only one may arm. After arm, either the timer completes (explode + attackers win) or defenders stop it (defenders win). You must not poll.

## Pattern

`race` runs async expressions; **first completion cancels the rest**.

```verse
race:
    block:
        ArmingPlayer := TimedObjectiveA.BeganEvent.Await()
        set BombState = bomb_state.BombAArmed
        TimedObjectiveB.Disable(ArmingPlayer)
        for (MapIndicator : BombBMapIndicators):
            MapIndicator.Disable()
    block:
        ArmingPlayer := TimedObjectiveB.BeganEvent.Await()
        set BombState = bomb_state.BombBArmed
        TimedObjectiveA.Disable(ArmingPlayer)
        for (MapIndicator : BombAMapIndicators):
            MapIndicator.Disable()

UpdateBeacons()

race:
    BombDetonated(TimedObjectiveA.CompletedEvent.Await())
    BombDetonated(TimedObjectiveB.CompletedEvent.Await())
    block:
        DisarmingPlayer := TimedObjectiveA.EndedEvent.Await()
        EndGameDevice.Activate(DisarmingPlayer)
    block:
        DisarmingPlayer := TimedObjectiveB.EndedEvent.Await()
        EndGameDevice.Activate(DisarmingPlayer)
```

`block` groups a wait + mutations so the *whole* arm-side is one racer.

Event names on `timed_objective_device` (**Observed**, not the tutorial):

| Meaning | Shipped | Tutorial (stale) |
|---|---|---|
| Player finished Arm interact | `BeganEvent` | `StartedEvent` |
| Timer ran out | `CompletedEvent` | Completed (prose) |
| Player finished Disarm | `EndedEvent` | Stopped (prose) |

`Disable(ArmingPlayer)` on the other Timed Objective is what enforces one site.

Closed state → `enum`. Don’t copy Commander’s abstract-class commands here.

## Don't

- Don’t subscribe and also Await the same event without a plan — this file is Await-only.
- Don’t emit `search_and_destroy` (tutorial class name).
- Don’t busy-loop Sleep checking a bool for “which bomb armed.”
