---
id: technique-014
kind: technique
---

# Ooze pools: default Water devices + authored sBRICK + local Day Sequence fog

**Sources:** dimension_x_template
**Tutorial:** Working with TURTLE Islands blurb only — **External**. Live values **Observed**.

Docs say “place pools of ooze.” The shipped look is **not** a custom Water Type.

## Problem

You need toxic alien pools that read as ooze (surface, bubbles, steam, green air) without a custom water shader on the Water device and without Verse.

## Pattern

Stack five layers in `Environment/Ooze/` (**Observed**):

| Layer | What ships | Job |
|---|---|---|
| **Water device** | 15× `Device_Water_V2`, `water Type=Default`, Always, 100% fill | collision / swim volume |
| **SBRICK mesh** | `SM_ENV_SBRICKA` (`ENV OozeA`…; default material, no override) | the goo surface |
| **Niagara** | 21× `NS OozeBubble` → authored `NS_GooBubble`, `bAutoActivate=true` | bursting bubbles |
| **Steam cards** | Engine `Plane` + `M_OozeSteam` override (`OozeSteam_*`) | local vapor |
| **Day Sequence volume** | Pool1–4, **Priority 2020**, `use Volume=true`, fog **10**, green | toxic air over the pool |

Island lighting stays on `Day Sequence Device_Main` at **Priority 2000**, `use Volume=false`, fog **0.2**. Pool devices win locally because 2020 > 2000 — same rule as day-sequence-priority-stack.md. Sample extents: Pool1 60×100×5, Pool2 65×60×5, Pool3 57×45×5, Pool4 64×60×5.

Water + VFX samples have **empty** `ListEventBindings`. Enable During Phase = Always; no Trigger bus.

Ambient color around the pools is **VFX Spawner**, not Niagara: 16× `LightFog` (orange), 1× `Ambient_Dust` (magenta), 1× `WaterfallSplash` (green). That is the Spawner recipe from vfx-surface-split.md, not a third surface.

## Don't

- Don’t set Water Type to something exotic and expect this island’s look. Live is **Default**.
- Don’t put the ooze fog on the island Day Sequence (Main fog is 0.2). The poison is the **volume** device at 2020.
- Don’t cite `M_Mutagen_*` / `MI_Mutagen_*` as the pool look. Those assets have **0** referencers; the live folders are `Ooze/` not `Mutagen/`.
- Don’t look for a Verse gather / TURTLE Character Spawner here — that is turtle_arcade_sample.
