---
id: technique-015
kind: technique
---

# Upgrade loadout on `EliminatedEvent`, track progress in `[agent]int`

**Sources:** elimination_template
**Snippet:** `../snippets/verse/elimination_game_device.verse`

## Problem

FFA where each elim upgrades the killer’s gun, first to N elims wins, N = number of item granters you wired.

## Pattern

Subscribe **character** events, not a Creative tracker:

```verse
FortCharacter.EliminatedEvent().Subscribe(OnPlayerEliminated)

OnPlayerEliminated(Result:elimination_result):void=
    if:
        FortCharacter := Result.EliminatingCharacter?
        EliminatingAgent := FortCharacter.GetAgent[]
    then:
        GrantNextWeapon(EliminatingAgent)
```

Optional eliminator: world/self-elim fails the `if`, no grant.

Per-player rung:

```verse
var AgentMap:[agent]int = map{}
# OnBegin: set AgentMap[Player] = 1; Grant granter[0]
# After kill: grant WeaponItemGranters[CurrentItemNumber], then CurrentItemNumber += 1, write map
```

Win: `NumberOfEliminationsToWin = WeaponItemGranters.Length`. Predicate:

```verse
IsVictoryConditionMet(EliminationNumber:int)<decides><transacts>:void=
    EliminationNumber = NumberOfEliminationsToWin
# call: if (IsVictoryConditionMet[CurrentItemNumber]):
```

Shuffle the `@editable []` once so designer order ≠ match order.

## Don't

- Don’t use `IsVictoryConditionMet(...) = true` — that’s the tutorial; shipped is failable `[]`.
- Don’t grant `[CurrentItemNumber - 1]` after increment unless you re-derive indices; shipped grants then increments from a map that starts at 1.
- Don’t Subscribe without handling empty EliminatingCharacter.
