---
id: technique-016
kind: technique
---

# Autonomous moving platforms: Zero-wiring ping-pong Prop Mover with EFCLMPS_Overlap prop selection

**Sources:** obstacle_course_starter  
**Tutorial:** Designing an Obstacle Course — **External**. Values and bindings **Observed** via `ObjectTools.get_properties` and `DeviceToolset.ListEventBindings`. Contrast: prop-mover-device-wiring.md (18 wired movers via helper Switches and Triggers).

## Problem

Creating dynamic translating platforms in an obstacle course often tempts creators to either write custom Verse scripts (`MoveTo` loops) or wire complex device networks (Switches, Triggers, and Timers) to reverse direction upon reaching path endpoints. obstacle-course obstacle courses require a lightweight, autonomous moving platform pattern with zero external dependencies.

## Pattern

### 1. Zero-wiring native ping-pong

`Device_PropMover_V2` features built-in autonomous oscillation. Unlike `obstacle_course_sample` (which wired 18 Prop Movers through hidden Switches and helper Triggers to toggle state on movement finish), `obstacle_course_starter` uses **zero device event bindings** (`ListEventBindings` returns `[]`):

```json
{
  "movementMode": "EFCLMM_Translation",
  "pathCompleteAction": "EFCLMMS_PingPong",
  "distanceInMeters": 15,
  "speedInMeters": 5,
  "shouldMoveFromStart": true,
  "delayFromStart": 0
}
```

- `pathCompleteAction: "EFCLMMS_PingPong"`: Automatically reverses linear translation direction when the 15-meter path limit is reached, running continuously without any external signal.
- `shouldMoveFromStart: true`: Movement begins immediately upon match start with 0 delay.
- At `speedInMeters: 5`, a full one-way pass takes exactly 3.0 seconds, providing a predictable timing window for Avatar platforming jumps.

### 2. Overlap selection (`EFCLMPS_Overlap`)

Rather than assigning an explicit target reference (`propReference: null`), the device binds to the intended platform by physical collision volume intersection:

```json
{
  "propSelection": "EFCLMPS_Overlap",
  "propReference": null
}
```

- In `obstacle_course_starter`, `Device_PropMover_V2_C_0` is placed at `(X: 776, Y: 5888, Z: 2192)`, precisely matching the center of `FG01_HoverPlatform_M_C_0` (distance = 0.0 units).
- Because `propSelection` is set to `EFCLMPS_Overlap`, the mover automatically adopts the intersecting obstacle-course hover platform prop without requiring actor reference wiring in the editor.

### 3. Collision behavior for moving obstacle platforms

When moving props interact with players and adjacent static geometry:

```json
{
  "onPlayerCollisionBehavior": "Continue",
  "playerDamageOnCollision": 0.0,
  "onPropCollisionBehavior": "Reverse",
  "propDamageOnCollision": 0.0
}
```

- `onPlayerCollisionBehavior: "Continue"`: The platform continues smoothly along its trajectory even when Avatars stand, jump, or dive onto it, carrying players across the void gap rather than stopping or pushing them off.
- `playerDamageOnCollision: 0.0`: Prevents accidental player elimination from physics contact.

## Don't

- Don't write Verse loops with `Sleep()` and `MoveTo()` for simple linear translating platforms when a single Prop Mover with `pathCompleteAction=EFCLMMS_PingPong` handles it natively.
- Don't build external Switch/Trigger ping-pong networks unless movement needs to be triggered by buttons or targets (as in `obstacle_course_sample`).
- Don't manually link `propReference` when placing the device inside modular platforms — `propSelection=EFCLMPS_Overlap` auto-binds cleanly.
