---
id: technique-017
kind: technique
---

# obstacle-course Avatar locomotion tuning: PlatformLocomotion, disabled combat/traversal, and mantling-first obby physics

**Sources:** obstacle_course_starter  
**Tutorial:** Working with obstacle-course Islands — **External**. Settings **Observed** via `ObjectTools.get_properties` on `Device_ExperienceSettings_FallGuys_C_0`. Project flags **Observed** in `obstacle_course_starter.uefnproject`.

## Problem

Building a obstacle-course obstacle course in UEFN requires player characters to look, weigh, wobble, jump, mantle, and ragdoll like obstacle-course Avatars rather than standard Fortnite Battle Royale avatars. Creators need to understand which project-level flags enable the obstacle-course ecosystem and how `Device_ExperienceSettings_FallGuys` tunes movement capabilities.

## Pattern

### 1. Project-level enablement (`PlatformLocomotion`)

In `obstacle_course_starter.uefnproject`:
```json
{
  "featureSet": "PlatformLocomotion",
  "dataSets": {
    "experimental": {
      "DefaultPawn": {
        "bUseDefaultCharacter": true
      }
    }
  }
}
```
- `featureSet: "PlatformLocomotion"`: Activates the obstacle-course game feature set (internal codename *platforming*). This mounts `/platformingLevelArt` (modular kits, props, obstacles) and `/platformingExperienceSettings`.
- `bUseDefaultCharacter: true`: Ensures players spawn directly as the customized theme avatar (the Avatar) rather than their Fortnite locker cosmetic skin.
- Player Spawner device `BP_Creative_Player_Spawner_Prop_C_0` is placed with `useAsIslandStart: true`. Avatars spawn directly on pads — Avatars do not skydive from a Battle Bus.

### 2. Disabling incompatible Fortnite traversal and combat

The obstacle-course Avatar uses custom animation and physics assets that do not support Fortnite tactical combat, vehicles, weapons, or advanced traversal:

| Property | Value | Rationale |
|---|---|---|
| `bAllowSprinting` | `false` | Avatars have fixed run speeds; tactical sprinting breaks platforming gap calculations. |
| `bAllowSliding` | `false` | Avatars do not slide on knees; slopes use physics momentum or ragdoll. |
| `bAllowVaulting` | `false` | Vaulting over obstacles is disabled in favor of mantling. |
| `bAllowWallKick` | `false` | Advanced wall movement is disabled. |
| `bAllowWallScramble` | `false` | Advanced wall movement is disabled. |
| `bAllowBoostedJump` | `false` | Jump heights remain deterministic. |
| `bAllowRollLanding` | `false` | Landings trigger standard recovery or ragdoll, not roll. |
| `bFallDamage` | `false` | Avatars bounce and ragdoll on hard landings; elimination is handled via hazard volumes. |
| `bFlyEnabled` | `false` | Flight is disabled. |
| `bInfiniteResources` | `true` | Shipped setting, though building is typically restricted. |

### 3. Preserving essential Avatar platforming mechanics

- **Mantling (`bAllowMantling: true`):** Crucial for obstacle-course obstacle courses. The Avatar automatically mantles ledges within reach. Mantling is slower than standard Fortnite mantling, matching Avatar momentum, but highly responsive to save missed jumps.
- **Hurdling (`bAllowHurdleOverJumpableObjects: true`):** Enabled in shipped settings to allow smooth traversal over low curbs and barriers without snagging.
- **Glider Redeploy (`bGliderRedeploy: true`):** Left enabled as a fallback/failsafe.
- **Dive & Ragdoll (Native Avatar Physics):** Built directly into the Avatar character class. Shift / RMB initiates a forward dive (ground or air). High impact or falling for >7 seconds induces a ragdoll state where players retain limited aerial steering before jumping back up upon deceleration.

## Don't

- Don't leave Fortnite sprinting or sliding enabled on a obstacle-course island — players will bypass designed platforming gaps.
- Don't rely on fall damage for hazard death — use Elimination Damage Volumes (`damageType: "Elimination"`) under gaps and sBRICK pits.
- Don't assume all Fortnite weapons, vehicles, or items work — obstacle-course islands strip weapon spawners and vehicle devices at the feature-set level.
