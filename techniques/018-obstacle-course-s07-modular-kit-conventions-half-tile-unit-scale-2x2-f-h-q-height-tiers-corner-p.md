---
id: technique-018
kind: technique
---

# obstacle-course S07 modular kit conventions: Half-tile unit scale (2x2), F/H/Q height tiers, corner pivots, and data asset palettes

**Sources:** obstacle_course_starter  
**Tutorial:** obstacle-course Obstacle Course Assets — **External**. Asset paths and properties **Observed** in `_raw/obstacle_course_starter/referenced-paths.txt` and live MCP actor queries.

## Problem

Building consistent obstacle-course courses requires understanding the modular geometry scale, naming taxonomy, pivot point alignment, and material palette system used by the `/platformingLevelArt` kit. Mixing these assets with standard Fortnite props often causes grid misalignment or unexpected collision behavior.

## Pattern

### 1. Half-tile unit scale and footprint taxonomy

Standard Creative grid tiles measure **512 cm × 512 cm × 384 cm** (Depth × Width × Height). obstacle-course modular building pieces use a coordinate system where **1 unit = half a tile (256 cm)**:

- A standard 1-tile floor is denoted as **`2x2`** (512 cm × 512 cm).
- A 2-tile wide by 1-tile deep platform is **`4x2`** (`FG01_Floor_Str_4x2s` — 1024 cm × 512 cm).
- Narrow walkways and extended platforms use **`4x1`** (`FG01_Floor_Str_4x1s`) and **`4x3`** (`FG01_Floor_Str_4x3s`).
- The `s` suffix designates "squares" (half-tile increments).

### 2. Height tiers: Full (F), Half (H), Quarter (Q)

Heights are standardized into three discrete tiers relative to the 384 cm tile height:

| Designation | Name | Height (cm) | Example Asset |
|---|---|---|---|
| **F** | Full Height | 384 cm | `FG01_Shape_Rect_F_2s` |
| **H** | Half Height | 192 cm | `FG01_Shape_Rect_H_2s`, `FG01_Wall_Str_H_1s`, `FG01_Ramp_Str_H_1x2s_B` |
| **Q** | Quarter Height | 96 cm | `FG01_Barrier_Str_Q_2s`, `FG01_Barrier_Pillar_Flat_Q` |

Ramps (`FG01_Ramp_Str_H_1x2s_B`) rise exactly half-height (192 cm) over a 2-unit (512 cm) run, creating comfortable traversal slopes for Avatars without requiring jump/mantle inputs.

### 3. Corner pivots vs center pivots

Standard Fortnite Creative props and building pieces place their origin/pivot at the geometric center. obstacle-course modular pieces place pivot points at the **corners**:
- Allows modular edge-to-edge and face-to-face snapping across varying dimensions without fractional offsets.
- When placing obstacle-course assets in UEFN, adjust the editor grid snap to 128 or 256 cm increments to align corner pivots.
- Because of corner pivots and custom building classes (`FG01_*`), standard Fortnite wall- and floor-mounted traps cannot attach directly to obstacle-course modular surfaces (use Hover Platforms or standard Fortnite base props if traps are required).

### 4. Data asset palettes and per-instance texture data

Modular pieces decouple geometry from color/pattern styling via soft-referenced data assets:
- Every modular piece references a centralized palette asset:
  ```json
  "palette": {
    "refPath": "/platformingLevelArt/Environments/S07/ModularKit/Data/DA_BS_Palette_S07.DA_BS_Palette_S07"
  }
  ```
- Visual variation (colors, stripes, checkers, arrow guides, team emblems) is configured per instance via `textureData[0..3]` properties without creating duplicated material graph instances.

## Don't

- Don't scale obstacle-course modular pieces manually using arbitrary transform scale — use the appropriate F, H, Q, or half-tile size variant to preserve texture UV fidelity and mantling reach.
- Don't assume obstacle-course floors snap to standard Fortnite center pivots — align using corner vertices or 256 cm grid steps.
