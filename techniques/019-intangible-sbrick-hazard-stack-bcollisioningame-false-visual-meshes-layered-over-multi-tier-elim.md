---
id: technique-019
kind: technique
---

# Intangible sBRICK hazard stack: bCollisionInGame false visual meshes layered over multi-tier elimination volumes

**Sources:** obstacle_course_starter  
**Tutorial:** Working with obstacle-course Islands — **External**. Property values and transforms **Observed** via `ObjectTools.get_properties` on `FG01_Cube_SBRICK_4s_C_0` and `BP_Device_DamageVolume_V2_C_*`.

## Problem

obstacle-course levels feature iconic pools of pink sBRICK and bottomless drop-offs that instantly eliminate players and send them back to the start. Creators need a reliable technique to render visible hazardous sBRICK surfaces while guaranteeing immediate player elimination and respawn without relying on weapon damage or standard fall damage (which is disabled for Avatars).

## Pattern

### 1. The intangible visual sBRICK mesh (`bCollisionInGame: false`)

Rather than giving the sBRICK mesh hazardous physical collision, the template turns in-game collision completely off on the visual prop:

```json
{
  "actor": "FG01_Cube_SBRICK_4s_C_0",
  "class": "/platformingLevelArt/VFX/Blueprints/FG01_Cube_SBRICK_4s.FG01_Cube_SBRICK_4s_C",
  "properties": {
    "bCollisionInGame": false
  }
}
```

- When a Avatar attempts to land or walk on the sBRICK cube, they don't bounce or walk on it; they sink straight through the mesh.
- This creates the visual effect of drowning or sinking into deep sBRICK.

### 2. Multi-tier elimination volume hierarchy

Elimination and respawn logic is entirely handled by `BP_Device_DamageVolume_V2` actors positioned beneath the intangible meshes:

```
[Course Level: Z ≈ 1200 - 2800 cm]
  │
  ├── Local Hazard Pit 0 (Y = -516, Z = 728 cm): DV 0 (1 × 1 × 1 tiles) under SBRICK Cube 0
  ├── Local Hazard Pit 1 (Y = 3352, Z = 1096 cm): DV 1 (2 × 2.5 × 1 tiles) under SBRICK Cube 1 + Drum S
  ├── Local Hazard Pit 2 (Y = 5976, Z = 1288 cm): DV 2 (2 × 2.5 × 1 tiles) under SBRICK Cube 2 + Moving Platform
  │
[Sea Level: Z = 728 cm]
  └── Global Ocean Catch-All: DV 3 (100 × 100 × 0.5 tiles) covering entire map perimeter
```

### 3. Damage Volume configuration

All 4 Damage Volumes share a unified configuration:

| Property | Value | Role |
|---|---|---|
| `damageType` | `"Elimination"` | Instantly eliminates the player upon contact regardless of shields or health. |
| `damage` | `10` | Base damage value (overridden by Elimination type). |
| `damage Tick Rate` | `"2 Seconds"` | Immediate trigger upon entering the volume. |
| `enabledOnPhase` | `"Gameplay Only"` | Inactive during pre-game and countdown. |
| `zoneVisibleDuringGame` | `false` | Invisible collision boundary hidden under the visual sBRICK/water. |

- **Local Pit Volumes (DV 0, 1, 2):** Fitted tightly inside the pit cavities below each obstacle challenge. When a player falls through the sBRICK cube or misses a jump, elimination occurs immediately within the pit, triggering the respawn sequence before the player falls into open sky.
- **Global Perimeter Volume (DV 3):** Spanning a massive 100 × 100 tile zone at Z = 728 cm directly above the ocean water surface (`VK_Waterbody_Ocean_Parent`). Any player bounced, ragdolled, or blown completely off the elevated platforms hits this plane and respawns cleanly.

## Don't

- Don't leave `bCollisionInGame: true` on sBRICK cubes if they are intended as pits — Avatars will walk across them as solid floors.
- Don't rely solely on water body damage or falling out of the world — water swimming is enabled for Avatars by default, allowing them to swim endlessly unless an elimination volume covers the water plane.
- Don't omit the global catch-all plane (DV 3) — ragdolling Avatars launched by bumper poles or moving platforms can easily overshoot small local pit volumes.
