---
id: technique-020
kind: technique
---

# Greybox with CubeGrid / Shape tools, then color-code intent — no art, no Verse

**Sources:** greyboxing101_template
**Tutorial:** Greyboxing — short process page; this island is the lab.

## Problem

You need a playable layout before art. Docs say Shape → Merge → color Attributes → playtest. The island adds **CubeGrid** (the actual volume editor) and **four meaning colors** so testers can read Goal / Danger / Navable / Building without a legend in their head.

## Pattern

Numbered halls (copy Modeling/Materials convention):

```
1 - Greyboxing 101/1.0 Greboxing (typo ships) … 1.3 Tools
2 - Greybox example/Example          ← finished blockout
3 - CubeGrid/2.0 Selection … 4.1 Corner Mode   ← tool GIFs as Level Sequences
4 - Shapetools/2.0 Examples/{Arrow,Box,Cone,…}
```

**Color grids (authored MIs — copy these names):**

| MI | Means |
|---|---|
| `MI_Greybox_Grid_Goal` | objective |
| `MI_Greybox_Grid_Danger` | kill / hazard |
| `MI_Greybox_Grid_Navable` | walk/drive path |
| `MI_Greybox_Grid_Building` | massing |

CubeGrid operations that need motion are **Sequencer** (`LS_Pulling_2_2`, `LS_Push_2_3`, `LS_Boolean_4_0`, plus `_Rev`) fired by `Device_CinematicSequence` + Switch — same hide-the-pawn family as Animation 101, but here the shot *is* the teaching.

Docs Create → **Merge** after the blockout sits. Don’t greybox in final materials.

## Don't

- Don’t ship art into the first playable — that’s the whole point.
- Don’t look up a Marketplace “greybox kit” when this island *is* the kit.
- Don’t write Verse for CubeGrid demos.
- Don’t “fix” `Grebixing` / `Move_Rotate` / nested `Content/Content/` when citing shipped names.
- Don’t confuse catalog `BP_Primitive_*` color massing (BRICK Obstacle Course) with CubeGrid authored MIs.
- Don’t treat BRICK Brick Editor Snap/BrickAssembly as CubeGrid — modular-BrickAssembly-building-modules.md (**0** CubeGrid there).
