---
id: technique-021
kind: technique
---

# Many arenas in one island: persistable pick + inverted teleporters

**Sources:** gunfight_template
**Snippet:** `../snippets/verse/island_selector_device.verse`
**Related:** speedway `persistable` stats — same `weak_map` family, different payload.

## Problem

One Fortnite island, many tiny FPS maps. Each round teleport everyone to a random arena they haven’t just played, optionally swap sides, survive JIP (persist maps disagree).

## Pattern

```verse
persist_island_data := class<final><persistable>:
    LastIslandIndex:int = 0
    ExclusionList:[]int = array{}

var IslandManagerMap:weak_map(player, persist_island_data) = map{}

island_data := class<concrete>():
    IslandsTeleporters : []teleporter_device = array{}
    InverseIslandsTeleporters : []teleporter_device = array{}
```

Pick: if `RandomEveryRound` or no round devices → every round; else `RoundBeginEvent` on listed Round Settings. `PickRandomIsland` uses **mode of LastIslandIndex** across `GetPlayers()`, then that player’s ExclusionList. Shuffle remaining indexes until one isn’t excluded. Full list → reset exclusion to `{Last}`.

Then Enable that arena’s teles (or Inverse if `InvertSides`). `TeleportZone.Enable()` dumps everyone. `IslandSelectionCompletedTrigger.Trigger()` for downstream.

`MakeIslandData<constructor>(PreviousData)` copies fields then overwrite — required because persistable classes are `<final>`.

## Don't

- Don’t store arena on a Creative device that resets each round — that’s why persistable exists.
- Don’t trust one player’s map after JIP — vote.
- Don’t Enable all 64 teles; only the chosen island’s array.
