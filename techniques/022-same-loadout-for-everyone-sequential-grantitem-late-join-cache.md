---
id: technique-022
kind: technique
---

# Same loadout for everyone: sequential GrantItem + late-join cache

**Sources:** gunfight_template
**Snippet:** `../snippets/verse/weapon_granting_device.verse`

## Problem

Every player must get the **same** gun+nade in **slot order**. Don’t grant in parallel (inventory order races). Late joiners must match the round already in progress. Keep the same loadout for two rounds.

## Pattern

```verse
item_pools := class<concrete>():
    Granters:[]item_granter_device = array{}
    NumberToGrant:int = 1

# Build DelayedGranters first, then grant:
for (Granter : Granters):
    Granter.GrantItem(Player)
    Granter.ItemGrantedEvent.Await()   # next only after this lands
```

Shuffle indexes per pool; take `NumberToGrant`. Save `[]tuple(poolIndex, granterIndex)` in `weak_map(player, …)` when `SaveLoadoutOnGrant`. Odd rounds: `RoundsToGrant` → `CreateGrantData`. Even: `LoadPreviousOnRounds` → consensus first-weapon index, rebuild DelayedGranters.

`UseLocalForLateJoiners`: once DelayedGranters is complete, JIP uses that array; else vote persist maps (`FindMostFrequentIndex`). Never compute-while-granting — a joiner mid-loop would get a partial kit.

## Don't

- Don’t fire every Granter in one frame.
- Don’t persist the granter **device** — persist indexes.
- Don’t skip `PlayerAddedEvent` or JIP has fists.
