---
id: technique-023
kind: technique
---

# Night-atmosphere stack: custom skybox + fog cards + ambient audio grid

**Problem.** A night scene needs a sky (Day Sequence only relights the stock sky), ground
fog that reads at walking pace, and ambience that doesn't feel like one global loop — all
without Verse.

**How UEFN's NightEnvironment does it** (all authored assets committed in the project).

## 1. Custom sky replaces Day Sequence entirely

No `DaySequenceActor`/`DSA_*` in the level. Instead, in the `Lighting/` folder:

- **`SM_SkySphere`** (the project's only authored mesh) as a `FortStaticMeshActor`, with
- **`M_ProceduralSkybox`** — material using `MF_SphericalUVsCap` + `MF_PlanetImage` and
  authored textures `T_Stars_A/B`, `T_SkyNebula_A`, `T_Moon`/`T_MoonBlur`, `T_Planet_A_N`:
  a baked night sky with moon and stars, not a sun simulation.
- **`M_LightClouds`** (with `T_Noise_CloudsB`) — moving cloud layer.
- Standard rig stays: `DirectionalLight` (moonlight role), `SkyLight`, `SkyAtmosphere`,
  `ExponentialHeightFog`.

**Lesson:** when the scene is permanently one time of day, a skybox material is cheaper and
more art-directable than Day Sequence. Day Sequence is for *changing* light (see
day_sequence_template); NightEnvironment never changes.

## 2. Fog is cards, not just exponential fog

`VFX/` folder: ~20 `NS FogCards` Niagara actors placed at POI boundaries (graveyard edge,
treeline, waterline), plus `ENV_PACK NS Biome Steam A`, orbiting seagulls and a bat swarm for
motion. Authored: `NS_FogCards` system + `M_FogCard` / `M_FogCard_Masked` materials +
`T_FogCard`, `T_Smoke_Puff_Tile_Up` textures, using `MF_ParticleRandomiser`.

**Lesson:** `ExponentialHeightFog` gives global depth cueing; **fog cards give placed,
local banks** — you put fog *where the silhouette needs softening*, not everywhere.

## 3. Ambience is a grid of 19 tiny audio players, not one island loop

`Audio/AmbientLoop/` holds 19 `Device_CRD_AudioPlayer` devices at POIs. Verified settings
(**Observed** via `ToyOptionsComponent.playerOptionData`):

| Option | Value |
|---|---|
| Audio | e.g. `Halloween_Whispers_Cue` (per-device cue) |
| AutoPlay | **Gameplay only** (not countdown/create/waiting) |
| Play On Hit | True |
| LoopAudio | **False** (cues are long-form) |
| Enable Volume Attenuation | **True**, linear, min 5 m / falloff 10 m |
| Enable Spatialization | False (attenuation carries locality) |
| Visible In Game | False |
| Volume | 2.0 |

**Lesson:** ambient audio devices are invisible, gameplay-gated, and *local* — placed per
POI with tight 5–10 m attenuation so walking between areas cross-fades loops. No triggers,
no Verse; `EnabledDuringPhase = Always` + autoplay-on-gameplay does the sequencing.

## 4. Supporting cast

- **Landscape**: `Landscape` + 16 streaming proxies, painted with **6 committed
  `Layer_*_LayerInfo`** assets (`NightEnvironment_StarterIsland_sharedassets/`) and
  `MI_Harrowvile_Landscape_A`; ocean via `VK_Waterbody_Ocean` + `WaterZone` +
  `WaterBrushManager` (FortWater).
- **Foliage**: 11 committed `*_FoliageType` assets (Helios grass variants, TREE_PACK
  grass/leaves) driven through `InstancedFoliageActor`s.
- **PPV split**: `GlobalPP` (island-wide grade) + `InteriorPP` (tighter volume over the
  haunted house) — interiors re-grade without touching the exterior look.

**Name-drift gotcha:** several authored assets ship with the typo `Harrowvile` (one 'r') —
`MI_Harrowvile_Ocean`, `M_Harrowvile_Landscape_A`. Grep both spellings.

**Sources:** night_environment_sample.

Contrast (do not duplicate): space_station_sample is a night-looking **space** sky via one DSD (`FixedTime` 1:00, `overrideSky=true`, stock `MI_DaySequenceSkydome`) — **not** an authored `SM_SkySphere`.
