---
id: technique-024
kind: technique
---

# Gate IceBase NPCs with Guard Character Definitions, patrol paths, and doorway volume swaps

**Sources:** ice_base_sample, volcanic_town_sample
**Tutorial:** Working With space-opera™ Islands + Vistas and Biome Environments — **External** (roaming NPCs + patrol; Mount CD). Bindings **Observed**.

Contrast (do not duplicate): Walker Type on walker-character-definition.md. Volume Enable for a horde on volume-checkpoint-gated-horde.md. Verse wander on npc-behavior-navigatable-focus.md. This island has **zero** `.verse` files.

## Problem

A snow base needs people outside, people inside, and ridable wildlife — without paying for all of them at once, and without a Verse `npc_behavior`. Interiors should not spawn until the player walks in; exteriors should despawn so the hangar is not fighting the yard.

## Pattern

### 1. Identity lives on an NPC Character Definition (Type = Guard, not Walker)

Authored: `/37632797-…/NPCs/CD_RebelNPC_GuardBehavior`. Tag `Pawn.Athena.NPC.Creative.Guard`. Type subobject `CharacterType_Guard`; preset `CR_BP_AISpawnerData_Guard_Trespassers`.

Live modifiers (**Observed**):

| Slot | Live |
|---|---|
| Health | 100 HP, 0 shield, `isVulnerable=false` |
| Team | Neutral |
| Movement | walk 1 / run 4.1 / sprint 5.5 (`bOverrideFortniteLimits=false`) |
| Effects | death/spawn |
| PatrolPath | `Group_None`, change target Never |
| GuardInventory | `WID_Akita_Galileo_Cosmos_AR_UC` qty **1**, `canDropLoot=false` |

Behavior: accuracy **MODERATE**, leash **on**, distance **30**, `bCanBeDownButNotOut=false`.

`NPCBehaviorScript` is unused. Do not subclass `npc_behavior` for these rebels.

Ridable Mount is a **mounted** definition, not authored here: `/FNEPLATFORM_PACK/CharacterDefinition/Mount_CharacterDefinition` (Type `BP_PLATFORM_PACK`, modifiers **Taming** + **Team**). Docs: predefined CDs under space-opera Content > AI.

### 2. Count / timer stay on the spawner

All **48** `Device_CharacterSpawner` share `spawnCount=1`, `allowInfiniteSpawn=true`, `totalSpawnLimit=1`, `spawnRadius=1`. The split is **when**:

| Cohort | n | `enabledAtGameStart` | `spawnTimer` | Folder |
|---|---|---|---|---|
| Exterior rebels + Guards | 24 | true | 0 | `NPCs/Exterior` (20 dir) + `Guards` |
| Mounts | 12 | true | 3 | `NPCs/Exterior/Mounts` |
| Interior | 12 | **false** | 3 | `NPCs/Interior` |

### 3. Doorway volumes swap the two worlds

Four volumes in `NPCs/SpawnVolumes/`. `OnEnter` bindings on the **spawners** (ListEventBindings on the volume itself is empty):

| Volume | Interior spawners | Exterior + Mount spawners |
|---|---|---|
| `VolumeEnter` | Enable | Disable |
| `VolumeExit` | Disable | Enable |
| `Volume2Enter` / `Volume2Exit` | **not wired** | same Disable/Enable pair (second doorway) |

Interior spawners only listen to doorway 1. Walking in the second hangar door still swaps the yard, but does not Enable the interior twelve.

`despawnAIsWhenDisabled=true` on every spawner — Disable is the exit hose.

### 4. Patrol Path devices, not a Verse leash

**82** `AIPatrolPath_Device`. Sample: `patrolling Mode=BackAndForth`, `enabled At Start=true`, group **6** (rebels) / **20** (storm sample). Nested `Patrol Paths/01–07` + `StaticPatrol` (7 ext / 3 int). The CD’s PatrolPath modifier stays `Group_None`; the **device** owns the group.

Vistas docs: “Add NPC Spawners and AI Patrol Path Node Devices.” That is this island.

### VolcanicTown variant (same devices, three zones + a Civilian CD)

volcanic_town_sample is the volcanic-city sibling (**120** spawners, **90** patrol, **6** volumes, **0** Verse).

Two authored CDs:

| CD | Type | Behavior | Inventory | Leash |
|---|---|---|---|---|
| `CD_RebelNPC_GuardBehavior` (IceBase) | `CharacterType_Guard` + Trespassers | MODERATE, leash **on** 30 | AR qty 1 | on |
| `CD_CivilianNPC_GuardBehavior` (VolcanicTown) | `CharacterType_Guard` + Trespassers | MODERATE, leash **off** (distance 30 unused) | **`items=[]`** | off |
| `CD_CivilianNPC` (VolcanicTown) | **`CharacterType_Custom`** | `DefaultBehavior` `canSwim=false` | none | n/a |

VolcanicTown volumes are **three** Enter/Exit pairs (`NPCs/NPCVolumes/`). Bindings still live on spawners:

| Zone | n | `enabledAtGameStart` | OnEnter |
|---|---|---|---|
| FrontStores | 13 | true | HallwayEnter **Disable** / Exit Enable |
| Hallway + Hallway2 | 24 | true | Hallway4Enter **Disable** / Exit Enable |
| MainSquare + MainSquare2 + SecondarySquare | 41 | **false** | MainSquareEnter **Enable** / Exit Disable |
| Hallway3 + Hallway4 | 16 | true | **no bindings** (always on) |
| Outskirt 01 + 03 | 16 | true | no bindings |
| Outskirt 02 | 10 | **false** | **no bindings** — never spawn |

Patrol groups **1–13** match folder names. Mode BackAndForth except folders 09 and 12 = Loop. No PatrolPath modifier on either VolcanicTown CD — the device owns the group (same as IceBase’s `Group_None`).

## Don't

- Don't Enable-at-start the interior twelve. The doorway Enable is the pacing (and the perf cut).
- Don't copy Walker Type / `MeadowDrosera` modifiers here. Type=`Guard` + Trespassers preset is the space-opera AllyFaction.
- Don't expect Volume2 to wake interiors. Only VolumeEnter/VolumeExit do.
- Don't look for a Verse `creative_device`. `GetDeviceProperties` is Verse-only; read spawners with `ObjectTools.get_properties` + `enable.eventSubscriptions`.
- Don't copy VolcanicTown’s Guard inventory or leash as “the” space-opera Guard. IceBase grants an AR and leashes; VolcanicTown’s Guard CD is empty-inventory / unleashed. Civilian townies are Type=`Custom`, not Guard.
- Don't expect every `enabledAtStart=false` spawner to have a volume. VolcanicTown Outskirt 02 is disabled with no Enable binding.
- Don't copy space_station_sample as a third volume-swap. Its **15** hangar spawners are all `enabledAtGameStart=true`; the one volume is the hyperspace trigger (hyperspace-attach-datalayer.md). Guard CD is IceBase-like (AR + leash on) with perception **off**.
- Don't copy desert_town_sample as a third volume-swap either. **0** volumes; Interior **10** + ImperialOutpost **6** spawners are `enabledAtGameStart=false` with empty bindings (dead, like VolcanicTown Outskirt 02). Guard CD is empty-inventory (VolcanicTown) + leash **on** (IceBase) + perception **off** (Space). Civilian Type=`Custom`.
