---
id: technique-025
kind: technique
---

# Entitlement is the item; offer is the SKU

**Sources:** in_island_transactions_template
**Snippets:** `entitlement_definitions.verse`, `offer_definitions.verse`, `buy_offer_buton_device.verse`

## Problem

Sell V-Buck items without mixing inventory rules with shop copy/price, and without treating `BuyOffer` succeeding as “give the item.”

## Two types

**Entitlement** (`class(entitlement)`): Name/Description/ShortDescription/Icon, `MaxCount`, `Consumable`, `PaidArea`, `PaidRandomItem`, `ConsequentialToGameplay`.

```verse
feature_example_entitlement := class<abstract><castable><unique>(entitlement){}
corn_seed := class<concrete>(feature_example_entitlement):
    MaxCount<override> : int = 100
    Consumable<override> : logic = true
primo_shovel := class<concrete>(feature_example_entitlement):
    Consumable<override> : logic = false   # durable ⇒ MaxCount 1
gnome := class<concrete>(feature_example_entitlement):
    ConsequentialToGameplay<override>:logic = false
vip_access := class<concrete>(feature_example_entitlement):
    PaidArea<override> : logic = true
```

**Offer** (`entitlement_offer` / `bundle_offer`): shop Name/Desc/Icon + `EntitlementType` + `Price := MakePriceVBucks(n)`. Same entitlement, two offers (e.g. 100 vs 50 pack).

**Bundle:** `Offers: []tuple(offer, int)` — **offers**, not entitlements. Nest ≤ 5. Total distinct entitlement ids ≤ 100. Price 50–5000, step 50.

## Buy vs grant vs consume

```verse
Result := BuyOffer(Player, Offer)
if (Result?):
    # do nothing — GetEntitlementsChangedEvent owns side effects
```

`GrantEntitlement` = promo/restore (still respects MaxCount). `ConsumeEntitlement` decrements consumables; fails on durable.

On join: `GetPurchasedEntitlements` vs your save.

Paid random: `PaidRandomItem=true` + odds in description; `RestrictPaidRandomItems[Player]` before showing. Direct prompts: `RestrictDirectPromptsToPurchase`.

## Don't

- Don’t set Consumable=false and MaxCount>1.
- Don’t put entitlements in a bundle — wrap as offers.
- Don’t treat Device Template potion/sword as this farm project.
- Don’t copy `buy_offer_buton` as a naming lesson — it’s a shipped typo.
