---
id: technique-026
kind: technique
---

# Hire hero squad from a Guard Spawner redesign (`DemonFaction Character`), not a Character Definition

**Sources:** music_fantasy_sample
**Tutorial:** Working with MUSIC_THEME Islands — hero squad NPCs — **External**. Bindings **Observed**.

Contrast (do not duplicate): IceBase Guard CD + patrol + doorway volumes on guard-patrol-volume-swap.md. TURTLE theme enum vs definition on TURTLE-character-spawner.md. This island has **zero** `.verse` files and **zero** authored NPC Character Definitions.

## Problem

A theme island wants HeroA / HeroB / HeroC as hireable allies with their pickaxes, without authoring a Guard Character Definition and without using the generic Guard Type dropdown.

## Pattern

### 1. Device is Guard Spawner with a faction character enum

Class: `Device_DemonFaction_CharacterSpawner` (`/DemonFaction_CharacterSpawner/Gameplay/…`). Docs: redesign of Guard Spawner — same settings, Guard Type dropdown swapped for **music-fantasy Character**.

Three placed, all Streets, `enabledAtGameStart=true` (**Observed**):

| Label | `DemonFaction Character` | `bCanBeHired` | `enablePatrol` | `guardType` | `spawnCount` |
|---|---|---|---|---|---|
| music-fantasy Character Spawner | HeroA | true | false | Shadow | 1 |
| …Spawner2 | HeroB | true | false | Shadow | 1 |
| …Spawner3 | HeroC | true | false | Shadow | 1 |

`allowInfiniteSpawn=false`. `characterCosmetic=null` — the theme enum owns the look, not a CID. `guardType=Shadow` is leftover Guard DNA; do not treat it as the identity.

Docs: Character / NPC Spawner / Guard Spawner can wear hero squad cosmetics, but **the pickaxe is only on this device**.

### 2. Dismiss on bathhouse exit, emote on plaza cine

`Vol_DisableBathhouseSpawners` `OnExit` → `Dismiss All Hired Guards` on all three spawners.

Plaza: `EmoteVolume` `OnEnter` → Trigger3 → Cinematic Sequence Device4 `PlayFunction` (`LS_PlazaDance`) + Character Device 1/2/3 `PlayEmote`. A second HUD Controller (`HUD Controller_for_cutscenes`) Enables on that Trigger (`showHUD=HideElement`, `priority=VeryHigh`, `enabledDuringPhase=None`) and Disables when the cine `OnStopped`.

## Don't

- Don't author a Guard CD for these three. The enum is the identity.
- Don't Enable patrol. Live `enablePatrol=false`.
- Don't confuse Character Device (mannequin / emote) with Character Spawner (hireable AI).
