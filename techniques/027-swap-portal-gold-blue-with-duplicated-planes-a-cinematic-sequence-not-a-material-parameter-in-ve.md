---
id: technique-027
kind: technique
---

# Swap Portal gold/blue with duplicated planes + a Cinematic Sequence, not a material parameter in Verse

**Sources:** music_fantasy_sample
**Tutorial:** Working with MUSIC_THEME Islands — Portal Shader + Bathhouse Reactive Environment — **External**. Bindings **Observed**.

Contrast (do not duplicate): Trigger → Cinematic as a VFX bus on trigger-cinematic-quake.md. This island has **zero** `.verse` files.

## Problem

A city needs a visible Portal barrier that starts demonic (blue) and turns gold after the bathhouse fight — without Verse driving a material parameter, and without deleting one set of meshes at runtime.

## Pattern

### 1. Two authored material instances, two Outliner folders of planes

Authored: `Content/Materials/VFX/MI_Portal_Blue` + `MI_Portal_Golden` (FName counts **11** each). ExtraAssets also ships `M_DemonFaction_SphereShield`.

Live folders (**Observed** `get_folders`):

| Folder | Count |
|---|---|
| `PortalBarriers_Demonic` | **10** |
| `PortalBarriers_Golden` | **10** |

Docs: duplicate a Place Actors plane; one gold MI, one blue MI. Modeling tools if you need custom silhouettes.

### 2. Autoplay the demonic sequence; play gold on volume exit

Five Cinematic Sequence devices; two are Portal:

| Label | `sequence` | `bAutoPlay` | Wake |
|---|---|---|---|
| Cinematic Sequence Device5 | `…/Sequences/LS_PortalChange_Demonic` | **true** | GameplayOnly |
| Cinematic Sequence Device | `…/Sequences/LS_PortalChange_Golden` | false | `Vol_DisableBathhouseSpawners` `OnEnter` → `PlayFunction` |

`showHUDDuringCinematic=Show`, `disablePlayerInput=No` on both — this is a shader swap, not a camera lock. Sequencer keyframes are **unanswerable**.

Leftover: root `Content/LS_PortalChange.uasset` plus `Content/Sequences/LS_PortalChange.uasset`. Live devices point at `_Golden` / `_Demonic`, not the leftover. Script-modules still mention previous-project GUID `/e455d126-4a95-8ed2-feec-6ea4d08ba827/Sequences/LS_PortalChange_*`.

### 3. Bathhouse water uses the same bus, different sequence

`Vol_EnableBathHouseSpawners` `OnEnter` → Trigger2 `Trigger` → Cinematic Sequence Device2 `PlayFunction` (`LS_CombatOn`). Score Manager `On Score Output` → Device3 `LS_CombatOff`. Docs’ “water material gradually changes colors” is this cine, not a Mutator Volume.

## Don't

- Don't expect a Storm Controller. Portal Storm is a **feature-set dropdown** on storm devices; this island places **0** storm devices.
- Don't treat `Content/LS_PortalChange` as the live shot. Devices bind `_Golden` / `_Demonic`.
- Don't hide the HUD for this swap. HUD hide is the **plaza** cutscene controller (`priority=VeryHigh`, `showHUD=HideElement`).
