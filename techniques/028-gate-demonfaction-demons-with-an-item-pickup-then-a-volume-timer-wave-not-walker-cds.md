---
id: technique-028
kind: technique
---

# Gate DemonFaction demons with an item pickup, then a volume→timer wave — not Walker CDs

**Sources:** music_fantasy_sample
**Tutorial:** Working with MUSIC_THEME Islands — Gameplay + Demon NPCs + Bathhouse Reactive Environment — **External**. Bindings **Observed**.

Contrast (do not duplicate): Walker volume/checkpoint horde on volume-checkpoint-gated-horde.md. Guard doorway swap on guard-patrol-volume-swap.md. This island has **zero** `.verse` files and **zero** Walker / Guard Character Definitions.

## Problem

A theme starter needs two combat beats (street, then bathhouse) without paying for every demon at game start, and without Verse. The first wave should wait on a named item; the second should wait on entering a building.

## Pattern

### 1. Identity lives on DemonFaction Creature devices, not a Character Definition

Class: `Device_DemonFaction_CreatureSpawner` (`/CRD_DemonFaction_FiendOverrides/…`) — docs: Creature Spawner redesign. Sample live (**Observed**):

| Label | Folder | `bright Coil Creature Type` | `number of Creatures` | `enabled at Game Start` |
|---|---|---|---|---|
| DemonSpawner | Streets | Small Demon | 4 | false |
| DemonSpawner2 | Streets | Demon | 4 | false |
| DemonSpawner (bath) | Bathhouse | (same pattern) | — | false |

Base `creature Type` still shows leftover **Cube Random** / Manager **Fiend**. The theme dropdown is `bright Coil Creature Type`. Docs list **11** types; live samples use Small Demon + Demon.

Demon Manager (`Device_DemonFaction_CreatureManager`): `health=1`, `damageToPlayer=0`, `movementSpeed≈0.33`, `enabled on Game Start=true`, `affected Creatures=NewPawnsOnly`. Tutorial enemies.

**8** spawners + **6** managers total. **0** Demon Placer placed (docs still list the device as available in the feature set).

### 2. Streets wake on sword pickup

`HeroA's Saingeom Sword` `On Item Picked Up` → Enable street spawners. `Vol_DisableStreetsSpawners` `OnEnter` → Disable. Docs: alley hire + HeroA’s Empowered Sword triggers demons.

### 3. Bathhouse wakes on Volume → Timer (3s) → Enable

Docs: a Volume enables a Timer and a Trigger on enter.

Live (**Observed**), same volume `Vol_EnableBathHouseSpawners`:

| Target | Function |
|---|---|
| Timer Device | `Start` |
| Trigger2 | `Trigger` → Cinematic Sequence Device2 `PlayFunction` (`LS_CombatOn`) |

Timer: `duration=3`, `startAtGameStart=false`, `successOnTimerEnd=true`, `showOnHUD=false`, `visibleDuringGame=Hidden`, `canInteract=Yes`. `OnSuccess` → Enable the **five** bathhouse spawners.

Disable hose: `Vol_DisableBathhouseSpawners` (+ a second volume) `OnEnter` → Disable those spawners.

### 4. Score is a cine cue, not a win table

Score Manager: `score Value=1`, `maximum Score=12`, five `On a Creature is Eliminated` → Increment (bathhouse spawners). `On Score Output` → `LS_CombatOff`. Island `eliminationsToEnd=1` is unrelated BR leftover, not this bus.

## Don't

- Don't Enable-at-start the street or bathhouse spawners. Both waves are `enabled at Game Start=false`.
- Don't copy Walker `WalkerDefinition` or IceBase Guard CDs. Hireable hero squad is the Character Spawner, not these Creature devices — hireable-theme-character-spawner.md.
- Don't treat Demon Placer as placed. Feature set ≠ this Outliner.
