---
id: technique-029
kind: technique
---

# Jam a Crowd Volume + gold aura from Mutator `On Player Starts Jamming`, not a Verse emote subscribe

**Sources:** music_fantasy_sample
**Tutorial:** Working with MUSIC_THEME Islands — Victory Jam Session + Gold and Demon VFX Auras — **External**. Bindings **Observed**.

This island has **zero** `.verse` files. Contrast Social dynamic-team emotes (Island Settings, not a Mutator jam event).

## Problem

A plaza stage should grow a concert crowd and swap the player from a demon aura to a gold aura when they jam — without Verse and without leaving Crowd Volumes Always-on.

## Pattern

### 1. Crowd Volumes start off; Mutator is the bus

**13** `BP_CrowdVolume_Device`. Sample: `crowdDensity=52`, `enabledDuringPhase=None`.

One Mutator Zone on `SodaPopStage`: `bAffectsPlayers/Creatures/Guards=true`, `enabledOnPhase=Gameplay Only`, `zoneVisibleDuringGame=false`. Bindings live on the Crowd Volume (**Observed**):

| Mutator event | Crowd |
|---|---|
| `On Player Starts Jamming` | Enable |
| `On Player Stops Jamming` | Disable |

Island `enableJam=Default`. Authored stage: `BP_SodaStage` / `BP_Stage` + `SM_SodaPop_Stage`. **14** Chair devices at the plaza.

### 2. Two VFX Powerups: gold on jam, demon as the default

Both `visualEffect=Custom Effect`, `spawnOnMinigameStart=true`, `infiniteEffectDuration=true`.

| Label | `custom Effect` | Jam Start | Jam Stop |
|---|---|---|---|
| VFX Powerup2 | `/DemonFactionExtraAssets/PlayerAuras/NS_DemonFaction_GoldAura` | Pickup | Clear |
| VFX Powerup | `/DemonFactionExtraAssets/PlayerAuras/NS_DemonFaction_DemonAura` | **Clear** | (no stop bind on dump) |

Docs: gold aura when you begin a jam session. Live also **clears** the demon aura on jam start so the two do not stack. Docs: assign player auras via the device (`Spawn` / Pickup), not by dropping the player Niagara in the viewport (fails validation).

Mutator `ListEventBindings` on the mutator itself is **[]** — incoming binds sit on Crowd + Powerup, same OFPA trap as IceBase volumes.

## Don't

- Don't Enable Crowd Volumes at game start. `enabledDuringPhase=None` until jam.
- Don't drag player-aura Niagara onto the floor. Use VFX Powerup Custom Effect.
- Don't look for Verse `EmoteStartedEvent`. The event name is **Starts Jamming**.
- Don't copy this Mutator jam → Crowd stack onto BRICK Music Concert. That island is Switch Enable/Disable on Patchwork sequencers (`enableJam=Off`). See patchwork-switch-stems.md.
