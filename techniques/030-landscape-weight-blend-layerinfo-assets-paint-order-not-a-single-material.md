---
id: technique-030
kind: technique
---

# Landscape weight-blend: LayerInfo assets + paint order, not a single material

**Sources:** landscape_template
**Tutorial:** Changing Terrain with Landscape Mode — Paint stage.

## Problem

A Landscape is not “one material on a heightmap.” Players see Grass vs Path vs Sand vs Stone as **weightmaps**. Painting an upper layer hides a lower one. Roads that must keep editing need **Splines**, not a Ramp sculpt.

## Pattern

Authored set (copy these, don’t invent names):

```
Materials/Landscape_WeightBlends/LI_Grass
                              /LI_Path
                              /LI_Sand
                              /LI_Stone
M_Landscape_WeightDemo + MI_Landscape_WeightDemo_A
```

1. Create Landscape (Manage → New). Assign the weight-demo material. UEFN cap: **2048×2048 vertices** (or equivalent: 4096×1024).
2. Paint **target layer** in Paint mode. Shift-paint removes weight. Layer **order = visibility**.
3. Foliage: FoliageType assets (`SM_ShrubDemo_01_FoliageType`, tree variants) → `InstancedFoliageActor` (15 on this island).
4. Roads: Manage → **Splines** (`Ctrl+Click`). Ramp sculpt is one-shot; tutorial says use Splines to keep editing.
5. Holes: **do not** Visibility-brush caves — tutorial: players despawn, vehicles teleport. This island uses `SM_Tunnel`.

World Partition: even a tiny demo ships 7 `LandscapeStreamingProxy` + HLOD. Expected.

## Don't

- Don’t paint a second Landscape material as a mesh override — use LayerInfo weight.
- Don’t Visibility-brush a cave and ship it.
- Don’t treat Landscape Mode docs as this island’s walkthrough — it’s the editor manual; the island is the *result*.
- Don’t spawn trees as `FortStaticMeshActor` when a FoliageType exists.
