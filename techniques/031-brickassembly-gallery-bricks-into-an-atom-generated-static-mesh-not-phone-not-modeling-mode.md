---
id: technique-031
kind: technique
---

# BrickAssembly gallery bricks into an Atom-generated static mesh (not Phone, not Modeling Mode)

**Sources:** brick_editor_sample
**Tutorial:** BRICK Brick Editor Template Challenges 1–3 + Working with the BRICK Brick Editor — **External**. Meshes / components **Observed**.

Do not duplicate phone-copy-paste-prefab-quickbar.md (Create-mode Phone on catalog pieces). Do not duplicate walkable-modeling-curriculum.md (Modeling Mode halls). Do not duplicate cubegrid-then-color-grids.md (**0** CubeGrid). Do not duplicate playset_sample (`Parent_Building*` kit). Assembly stays on action_adventure_sample — Brick Editor meshes **cannot** feed the Assembly device.

## Problem

A BRICK island needs custom props that look like real bricks (studs, tubes, authentic colors) without leaving thousands of individual brick actors in the world, and without authoring a new brick mesh in Modeling Mode.

## Pattern

**1. Place gallery bricks in BRICK Brick Editor mode** (`Shift+6`). Live duck kit is native `Actor`s labeled `PLATE 2X4` / `BRICK 2X10` / … with:

- `AtomCollisionComponent`
- `AtomPrimitiveStaticMeshComponent`

Mount: `/BRICKBasicBrickGallery_Assets/Primitives/VX*`. These are the unclassified OFPA actors. `bIsSpatiallyLoaded=true`. They are **not** `StaticMeshActor`.

**2. BrickAssembly the selection.** Docs: Snap / BrickAssembly / Separate. BrickAssembly writes a static mesh into the project Content Browser (`Examples/` on this island). Sampled authored meshes (**Observed** `get_asset_tags`):

| Tag | Typical value |
|---|---|
| `AtomGeneratedAsset` | **True** |
| `StaticMaterials` | `MI_Primitive_Standard_VC_LUT_FakeBevel` + `_Underside` + `_CP` (`/BRICKAtomAssets`) |
| `CollisionComplexity` | `CTF_UseSimpleAsComplex` |
| `NaniteEnabled` | True; 6 LODs |
| `UVChannels` | 2 |

Do not swap those LUT materials (theme Rules / validator). Do not scale the mesh (studs/tubes). Docs: recolor requires Separate → re-BrickAssembly. Collision/pivot edits are allowed but **not preserved** on re-BrickAssembly.

**3. Reuse the SM.** Duplicates instance the mesh (memory win vs un-BrickAssembled brick instances). Challenge 3: Separate the example fence, isolate one bay, BrickAssembly, duplicate. Live `SM_Fence` is `AtomGeneratedAsset=True`. DemoDisplay C: *“The best way to get the most BRICK bricks into your games is to BrickAssembly your models and reuse them.”*

Teaching pods 1–2 (`1. Ducks` 26 actors / `2. Hearts` 2) are un-BrickAssembled practice kits plus one example `FortStaticMeshActor`. Pod 3 is the BrickAssembly lesson.

## Don't

- Don't Phone-copy catalog `BrickSimple_*` / `SM_FORT_*` and call it Brick Editor — that is Home Builder.
- Don't Extrude/Bevel a new brick in Modeling Mode halls — gallery bricks are the palette; BrickAssembly is the bake.
- Don't feed BrickAssembled meshes to an Assembly device (static mesh ≠ geometry collection).
- Don't BrickAssembly a walkable interior as one mesh (docs strip occluded triangles).
- Don't “fix” `AtomGeneratedAsset` or the three LUT slots when citing shipped examples.
