---
id: technique-032
kind: technique
---

# BrickAssembly wall / pillar / doorway / roof modules sized to the Fortnite building grid

**Sources:** brick_editor_sample
**Tutorial:** BRICK Brick Editor Template Challenge #4 — **External**. Folders **Observed**.

Do not duplicate playset_sample / modular-kitbash-blueprints.md (`Parent_Building*` harvest/edit-mode kit). Do not duplicate BRICK-fortnite-mix-no-scale.md (catalog Fortnite houses + `BrickSimple_*`, no BrickAssembly). Greyboxing CubeGrid stays on cubegrid-then-color-grids.md.

## Problem

A custom BRICK building should snap together like Fortnite modular walls without authoring a playset Blueprint hierarchy, and without placing every brick by hand on the finished structure.

## Pattern

Docs: *“similar to the building actor system in UEFN. Build your own reusable modules out of BRICK bricks, then BrickAssembly them and fit them together.”* Live teaching pod 4 + two worked buildings:

```
4. Building Systems/          Den_* floor / wall / pillar / doorway  (16)
4. Building Systems/Castle/   CastlePillar / CastleWindow / CastleDoorway / SpireRoof  (34)
4. Building Systems/Temple/   AsWall / AsPillar / AsDoorway / AsPeakRoof / AsRoofMid  (18)
```

Authored `Content/Examples/SM_{Den,Castle,As,Med}*`. DemoDisplay C3: *“build your own wall, pillar, doorway, and roof/floor modules… designed to fit the existing BRICK modules within the Fortnite building system. The temple and castle next to this sign are also made from building modules!”*

Workflow (**External** + **Observed**):

1. Duplicate an existing module (or start from gallery bricks).
2. Separate → edit bricks → BrickAssembly.
3. Place copies of the SM like tiles. Do not scale.

Working-with FAQ: BrickAssembled meshes **can** be used as Fortnite Building Prop / Building Static Mesh actors. This island leaves them as `StaticMeshActor` / `FortStaticMeshActor` exhibits, not a `Parent_Building*` kit. Sampled `bCanBeDamaged=false` on `SM_Fence`.

## Don't

- Don't inherit harvest / edit-mode / nav-area from fort_playset — these are BrickAssembled SMs, not a playset scaffold.
- Don't mix Fortnite `SM_FORT_*` house tiles here and call it Brick Editor — Home Builder already taught catalog mix.
- Don't BrickAssembly the whole castle as one mesh (memory + collision + interior-strip). Modules are the point.
