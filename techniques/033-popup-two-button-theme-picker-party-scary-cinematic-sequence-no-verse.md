---
id: technique-033
kind: technique
---

# Popup two-button theme picker → Party / Scary Cinematic Sequence (no Verse)

**Sources:** brick_editor_sample
**Tutorial:** BRICK Brick Editor Template — Cinematic Podium — **External**. Bindings **Observed** (`_raw/brick_editor_sample/event-graph.md`).

Do not duplicate cartoon-comic-unbound-ppv.md (two Buttons each Play a look cine). Do not duplicate trigger-cinematic-quake.md (Trigger → one quake sequence). VFX Creator vs Spawner surfaces stay on vfx-surface-split.md — this island only uses them as cine dressing.

## Problem

A player should drop a BrickAssembled build on a podium and pick a presentation (party vs scary) without Verse, Class Designer, or a Switch per look.

## Pattern

Outliner `Cinematics/`:

1. **Place the build** on `CinematicPodium` (`FortStaticMeshActor`, authored `SM_CinematicPodium`). Boards: “YOUR BUILD HERE!” + face the red arrow.
2. **PLAY** `PlayCinematicButton` (`interaction Text` = `Play Cinematic`, `timesCanTrigger=0`, `enabledAtGameStart=true`).
3. **Popup** `Show` (`On Interact` is stored on the dialog, not the button):

| Field | Live |
|---|---|
| title | Choose Theme |
| description | Starts a display video of your build |
| button1Text | Party |
| button2Text | Scary |
| button3Text | Cancel |
| autoDisplay | None |
| enabledDuringPhase | Always |

4. **Button 1** → `Party Sequence` `PlayFunction` (`PartySequence.uasset`). Folder `Cinematics/PartyCinematic`: concert stages, balloons, Mascot/Frog/Snail party animals (`/Musical_Walk_Assets`), VFX Creator (Bubble / Hearts / Confetti), `PP_ActionLines` + `PP_HappyPlace`.
5. **Button 2** → `Scary Sequence` `PlayFunction` (`ScarySequence.uasset`). Folder `Cinematics/ScaryCinematic`: VFX Spawner (SpookyGhosts / AreaOfFog / Bats), `PP_Vignette`, `Scary Day Sequence Device` (priority **2000**, FixedTime **21.5**, `sunlightIntensity=11.76`, `use Volume=false`, `enableDuringPhase=None`), `/ENV_PACK_UnplayableTestMap` skull box.

Both cines: `bAutoPlay=false`, `bLoopPlayback=false`, `enabledOnPhase=Always`, `showHUDDuringCinematic=Show`, `disablePlayerInput=No`. Play rate ~1.

**HUD around the shot:** controller starts `enabledDuringPhase=None`. Popup Party/Scary **Enable** it; both cines `OnStopped` **Disable**. Overrides **null**. HideElement HUD/minimap/build/inventory; health/shields DoNotOverride. ExtraAssets stays **four**.

**PP start-off:** each Post Process `On Player Spawned` → `DisableForAll`, `enabledDuringPhase=None`. The cine turns them on (Sequencer tracks **unreachable**).

VFX: Party uses **Creator** sprites (Bubble / Heart / Triangle-as-Confetti); Scary uses **Spawner** presets. `Fireworks VFX` live `visual_Effect=Bubbles` (label drift). Most `enabled On Phase` / `enabledDuringPhase` = **None** until the sequence.

Playground DemoDisplay C6 tells you to involve Verse / sequencer — this podium is the shipped sequencer half. **0** Verse.

## Don't

- Don't wire one Button per look (City Visual Styles). One button opens a popup with Party / Scary / Cancel.
- Don't auto-play the sequence. `showHUDDuringCinematic=Show`; input stays on. HUD Controller is Enable/Disable around the shot, not ExtraAssets widgets.
- Don't copy ExtraAssets HUD here. Overrides **null**. HideElement HUD/minimap/build/inventory.
- Don't treat Musical Walk animals as Patchwork — they are cine dressing (`enableJam=Default` on Experience Settings V2).
