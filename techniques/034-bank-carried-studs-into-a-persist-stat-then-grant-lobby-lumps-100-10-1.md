---
id: technique-034
kind: technique
---

# Bank carried studs into a persist Stat, then grant lobby lumps 100/10/1

**Sources:** brick_extraction_sample
**Snippets:** `../snippets/verse/BRICK_goldstuds_manager.verse`
**Official tutorial:** BRICK PvP Extraction Template — Gold Studs Management. Docs StatBankedGoldStuds vs shipped `Stat_BankedGoldStuds`.

World pickup bursts stay on stud-itemspawner-teleport-pool.md. Tagged Collectible → Granter stays on tagged-collectible-stud-granter.md (**0** BRICK Collectibles here; `BRICK_collectibe_stud_granter.verse` ships unused). Sequential GrantItem+Await loadouts stay on sequential-weapon-grant.md. Persistable quest logs stay on persistable-quest-line-verse-dialog.md.

## Problem

Players should drop carried studs on death, keep a **banked** total across extracts / sessions, and spend that bank in the lobby without a second currency item type.

## Pattern (shipped)

Three Stat Creators + one Conditional Button as the inventory counter:

| Device | Live | Role |
|---|---|---|
| Stat_GoldStuds | persist **false**, HUD **Non-Zero**, widget `WBP_PlayerStuds` | carried count in the map |
| Stat_BankedGoldStuds | persist **true**, HUD **No**, auto save/load | bank (never shown) |
| Stat_BankedGoldStuds_VisualsOnly | persist **false**, HUD **Yes**, widget `WBP_BankedStuds` | lobby HUD mirror |

`BRICK_goldstuds_manager` (label **BRICKExtraction-goldstuds management device**, unfoldered) `GetItemCount` on `ConditionalButton_GoldStuds_CheckCount`, then:

```verse
ItemAmount : int = ConditionalButton_GoldStuds_CheckCount.GetItemCount(ValidAgent, 0)
if (ValidExtractionPlayer.InLobby?):
    Stat_GoldStuds.SetValue[ValidAgent, 0]   # Non-Zero HUD hides carried
    Stat_BankedGoldStuds_VisualsOnly.SetValue[ValidAgent, ItemAmount]
else:
    Stat_GoldStuds.SetValue[ValidAgent, ItemAmount]
```

Deposit: add `ItemAmount` to persist bank → `ItemRemover_DepositGoldStuds.Remove` → Sleep(2) cine/HUD/audio → `GrantLobbyStuds`. Grant is Floor/Mod **100 then 10 then 1** into three Item Granters (no `ItemGrantedEvent.Await`). `InLobby` lives on `BRICK_fortplayer_extraction`, flipped by Trigger_OnEnterLobby / OnLobbyExit. Elim Manager **On Eliminated** → Reset Stat + ItemRemover_GoldStuds_OnDeath.

Live: goldstuds schema lists the three Stats, Conditional Button, two Removers, deposit cine/HUD/audio, X1/X10/X100 granters. Folder walk missed this VerseDevice (`folderPath` None); `find_actors` name `Gold` found it.

## Don't

- Don't persist the HUD Stat — persist `Stat_BankedGoldStuds` (HUD No) and mirror with VisualsOnly.
- Don't treat the demo `Button_Deposit_GoldStuds` as the live loop. Docs: deposit pod is **not used**; extract banks. Button is still wired to `Trigger_DepositGoldStuds`.
- Don't copy gunfight sequential Await — these GrantItem loops are lump-sum lobby fill.
- Don't expect BRICK Collectible Gold Studs. This island uses Item Spawners + chests.
