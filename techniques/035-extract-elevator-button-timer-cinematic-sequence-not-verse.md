---
id: technique-035
kind: technique
---

# Extract elevator: Button → Timer → Cinematic Sequence, not Verse

**Sources:** brick_extraction_sample
**Official tutorial:** BRICK PvP Extraction Template — Emergence & Extraction. Docs `Button_Extract_Elevator` vs shipped `Button_ExtractPoint_Elevator_GoingUp` / `_GoingDown`.

Gunfight arena rotation stays on persistable-arena-rotation.md. VFX Trigger→cine stays on trigger-cinematic-quake.md. Map Close/Long/Mid/Town arenas here are **set-dress only** (meshes, no rotation).

## Problem

An extract shaft needs a call delay, a fight window, a descent, then a lobby teleport — visible to everyone at that pad — without a Verse state machine.

## Pattern (shipped)

Details graph (one pad; copies `_1/_2/_3` for the other shafts):

1. Call button **On Interact** → GoingUp timer **Start** (duration **1**, name Extract Point Opening, `showOnHUD=false`) + siren + red light.
2. Timer **OnSuccess** → `CS_Extraction_TunnelOpen` **PlayFunction**.
3. CS **OnStopped** → AutoClose timer **Start** (duration **30**, name Extract Point Closing, HUD off) + enable the in-elevator GoingDown button.
4. AutoClose **OnSuccess** *or* GoingDown **On Interact** → GoingDown delay **Start** (duration **5**, also named Extract Point Closing) → `CS_Extraction_TunnelClose`.
5. Close **OnStopped** → MutatorZone_ExtractPoint_Tunnel **Disable**, Barrier **Enable**.
6. Volume **OnEnter** → Teleporter_Lobby_Enter **Teleport** → **On Teleported** → Trigger_OnEnterLobby → Health Powerup Pickup.

Billboard_GameIntro2: call the elevator, then use the button on the elevator to bank studs. DemoDisplay_ExtractPoint: AutoClose creates steal / share / sneak tension.

Eight stacked lobby teles + eight BRICK spawners (`useAsIslandStart=true`, Always). Emergence Point (265 actors) is the return shaft: Mutator Entering Zone → Fixed Point camera AddToPlayer + CS TunnelOpen_N on Teleported.

## Don't

- Don't put the 30s fight window in Verse Sleep — AutoClose is a Timer device (`showOnHUD=false`; name collides with the 5s descent timer).
- Don't rotate the four named arenas like Gunfight. They are BrickAssembly houses, not a persistable map list.
- Don't assume one extract pad. Live copies `_1/_2/_3` plus the demo pad.
