---
id: technique-036
kind: technique
---

# Gate lobby rooms with Barrier ignore-list from Stat level

**Sources:** brick_extraction_sample
**Snippet:** `../snippets/verse/BRICK_lobby_room_unlock_manager.verse`
**Official tutorial:** BRICK PvP Extraction Template — Lobby Unlock Zone Manager (58-line paste).

Sequential capture unlocks stay on sequential-capture-unlock.md. Weighted domination spawns stay on weighted-capture-spawns.md.

## Problem

Upgrade rooms in a shared lobby should open **per player** at a Stat level, without disabling the Barrier for everyone else.

## Pattern (shipped)

`BRICK_lobby_room_unlock_manager` (label **Lobby unlock zone manager**):

```verse
PlayerLevel.LevelChangedEvent.Subscribe(OnPlayerLevelChanged)

OnPlayerLevelChanged(Agent : ?agent, CurrentPlayerLevel : int) : void =
    if (ValidAgent := Agent?):
        for (UnlockZone : UnlockZones):
            if (CurrentPlayerLevel >= UnlockZone.UnlockLevel):
                UnlockZone.Barrier.AddToIgnoreList(ValidAgent)
            else:
                UnlockZone.Barrier.RemoveFromIgnoreList(ValidAgent)
```

`unlock_zone` is a `class()` with `@editable UnlockLevel : int = 5` and a `barrier_device`. Join path: Trigger_OnBRICKFortPlayerAdded → `GetLevel` → same ignore-list update.

Live UnlockZone_2: `enabledOnPhase=Gameplay Only`, `invisibleToIgnoredPlayers=true`. Barrier Device_1–8: `enabledOnPhase=None` (emergence-shaft copies driven by cine OnStopped Enable, not lobby gates). Billboard **Level Requirement: 2**. Schema confirms `playerLevel` / `unlockZones[]` / `trigger_OnBRICKFortPlayerAdded`. Subobject array values Unverified.

UnlockZone_2 also has Details **On Level Up → Disable** (global). Ignore-list is the per-player API; Disable is everyone.

Trigger_OnEnterLobby (Teleporter On Teleported) → Health Powerup_GrantOnLobbyEnter Pickup.

## Don't

- Don't `Disable` the Barrier to “unlock” one player — that opens it for the whole island.
- Don't trust the Verse default UnlockLevel **5** as the shipped gate; the intro board says **2**. Read the placed `unlock_zone` array (GetDeviceProperties subobjects often fail).
- Don't confuse these with capture-area sequential unlocks (Team Conquest) or gunfight arena rotation.
