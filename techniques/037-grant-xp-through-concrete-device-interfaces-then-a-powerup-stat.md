---
id: technique-037
kind: technique
---

# Grant XP through `<concrete>` device interfaces, then a Powerup Stat

**Sources:** brick_extraction_sample
**Snippet:** `../snippets/verse/BRICK_xp_level_manager.verse`
**Official tutorial:** BRICK PvP Extraction Template — Player’s Level System (127-line paste).

Persistable quest XP stays on persistable-quest-line-verse-dialog.md. Tracker→End Game stays on collectible-tracker-endgame.md.

## Problem

Several Creative event families (tracker, elim, extract, chest grant) should feed one Stat level curve without one mega device listing every Trigger.

## Pattern (shipped)

`BRICK_xp_level_manager` (label **LevelManager**) takes four `@editable` `<concrete>` structs. OnBegin only **subscribes**; each struct’s `GrantXP` fires its own `Trigger_GrantXP`, which Details-wires to a Powerup Stat Pickup:

| Struct | Subscribe | Grant |
|---|---|---|
| `xp_level_manager_trackers_fndevice_interface` | `Trigger_OnEventProgress.TriggeredEvent` | `Trigger_GrantXP.Trigger` |
| `xp_level_manager_elimination_fndevice_interface` | EliminatedEvent + EliminationEvent | two GrantXP triggers (victim vs killer) |
| `xp_level_manager_extraction_fndevice_interface` | `OnExtractedTrigger` | Grant only if `Stat_GoldStuds.GetValue > 0` |
| `xp_level_manager_loot_chest_fndevice_interface` | `ItemGranter.ItemGrantedEvent` | `Trigger_GrantXP.Trigger` |

Live Powerups: StatPowerup_XP_Reward_Small / Medium / Large. Stat_PlayerLevel: persist **true**, HUD Yes, widget `WBP_PlayerLevelStat`, `maxLevel=100`, `maxValue=3`, `perLevelPointsMultiplier=1.5`. On Level Up → HUD_LevelUp Show.

Schema: `trackerEvents[]`, `eliminationEvents`, `extractionEvent`, `lootChestEvents[]`. Bindings `[]`.

## Don't

- Don't put Stat.SetValue in the manager — the Powerup Stat is the grant.
- Don't award extract XP with empty pockets (`NumberOfStuds > 0` is the gate).
- Don't treat this as a persistable quest log. The Stat is the level; fortplayer UUID persist is a different wrapper.
