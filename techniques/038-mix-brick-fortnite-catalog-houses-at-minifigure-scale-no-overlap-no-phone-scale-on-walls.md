---
id: technique-038
kind: technique
---

# Mix BRICK + Fortnite catalog houses at Minifigure scale — no overlap, no Phone scale on walls

**Sources:** home_builder_sample
**Tutorial:** Home Builder — Customizing — **External**. Pieces **Observed**.

Do not duplicate homestead-claim-spawn.md (Verse pad claim — **0** Verse here; all 8 BRICK spawners `useAsIslandStart=true`). Do not duplicate playset_sample (authored buildable kit — **0** PlaysetRoot; sampled walls `bIsPlayerBuildable=false`). Greyboxing CubeGrid stays on cubegrid-then-color-grids.md.

## Problem

A Minifigure neighborhood should look like tiny Fortnite houses mixed with BRICK brick walls and NEIGHBORHOOD props, without teaching player-build or a custom playset.

## Pattern

Catalog wrap (**Observed** OFPA + live finds):

| Family | Live / OFPA |
|---|---|
| `SM_FORT_*` floors/roofs | 91 BasicTile + roofs; find `"SM_FORT"` **117** |
| `BrickSimple_*` walls | 48 AAA SolidWall; find `"Brick"` **112** |
| `/NEIGHBORHOOD_Assets` `BP_SUB_*` | furniture, roofs, sidewalks, interior lights |
| Athena / Apollo props | plants, kitchens, cars, doorbells ×3 |

Sampled `BrickSimple_AAA_SolidWall` and `SM_FORT_Floors_Generic_BasicTile`: `buildingType` Wall/Floor, `resourceType=Wood`, **`bIsPlayerBuildable=false`**, `bPlayerPlaced=false`, `bCanBeDamaged=true`. They are **placed catalog pieces**, not an authored `Parent_Building*` kit.

Docs constraints (**External**):

- BRICK assets cannot overlap or scale. Grid snap inserts a pillar gap; overlap crumbles.
- Fortnite walls / floors / roofs / stairs also cannot scale on a BRICK island (0.5 UU grid). Fortnite **props** can still scale on the Phone.
- Mix BRICK + Fortnite, or BRICK-only. Billboard6: “Wait a minute! This appears to be Fortnite.” Billboard11: mix categories. Billboard7: mailbox looks different.

Live `featureSet=BrickGameplay`. `enableJam=Off`. DSD is **2000-only** (`FixedTime` **7**, `sunlightIntensity=8`, `fogDensity=0.05`, `use Volume=false`) — not a lighting-curriculum rig.

## Don't

- Don't inherit `Parent_Building*` / harvest / edit-mode from fort_playset — these tiles are not player-buildable.
- Don't assign homestead pads with Verse — eight BRICK spawners are all island starts.
- Don't scale BRICK Elements in UEFN to “fix” the gap; moderation fails on overlap.
- Don't copy Obstacle Course `allowBuilding=None` as “building is off forever” — CREATE Phone still places; PLAY player-build is what is off.
- Don't BrickAssembly `BrickSimple_*` / `SM_FORT_*` into Atom meshes — BrickAssembly-atom-static-mesh.md is gallery bricks, not this catalog mix.
