---
id: technique-039
kind: technique
---

# Phone tool + Content tab prefab → Quick Bar on an empty lot (no Verse)

**Sources:** home_builder_sample
**Tutorial:** Home Builder — Playing Home Builder — **External**. Boards **Observed** (`_raw/home_builder_sample/billboard-texts.md`).

This island has **zero** `.verse` files and **zero** device event bindings. Greyboxing CubeGrid stays on cubegrid-then-color-grids.md (**0** CubeGrid here). Scene Graph prefab spawn stays on spawn-remove-prefabs.md. BrickAssembly / Brick Editor mode stays on BrickAssembly-atom-static-mesh.md.

## Problem

A BRICK neighborhood should teach Create-mode placement — copy, paste, delete, rotate, and drop a prefab from the Content tab — without a minigame loop.

## Pattern

**15** `Device_Billboard_V2` (`enabledDuringPhase=Create Only`). The syllabus is the boards plus the public Home Builder page, not a device graph.

| Board | Text (live) |
|---|---|
| Billboard12 | Welcome to the neighborhood! |
| Billboard8 | Activate the PHONE tool and the Controls will appear to the left! |
| Billboard3 | Point your Phone to COPY and PASTE new objects. |
| Billboard1 | Yuck, trash! Use DELETE to clean it up. |
| Billboard10 | Rotate pieces to help build out the structure of your home. |
| Billboard2 | Open the Content Menu to select from many different building pieces to add to the Quick Bar. |
| Billboard9 | Remember you can also Open a Prefab or Gallery to select from any of the individual pieces. |
| Billboard5 | Go to the empty plot to build your home, then enjoy your island in PLAY mode! |
| Billboard0 | Build Your Dream Home Today! |

Docs: empty lot → Content tab → prefab → Quick Bar; Create-mode hotkey menu for rotate / snap / place. Phone copy/paste/delete is the same tool.

`allowBuilding=None` (**Observed** Island Settings). Placement is the **Phone**, not Fortnite player-build. `bInfiniteResources` / `bInfiniteBuildingResources` / `bInfiniteWorldResources` **true**.

## Don't

- Don't wire a Tracker / End Game — docs: not a mini-game.
- Don't expect `allowBuilding` On — that is player-build, not Phone.
- Don't greybox with CubeGrid or `BP_Primitive_*` color cubes (Obstacle Course) when the syllabus is Content/Prefab/Gallery.
- Don't treat doorbells as devices; they are catalog props.
- Don't enter BRICK Brick Editor (`Shift+6`) / BrickAssembly catalog houses here — that is brick_editor_sample.
