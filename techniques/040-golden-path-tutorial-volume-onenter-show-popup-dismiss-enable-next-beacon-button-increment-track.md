---
id: technique-040
kind: technique
---

# Golden-path tutorial: Volume `OnEnter` Show popup, dismiss Enable next Beacon, button Increment Tracker

**Sources:** music_concert_sample
**Tutorial:** Music Concert — Playing Music Concert — **External**. Bindings **Observed**.

Combat arena trackers stay on tracker-gated-encounter.md (Trigger Enable + spawner elim Increment). This island is a **visit trail**, not a fight.

This island has **zero** `.verse` files.

## Problem

A desert festival should lead a player from animal to animal with a numbered beacon, a named popup, and a HUD visit counter — without Verse and without locking them to a rail.

## Pattern

### 1. Invisible volumes fire the dialog

**6** `Device_Volume` (`visibleInGame=false`, player events on). Each `OnEnter` → Popup `Show`; `OnExit` → Popup `Disable`.

Live titles (**Observed**):

| Popup | Title |
|---|---|
| Device1 | Welcome to the Music Concert! |
| Device3 | DJ Mascot |
| Device2 | Piano Rabbit |
| Device4 | Organ Frog |
| Device0 | Vibe Turtle |

Docs also describe Beacon 6 (backstage cave). Live **5** `Device_Beacon_V2` (`enabledOnPhase=None` until a dismiss Enable).

### 2. Dismiss is the next-beacon bus

Incoming binds sit on the **Beacon** (`ListEventBindings` on volumes is `[]` — same OFPA trap as IceBase / music-fantasy). Sample:

| Beacon | Enable | Disable |
|---|---|---|
| Beacon4 | Welcome `On Dismissed` | DJ Mascot `On Dismissed` |
| Beacon2 | Vibe Turtle `On Dismissed` | Volume3 `OnEnter` |

The player can wander; beacons only advertise the next station.

### 3. Tracker counts dialog OK, not native elims

`Tracker0`: `targetValue=4`, `assignOnGameStart=true`, `whenTargetIsReached=Complete Tracker`. `statToTrack=Eliminations` is leftover DNA — progress is **Popup `On Responding Any Button` → Increment Progress** on Mascot / Rabbit / Frog / Turtle (not Welcome). Same leftover-stat trap as Arcade (`statToTrack=None` there).

HUD Controller `showHUDMessages=ShowElement` so the tracker box is visible; minimap / health / shields `HideElement`. Widget overrides **null** (not ExtraAssets).

## Don't

- Don't put `ListEventBindings` on the Volume and conclude the trail is unwired. Incoming binds live on Popup + Beacon + Tracker.
- Don't trust `statToTrack=Eliminations` as the visit counter. The Increment events are the bus.
- Don't expect docs' six numbered beacons to match live Beacon0–4 labels.
- Don't copy Brick Editor’s one-shot Choose Theme popup (Party/Scary cine) as a visit trail — popup-theme-cinematic.md.
