---
id: technique-041
kind: technique
---

# Remix a Patchwork song with Switch Enable/Disable on sequencers, not Mutator jam → Crowd

**Sources:** music_concert_sample
**Tutorial:** Music Concert — Main Stage switches — **External**. Bindings **Observed**.

music-fantasy plaza jam stays on jam-mutator-crowd-aura.md (`On Player Starts Jamming` → Crowd Enable). This island places **0** Crowd Volumes and **0** Mutators; Island `enableJam=Off`.

Village Starter's Outliner `Patchwork` folder is a **room name**, not Fabric devices — village_starter_template.

This island has **zero** `.verse` files. Patchwork knobs are player-edited with the Patchwork Tool (BRICK Item Spawner + Team Settings grant), not Verse.

## Problem

A main-stage band should let players mute Drum / Bass / Melody / Chords in Play mode, with matching VFX, while one Music Manager keeps tempo and key for every Patchwork device.

## Pattern

### 1. One Music Manager owns tempo / key / mode

`Patchwork Music Manager0` (`Device_ModularGlobalMusicController`): `tempoValue=140`, `key=C`, `keyMode=Major`, `jamSettingsControl=FabricControlsJam`, `enabledOnPhase=Always`. Billboard: it controls tempo, key, and mode for all Patchwork devices.

Live Patchwork classes (**Observed**): 6 Note Sequencer (`Device_FabricSequencer`), 2 Drum Sequencer, 6 Note Progressor, 5 Instrument Player, 2 Drum Player, 8 Echo Effect, 1 Distortion, 1 Omega Synth (`Device_ModularBassSynth`). OFPA histogram also lists Wave Generator ×2 and Modular Emitter ×5 — live `find_actors` by those labels returned **0** (extractor class names vs device labels).

### 2. Four Switches are the stem bus

Billboards label Drum / Bass / Melody / Chords. Live: `enabledAtGameStart=true`, `initialState=true`, `visibleDuringGame=true`. Docs: switches work in Play, not Edit.

Incoming binds sit on the **sequencer / VFX**, not the Switch (`ListEventBindings` on Switch is `[]`):

| Switch | Sequencer Enable/Disable | VFX Creator Start/Stop |
|---|---|---|
| Switch0 | Drum Sequencer0 | VFX Creator1 |
| Switch1 | Note Sequencer1 | VFX Creator0 |
| Switch2 | Note Sequencer2 | VFX Creator3 |
| Switch3 | Note Sequencer5 | VFX Creator2 |

VFX Creator: `enabledDuringPhase=Always`, `startEffectsWhenEnabled=true`, `spawnMode=Continuous`. Surface split stays on vfx-surface-split.md — this island only uses Creator, no authored Niagara.

### 3. Patchwork Tool is an item, not a Verse give

Billboards: spawn the tool on the BRICK Item Spawner; also grant from Team Settings so it is in inventory at spawn. Live: 2 `Device_BRICKItemSpawner` (`enabled At Game Start=true`, `items Respawn=true`); Team Settings `grantCondition=Always`, `onGrantBehavior=Clear Inventory`, `bGrantItemsOnRespawn=false`. The granted-item array is not in the property dump (same GetObjectProperties trap as music-fantasy item class).

Audio Mixer ×2: billboard says they duck footsteps and ambience.

## Don't

- Don't copy music-fantasy Mutator `On Player Starts Jamming` onto this concert. `enableJam=Off`; the remix event is Switch `On Turned On/Off`.
- Don't look for Verse `EmoteStartedEvent` or Crowd Volume Enable. The graph is Fabric cables + Switch stems.
- Don't assume Wave Generator / Emitter labels match the OFPA histogram — confirm with `find_actors` after `load_actors`.
