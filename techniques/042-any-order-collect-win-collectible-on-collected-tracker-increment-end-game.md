---
id: technique-042
kind: technique
---

# Any-order collect win: Collectible `On Collected` → Tracker Increment → End Game

**Sources:** obstacle_course_sample
**Tutorial:** Obstacle Course — Collectibles — **External**. Bindings **Observed**.

Ordered Verse collect stays on sequential-collect-race.md (subscribe spawners, `SpawnItem` the next index, timer). Combat trackers stay on tracker-gated-encounter.md. Visit-trail Increment stays on beacon-volume-popup-trail.md. Speedway persistence/leaderboards stay on persistable-stats-and-sorted-leaderboard.md.

This island has **zero** `.verse` files, **0** Item Spawners, **0** Timers.

## Problem

An obby should end when the player has found **N** pickups in **any order**, with a HUD counter, without Verse and without forcing a path.

## Pattern

### 1. Place N Collectible Object devices

Live: **5** `Device_Collectible_Object_V2`. All `collectibleObject=Gold Coin`, `consumeIfCollectedBy=Anyone`, `displayScoreUpdateOnHUD=false`. Incoming `ListEventBindings` on the collectibles is `[]` — binds sit on the Tracker (same OFPA trap as IceBase / Music Concert).

Billboard: “Try changing the collectible type to something else... maybe an egg?” Companion Collectible Devices lists **Gold/Blue/Purple/Silver Stud** and points at **Action Adventure**, not this island.

### 2. Tracker is the win counter

`Tracker0`: `trackerTitle="Collect the Golden Coins"`, `targetValue=5`, `statToTrack=None`, `whenTargetIsReached=Complete Tracker`, `showOnHUD=Both`.

Each collectible `On Collected` → Tracker `Increment Progress`.

### 3. Tracker complete activates End Game

`Tracker0` `When Complete` → `End Game Device0` `Activate`. `postGameType=Classic`.

Popup (`autoDisplay=Game Start`, title **Welcome!**, body “There are 5 golden coins… Collect them all to win!”) is the syllabus, not the counter.

## Don't

- Don’t train Parkour `item_spawner_device` + Verse array order from this island.
- Don’t assume `statToTrack=Eliminations` leftover (Music Concert); here it is `None`.
- Don’t emit Gold Stud / Action Adventure unlock from this island’s coins.
- Don’t train tagged Collectible → Item Granter stud currency from this obby — that is tagged-collectible-stud-granter.md.
