---
id: technique-043
kind: technique
---

# Moving platforms without Verse: Switch ping-pong, damage-Trigger raise, button/coconut extras

**Sources:** obstacle_course_sample
**Tutorial:** Obstacle Course — Game Flow / Customizing — **External**. Bindings **Observed**. Graph: `_raw/obstacle_course_sample/event-graph.md`.

Scene Graph keyframed movement stays on simple-movement-component.md. Music Concert Switch Enable/Disable of Patchwork sequencers stays on patchwork-switch-stems.md — **0** Fabric devices here.

This island has **zero** `.verse` files. **18** `Device_PropMover_V2` (`movementMode=EFCLMM_Translation`, `enabledDuringPhase=Gameplay Only`). Incoming binds sit on the **mover**, not the Switch (`ListEventBindings` on Switch is the Toggle/Turn source only).

## Problem

An obby needs looping moving platforms, shoot-to-raise platforms, a coconut-gated mover, and a button that drains water — all device wiring.

## Pattern

### 1. Hidden Switch is the ping-pong bus

Switches `visibleDuringGame=false`. A Button `On Interact` or a Trigger `OnTriggered` **Toggle State**. Each Switch `On Turned On` / `On Turned Off` → Prop Mover `Advance` / `Reverse` (polarity varies per switch).

To loop without a player standing on a button: Prop Mover `On Movement Finish` → a helper Trigger `Trigger` → that Trigger `OnTriggered` Toggle the Switch. Helper triggers `triggeredByPlayer=false`, `triggeredByDamage=false`, `timesCanTrigger=1`, `triggerDelay=1`.

Live groups (**Observed**):

| Switch source | Movers |
|---|---|
| Switch0 ← Button2 | 16, 10 |
| Switch1 ← Trigger2 (Mover8 finish) | 13, 8 |
| Switch2 ← Trigger0 (Mover5 finish) | 5 |
| Switch3 ← Button1 | 6, 17, 2 |
| Switch4 ← Trigger3 Turn On / Trigger1 Turn Off | 0 (`speedInMeters=1.5`) |

Most movers `speedInMeters=1`. Billboard: change speed to go faster.

### 2. Damage Triggers raise platforms (shooting break)

Four Triggers `triggeredByDamage=true`, `triggeredByPlayer=false`, `triggerDelay=0`: **3, 4, 5, 6**. Billboard: a Trigger in front of each target reacts to damage.

`OnTriggered` → Prop Mover `Advance` (one-shot raise): Trigger4 → movers 9/4/3; Trigger5 → 1/14/15; Trigger6 → 11/12. Trigger3 **Turn On** Switch4 (starts Mover0). Docs: vending Burst Pulse Rifle, then shoot targets.

### 3. One-shot extras

- Conditional Button `On Activated` → Prop Mover7 `Advance` (`keyItemsRequired=1`, `consumeKeyItems=true`). Billboard: **Find a coconut!** (`/BRICKCoconut` mount).
- Button0 `On Interact` → Water2 `Start Vertical Emptying`.
- Teleporters ×4 `enabledDuringPhase=Always` (docs: cliffs). No Verse offset gather — contrast multiplayer-teleporter-offset.md.

## Don't

- Don’t write Verse `MoveTo` / keyframed components for these platforms.
- Don’t copy Music Concert Switch→sequencer Enable onto these Switches.
- Don’t assume every Trigger is a damage target — 0/1/2 are ping-pong helpers.
