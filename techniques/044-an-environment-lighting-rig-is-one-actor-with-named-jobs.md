---
id: technique-044
kind: technique
---

# An Environment Lighting Rig Is One Actor With Named Jobs

**Sources:** lighting_template

## The ratio

**1 Blueprint → 7 lighting systems → 3 named post-process jobs** (**Observed**).

character_environment had a character lighting rig for cinematics
(cinematic-sequences.md
— a **character_environment_sample** note, not this sample's).
That is a *subject* rig — key / fill / rim aimed at a person. This is an *environment*
rig — the sky, the sun, the fog, and the exposure that everything else sits under. They
are different jobs. Do not copy one when you need the other.

## What lives on `BP_Environment_Light_Rig`

**Observed** — live editor, `ActorTools.get_components` + `ObjectTools.get_properties`
on `/lighting_template` `Environment Light Rig`. Every value below is from the placed
instance, not the CDO.

| Component (name, not class) | Role | Worked values |
|---|---|---|
| `DirectionalLight` | Sun. `bAtmosphereSunLight` **ON**, otherwise the sky doesn't know this is the sun | intensity **10**, 6500 K (temperature unused), `lightSourceAngle` 1.5, `indirectLightingIntensity` **2** |
| `SkyAtmosphere` | Rayleigh / Mie aerial perspective | `rayleighScatteringScale` 0.0331, `mieAnisotropy` 0.8, `multiScatteringFactor` 1 |
| `SkyLight` | Indirect fill from a captured (or specified) cubemap | intensity **0.2**, `SLS_SpecifiedCubemap`, `bRealTimeCapture` **ON** |
| `ExponentialHeightFog` | Height fog + volumetric | density **0.03**, heightFalloff 0.2, volumetric **ON**, scatteringDistribution 0.2, albedo white, extinctionScale 1 |
| `LumenExposure` | Unbound post-process. **The live look.** | Histogram auto-exposure, min 0.075 / max 1.0, bias 0.5, GI+reflections = **Lumen** |
| `BasicExposure` | Unbound post-process. **Disabled fallback.** | Basic auto-exposure, min=max=1, bias **3**, whiteTemp **11000**, GI+reflections = **None** |
| `ColorGrading` | Unbound post-process. Always on, stacked | vignette 0.6, bloom 0.675, fringe 0.4, local-exposure detail 1.2 |

Plus editor-only helpers (`EditOnly_WireSphere`, `EditOnly_Icon`, `EditOnly_Arrows`)
and a `StaticMeshComponent0` that does not carry the look.

The four "atmosphere" components are the same ones a default Unreal level already has.
**The reusable bit is that they are one actor, and that exposure is three named
post-process components rather than one overloaded volume.**

## Name post-process components for their job

**This is the finding that was not obvious from the class list.** Filtering components
for `"PostProcess"` in the name finds nothing on this rig. The three post-process
components are named `LumenExposure`, `BasicExposure`, `ColorGrading`.

The pattern:

```
one unbound PostProcessComponent per concern
    ├── named for the concern, not the class
    ├── bUnbound = true          (whole-world, not a boxed volume)
    ├── priority = 0             (stack order is enable/disable, not priority)
    └── settings.bOverride_*     (only the fields this job owns are flagged)
```

**LumenExposure owns:** histogram auto-exposure, Lumen GI / reflections, Lumen diffuse
boost 1.25, skylight leaking 0.02, bloom 1.5, chromatic aberration 0.3.

**BasicExposure owns:** basic auto-exposure locked at brightness 1, exposure bias 3,
white balance 11000 K, GI / reflections **None**. And `bEnabled = false` in the shipped
instance — it is the no-Lumen path, sitting ready.

**ColorGrading owns:** vignette, bloom threshold, fringe, local-exposure detail.
Always enabled. Independent of which exposure path is live.

Toggling scalability is then **enable one component, disable the other** — not rewriting
a 200-field `FPostProcessSettings` blob, not swapping actors, not a data layer. A
`BP_LightingScalabilityManager` referenced from `/Game/TimeOfDay/TODM_Sequence/UEFN/`
sits next to the rig as the hook that would flip those enables; the placed instance
itself has no interesting properties.

Exposure values live in the nested `settings` struct, not on the component. Reading
`autoExposureMethod` on the component returns nothing; reading `settings` and then
walking `bOverride_*` flags is the way.

## Copy these three together on every local light

Every engine light sampled in lighting_template rooms 01–02:

| | |
|---|---|
| `intensityUnits` | **Candelas** |
| `bUseInverseSquaredFalloff` | **true** |
| `lightFalloffExponent` | **8** (only consulted if inverse-squared is off; they set it anyway) |

Plus `bUseTemperature = true` and a real Kelvin value rather than tinting `lightColor`.
Sampled temperatures in the engine-light rooms: **3800** (warm practicals), **5000**
(rect-light fill), **5500** (the volumetric hero spot), **6500** (cool / exterior).

The volumetric hero (Room 01 `SpotLight6`): intensity 3 cd, 5500 K, cone 25°/65°,
source radius 20, **`volumetricScatteringIntensity = 4`** against the rig's default of 1.
Four times the fog response is how a single practical reads in volumetric fog without
raising its intensity into blown-out.

## Engine lights and Creative light devices are different tools

lighting_template teaches them as a **progression**, not as equivalents.

```
Rooms 01–02     Engine.SpotLight / PointLight / RectLight
                full property surface: cone, source radius, IES-less shape,
                volumetric scattering, temperature

Rooms 03–Exterior
                Device_PointLight_V2  (Creative light *device*)
                still owns SpotLightComponent + PointLightComponent internally
                authorable surface is the device: dim / undim via gameplay messages
```

A Creative point-light device carries **both** a spot and a point component and leaves
the unused one at intensity 0. Do not "fix" that.

Use engine lights for an authored look you fully control. Use Creative devices when
players or Verse need to dim them. Mixing without a reason means you have two property
surfaces and two replication paths for the same job.

The `_NoLight` suffix on fixture meshes (`CP_Apollo_MountainBase_WallLight_04_NoLight`,
`CP_SharkIsland_Chandelier_NoLight`) is the matching convention: the **mesh looks like
a light, the sibling actor *is* the light**. Don't bake an emissive into a fixture and
also put a real light in it unless you want both.

## What not to copy from this sample

- The `Geometry/` Modelling-Mode intermediates (`Box_<hash>`, `Extrude_<hash>`). Demo
  mesh for the rooms, not an art kit.
- `Sky_Sphere` as a `FortStaticMeshActor` next to the rig. The sky look is
  `SkyAtmosphere` + `SkyLight`. The sphere is leftover.
- The 16 `ObjectRedirector`s at the content root (`LS_*`) left by moving sequences
  into `Sequences/`. If you adopt the four sun sequences, copy `Content/Sequences/LS_Sun_*`,
  not the root aliases.

## Capture backlog

**Cleared for the rig itself.** Open on the sample digest, not here: the Sequencer
tracks inside `LS_Sun_01`..`04` (which property of the directional light each phase
drives). The component set above is complete regardless of what those sequences animate.
