---
id: technique-045
kind: technique
---

# Swap a kyber crystal: `RemoveFromEntity` + `Sleep(0.1)` + `AddComponents`

**Sources:** energy_ability_sample
**Snippets:**
- `../snippets/verse/kyber_crystal.verse`
- `../snippets/verse/EnergyBlade_spawner_device.verse`
- `../snippets/verse/EnergyBlade_color_changer_device.verse`
- `../snippets/verse/menu_EnergyBlade_customization_device.verse`

## Problem

A EnergyBlade entity already ships a lore-matching `kyber_crystal_[color]_component`. You need a different blade color (or Arc Intensity / Stability) at runtime. Scene Graph cannot keep two kyber components of that family on one entity.

## The pattern (shipped)

1. Build the hilt entity from the `EnergyBlade` enum (`CreateEnergyBladeEntity` → `EnergyBladeHeroA_V1` … `EnergyBladeHeroC_V1`).
2. Read the existing crystal with `GetComponent[kyber_crystal_component]`.
3. Call `RemoveFromEntity()` on that component.
4. `Sleep(0.1)`.
5. Construct the new typed component (`kyber_crystal_red_component{ Entity := InEntity }`), set `ArcIntensity` / `Stability` **before** attach, then `AddComponents(array{ NewCrystal })`.

```verse
OldCrystal.RemoveFromEntity()
# Workaround for a race conditon when trying to replace a component.
# If you try to remove a component and then replace it with another
# sometimes the remove and add replicate in the wrong order.
Sleep(0.1)
NewCrystal := CreateKyberCrystalComponent[EnergyBladeEntity, KyberCrystal]
set NewCrystal.ArcIntensity = ArcIntensityOverride
set NewCrystal.Stability = StabilityOverride
EnergyBladeEntity.AddComponents(array { NewCrystal })
```

`AddComponents` / `RemoveFromEntity` are **no-rollback**, so they cannot live inside a `<transacts>` failure path. `CreateKyberCrystalComponent` is `<decides><transacts>` and **does not** call `AddComponents` for that reason. The spawner / color changer / workbench `ApplyLoadoutFor` all `spawn` a `<suspends>` function around the sleep.

Workbench extra: after `AddItemDistribute`, `Unequip()` + `Sleep(0.1)` + `Equip()` so the visual actor reloads. If only arcing/stability changed, it mutates the live `kyber_crystal_component` and skips a new grant.

## Don't

- Don't `AddComponents` a second kyber without removing the first — the client add fails if remove/add replicate out of order.
- Don't treat Arc Intensity / Stability as “max at 1 = more flicker.” Shipped boards: Arc **0 = none / 1 = max arcing**; Stability **1 = no flicker / 0 = max flicker**.
- Don't spawn custom Scene Graph sabers from a stock Item Spawner — docs and this island use Verse `AddItemDistribute` for custom entities. Stock holocrons on Item Spawners are a different path (weapon-vs-EnergyBlade-augment).
