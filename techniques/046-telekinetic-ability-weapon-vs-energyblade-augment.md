---
id: technique-046
kind: technique
---

# telekinetic ability weapon vs EnergyBlade augment

**Sources:** energy_ability_sample
**Snippets:**
- `../snippets/verse/Telekinetic_power_spawner_device.verse`
- `../snippets/verse/EnergyBlade_spawner_device.verse`
- `../snippets/verse/ui_config.verse`

## Problem

The same telekinetic ability exists as a Holocron you put on the quick bar **and** as an augment that only fires while a EnergyBlade is equipped. Mixing the two types, or mapping a designer enum to the wrong entity, grants the wrong item.

## Two entity families (shipped)

| Form | Verse type suffix | How players get it | Quick bar |
|---|---|---|---|
| Weapon / Holocron | `TelekineticAbility*_Weapon_V1` | Item Spawner (stock) or Verse `AddItemDistribute` | Yes |
| EnergyBlade augment | `TelekineticAbility*_EnergyBladeAugment_V1` (Training: `TelekineticAbilityTraining_Augment_V1`) | Verse only | No (default) |

Docs name the pair `TelekineticAbility[NAME]_Weapon` / `TelekineticAbility[NAME]_EnergyBladeAugment`. Snippets add `_V1`. Both also need `Telekinetic_power_item_component` + `Telekinetic_power_[name]_component` in Scene Graph.

**One augment at a time** — boards, dedicated chapter, and telekinetic ability Component Reference all say so.

Telekinetic Training is **not an item**. Training Grounds has no 4.11 Item Spawner; the placed `EnergyBlade_spawner_device` uses `powerOverride=TrainingAugment`.

Grant path is Itemization `inventory_component.AddItemDistribute`, then `EnergyBlade_item_component.Equip()` — not Marketplace `BuyOffer`. Entitlements live on entitlement-vs-offer.

## Enum aliases (train on snippets)

`Telekinetic_power` is one open enum used by **both** spawners. The `case` arms are not 1:1 with lore names:

| Enum | `Telekinetic_power_spawner_device` (mostly weapons) | `EnergyBlade_spawner_device` (always augments) |
|---|---|---|
| `Cloak` | `TelekineticAbilityTrick_Weapon_V1` | `TelekineticAbilityTrick_EnergyBladeAugment_V1` |
| `Decoy` | `TelekineticAbilityDistraction_Weapon_V1` | `TelekineticAbilityDistraction_EnergyBladeAugment_V1` |
| `Sense` | `TelekineticAbilityAwareness_Weapon_V1` | **`TelekineticAbilityHeal_EnergyBladeAugment_V1`** |
| `Zap` | `TelekineticAbilityThrow_Weapon_V1` | `TelekineticAbilityThrow_EnergyBladeAugment_V1` |
| `LightningAugment` / `ThrowAugment` / `TrainingAugment` / `PullAugment` / `PushAugment` | matching augment types | matching augment types |

Workbench `ui_config.GivePower` maps UI ids 200–209 **only** to augments (Push/Pull/Throw/Shield/Lightning/Heal/Awareness/SaberThrow/Distraction/Trick). It does not use the `Telekinetic_power` enum. Awareness is correct there (`TelekineticAbilityAwareness_EnergyBladeAugment_V1`). Do not copy `Sense => Heal` from the EnergyBlade spawner.

Placed spawners (**Observed**): three `EnergyBlade_spawner_device` (EnergyBladeHeroC + red crystal, arc 0, stability 1) with `LightningAugment` / `Shield` / `TrainingAugment`. `Telekinetic_power_spawner_device` is **not placed**. Color/stability/arcing/remover Verse devices are files only.

## Don't

- Don't put a custom Scene Graph Telekinetic-power entity on a stock Item Spawner. This island's 11 Item Spawners are **standard** holocrons + a VillainA saber (timer 10s). Custom grants are Verse.
- Don't parent an augment under the saber entity in Verse — shipped grant is a **second** `AddItemDistribute` of the augment entity (docs “add as child” is Scene Graph authoring, not the Verse path).
- Don't emit tutorial using-path `/Fortnite.com/Items/TelekineticAbility/SPACE_OPERA` or `/Fortnite.com/Items/EnergyBlade/SPACE_OPERA`. Shipped: `/Fortnite.com/Items/SPACE_OPERA` + `/Fortnite.com/Weapons/SPACE_OPERA`.
