---
id: technique-047
kind: technique
---

# EnergyBlade workbench: per-player session + UMG overlay + Itemization apply

**Sources:** energy_ability_sample
**Snippets:**
- `../snippets/verse/menu_EnergyBlade_customization_device.verse`
- `../snippets/verse/ui_config.verse`
- `../snippets/verse/item_utils.verse`

Clickable `button` + `button_slot` wrapping UMG is verse-clickable-button (now two sources). This note is the **session + apply** stack around it.

## Problem

Players need an in-island saber builder (hilt, crystal, Telekinetic augment, stability, arcing) without Marketplace offers, without stacking a second Fortnite HUD, and without two players sharing one UI state.

## The pattern (shipped)

**World:** `trigger_device` → enable a workbench `hud_controller_device` (`showHUD=HideElement`, no widget overrides) + `gameplay_camera_fixed_angle_device` (distance 375, pitch −15.76, yaw 130, FOV 50). Classic bindings also `OnTriggered` → camera `AddToPlayer`. Mutator volume: no building, no weapon fire, Gameplay Only.

**Verse device** `menu_EnergyBlade_customization_device`: `@editable` ShowTrigger / HUDController / FixedAngleCamera. `GetDeviceProperties` returns **subobject** `refPath`s (Social Deduction trap) — confirm wiring from the Trigger/HUD/Camera actors, not those paths.

**Session:** `[player]EnergyBlade_customization_session`. `Open()` `RemoveWidget` then `AddWidget(RootOverlayWidget, player_ui_slot{ InputMode := ui_input_mode.All })`, refresh UMG materials, `SetFocus`. Chrome is authored `UW_Overlay` / `UW_Header` / `UW_Tab` / `UW_Item` (Verse constructs at runtime — `get_referencers` on the widgets is `[]`).

**Tabs:** `TabID_Hilt=0`, `TabID_Ability=1`, **`TabID_Crystal=3` (2 skipped)**, `TabID_Stability=4`, `TabID_Arcing=5`. Hilt labels in `ui_config` are lore-safe (“Hope Hilt”, …) while `GiveEnergyBlade` still maps 0→`EnergyBladeHeroC_V1`.

**Debounce:** `UpdateToken` increments on slider/grid input; `Sleep` then apply only if `Token = UpdateToken`.

**Apply:** Itemization `AddItemDistribute` of the saber + augment (weapon-vs-EnergyBlade-augment); kyber swap (kyber-component-swap). Inventory helper: `Agent.FindDescendantComponents(inventory_component).First[]`.

Two copies in the Outliner (Workshop + Training Grounds `device2`) share the same Verse class.

Docs list `EnergyBlade_color_changer_device` / `stability` / `arcing` as workbench devices. Those files exist; **none are placed**. The live bench is the menu device only.

Requires Custom Items and Inventory (`.uefnproject` `experimental.inventory.bIsCustomInventoryAllowed` is **Unverified** as a flag flip; Island Settings `customInventoryConfiguration=/Itemization/BRStyle/...`, `maxItemSlots=5` is **Observed**).

## Don't

- Don't `BuyOffer` saber parts — this is not in-island transactions.
- Don't copy ExtraAssets HUD here. Island HUD Controller (folder `UI/`) overrides PlayerInfo to **`/ENV_SPACE_THEME_UI_Assets/PlayerHUD/UW_HUD_PlayerInfo_C`**. Workbench HUD Controllers only hide HUD. ExtraAssets coincidence stays three (Social / Mastery / Walker). IceBase in-project `Content/UI/PlayerHUD` is a different recipe.
- Don't Migrate the Verse files with the UI folder — docs: copy `EnergyBladeWorkbenchCore` (`item_utils.verse`, not docs `items_utils.verse`), `menu_EnergyBlade_customization_device.verse`, `ui_config.verse` in Explorer.
