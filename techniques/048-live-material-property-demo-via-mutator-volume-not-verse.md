---
id: technique-048
kind: technique
---

# Live material-property demo via Mutator Volume, not Verse

**Sources:** materials_template
**Tutorial:** Material Nodes and Settings — property defaults and node table; this island is the walkable lab, same convention as `modeling_template`'s numbered halls.

## Problem

Show a player 30 different material-property states (Blend Mode, Shading Model, Two Sided, node pins…) without writing Verse — each exhibit just needs to visibly change when the player enters or hits a switch.

## Pattern

Numbered Outliner = the syllabus (copy this convention for any authoring-tool showcase):

```
1 - Mat Properties/1_1_Blendmode … 1_10_RefractionDepthBias
    each exhibit: Layout/Camera + Layout/Switch + Device_MutatorVolume_V2
2 - Materials/1_1_BaseColor … 1_12_Refraction     (Main Material node pins)
3 - Instances/1_1_Material → 1_2_Instance → 1_3_InstanceOfInstance → 1_4_Groups
4 - Advanced/1_1_MasterMaterial … 1_4_VertexAnimation
```

**Zero Verse.** `Device_MutatorVolume_V2` (30 total, ~2–3 per hall-1 exhibit) does the swap — walking into the volume or hitting its paired `Device_Switch_V2` changes the live material variant on the preview mesh. `Device_CameraModes_Perspective` (13) frames each exhibit consistently regardless of room layout.

`BP_LumenExposureManager` (3) reused verbatim from the Day Sequence pattern — one per room where Lumen/Non-Lumen parity matters for the demo to read correctly at low scalability.

Instance chain (hall 3) is an actual **object hierarchy exhibit** — walk Material → Instance → Instance-of-Instance → Groups and see which parameters each level can still override.

## Don't

- Don’t reach for Verse to swap a material for a demo — Mutator Volume is the idiom.
- Don’t confuse hall 1 (mesh-authoring settings: Blend Mode, Two Sided, Domain) with hall 2 (Main Material node **pins**: BaseColor, Normal, WPO). Different lessons.
- Don’t cite “Device materials” (Barrier/Switch) as part of this island — that’s the separate Material Library catalog page, not a hall here.
