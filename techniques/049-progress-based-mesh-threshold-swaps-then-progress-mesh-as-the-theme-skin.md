---
id: technique-049
kind: technique
---

# Progress Based Mesh: threshold swaps, then progress mesh as the theme skin

**Sources:** minigame_mastery_template
**Tutorial:** Tracking Game Progression — **External**. Fill values **Observed** live.

## Problem

You want an in-world meter (jar filling, tree growing, piggy bank swelling) without Verse and without a HUD widget. Score Manager text is not the same as a mesh that *changes shape* at 25 / 50 / 75 / 100.

## Pattern

The Creative device is `Device_Fillable_Container` (Outliner: Progress Based Mesh). Primary knobs:

| Field | Tree (`Mesh Sequence_Tree`) | progress mesh (progress room) | progress mesh (throwables score) |
|---|---|---|---|
| `fillTarget` | 100 | 100 | 100 |
| `gameStartFillAmount` | **10** | 0 | 0 |
| `fillType` | **Instant** | ContinuousRate | **Instant** |
| `fillRate` | 25 | **25** | 10 |
| `drain Rate` | 5 | — | — |
| `thresholdMeshes` | `[0..3]` | ExtraAssets bank meshes | ExtraAssets bank meshes |

Instant + a Button is “next threshold on press” (the tree). ContinuousRate is “start filling until full / until Stop” (the jars). `onThresholdMeshChanged` / `onFilled` fire the rest of the room (Placeable Ledges on the grown tree; Prop Mover on the CTL doll; HUD “Grow Tree”).

**progress mesh** (`BP_ProgressMesh_Device` in `/SOCIAL_WAVE_ExtraAssets/SetupAssets/`) is the social-elimination *variant* of the same device: same fill fields, branded piggy. Docs: All > social-elimination > Devices. This island also uses two progress meshs as **score displays** in Throwables (`fillType=Instant`), not only in the progress demo.

Starter Island places one progress mesh as a sandbox toy with `fillRate=5` ContinuousRate — social_game_sample. Same class; different job.

## Don't

- Don't author a new UMG bar when a threshold mesh already communicates the number.
- Don't assume progress mesh is a unique device class. It is Fillable Container with a Social mesh.
- Don't cite `Content/NS/VIPMask` as the Mascot/progress lesson — that Niagara has **0** asset referencers.
