---
id: technique-050
kind: technique
---

# Multiplayer Skilled Interaction: queue types vs one SID per tower

**Sources:** minigame_mastery_template
**Snippet:** `../snippets/verse/qte_towers_device.verse` + `../snippets/verse/qte_game_device.verse`
**Tutorial:** Multiplayer Skill Checks — **External**. Device counts **Observed** live.

## Problem

A QTE that several players must pass together used to mean one Skilled Interaction Device per floor, plus a Player Counter and Timer on each floor, plus a lava Volume to detect a wipe. That does not scale, and it hides the actual skill-check settings in a pile of bindings.

## Pattern

The **Queue** category on `skilled_interaction_device` is the Creative answer:

| `queueExecutionType` | What it does (corridor billboards) |
|---|---|
| **Synchronous** | Everyone in the queue plays at once; `successTarget` / `failureLimit` can require the whole squad |
| **Sequential** | Relay: same QTE, join order |
| **Random** | One player from the queue gets the check |

Live **device-only** map: **11** SID (docs **10** — extra is the corridor sample). Sample SID: `queueExecutionType=None`, `successTarget=1`, `failureLimit=1`. Tower floors are one SID each, plus per-floor Player Counters and Timers, plus `Game Lost` Volumes over lava.

Live **`_Verse` map**: **3** SID (two towers + the same corridor sample). Tower 1 SID: `queueExecutionType=Synchronous`, `synchronousPlayerLimit=100`, `successTarget=5`, `failureLimit=5` (**Observed**). `qte_tower_device` keeps **one** SID, **one** Timer (`NextFloorTimerDevice`), **one** Player Counter, and `[]cinematic_sequence_device` for the five floors (`towerFloors` length **5**). Fail → `Play()` the current floor cine (hide the floor) → increment index → timer → `BeginInteraction` again. Last floor → `FailedEvent` → `qte_game_device` `ArrayRace`s the two towers and ends the round.

Do not stack both graphs in one Outliner. Twin maps: twin-level-device-vs-verse.md. First-to-fail wait: array-sync-race-primitive.md.

## Don't

- Don't assume the Project Browser opened `_Verse`. Default map has **0** Verse devices.
- Don't copy the per-floor SID graph onto the Verse map. The files already collapse it.
- Don't train on docs SID×10 as the live default count — corridor sample makes **11**.
