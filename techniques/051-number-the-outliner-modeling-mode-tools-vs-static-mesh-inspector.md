---
id: technique-051
kind: technique
---

# Number the Outliner: Modeling Mode tools vs Static Mesh inspector

**Sources:** modeling_template
**Tutorial:** Modeling Mode — tool catalog; this island is the *lab*.

## Problem

UEFN Modeling Mode is a dozen palettes (Create, PolyEdit, Deform, UVs, Volumes…). A landing-page GIF doesn’t stick. Need one island where each tool is a physical exhibit, plus the *inspector* settings (collision, extra UV, cooked) that Mode does not show.

## Pattern

Three halls, folders numbered like a textbook:

```
1-Tools/1.0 Modeling/1.1 Create … 1.6 Bevel
        2.0 Attributes/2.0 Material ID … 2.5 UV's
        4.0 Extra Polish/4.1 Sculpting … 4.3 Stylize
2 - Static Meshes/3.0 Collision/3.1 No Collision … 3.6 DOP
3 - Units and Grids/5.0 Prop Cooked  vs  6.0 Prop Uncooked
```

One `BP_DemoDisplay` + one `FortStaticMeshActor` per step. Teleporters hop halls.

**Collision wall (copy this set, don’t invent names):**

| Folder | Mesh |
|---|---|
| No Collision | `SM_Mesh_NoCollision` |
| Simplified | Box / Capsule / Sphere |
| Auto Convex | `SM_Mesh_CollisionAutoConvex` |
| Complex as Simple | `SM_Mesh_CollisionCOmplex` (typo ships) |
| DOP | `SM_Mesh_Collision24DOP` |

**Cooked vs uncooked:** `SM_Locker_Uncooked` next to the cooked locker — same prop, different cook.

UVs: `SM_Mesh_NoUVs` / `SM_Mesh_GoodUVs` / extra texcoord exhibit — adjacent, not a slideshow.

## Don't

- Don’t treat the Modeling hub as this island’s walkthrough — it’s the discipline index.
- Don’t dump Modeling Mode tools into hall 2 (that’s Details-panel / collision / UV).
- Don’t “fix” `CollisionCOmplex` or `modeling_templateTools` when citing shipped names.
- Don’t BrickAssembly BRICK gallery bricks in these halls — that is BrickAssembly-atom-static-mesh.md (`Shift+6`, not Modeling Mode).
