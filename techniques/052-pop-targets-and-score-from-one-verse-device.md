---
id: technique-052
kind: technique
---

# Pop targets and score from one Verse device

**Sources:** my_first_island_template
**Snippet:** `../snippets/verse/shooting_range_manager_device.verse`
**Tutorial:** ch.5 Spice Up the Gameplay with Verse — matches.

## Problem

Shooting-gallery dummies don’t die in one hit from device settings alone. You want +score on “good” tracks, −score on “bad” dummies, one Score Manager.

## Pattern

`@editable` each target (don’t hardcode names). Track vs dummy are **different Verse types**.

```verse
GoodTarget1:shooting_range_target_track_device = shooting_range_target_track_device{}
BadTarget1:shooting_range_target_device = shooting_range_target_device{}

OnBegin<override>()<suspends>:void=
    GoodTarget1.HitEvent.Subscribe(OnGoodTarget1Hit)
    BadTarget1.HitEvent.Subscribe(OnBadTarget1Hit)

OnGoodTarget1Hit():void=
    AdjustScore(GoodTargetScore)
    GoodTarget1.PopDown()

AdjustScore(Value:int):void=
    ScoreManager.SetScoreAward(Value)
    if (Player:player = GetPlayspace().GetPlayers()[0]):
        ScoreManager.Activate(Player)
```

`SetScoreAward` then `Activate` = both signs on one device. Scores themselves `@editable` (100 / −100) so designers tweak without rebuild.

This island is 1P — `[0]` is fine. Don’t copy that into an 16-player mode.

## Don't

- Don’t write `shooting_range_track_device` (docs slip).
- Don’t put combo/timer/weapon-level Verse here — that’s the *Level Up* follow-on, not this file.
- Don’t duplicate Score Managers for plus vs minus.
