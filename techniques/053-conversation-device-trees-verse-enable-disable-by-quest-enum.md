---
id: technique-053
kind: technique
---

# Conversation Device trees + Verse enable/disable by quest enum

**Sources:** narrative_roleplaying_template
**Snippets:**
- `../snippets/verse/dialog_device.verse`
- `../snippets/verse/quest_manager_device.verse`
- `../snippets/verse/utility_module.verse`
**Official tutorial (outline only):** Conversations and Quests + Warren NPC Conversations. Tutorial *Verse* is untrusted if it conflicts; **`snippets/` win**.

This is **scripted Conversation Device banks**, not LLM Persona. Persona stays on persona-voice-structured-output.md. `.uefnproject` `experimental.personaDevice.bEnablePersonaDevice` is **Unverified** as a flag; content-side: **0** Persona devices, **18** `Device_Conversation`.

## Problem

Several NPCs share one quest. Each NPC has *many* conversation banks (one per stage / branch). Only the bank for the current objective should be interactable. Choice events must also bump reputation/attitude, swap sub-dialogs, or jump the quest — without wiring every Enable/Disable in the Details panel.

## Pattern (shipped)

**World:** `Device_Character_V2` `interactType=Send Event Only`, `interactText=Talk`. Classic binding: Character `OnInteractedWith` → Conversation `InitiateConversation`. Most conversations also `Disable` on Player Spawner `On Player Spawned`; Verse then `Enable()`s the current path.

**Banks:** authored `Content/Dialogue/<NPC>/*.uasset` (`ConversationDatabase`). Live `conversationType=Custom`, `customWidget=UW_UI_Conversation_Lightside` (all **18**). `conversationMaterials[0]` is the NPC portrait MI (`M_Character_Caris` / `DJDroid` / `Scavenger`); Captain materials are **null**.

**Verse `dialog_device`** (one placed per quest NPC: DJ, Caris, Scavenger, Captain):

```verse
dialogpath := class():
    @editable ConversationDevice:conversation_device = conversation_device{}
    @editable Objective:cantina_objectives = cantina_objectives.GameStart
    @editable Events:[]conversation_event_class = array{}
    @editable MaybeSubDialog:?sub_dialog_class = false
```

`OnBegin`: `Sleep(0.1)`, then `OnConversationEvent.Subscribe(OnEvent)` and `CancelEvent.Subscribe(UpdateProgressOnCanceled)`. `OnEvent` switches on `event_type` (`ChangeReputation` / `NextConversation` / `NextObjective` / `StartQuestDJ`). `UpdateSubDialog` / `UpdateObjectiveDialog` **Enable** the matching bank and **Disable** the rest.

**Verse `quest_manager_device`** (one placed): `@editable CantinaQuest:[]cantina_quest` (**9** rows live), `Conversations:[]dialog_device` (**4**), plus Stat Creator / Tracker / Popup refs. `ProgressQuest` writes tracker `SetDescriptionText(ToText(…))` via `utility_module.verse` `ToText<localizes>`, pulses `map_indicator_device`, and on the last index `PopupDevice.Show` + `QuestTrackerDevice.Complete`.

Caris-first vs DJ-first is **not** identical: if `TalkToDJAboutEskay` arrives with `StartQuestWithDJ=false`, Verse awards the `QuestAccepted` row then jumps to `TalkToDJAboutEskay1_2`. Docs say starting NPC has no impact.

`GetDeviceProperties` on the Verse devices returns **subobject** `refPath`s (`Scripts-cantina_quest_N`, `Scripts-dialogpath_N`) — same trap as EnergyBlades / Social Deduction. ObjectTools cannot read those Verse struct fields by friendly name.

## Don't

- Don't train Persona `PromptToTalk` / `RegisterAction` as this island.
- Don't drive conversation chrome with Verse `button`+`button_slot` — that is verse-clickable-button / EnergyBlades workbench. Here the Conversation device owns `customWidget`.
- Don't `BuyOffer` quest rewards — no Marketplace; reputation is Stat Creator + HUD Message.
- Don't assume classic `ListEventBindings` on a Keylock explains quest doors. Three of four locks have **[]** bindings; Verse has no `keylock` API. Billboard 32 is the syllabus claim.
- Don't train persistable Verse canvas quest dialog from this island — that is persistable-quest-line-verse-dialog.md.
