---
id: technique-054
kind: technique
---

# space-opera Custom UI lives in `Content/UI/Widgets`, HUD Controller in Outliner `UI/`

**Sources:** narrative_roleplaying_template, team_conquest_sample
**Official tutorial:** space-opera™ Custom UI — Narrative is the chapter’s cantina / The Warren example. Team Conquest is the PvP sibling (same `Content/UI/Widgets` tree + a `HUDMessage/` folder).

IceBase’s Custom UI notes were Narrative-shaped. Do **not** duplicate that HUD recipe on ice_base_sample; IceBase is the contrast (Outliner `Devices/`, `Content/UI/PlayerHUD` with no `Widgets/` folder, Equipped + Quickbar overrides **on**). EnergyBlades PlayerInfo is a **mount** `/ENV_SPACE_THEME_UI_Assets/…`, not authored PlayerHUD. ExtraAssets HUD coincidence stays **four** (Social / Mastery / Walker / music-fantasy). Action Adventure HUD is in-project `WBP_BRICK_PlayerInfoStack`, a fourth in-project family — action_adventure_sample. Deeper viewmodel wiring: umg-viewmodel-device-binding.md. Team Conquest details: team_conquest_sample.

## Problem

A 1-player story island needs space-opera chrome (HUD, conversation, popup, tracker, stat bar) without ExtraAssets widgets and without showing minimap / build / inventory / team.

## Pattern (shipped)

Authored tree (**Observed** `folders.txt` + live overrides):

```
Content/UI/Widgets/
  Conversation/   UW_UI_Conversation_{Lightside,Darkside,Neutral} + UW_BTN_Conversation
  PlayerHUD/      UW_HUD_{PlayerInfo,Equipped,Quickbar,PlayerVitals,TeamVitals}
  Popup/          UW_UI_Popup_{Lightside,Darkside,Neutral} + UW_BTN_Popups
  Stats/          UW_UI_Stats
  Tracker/        UW_UI_Tracker_Lightside
```

| Device | Outliner | Live override |
|---|---|---|
| HUD Controller | `UI/` | `playerInfoWidgetOverride` → `…/PlayerHUD/UW_HUD_PlayerInfo_C`. Equipped + Quickbar **null**. `showMinimap` / `showTeamInfo` / `showBuildMenu` / `showPlayerInventory` = **HideElement**. Health/shield/HUD **DoNotOverride**. |
| Welcome + Complete popups | `UI/` | `templateOverrideClass` → `UW_UI_Popup_Lightside_C`. Welcome `autoDisplay=Game Start`, title **The Warren**. Complete `autoDisplay=None`. |
| 18 Conversation devices | per-NPC folders | `customWidget` → `UW_UI_Conversation_Lightside_C` (**all 18**; Darkside/Neutral authored, unused) |
| 4 Stat Creators | Setup / Attitude folders | `user Widget` → `UW_UI_Stats_C`. Reputation `enabled Phase=Always`; NPC attitude stats `None` until interact. |
| Tracker | `UI/Objectives` | `hUD Widget` → `UW_UI_Tracker_Lightside_C` (**Observed**). Title `Quest: Anchor Point`. `assignOnGameStart=false` (Verse assigns). |

Docs: Player Info override should be `UW_HUD_Vitals`. Shipped is **`UW_HUD_PlayerInfo`**. Docs “Show Minimap = No” = live `HideElement`.

`get_referencers` on the widgets returns **[]** — confirm on the placed devices (same IceBase trap).

## Team Conquest (second source)

Same `Content/UI/Widgets/` tree, plus `HUDMessage/` (contested vs capture) and `Popups/` (plural). HUD Controller is **unfoldered** (not Outliner `UI/`). Overrides PlayerInfo **and** Equipped + Quickbar (IceBase-like), `showTeamInfo=DoNotOverride` (PvP). ExtraAssets stays four. Details: team_conquest_sample.

## Don't

- Don't copy ExtraAssets `UW_HUD_PlayersInfoStack`. ExtraAssets HUD stays three sources.
- Don't treat IceBase `Content/UI/PlayerHUD` (no `Widgets/`) as this folder layout.
- Don't treat EnergyBlades workbench `UW_Overlay` as PlayerHUD.
- Don't expect Equipped/Quickbar overrides on Narrative — `allowBuilding=None`, nothing to equip. Team Conquest *does* override Equipped + Quickbar.
