---
id: technique-055
kind: technique
---

# Ship BrickAssembly Atom modules as a FortStaticMeshActor world kit (not a viewport lab)

**Sources:** martial_arts_sample
**Tutorial:** BRICK Templates Starter Island Templates — BRICK MartialArts Starter — **External**. Folders / tags **Observed**.

Do not duplicate BrickAssembly-atom-static-mesh.md / modular-BrickAssembly-building-modules.md (Brick Editor *makes* the SM in `Examples/`). Do not duplicate BRICK-fortnite-mix-no-scale.md (catalog `BrickSimple_*` / `SM_FORT_*`, **0** Phone here). Do not duplicate playset_sample (`Parent_Building*`; **0** PlaysetRoot). Assembly stays on action_adventure_sample — BrickAssembled SMs cannot feed Assembly.

## Problem

A theme starter island needs a large walkable MartialArts set (monastery, dojo, village) that still looks like real bricks, without leaving thousands of un-BrickAssembled gallery actors in the world, and without teaching Brick Editor mode in PLAY.

## Pattern

Docs: *“MartialArts-themed modular pieces constructed from the BRICK Brick Editor”* and *“reassemble, add to, or use them.”* This island is the **result**:

| Folder (live recursive counts) | Authored family |
|---|---|
| `NinjaBase/` **3063** | `SM_BRICK_MartialArts_*` BrickAssembled + `SM_HotSauce_*` monastery kit + catalog bamboo/TREE_PACK |
| `Whispering Woods/` **2488** | `SM_Fairy_*` monastery/village + authored `M_wWoods_Tree` |
| `Valley/` **559** | bridges / shrines + catalog foliage |

Placed class is **`FortStaticMeshActor`** (4531), not gallery `Actor` + Atom components. Sampled `SM_BRICK_MartialArts_Pillar_BrickAssembled_A` (`get_asset_tags`):

| Tag | Value |
|---|---|
| `AtomGeneratedAsset` | **True** |
| `StaticMaterials` | `/BRICKAtomAssets` LUT FakeBevel + `_CP` |
| `CollisionComplexity` | `CTF_UseSimpleAsComplex` |
| Nanite | True; 6 LODs; 2 UV channels |

Outliner POIs are copies of those modules (Dojo 335, Monastery 645, Pagoda 116, houses A–F, Whispering Woods Monastery 654). `allowBuilding=None` — players do not BrickAssembly on this island. Recolor / Separate / re-BrickAssembly stays on Brick Editor.

Catalog wrap fills the gaps (bamboo, Nightmare jungle, TREE_PACK pines, `/BRICKFairy`, `/BRICKHotSauce`). That is dressing, not the kit.

## Don't

- Don't train BrickAssembly hotkeys or gallery `PLATE 2X4` actors from this packet — that is Brick Editor pods 1–3.
- Don't Phone-copy `BrickSimple_*` and call it MartialArts — Home Builder.
- Don't feed these SMs to an Assembly device.
- Don't treat Musical Walk ConcertVenue props as the modular kit.
