---
id: technique-056
kind: technique
---

# Dress a BRICK Guard CD with a Fortnite CID (`characterLook=FortniteCharacter`)

**Sources:** martial_arts_sample
**Tutorial:** none for these NPCs (Templates index does not mention citizens). Live values **Observed**.

Contrast (do not duplicate): IceBase Guard CD + patrol + doorway volumes on guard-patrol-volume-swap.md. TURTLE theme enum vs definition on TURTLE-character-spawner.md. music-fantasy hireable `DemonFaction Character` on hireable-theme-character-spawner.md. Action Adventure `npc_behavior` quest giver / `npc_attack_controller` on persistable-quest-line-verse-dialog.md / npc-attack-stasis-aoe.md. This island has **zero** `.verse` files, **zero** patrol paths, **zero** volumes.

## Problem

A BRICK art island wants a few wandering people without a MartialArts character enum and without authoring a minifig mesh. The generic BRICK Guard type plus a Fortnite cosmetic is enough.

## Pattern

Five authored `NPCCharacterDefinition` assets (`Content/NPCDefinitions/NewNPCCharacterDefinition_Citizen1–5`). Tags: `CharacterDefinitionTag=Pawn.Athena.NPC.Creative.Guard`. Nested type is **`CharacterType_BRICKGuard`**. Behavior: `CharacterModifier_BRICKGuardBehavior`. Modifiers: BRICKHealth (100/100, 0 shield), BRICKTeam (`teamOption=Index`, `teamIndex=1`), BRICKGuardInventory, BRICKCosmeticSpawn, BRICKMovement.

CosmeticSpawn (**Observed** `characterLook=FortniteCharacter`, `characterBlueprint=null`):

| CD | `characterCosmetic` |
|---|---|
| Citizen1 | `/COSMETIC_ASSETS/…/Character_ScallopLava` |
| Citizen2 | `/COSMETIC_ASSETS/…/Character_TicketPoke` |
| Citizen3 | `/COSMETIC_ASSETS/…/CID_086_Athena_Commando_M_RedSilk` |
| Citizen4 | `/COSMETIC_ASSETS/…/Character_BlessFlan` |
| Citizen5 | `/COSMETIC_ASSETS/…/Character_OvenDrastic` |

Five `Device_CharacterSpawner` labeled **`NPC Spawner_Citizen1–5`** (find `"Character Spawner"` returns **0**). Live: `enabledAtGameStart=true`, `spawnCount=1`, `allowInfiniteSpawn=true`, `totalSpawnLimit=1`, `spawnTimer=3`, `spawnRadius=1`, `spawnOnEnabled=true`. `get_referencers` on the CD assets is **[]** — the label is the join.

No `AIPatrolPath`. No volume Enable/Disable. They are always-on dressing.

## Don't

- Don't look for a MartialArts / Tabasco character enum on these spawners. Identity is BRICK Guard + CID.
- Don't copy IceBase doorway volumes or TURTLE `TURTLE Character`.
- Don't subclass `npc_behavior` for these five.
- Don't treat the Fortnite CIDs as theme-legal MartialArts minifigs.
