---
id: technique-057
kind: technique
---

# Starter-island music bed: Switch On/Off → Audio Player Play/Stop (not Patchwork)

**Sources:** martial_arts_sample
**Tutorial:** BRICK Templates — art template “includes … audio” — **External**. Bindings **Observed**.

Do not duplicate patchwork-switch-stems.md (Jam / MUSIC_SEQR stems; **0** Patchwork here, `enableJam=Off`). Do not duplicate beacon-volume-popup-trail.md (**0** Beacon / popup). Do not duplicate jam-mutator-crowd-aura.md.

## Problem

A theme art island wants a looping-feeling bed the player can mute, without a Patchwork lab and without Verse.

## Pattern

Four actors in `World Setup/Music Control` (**Observed** `get_actors_in_folder`):

| Label | Class | Live |
|---|---|---|
| `Billboard_Music` | `Device_Billboard_V2` | `text="Turn on/off Music"`, Always, One Sided, size 12 |
| `Switch_Music` | `Device_Switch_V2` | `enabledAtGameStart=true`, `initialState=true` (On), `visibleDuringGame=true`, `turnOnText="Turn on Music"`, `turnOffText="Turn off Music"` |
| `AudioPlayer_BKG` | `Device_CRD_AudioPlayer` | `audio=/…/Music/MUSCSong_MartialArtsArtTemplate_Cue`, `volume=2`, `loopAudio=false`, `autoPlay - Gameplay=true`, `autoPlay - Create=false`, `can Be Heard By=Everyone`, `visible In Game=false`, Always |
| `Cane SpeechBubble A` | `BP_Cane_SpeechBubble_A` | dressing |

Bindings live on the **Audio Player**, not the Switch (`ListEventBindings` on the Switch is **[]**):

```
Switch_Music  On Turned On   → AudioPlayer_BKG  Play
Switch_Music  On Turned Off  → AudioPlayer_BKG  Stop
```

Cue is authored (`Content/Music/MUSCSong_MartialArtsArtTemplate` + `_Cue`; import path still says `FortniteProjects/HotSauce/…`). Duration ~123 s, not looping on the wave — Gameplay autoPlay + the switch are the bed.

ConcertVenue (47) is Musical Walk speakers / lute / pink guitar / SpaceFriends mic. It is **not** wired to this player.

## Don't

- Don't Enable/Disable Patchwork devices from this Switch.
- Don't look for Play on the Switch’s binding list.
- Don't set `enableJam=Default` and call it this island — live `enableJam=Off`.
