---
id: technique-058
kind: technique
---

# Sequential collect: subscribe all spawners, `SpawnItem` only the next

**Sources:** parkour_template
**Snippet:** `../snippets/verse/parkour_race_device.verse`

## Problem

N pickups in a designed path. Player must get them **in order**. Win when `collected >= spawners.Length`. Optional secret pickup that adds capped timer time. Timer death respawns with a full clock.

## Pattern

```verse
for (BatterySpawner : BatteryItemSpawners):
    BatterySpawner.ItemPickedUpEvent.Subscribe(HandleBatteryPickedUp)

HandleBatteryPickedUp(Agent:agent):void=
    set BatteriesCollected = BatteriesCollected + 1
    if (BatteriesCollected >= BatteryItemSpawners.Length):
        spawn { EndGame(Agent) }
    else if (NextBatterySpawner := BatteryItemSpawners[BatteriesCollected]):
        NextBatterySpawner.SpawnItem()
```

Index **after** increment = next spawner. First battery must already be spawned (editor: only [0] active at start).

Timer:

- SpawnedEvent → `Reset` + `Start`
- FailureEvent → `GetFortCharacter[].Damage(500)`
- Secret: `SetActiveDuration(Min(remaining+reward, GetMaxDuration()), Agent)`

Parameterized HUD: `BatteryCollectedMessage<localizes>(Amount:int):message = "You collected {Amount} battery"`

This is **not** `race`. Docs are right that the teaching frame is loops + Subscribe + arrays.

Use **item_spawner_device** (world pickup). Item Granter is Elimination’s inventory grant — tutorial table mixed them.

## Don't

- Don’t Subscribe only the current spawner unless you cancel/resubscribe; this file subscribes all and gates with `SpawnItem`.
- Don’t skip `Reset` on spawn or respawns inherit leftover duration.
- Don’t train Granter APIs from this sample’s docs device list.
- Don’t use Verse sequential spawners when any-order Collectible→Tracker→End Game is enough — collectible-tracker-endgame.md.
- Don’t train a tagged Item Spawner teleport pool from this race — that is stud-itemspawner-teleport-pool.md.
