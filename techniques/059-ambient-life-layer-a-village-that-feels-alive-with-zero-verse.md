---
id: technique-059
kind: technique
---

# Ambient-life layer: a village that feels alive with zero Verse

**Problem.** A cozy starter village feels dead if it's only geometry — but you don't want
custom AI or Verse in a *starter* template. What devices and assets make it feel inhabited?

**How UEFN's AmbientEnvironment does it** (daytime counterpart to NightEnvironment's night stack).

## 1. Light = Day Sequence + scalability tiers

Unlike NightEnvironment (frozen night, custom skybox), AmbientEnvironment runs a **Day Sequence Device**
(`BP_DaySequenceModifier_Device`) so the village cycles. Beside it in `00_Lighting/`:

- `BP_LightingScalabilityManager` (`/Game/TimeOfDay/TODM_Sequence/UEFN/`)
- **Three PostProcessVolumes named `…_Low`, `…_Med`, `…_High`** — the grade is authored
  per scalability tier, not one look forced everywhere.
- One `ExponentialHeightFog`, no fog cards (day scene; fog cards are a night trick).

**Lesson:** daytime islands still author atmosphere — but the unit is the *scalability tier*,
not the POI. Pair with day_sequence_template for multi-biome priority stacking.

## 2. Ambient fauna and people come from spawner devices, not code

| Device | Count | Placement pattern |
|---|---|---|
| `Device_WildlifeSpawner_V2` | 7 | One per biome pocket (farm fields, forest, shoreline) |
| `Device_FishingZone_V2` | 4 | Coast, river, lake — matches the three waterbodies |
| `Device_FireflySpawner_V2` | 1 | The single night accent in a day scene |
| `Device_CharacterSpawner` | 9 | Named characters ("CID 320 Commando M Nautilus", "SnowFairy", "WeepingWoods") — villagers as NPC outfits |
| `Device_Character_V2` (Mannequin) | 1 | Static posed character |

No triggers, no events, no Verse: these devices self-run when enabled. **The pattern is
density-over-logic** — 20+ ambient spawners at ~zero behavioural complexity each.

## 3. Companions are data assets, not Verse classes

AmbientEnvironment ships four authored NPC definitions:
`Sidekick_NPC_Example{,_Cat,_Flopsticks,_FollowPlayer}`. Live-read structure:

```
type       → CharacterType_Custom
behavior   → CharacterModifier_VerseBehavior
modifiers  → [CharacterModifier_CosmeticSpawn, CharacterModifier_Health]
```

A "sidekick" (pet/companion) is assembled from a **type + behaviour + modifier list** in a
data asset; the Verse behaviour is referenced through a modifier, so the island itself
contains no `.verse` files. `…_Cat` vs `…_FollowPlayer` shows the variation axis: swap the
behaviour modifier, keep everything else.

## 4. Farm dressing encodes time as prop families

Crops are placed from `/AmbientFarm` as **lifecycle families** —
`*_{Seedling, Grown, Produce, Dying}` for wheat, carrot, cabbage, tomato, sweetcorn, pumpkin,
apple. Fields read as "a working farm" because rows mix stages. It's static dressing (no
simulation), but the four-stage vocabulary is what makes it read as *farmed*.

## 5. Water is three bodies, fishing is zoned to match

`VK_Waterbody_Ocean` + `VK_Waterbody_Lake` + `VK_Waterbody_River` (FortWater
`WaterBrushManager`), with the four FishingZone devices placed one per body. Fish market on
the beach closes the loop fictionally.

**Contrast pair to remember:** NightEnvironment = night + fog cards + audio grid (mood);
AmbientEnvironment = day + spawner density + NPC data assets (life). Same skeleton (atmosphere layer
over a mounted village kit), different authored tier.

**Sources:** ambient_environment_sample; contrast night_environment_sample.
