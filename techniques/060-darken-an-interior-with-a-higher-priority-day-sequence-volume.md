---
id: technique-060
kind: technique
---

# Darken an interior with a higher-priority Day Sequence volume

**Sources:** prison_sample
**Tutorial:** Working With TWDU Islands — **External** (blurb only; no DSD numbers). Live values **Observed**.

Same 2000-vs-higher rule as day-sequence-priority-stack.md (biomes) and ooze-pool-stack.md (pool fog). This island uses it to **kill daylight in a basement**, not a biome or a poison puddle.

## Problem

A kitbashed prison has a large underground. The outdoor Day Sequence still lights the cellar through missing roofs and World Partition holes. You need the basement dimmer and foggier without Verse and without a second island Time of Day.

## Pattern

Two `BP_DaySequenceModifier_Device`s, both `FixedTime` **20**, `enableDuringPhase=Always`, all override toggles on, skybox component **enabled**:

| Device | Priority | `use Volume` | Sun | Fog | Clouds |
|---|---|---|---|---|---|
| `Day Sequence Device (Outside)` | **2000** | **false** (island-wide) | **4** | **0.01** | 0.75 |
| `Day Sequence Device (Underground)` | **2050** | **true** | **1.5** | **1** | 0.90 |

2050 > 2000, so the volume wins while the player is inside it. Outdoor fog stays a haze; underground fog is the “closed air.” Sun rotation **Z=230** is shared (inherit vs override: they both set it).

Live volume extents on the underground device are `200×200×250` (the authored box). Actor bounds are much larger — trust `use Volume` + priority, not the outliner gizmo size.

**Geo occluder:** five `GridPlane`s in `Lightblocker/` sit under the yard so sun doesn’t leak through the floor while the volume fades.

Global `PostProcessVolume` is unbound, `blendWeight=1`, **empty** `weightedBlendables`. Grade is the DSD + `LumenExposureManager C`, not a custom PP material stack. Contrast City’s cartoon/comic blendables: cartoon-comic-unbound-ppv.md.

## Don't

- Don’t put a second Day Sequence at priority **2000** and expect a hard interior cut — they blend.
- Don’t look for an authored `SM_SkySphere` here. Sky is the DSD skybox component plus seven `BackDrop` static meshes in `Background/`.
- Don’t copy Walker NPC’s Heavy Linework Post Process device. This island has **no** `Device_PostProcess` and **no** HUD.
