---
id: technique-061
kind: technique
---

# Asymmetric teams: abstract base + shuffle + race to empty

**Sources:** prop_hunt_tutorial
**Snippets:** `base_team.verse`, `prop_hunt.verse`, `hunter_team.verse`, `waiting_for_more_players.verse`

## Problem

Hide-and-hunt: random hunters vs props each round, props get a head start, last prop is worth the most, mid-joiners become hunters, round ends when either team empties or the clock dies — without a forest of event-binding devices.

## Pattern

```verse
base_team := class<abstract>:
    var TeamAgents<private>:[]agent = array{}
    TeamEmptyEvent:event() = event(){}
    InitializeAgent(Agent): Add + ChangeTeamAndClass + HUD
    RemoveAgentFromTeam: RemoveAllElements; if Count() < 1 then TeamEmptyEvent.Signal()
```

`prop_team` / `hunter_team` inherit. Orchestrator `prop_hunt`:

1. Wait `MinimumNumberOfPlayers` (2) then timer.
2. `Ceil[N / Max(1.1, HunterAgentPerNumberOfPlayers)]` hunters; Shuffle; Slice.
3. Props teleport now; hunters `WaitTimer` 15s then teleport (`HuntersGo`).
4. `race`: prop empty vs hunter empty vs `RoundTimer.AwaitEnd()`.
5. Mid-round: `ParticipantAdded` → hunter if `PropTeam.Count() > 0`. Removed → that team’s `EliminateAgent`.
6. Hunter elim award: `Floor(MaxEliminationScore / PropTeam.Count())` then `SetScoreAward` + `Activate`.
7. Dead prop: `PropTeam.EliminateAgent` then `HunterTeam.InitializeAgent`.

`1.1` floor on the divisor guarantees ≥1 prop.

## Don't

- Don’t bind this in the Creative UI — this *is* the round manager.
- Don’t start hunters with props (head-start is the mode).
- Don’t use `GetPlayers()[0]` except as `EndRound` instigator (comment in file).
