---
id: technique-062
kind: technique
---

# Reveal idle hiders: movement `race` + pooled heartbeat VFX

**Sources:** prop_hunt_tutorial
**Snippets:** `prop_team.verse`, `heartbeat.verse`

## Problem

Props that camp still must be findable. After 15s without moving ≥100 cm, show VFX over their head and play a heartbeat. Moving cancels it. Many props can beat at once; one radio.

## Pattern

Per prop, forever (until not on team):

```verse
race:
    AwaitNoLongerAProp()
    loop:
        AwaitStopMoving(MinimumMoveDistance)   # 100
        race:
            AwaitStartMoving(MinimumMoveDistance)
            block:
                CountdownTimer(PropAgent)      # last 5s warning UI
                StartHeartbeat()               # Sleep(Inf) until this race loses
        Sleep(0.0)
```

Move check: sample `GetTransform().Translation` each tick; `Distance` vs last sample.

Heartbeat: index `AgentVFX[FindOnTeam[PropAgent]]`. `TeleportTo` character + (0,0,110), `Enable`. Radio `Register` that agent; `NumberOfHeartBeats` 0→1 starts Play, Disable Unregister and decrement (clamp ≥0). `defer` on `StartHeartbeat` always Disable.

Warning canvas per player in `WarningUI` map; `defer` clears text if they move mid-countdown.

## Don't

- Don’t spawn a new Niagara per pulse — pool `[]heartbeat_vfx` (one per max player).
- Don’t Play the radio per prop; one device, Register/Unregister.
- Don’t skip `AwaitNoLongerAProp` or eliminated props keep looping.
