---
id: technique-063
kind: technique
---

# Analog 3D clock hands driven by Verse trigonometric rotation

**Sources:** gift_factory_sample
**Snippets:**
- `../snippets/verse/real_time.verse`
- `../snippets/verse/date_and_time.verse`
**Official tutorial:** BRICK gift factory.

## Problem

Games frequently need physical world props (such as clock faces, dials, or compass needles) to rotate accurately based on real-world time or in-game measurements, regardless of how the clock housing actor is rotated or pitched in 3D level space.

## Pattern (shipped)

In `real_time.verse`, an analog clock consists of four props:
- `ClockProp`: The static clock housing.
- `HourHandProp`: The hour pointer prop.
- `MinuteHandProp`: The minute pointer prop.
- `SecondHandProp`: The second pointer prop.

### 1. Radian calculation from UTC time

Every second, `GetDateAndTime()` queries current hours, minutes, and seconds. The rotation angle around the clock dial's normal vector is computed in radians ($2\pi \text{ radians} = 360^\circ$):

```verse
Now := GetDateAndTime()
RotationAxis := vector3{X := 1.0, Y := 0.0, Z := 0.0}

# Hour hand (12-hour cycle)
HourHandAngle := 1.0 * Now.Hours / 12.0 * (2.0 * PiFloat)
LocalHourHandRotation := MakeRotation(RotationAxis, HourHandAngle)

# Minute hand (60-minute cycle)
MinuteHandAngle := 1.0 * Now.Minutes / 60.0 * (2.0 * PiFloat)
LocalMinuteHandRotation := MakeRotation(RotationAxis, MinuteHandAngle)

# Second hand (60-second cycle)
SecondHandAngle := Now.Seconds / 60.0 * (2.0 * PiFloat)
LocalSecondHandRotation := MakeRotation(RotationAxis, SecondHandAngle)
```

### 2. World-space orientation transformation via `RotateBy`

To ensure hands rotate in the plane of the clock face even if the clock is tilted, placed on a wall, or rotated at an arbitrary angle, the local rotation is transformed by the parent clock prop's world rotation:

```verse
ClockTransform := ClockProp.GetTransform()
GlobalHourHandRotation := LocalHourHandRotation.RotateBy(ClockTransform.Rotation)
GlobalMinuteHandRotation := LocalMinuteHandRotation.RotateBy(ClockTransform.Rotation)
GlobalSecondHandRotation := LocalSecondHandRotation.RotateBy(ClockTransform.Rotation)
```

### 3. Prop teleportation loop

The hands are updated in place using `TeleportTo[]` preserving their original translation:

```verse
if:
    HourHandProp.TeleportTo[HourHandProp.GetTransform().Translation, GlobalHourHandRotation]
    MinuteHandProp.TeleportTo[MinuteHandProp.GetTransform().Translation, GlobalMinuteHandRotation]
    SecondHandProp.TeleportTo[SecondHandProp.GetTransform().Translation, GlobalSecondHandRotation]
Sleep(1.0)
```

## Don't

- Don't apply Euler angles directly in world space — if the clock face is placed on an East-West wall or angled ceiling, rotating around world Z spins the hands perpendicular to the dial.
- Don't spawn separate devices or skeletal meshes for simple dials when static `creative_prop`s with `TeleportTo` provide zero-overhead kinematic rotation.
- Don't forget that Verse `MakeRotation(Axis, Angle)` expects **radians**, not degrees. Passing $360$ produces multi-turn wraps and jitter.
