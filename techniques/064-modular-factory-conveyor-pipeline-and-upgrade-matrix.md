---
id: technique-064
kind: technique
---

# Modular factory conveyor pipeline and upgrade matrix

**Sources:** gift_factory_sample
**Snippets:**
- `../snippets/verse/factory_manager.verse`
- `../snippets/verse/factory_element.verse`
- `../snippets/verse/factory_belt.verse`
- `../snippets/verse/factory_mould.verse`
- `../snippets/verse/factory_wrapper.verse`
- `../snippets/verse/factory_storage.verse`
- `../snippets/verse/upgrade_config.verse`
**Official tutorial:** BRICK gift factory.

## Problem

Tycoon, automation, and idle games require scalable production pipelines where products flow sequentially from raw input through transformation machines into finished goods and storage. Hardcoding machine behaviors creates duplicate code, while rigid structures make balancing upgrade prices and production speeds painful.

## Pattern (shipped)

`gift_factory_sample` implements a 4-line industrial production system. Each line consists of four stages:
1. `factory_belt`: Conveyor track moving products forward.
2. `factory_mould`: Machine stamping/pressing products into form.
3. `factory_wrapper`: Packaging unit boxing products for delivery.
4. `factory_storage`: Central silo aggregating inventory from all 4 lines.

### 1. Abstract polymorphic base class: `factory_element`

Every station on the factory floor derives from `factory_element`:

```verse
factory_element<public> := class<abstract>(creative_device):
    @editable UpgradeButton:button_device = button_device{}
    @editable BuildProp:creative_prop = creative_prop{}
    @editable BuildPreviewProp:creative_prop = creative_prop{}
    @editable UpgradeProp:creative_prop = creative_prop{}

    # Common virtual methods overridden per station
    IsBuilt<public>()<transacts>:logic
    GetLevel<public>()<transacts>:int
    GetMaxLevel<public>()<transacts>:int
    GetUpgradePrice<public>()<transacts>:int
    Build<public>()<transacts>:void
    Upgrade<public>()<transacts>:void
```

Buttons handle interactions, props switch between `BuildPreviewProp` (ghosted), `BuildProp` (placed), and `UpgradeProp` (active upgrades), and each station queries its own persistent slice from `player_info`.

### 2. Multi-column configuration matrix (`upgrade_config.verse`)

Instead of hardcoding prices in Verse classes, economy parameters are defined as 2D tables:

```verse
# Rows = Upgrade Level (0..14)
# Columns = Factory Line (0..3)
FactoryBeltPrices: [][]int = array{
    array{   5,    25,    500,    5000 },
    array{  25,   125,   2500,   25000 },
    array{ 100,   500,  10000,  100000 },
    # ... up to level 15
}

FactoryMouldPrices: [][]int = ...
FactoryWrapperPrices: [][]int = ...
FactoryStorageCapacities: []int = array{ 50, 100, 250, 500, 1000, 2500, 5000 }
```

Adding lines or adjusting inflation rates requires editing only numeric values in `upgrade_config.verse`.

### 3. Conveyor product animation & kinematic movement

During active play (`RunProduction`):
- When a mould cycles, a `factory_product` prop spawns at the belt origin.
- The product slides along waypoints using `MoveTo(TargetLocation, Rotation, Duration)`.
- When entering the wrapper, the wrapper's pneumatic ram animates (`LS_Factory_Wrapper_N`), the product mesh swaps from raw cane to wrapped box, and continues down the belt into storage.
- Storage adds $+1$ to `ProductsInStorage[LineIndex]` until reaching `StorageCapacity`.

### 4. Visual status indicators

Each station controls visual indicators:
- `factory_progress_lights`: Step lights indicating machine cycle readiness (`BP_Light_Green`, `BP_Light_Orange`, `BP_Light_Red`).
- `factory_rotor_light`: Spinning warning light alerting when lines stall or finish.
- `factory_intensity_lights`: Dynamic spot/point lights brightening as machines upgrade to higher tiers.

## Don't

- Don't write separate purchase and upgrade logic for belts, moulds, and wrappers — unifying under `factory_element` allows the UI manager and tutorial to treat any machine identically.
- Don't hardcode upgrade costs in device `@editable` properties — storing economy tables in a central Verse config module (`upgrade_config.verse`) ensures multiplayer synchronization and zero asset re-serialization when rebalancing.
- Don't let production run unthrottled — `FactoryBeltMinDistanceToProduce` (4.0 units) prevents product meshes from stacking on top of each other when speed modifiers spike.
