---
id: technique-065
kind: technique
---

# Persistable real-time calendar and offline catch-up simulation

**Sources:** gift_factory_sample
**Snippets:**
- `../snippets/verse/real_time.verse`
- `../snippets/verse/date_and_time.verse`
- `../snippets/verse/time_interval.verse`
- `../snippets/verse/player_info.verse`
- `../snippets/verse/factory_manager.verse`
- `../snippets/verse/test.verse`
**Official tutorial:** BRICK gift factory.

## Problem

Idle and tycoon games require calculating resource progression that occurred while the player was offline (or between match sessions), without server ticks running in their absence. Creators also need a deterministic way to simulate time-travel during testing and detect daily calendar rollovers (e.g. daily gifts at midnight UTC).

## Pattern (shipped)

### 1. Common Era epoch normalization

Verse's `GetSecondsSinceEpoch()` returns Unix epoch seconds (since Jan 1, 1970 00:00:00 UTC). The template normalizes this to the Common Era (CE: Jan 1, 0001 00:00:00 UTC) by adding `62135596800.0` seconds, enabling pure calendar mathematics:

```verse
# In real_time.verse:
GetDateAndTime<public>()<transacts>:date_and_time=
    Seconds := GetSecondsSinceEpoch() + 62135596800.0 + 1.0 * DebugTimeOffsetSecondsTracker.GetValue()
    return MakeDateAndTime(Seconds)
```

### 2. Persistable Gregorian calendar class

Instead of storing arbitrary floating-point timestamps that suffer precision drift or lose calendar semantics, date and time is stored in an immutable `<final><persistable>` class:

```verse
# In date_and_time.verse:
date_and_time<public> := class<final><persistable>:
    @editable Years<public>:int = -1
    @editable Months<public>:int = -1
    @editable Days<public>:int = -1
    @editable Hours<public>:int = -1
    @editable Minutes<public>:int = -1
    @editable Seconds<public>:float = -1.0

    IsValid<public>()<decides><transacts>:void=
        Years > 0 and Months > 0 and Months <= 12 and Days > 0 and Days <= GetDaysFromMonth(Months)
        Hours >= 0 and Hours < 24 and Minutes >= 0 and Minutes < 60 and Seconds >= 0.0 and Seconds < 60.0

    IsLeapYear<public>()<decides><transacts>:void=
        Mod[Years, 4] = 0 and (Mod[Years, 100] <> 0 or Mod[Years, 400] = 0)
```

### 3. Date arithmetic and intervals

`Subtract(DAT:date_and_time)` calculates total seconds elapsed since Common Era and constructs a `time_interval` struct (`Days`, `Hours`, `Minutes`, `Seconds`, `GetTotalSeconds()`). Conversely, `Add(TI:time_interval)` advances a `date_and_time` across month/year boundaries.

### 4. Offline progression catch-up loop

When a player loads their factory (`factory_manager.CalculateOfflineProduction()`):
1. Reads `Now := RealTime.GetDateAndTime()`.
2. For each active production line, computes elapsed seconds since that line's `LatestProductionTime` stored in `factory_mould_info`.
3. Divides elapsed time by that mould's current `ProductionInterval` (from `upgrade_config.verse`) to get raw offline units.
4. Simulates capacity caps: products accumulate into `factory_storage_info.ProductsInStorage`, strictly bounded by `FactoryStorageCapacities[StorageLevel]`.
5. Emits `OfflineProductionEvent([]int)` and presents the `offline_production_ui` modal showing exactly what accumulated while away.

### 5. Daily rollover detection (midnight UTC)

Daily rewards check whether the current date's calendar values differ from the last recorded opening date:

```verse
IsDateRolledOver()<decides><transacts>:void=
    Now := RealTime.GetDateAndTime()
    SelectedPlayer := MaybeSelectedPlayer?
    GiftOpeningTime := GetGiftOpeningTime[SelectedPlayer]
    Now.Years <> GiftOpeningTime.Years or Now.Months <> GiftOpeningTime.Months or Now.Days <> GiftOpeningTime.Days
```

### 6. Persistent debug time-offset testing

Instead of waiting hours or altering system clocks, `real_time` injects an offset into every query using a persistent `tracker_device` (`DebugTimeOffsetSecondsTracker`). In `test.verse`, a creator button calls `RealTime.AddDebugTimeOffsetHours(1)` and immediately recalculates offline production, verifying offline loops without closing the session.

## Don't

- Don't store raw floats across sessions for calendar logic — floating-point seconds lose day/month boundary precision and fail leap-year math.
- Don't allow offline catch-up without storage capacity bounds — unconstrained simulation creates economy hyperinflation when players return after weeks.
- Don't rely on local client time for daily reward resets — `GetSecondsSinceEpoch()` is synchronized UTC time, preventing client clock manipulation.
