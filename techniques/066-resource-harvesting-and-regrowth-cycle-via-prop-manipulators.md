---
id: technique-066
kind: technique
---

# Resource harvesting and regrowth cycle via prop manipulators

**Sources:** gift_factory_sample
**Snippets:**
- `../snippets/verse/candy_cane_field.verse`
- `../snippets/verse/daily_gift.verse`
- `../snippets/verse/player_info.verse`
**Official tutorial:** BRICK gift factory.

## Problem

Farming, survival, and resource-gathering loops require props in the world that players can strike with their pickaxe to harvest resources. Once harvested, the props must disappear, grant currency or items, and regrow after a cooldown — even if the player leaves and rejoins the island later.

## Pattern (shipped)

In `gift_factory_sample`, both the Candy Cane field (`candy_cane_field.verse`) and the Daily Gift (`daily_gift.verse`) use this pattern.

### 1. Concrete helper wrapper around `prop_manipulator_device`

Each plant in the field is encapsulated in a class:

```verse
candy_cane := class<concrete>:
    @editable PropManipulator:prop_manipulator_device = prop_manipulator_device{}
    @editable Number:int = 0
    var Grown:logic = false
```

In the UEFN editor, each candy cane prop is associated with a placed `Device_PropManipulation_V2`. Crucially, the prop actor itself must be marked as a **damageable resource** so the pickaxe registers hits.

### 2. Harvest event handling

The manager device subscribes to `PropManipulator.HarvestingEvent`:

```verse
OnHarvesting(Agent:agent, CandyCane:candy_cane):void=
    # Hide the physical prop instantly
    set CandyCane.Grown = false
    CandyCane.PropManipulator.HideProps()

    # Play harvesting VFX & Audio
    spawn: ShowCandyCaneEffect(CandyCane)
    HarvestAudioPlayer.Play()

    # Grant gold & record harvest time in persistent data
    if (SelectedPlayer := MaybeSelectedPlayer?):
        GrantGold[SelectedPlayer, GoldPerCandyCane]
        Now := RealTime.GetDateAndTime()
        SetCandyCaneHarvestTime[SelectedPlayer, CandyCane.Number, Now]
        spawn: Regrow(CandyCane)
```

### 3. Regrowth with offline persistence

When a candy cane is harvested, `Regrow` starts a timer:

```verse
Regrow(CandyCane:candy_cane)<suspends>:void=
    Sleep(CandyCaneRegrowthTime)
    if (not CandyCane.Grown?):
        set CandyCane.Grown = true
        CandyCane.PropManipulator.ShowProps()
        CandyCaneRegrowAudioPlayer.Play()
```

When a player enters the island, `SelectPlayer` inspects `PlayerInfoMap[Player].CandyCaneHarvestTimes`:
- Calculates elapsed time since each candy cane was harvested using `RealTime.GetDateAndTime().Subtract(HarvestTime)`.
- If `Interval.GetTotalSeconds() >= CandyCaneRegrowthTime`, the candy cane is immediately unhidden (`ShowProps()`).
- If not yet grown, a timer sleeps for the *remaining* seconds (`CandyCaneRegrowthTime - Elapsed`), then shows the prop.

## Don't

- Don't destroy actor props when harvesting — destroying and re-spawning actors leaks memory and breaks references. Calling `HideProps()` and `ShowProps()` on `prop_manipulator_device` is instantaneous and garbage-free.
- Don't rely solely on a running `Sleep(Duration)` loop for regrowth — if players log out or restart the session, in-memory loops are destroyed. Persisting `LatestHarvestTime` in `player_info` ensures crops regrow accurately across sessions.
- Don't forget that props assigned to `prop_manipulator_device` must have collision and material destruction settings enabled, otherwise `HarvestingEvent` will not fire when hit by pickaxes.
