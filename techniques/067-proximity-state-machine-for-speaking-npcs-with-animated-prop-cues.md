---
id: technique-067
kind: technique
---

# Proximity state machine for speaking NPCs with animated prop cues

**Sources:** gift_factory_sample
**Snippets:**
- `../snippets/verse/speaking_character.verse`
- `../snippets/verse/tutorial.verse`
**Official tutorial:** BRICK gift factory.

## Problem

Creating expressive NPC guides or shopkeepers usually requires heavy Conversation Devices, trigger volumes, or Verse NPC behavior trees. When guides need tiered attention-grabbing behavior (an "Attract" emote from medium distance and a "Speak" dialogue bubble at close range) without blocking player movement or spamming trigger overlaps, a lightweight proximity polling state machine is ideal.

## Pattern (shipped)

Five NPC guide characters ship in `gift_factory_sample` (Sir Pawlar, Sub Zero, Professor Flopkins, Freshman Flapjack, and Blubberchops). Each is powered by `speaking_character.verse`.

### 1. 4-State Enum

```verse
speaking_character_state := enum:
    None,
    Idle,
    Attract,
    Speak
```

### 2. Distance-based polling loop without trigger volumes

Rather than managing overlapping `mutator_volume` or `trigger_device` volumes, the device runs a continuous polling loop measuring squared Euclidean distance to each active player pawn:

```verse
PropTransform := Prop.GetTransform()
loop:
    var MinCharacterDistanceSquared:float = Inf
    for (Player : GetPlayspace().GetPlayers(), Character := Player.GetFortCharacter[]):
        CharacterTransform := Character.GetTransform()
        CharacterDistanceSquared := DistanceSquared(CharacterTransform.Translation, PropTransform.Translation)
        if (CharacterDistanceSquared < MinCharacterDistanceSquared):
            set MinCharacterDistanceSquared = CharacterDistanceSquared

    if (MinCharacterDistanceSquared < DistanceToSpeak * DistanceToSpeak):
        UpdateState(speaking_character_state.Speak)
    else if (MinCharacterDistanceSquared < DistanceToAttract * DistanceToAttract):
        UpdateState(speaking_character_state.Attract)
    else:
        UpdateState(speaking_character_state.Idle)

    Sleep(0.2)
```

Default thresholds:
- `DistanceToAttract = 1200.0` (~12 meters / 2.3 grid tiles)
- `DistanceToSpeak = 500.0` (~5 meters / 1 grid tile)

### 3. Visual & Audio prop transitions

Entering a new state triggers state transitions:
- In `Idle`: Props remain hidden underground (`HideOffset := vector3{Z := -500.0}`).
- In `Attract`: `AttractProp` (an attention icon or waving prop) pops up above the character (`ShowOffset := vector3{Z := 20.0}`), and `AttractAudioPlayer.Play()` chimes.
- In `Speak`: `SpeakProp` (e.g. `BP_Cane_SpeechBubble_A/B`) appears above the character's head, `SpeakAudioPlayer.Play()` speaks a vocal chirp, and `SpeakEvent.Signal()` triggers any linked tutorial or camera events.
- Leaving a state hides the corresponding prop back to `-500.0` Z and signals `StopEvent`.

## Don't

- Don't place multiple overlapping trigger volumes around every NPC — radius triggers cause outliner clutter and trigger volume limit exhaustion in large hubs.
- Don't compute `Distance()` with square roots inside frequent polling loops — compare against `Threshold * Threshold` (`DistanceSquared`) to avoid unnecessary square root computation.
- Don't spawn/destroy widgets or props every time a player walks in and out — pre-placing the indicator props and moving them between `HideOffset` and `ShowOffset` with `TeleportTo` eliminates garbage collection stutter.
