---
id: technique-068
kind: technique
---

# Iterate a Scene Graph prefab by adding components

**Sources:** scene_graph_template

## Problem

You have a game object (a light post) and you want to grow it from "a mesh in the world" to "a swaying, glowing, interactable prop" without rebuilding it each time.

## What Scene Graph actually is here

Unlike Fortnite Creative devices (actors with a ToyOptionsComponent), a Scene Graph object is an **`entity` tree**. Each node holds **`component`s**. A prefab (`P_LightPost`) is that tree saved as one asset and instantiated as a `_C` class.

**Observed** — live `EntityToolset.GetComponents` on the last light-post instance:

```
P_LightPost                         EntityPrefab P_LightPost_C
  mesh SM_Lightpost_01
  lantern_interaction_component     (only on the finished instance)
  Pivot                             plain entity
    keyframed_movement_component
    simple_movement_component
    Lantern                         plain entity
      mesh SM_Lightpost_Lantern_01
      interactable_component
      SphereLight                   sphere_light_component
    Sound_*                         sound_component (MetaSound)
    VFX                             particle_system_component
```

Child nodes of a prefab instance are class `entity`, not `P_LightPost_C`. The prefab class is only the root.

## The curriculum (five instances of the same prefab)

The hall places five `P_LightPost` along X ≈ −6462, −5472, −4440, −3443, −2470 (**Observed**). Each instance is the previous plus one concern:

1. **Hierarchy + meshes.** Post, Pivot, Lantern. Each has `transform_component`. Board: `Basic hierarchy of Entities and Mesh Components`.
2. **Light.** Child `SphereLight` with `sphere_light_component`. Board: `Light Component Added`.
3. **VFX + extra audio.** `NSYS_LampFlies` on a `VFX` child; moth-loop MetaSound. Board: `Particle System Component adds the moths VFX`.
4. **Motion.** `simple_movement_component` on **Pivot** (`MovementMode = PingPong`, `AutoPlay = true`) driving `keyframed_movement_component`. Descendants (lantern, light, VFX, sounds) ride the pivot.
5. **Interaction.** `lantern_interaction_component` on the **root**, `interactable_component` (`Enabled = true`) on **Lantern**.

Instancing a prefab is cheap on unique assets (meshes/textures are shared). Overrides on one instance (extra components, different keyframes) do not mutate the prefab asset — that is the point of the last two booths.

## Authoring rule

Put the component on the entity that *owns the concern*:

| Concern | Lives on |
|---|---|
| Sway | Pivot (rotation origin) |
| Glow | SphereLight under Lantern |
| Player interact prompt | Lantern mesh (the thing you look at) |
| Toggle logic | Prefab **root** (`FindDescendantComponents` reaches everything) |

Logic that says "turn off every light on this prop" belongs on the root and queries by type. Logic that says "this mesh can be interacted with" belongs on that mesh.

## Generated asset classes

With `EnableVerseAssetReflection`, the prefab and its meshes become Verse classes in `Assets.digest.verse`:

- `LightPost.P_LightPost{}` — construct a new instance in Verse
- `LightPost.SM_Lightpost_Lantern_01` — the lantern *mesh component type*
- `Meshes.SM_Toy_Cow` — used as a typed cast in overlap queries

`lantern_interaction_component` does `FindDescendantComponents(LightPost.SM_Lightpost_Lantern_01)` then `set MeshComponent.M_Lantern = MaterialInstance`. That only compiles because reflection emitted those names.

## Don't

- Don't parent the lantern to the post without a Pivot if you want it to swing — the pivot *is* the animation origin.
- Don't assume every instance of a prefab has the same component set. Stations 1–3 have no `simple_movement_component`; station 5 is the only one with `lantern_interaction_component`. Overrides are per instance.
- Don't look for this pattern in Creative devices. Devices don't nest like this.
