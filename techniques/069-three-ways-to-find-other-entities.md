---
id: technique-069
kind: technique
---

# Three ways to find other entities

**Sources:** scene_graph_template
**Snippets:**
- `../snippets/verse/EntityEnteredExampleComponent.verse`
- `../snippets/verse/FindOverlapHitsExampleComponent.verse`
- `../snippets/verse/FindSweepHitsExampleComponent.verse`

## Problem

You need to know when one Scene Graph entity meets another: a cow flying into a UFO, a beam covering a cow, a beam *blocked* by a shield. Creative volumes only tell you about *players*. Entity-vs-entity uses Scene Graph queries.

## Pick by precision

| Need | API | this sample |
|---|---|---|
| "Did *this mesh* overlap *any* other mesh?" | `mesh_component.EntityEnteredEvent` | Cow lifts itself; hides when it enters the UFO |
| "What is inside this volume *right now*?" | `entity.FindOverlapHits(Transform, Shape)` | UFO builds a `collision_capsule` under itself and casts `Meshes.SM_Toy_Cow[Overlap.TargetComponent]` |
| "What is the *first* thing along this ray/sweep?" | `entity.FindSweepHits(DisplacementVector)` | Sweep `Up := -300` from the UFO mesh; a globe in between blocks the cow |

All three booths then share the same abduction follow-through: build a `keyframed_movement_delta` (or play authored keyframes), `Play()`, wait until the cow meets the UFO, `RemoveFromParent()` or hide.

## 1. Mesh event — `EntityEnteredEvent`

Shipped code (not the tutorial) waits on the event inside the animation loop:

```verse
loop:
    Entity.SetGlobalTransform(StartTransform)
    set Mesh.Visible = true
    MovementComponent.SetKeyframes(Keyframes, oneshot_keyframed_movement_playback_mode{})
    MovementComponent.Play()
    Mesh.EntityEnteredEvent.Await()
    Sleep(0.2)
    MovementComponent.Stop()
    set Mesh.Visible = false
    Sleep(2.0)
```

Round-gated via `GetFortRoundManager[].SubscribeRoundStarted`. The cow entity owns the component; the UFO is just scenery the mesh happens to enter.

**When:** the moving thing can detect the overlap itself. No shape math.

## 2. Overlap query — `FindOverlapHits`

The UFO owns the query. Player steps on `Volume: HMD2` (`@editable Trigger: volume_device`). UFO `Pause()`s patrol, enables beam mesh + `spot_light_component` + beam audio, then:

```verse
CollisionCapsule := collision_capsule{Radius := 36.0, Length := 328.0}
var CollisionTransform:transform = Entity.GetGlobalTransform()
set CollisionTransform.Translation.Up = CollisionTransform.Translation.Up - 164.0

for:
    Overlap : Entity.FindOverlapHits(CollisionTransform, CollisionCapsule)
    CowMesh := Meshes.SM_Toy_Cow[Overlap.TargetComponent]
    CowPrefab := CowMesh.Entity
do:
    spawn { AbductCow(CowPrefab) }
```

The `Meshes.SM_Toy_Cow[...]` failable cast **is** the filter. Non-cow hits fail the `for` guard and are ignored. Length 328 / offset 164 = capsule centered under the hull.

Abduction then waits `CowMesh.EntityEnteredEvent.Await()` before `CowEntity.RemoveFromParent()` — overlap finds the candidate, the mesh event confirms it arrived.

## 3. Sweep query — `FindSweepHits`

Same player-volume pattern (`Volume: HMD3`). Difference: first hit along a vector, so occluders work.

```verse
if (Child := Entity.GetEntities()[0]):
    DisplacementVector := vector3{Left:=0.0, Up:=-300.0, Forward:=0.0}
    FirstSweepHitEntity := FindFirstSweepHit(Child, DisplacementVector)
    if (HitEntity := FirstSweepHitEntity?, HitEntity.GetComponent[Meshes.SM_Toy_Cow]):
        spawn { AbductCow(HitEntity) }

FindFirstSweepHit(InEntity:entity, DisplacementVector:vector3):?entity =
    for (SweepHit : InEntity.FindSweepHits(DisplacementVector)):
        return option{ SweepHit.TargetComponent.Entity }
    return false
```

`GetEntities()[0]` is the UFO's first child (BodyMesh) — sweep from the *mesh*, not the empty root. Two `Globe*` entities (`Meshes.Sphere_93D4BF41`) sit over cows in this booth; they are the occluders.

A `sweep_hit` also has contact position / normal / distance; this sample only uses `TargetComponent.Entity`.

## Shared follow-through (`AbductCow`)

Both overlap and sweep:

1. `set IsAbducting = true` so re-entry doesn't double-fire
2. Delta transform UFO − cow, scale 0, 2 s, `ease_in_cubic_bezier_easing_function{}`
3. `SetKeyframes` + `Play` on the **cow's** `keyframed_movement_component`
4. Cow `sound_component.Play()`
5. `CowMesh.EntityEnteredEvent.Await()`
6. `CowEntity.RemoveFromParent()`
7. Resume UFO `Play()` and disable the beam

`EnableAbductionBeam` / `Disable` are `FindDescendantComponents` on `Meshes.S_EV_SimpleLightBeam_01`, `spot_light_component`, and `Sounds.MSS_Ship_Beam`.

## Don't

- Don't use a Creative volume to detect *cows*. Volumes are for agents. The three `Volume: HMD*` devices in this sample only start the query.
- Don't string-compare mesh names. Cast to the generated class.
- Don't sweep from an entity with no mesh if you meant to hit geometry — take `GetEntities()[0]` (or a named child) as they did.
- Don't trust chapter 03's `Subscribe(OnEntityEntered)` or `PlayCowAbdcutionAnimation`. Shipped awaits `EntityEnteredEvent` inside a `loop` and hides the mesh (`Visible = false`), it does not Subscribe + remove. Capsule offset in docs is 248, shipped is 164. Use the snippets. See digest *Official tutorial vs shipped code*.
