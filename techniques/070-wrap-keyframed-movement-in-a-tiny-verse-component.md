---
id: technique-070
kind: technique
---

# Wrap keyframed movement in a tiny Verse component

**Sources:** scene_graph_template
**Snippet:** `../snippets/verse/SimpleMovementComponent.verse`

## Problem

`keyframed_movement_component` plays a list of `keyframed_movement_delta` with a playback mode. It does **not** expose those as designer-friendly `@editable`s in this sample — live `ListComponentProperties` on the native component returned **empty**. You need a Verse wrapper that:

- owns the keyframes and the playback enum
- guarantees a `keyframed_movement_component` exists (`GetComponent` or `AddComponents`)
- maps `OneShot` / `Loop` / `PingPong` onto the three `*_keyframed_movement_playback_mode` types
- starts playback at round start if `AutoPlay`

## The wrapper

```verse
movement_mode<public> := enum:
    OneShot
    Loop
    PingPong

simple_movement_component<public> := class<final_super>(component):
    @editable
    var Keyframes: []keyframed_movement_delta = array{}
    @editable
    var AutoPlay: logic = true
    @editable
    MovementMode: movement_mode = movement_mode.Loop
```

On simulate: if the entity already has `keyframed_movement_component`, initialize it; else construct one with `{ Entity := Entity }` and `AddComponents`. The `{ Entity := Entity }` initializer is required (Verse error 3600 otherwise — same trap as `particle_system_component`).

Playback mapping:

| `movement_mode` | Playback object |
|---|---|
| OneShot | `oneshot_keyframed_movement_playback_mode{}` |
| Loop | `loop_keyframed_movement_playback_mode{}` |
| PingPong | `pingpong_keyframed_movement_playback_mode{}` |

Then `SetKeyframes(Keyframes, PlaybackMode)`. `Play()` is deferred to `OnRoundStarted` gated on `AutoPlay?` and `GetFortRoundManager[].SubscribeRoundStarted`.

**Shipped vs docs:** chapter 02 still names this `basic_movement_component`, uses `OnSimulate` + `Sleep(0.1)`, `AddComponents of array {…}` (not valid Verse), and `Play()` inside initialize. Shipped file is `simple_movement_component`, `OnBeginSimulation` + round subscribe, `AddComponents(array{…})` with `{ Entity := Entity }`, Play in `OnRoundStarted`. See digest *Official tutorial vs shipped code*. Train on this snippet.

## Where it is reused (**Observed** live)

| Entity | MovementMode | Keyframe count | Job |
|---|---|---|---|
| Light-post Pivot (×2) | PingPong | 1 | Lantern sway |
| UFO root (first) | PingPong | 2 | Hull patrol |
| UFO root (overlap / sweep) | Loop | 4 | Hull patrol |
| UFO `TopMesh` (all three) | Loop | 2 | Hub spin |

Same component, different `@editable` values. That is the reusable unit — not a per-prop movement script.

The sweep-UFO instance's Keyframes **refPath into the overlap-UFO's delta objects**. Copy-paste of the array across instances; it still plays. Don't cite that sharing as a technique.

## Building a keyframe in Verse (not in the outliner)

Overlap/sweep abduction doesn't use `@editable` keyframes. It builds one at runtime from the delta between UFO and cow:

```verse
DeltaTransform:transform = transform:
    Translation:= Entity.GetGlobalTransform().Translation - CowEntity.GetGlobalTransform().Translation
    Scale := vector3{Left:= 0.0, Up:= 0.0, Forward:= 0.0}

Keyframe := keyframed_movement_delta:
    Transform := DeltaTransform
    Duration := 2.0
    Easing := ease_in_cubic_bezier_easing_function{}

MovementComponent.SetKeyframes(array{ Keyframe }, oneshot_keyframed_movement_playback_mode{})
MovementComponent.Play()
```

`Scale := 0` in the delta is how they shrink the cow during abduction.

## Controls on the native component used elsewhere

`Pause()` / `Play()` / `Stop()` — UFO overlap/sweep pauses hull patrol while abducting, then `Play()` to resume. `Stop()` on the cow's component before hiding/removing it.

## Don't

- Don't author a new movement component per prop. Parameterize this one.
- Don't call `Play()` in `OnBeginSimulation` if the animation should line up with round start — this sample waits for `SubscribeRoundStarted`.
- Don't pass degrees into SpatialMath rotation helpers (not used here, but the same file family). Deltas here are transforms, not yaw/pitch/roll.
- SpatialMath vectors are `Left` / `Up` / `Forward`, never `X` / `Y` / `Z`.
