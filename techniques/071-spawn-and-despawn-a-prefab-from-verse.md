---
id: technique-071
kind: technique
---

# Spawn and despawn a prefab from Verse

**Sources:** scene_graph_template
**Snippet:** `../snippets/verse/SpawnPrefabDevice.verse`

## Problem

You want a Scene Graph prefab to appear when a player enters a volume and disappear when they leave. Creative prop spawners don't spawn `entity` prefabs. The simulation entity is the parent.

## Shipped pattern

`SpawnPrefabDevice` is the template's only `creative_device`. It `@editable`s a `volume_device` (`Volume: HMD01` on the floor pad — **Inferred** from position; live binding not read). Board: `Add & Remove Entities From the World`.

```verse
SpawnPrefabDevice := class(creative_device):
    @editable
    TriggerVolume:volume_device = volume_device{}

    OnBegin<override>()<suspends>:void =
        SpawnTransform:transform = transform:
            Translation := vector3:
                Left := -9084.0
                Up := -8.0
                Forward := -919.0

        if:
            SimulationEntity := GetSimulationEntity[]
        then:
            loop:
                SpawnedPrefab:entity = LightPost.P_LightPost{}
                TriggerVolume.AgentEntersEvent.Await()
                SimulationEntity.AddEntities(array{SpawnedPrefab})
                SpawnedPrefab.SetGlobalTransform(SpawnTransform)
                TriggerVolume.AgentExitsEvent.Await()
                SpawnedPrefab.RemoveFromParent()
```

Construct the prefab **inside** the loop so each enter is a new instance. `LightPost.P_LightPost{}` exists because Verse asset reflection emitted the prefab class.

`GetSimulationEntity[]` is failable — must sit in an `if`. `AddEntities` takes an array. `SetGlobalTransform` after add (not before). `RemoveFromParent` is the despawn; there is no `Destroy()`.

Hardcoded spawn translation is a template shortcut, not a lesson. Drive it from an `@editable` transform or a marker entity in real work.

## Docs drift

Chapter 03 constructs `PrefabToSpawn` **once**, wraps enter vs exit in `race`, and re-`AddEntities` the same object after `RemoveFromParent`. Shipped code does not race and **constructs a new** `LightPost.P_LightPost{}` every enter. Do not emit `race` here — sequential Awaits are what runs. If you re-add a removed entity it may or may not be valid; the shipped author didn't rely on it. Full table in the digest *Official tutorial vs shipped code*.

## Related: removing from a query

Overlap/sweep abduction also despawns with `CowEntity.RemoveFromParent()` after `EntityEnteredEvent.Await()`. Same primitive, different caller (a `component` on the UFO, not a `creative_device`).

## Don't

- Don't spawn into the level actor list. Simulation entity is the Scene Graph root at runtime.
- Don't use XYZ. `vector3{Left, Up, Forward}`.
- Don't expect `volume_device` to spawn the prefab by itself — it only signals the agent. The device is glue.
