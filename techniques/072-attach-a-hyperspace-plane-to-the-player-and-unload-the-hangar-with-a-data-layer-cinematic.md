---
id: technique-072
kind: technique
---

# Attach a hyperspace plane to the player and unload the hangar with a data-layer cinematic

**Sources:** space_station_sample
**Tutorial:** Vistas and Biome Environments — Hyperspace — **External** (62-line snippet matches shipped Verse). Bindings **Observed**.

Contrast (do not duplicate): no-Verse Trigger→Cinematic on trigger-cinematic-quake.md. Button→sequence on cinematic-sequences.md. Data-layer *naming / streaming-off* on world-partition-data-layers.md. Chair-gated cine on chair-gated-cinematic.md. This island’s NPC doorway swap is **absent** — hangar Guards stay on guard-patrol-volume-swap.md.

## Problem

A hangar jump-to-hyperspace shot needs (1) streaks that stay in front of the moving player, (2) enter/exit VFX that designers can retune in Sequencer, and (3) the hangar itself gone for the duration so the sky reads as space — without simulating 1500 props off-screen.

## Pattern

### 1. One Verse device owns the clock; cinematics own the look

Authored: `Content/Attach_Hyperspace.verse` (`hyperspace_sequence_device`). Placed in `Hyperspace/`, `visibleInGame=false`, `enabled at Game Start=true`.

Live `@editable` values (**Observed** — `GetDeviceProperties` + `savedActor` on the mangled field objects):

| Field | Bound actor | Notes |
|---|---|---|
| `HyperspacePlane` | `BP_Hyperspace` “Hyperspace Plane” | `creative_prop`; `MoveTo` target |
| `VolumeStart` | `Device_Volume` “Hyperspace Trigger Volume” | box 3×1×2 m; `visibleInGame=false`; player+guard events on |
| `BeginHyperspace` | Cinematic Sequence — Enter | `LS_Enter_Hyperspace`; GameplayOnly; HUD shown; input **not** disabled |
| `EndHyperspace` | Cinematic Sequence — Exit | `LS_Exit_Hyperspace` |
| `DataLayerController` | Cinematic Sequence — Data Layer | `LS_DataLayerController` |
| `HyperspaceDuration` | **6.0** | race timeout vs attach loop |
| `AttachmentOffset` | `{X:0, Y:-1500, Z:700}` | cm in front / above; rotation frozen at enter |

`ListEventBindings` on the volume and the three cines is **[]**. The volume fires through Verse `VolumeStart.AgentEntersEvent.Subscribe(GetAgent)`, not a device binding.

### 2. `race` attach-loop against a Sleep timeout

```
OnBegin → Volume.AgentEntersEvent.Subscribe(GetAgent)
GetAgent → spawn{ RunHyperspaceSequence }
RunHyperspaceSequence:
  capture StartingRotation
  BeginHyperspace.Play()
  DataLayerController.Play()     # hide hangar
  race:
    loop: Sleep(0.01); MoveTo(player.Translation + offset, StartingRotation, 0.1)
    Sleep(HyperspaceDuration)    # 6s
  BeginHyperspace.Stop()
  EndHyperspace.Play()
  Sleep(3.0)                     # “length of HyperspaceExit NS”
  EndHyperspace.Stop()
```

`MoveTo` uses the **enter** yaw, not current yaw — looking around does not spin the tunnel. `spawn{}` so a second enter can overlap (no debounce). `Diagnostics` is imported and unused (shipped).

Docs tutorial Verse is the same 62 lines. Filename in docs is `hyperspace_sequence_device.verse`; shipped file is **`Attach_Hyperspace.verse`**. Train on snippets/verse/Attach_Hyperspace.verse.

### 3. `BP_Hyperspace` is the prop the sequences dress

Components (**Observed** — `ActorTools.get_components`):

| Component | Role |
|---|---|
| `SM_Hyperspace_Plane` | authored `Content/Hyperspace/SM_Hyperspace_Plane` |
| `Niagara - Enter` | `/SPACE_THEME_D_Assets/…/NS_Enter_Hyperspace`, `bAutoActivate=true` |
| `Niagara - Exit` | `/SPACE_THEME_D_Assets/…/NS_Exit_Hyperspace`, `bAutoActivate=true` |
| `PointLight` | cyan, intensity **1e6**, attenuation **10000** |

Cines keep `showHUDDuringCinematic=Show` and `disablePlayerInput=No` — the player still flies the starfighter (or walks) while the tunnel tracks them.

Audio: `Device_CRD_AudioMixer` in `Devices/` with mix `Content/Hyperspace/Audio/Hyperspace_Mix`, `bActivateOnGameStart=true`, `canBeHeardBy=Everyone`. Launch/Arrive/Travel cues live next to the mix. Mixer bindings empty — the mix is on for the match, not toggled by Verse.

### 4. One data layer is the hangar hose

`Content/Datalayers/DL_Hangar`. World instance: `initialRuntimeState=Activated`, `bIsInitiallyVisible=true`, streamingPriority 0, `overrideBlockOnSlowStreaming=NoOverride`.

`get_actors_in_data_layer` → **1554** actors (the `/SPACE_THEME_B_*` hangar + `PID ENV_PACK SW HangarRoom Prefab A`). `get_referencers` on the layer asset: `LS_DataLayerController` **and** `LS_Enter_Hyperspace`.

The map tag is **`LevelHasStreamingDisabled=1`**. The layer is not a World Partition streaming lesson. It is a Sequencer-driven hide so 1554 hangar actors do not fight the sky during the jump. Docs call this “unload the Death Star prefab”; the playset label is **HangarRoom**.

Keyframe times inside the three `LS_*` assets are **unanswerable**.

## Don't

- Don't wire Volume `OnEnter` → Cinematic `PlayFunction`. This shot is Verse subscribe + `Play()` so the attach loop can `race` the duration.
- Don't `MoveTo` with live player rotation. Freeze yaw at enter or the tunnel wobbles.
- Don't expect `GetDeviceProperties` friendly names to return the bound actor path. Read `savedActor` on `__verse_0x…` field objects.
- Don't treat `DL_Hangar` as character_environment streaming choreography. Streaming is **off**; the cine unloads a single Activated layer.
- Don't copy IceBase/VolcanicTown doorway volumes onto this volume. `NPCs/01` spawners are all `enabledAtGameStart=true` with empty `enable.eventSubscriptions`.
