---
id: technique-073
kind: technique
---

# `ArraySync` / `ArrayRace`: divide-and-conquer concurrency over an array

**Sources:** speedway_race_template, minigame_mastery_template
**Snippet:** `../snippets/verse/sync_array.verse`
**Tutorial:** hub page names and links `ArraySync()` (as an external Community snippet, not part of this tutorial's own code listing) — prose description matches the shipped file.

## Problem

You have an array of N players/elements and need to run the same async function over every element, then either:
- **wait for all of them** to finish and collect their results (`sync`), or
- **wait for the first one** to finish and cancel the rest (`race`)

...without writing a bespoke `sync`/`race` block by hand every time N changes at runtime (you don't know N until the match starts).

## Pattern

Both functions recursively split the array in half and let Verse's native `sync`/`race` expressions do the divide-and-conquer:

```verse
ArraySync<public>(Array:[]t, Function(:t)<suspends>:u where t:type, u:type)<suspends>:[]u=
    if (Array.Length > 1):
        Mid := Floor(Array.Length / 2)
        LeftArray := Array.Slice[0, Mid]
        RightArray := Array.Slice[Mid]
        SyncResult := sync:
            ArraySync(LeftArray, Function)
            ArraySync(RightArray, Function)
        SyncResult(0) + SyncResult(1)
    else if (First := Array[0]):
        array{Function(First)}
    else:
        array{}

ArrayRace<public>(Array:[]t, Function(:t)<suspends>:u where t:type, u:type)<suspends>:u=
    if (Array.Length > 1):
        ...
        race:
            ArrayRace(LeftArray, Function)
            ArrayRace(RightArray, Function)
    else if (First := Array[0]):
        Function(First)
    else:
        Sleep(Inf)
        Err()
```

Base case for `ArrayRace` with zero elements sleeps forever and calls `Err()` — a `race` with no branches never completes, so this documents "shouldn't happen" rather than silently returning a wrong value.

## Usage in this sample

- `race_end_manager.verse`: `ArraySync(Players, WaitForPlayerToFinishRace)` — wait for every player to either finish or leave before ending the game.
- `starting_game_sequence.verse`: `ArraySync(PlayersAndOrder, WaitForPlayerIntro)` — run N concurrent "wait for my turn in the cinematic" coroutines, one per player, without knowing N until the race starts.

Both Speedway call sites use `ArraySync`; `ArrayRace` ships in the same file but **is not called anywhere else in this sample** (Observed — no other references found). It exists as the natural pair.

**Minigame Mastery** is the first corpus call site for `ArrayRace`: `qte_game_device.WaitForEnd` does `ArrayRace(Towers, WaitForTower)` so the first tower `FailedEvent` ends the round and cancels the other wait. File: `utils.verse` (community snippet oYRX; no `ArraySync` sibling). Docs even suggest changing it to last-tower-standing.

## Don't

- Don't write a fixed 2-, 4-, or 8-way `sync`/`race` block for a player-count that varies — this scales to any N via recursion.
- Don't reach for this if you only ever have a fixed small N known at compile time — a literal `sync: A() B() C()` is more readable for exactly 3 known branches.
