---
id: technique-074
kind: technique
---

# Persistable stats table + generic sort = a leaderboard

**Sources:** speedway_race_template
**Snippets:** `player_stats_table.verse`, `player_stats_manager.verse`, `player_leaderboards.verse`, `merge_sort.verse`
**Tutorial:** Make Your Own In-Game Leaderboard in Verse — closely matches; shipped is a superset (adds Podiums + BestLapTime to the tutorial's Points/Wins).

## Problem

Track lifetime stats per player across sessions, then display the *sorted* top players on billboards and player-reference props — without hardcoding a sort per stat.

## Data layer

```verse
player_stats_table<public> := class<final><persistable>:
    Version:int = 0
    Points:int = 0
    Podiums:int = 0
    BestLapTime:float = -1.0

player_and_stats<public> := struct:
    Player<public>:player
    StatsTable<public>:player_stats_table

var PlayerStatsMap:weak_map(player, player_stats_table) = map{}
```

`<final><persistable>` is required for anything stored across sessions — Verse persistable data must live in a `weak_map`.

## Comparator convention

Every comparator takes `(Left, Right):player_and_stats`, checks a condition, and **returns `Left`** — meaning "fails if Left doesn't win, succeeds returning Left if it does." This shape is exactly what generic `MergeSort`'s `Compare` parameter expects:

```verse
MorePointsComparison<public>(Left:player_and_stats, Right:player_and_stats)<decides><transacts>:player_and_stats=
    Left.StatsTable.Points > Right.StatsTable.Points
    Left
```

Six of these ship: More/Less × Points/Podiums, plus Slower/FasterLapTimeComparison (note: lower lap time is *better*, so "Faster" is the winning direction — comment this explicitly, the polarity is easy to get backwards).

## Sort + apply

```verse
UpdateStatsBillboards<public>(Players:[]agent, StatsBillboards:[]billboard_device):void=
    var PlayerAndStatsArray:[]player_and_stats = for (Agent:Players, Player := player[Agent], PlayerStats := GetPlayerStats[Player]):
        player_and_stats:
            Player := Player
            StatsTable := PlayerStats

    SortedPlayersAndStats := SortingAlgorithms.MergeSort(PlayerAndStatsArray, MorePointsComparison)

    for (PlayerIndex -> PlayerAndStats : SortedPlayersAndStats, StatsBillboard := StatsBillboards[PlayerIndex]):
        UpdateStatsBillboard(PlayerAndStats.Player, StatsBillboard)
```

`UpdatePlayerReferences` is the identical shape against `player_reference_device.Register()` instead of billboard `SetText`. Swap the comparator to change what "best" means without touching either function.

## Don't

- Don't hand-roll a stat-specific sort. Write one comparator per stat, reuse the same generic `MergeSort`.
- Don't forget `<final><persistable>` — a non-persistable stats class silently won't survive across sessions.
- Don't get the lap-time polarity backwards — lower is better, comparator names it explicitly (`FasterLapTimeComparison`, not `MoreLapTimeComparison`).
- Don't train quest-name persistable logs from this island — that is persistable-quest-line-verse-dialog.md.
