---
id: technique-075
kind: technique
---

# Two lifetimes, two maps: session-scoped rounds vs cross-session persistence

**Sources:** speedway_race_template
**Snippet:** `race_order.verse`
**Tutorial:** Custom Round Logic Using Verse — **unsuffixed URL**. Chapter and file agree. Extra teaching from docs: session weak_map resets **every round**; max **two** player weak maps per project; no round API so you store last-completed-round yourself.

## Problem

Two different kinds of state in the same racing game:

1. **Lifetime stats** (points, podiums, best lap) — must survive the player leaving and returning days later.
2. **Round-scoped bookkeeping** (finish order, last completed round) — must survive **round** boundaries inside one game, then reset when the game ends or a player leaves. A `weak_map(session, …)` **cannot** do this: session maps **reset every round** (Custom Round Logic, **External**). So this data lives on a *second* persistable **player** weak map and you reset it yourself.

## Pattern: two `weak_map`s with different key types

```verse
# Survives session end for this player, forever (until reset)
var CircuitInfo<public>:weak_map(player, player_circuit_info) = map{}

# Survives only as long as this session is running
var RoundInfo:weak_map(session, round_info) = map{}
```

`player_circuit_info` is `<final><persistable>` — tiny (`LastRoundFinishOrder`, `LastCompletedRound`), reset explicitly (`ResetCircuitInfo`) when a player leaves so stale data doesn't leak into a future unrelated session.

`round_info` is a **plain class**, not persistable — keyed by `session` (the current game instance), not `player`. It holds `PlayerStartPositions:[agent]start_position` and `CurrentRound:int`. When the session ends, this map (and everything in it) is gone — exactly the semantics you want for "who's racing this round."

## Why not one map

If round bookkeeping were persistable and keyed by player, a player rejoining a *new* session days later would see stale round/position data from last time. Keying the *round* state by `session` instead of `player` makes the lifetime automatic — no manual reset needed for that half. Keying the *stats* by `player` and marking them `persistable` is what makes them survive on purpose.

## Round advancement

```verse
RecordCurrentRound(Players:[]player, TotalRounds:int):void=
    var CurrentRoundInfo:round_info =
        if (Info := RoundInfo[GetSession()]):
            Info
        else:
            LastCompletedRound := GetLastCompletedRound(Players, TotalRounds)
            round_info:
                CurrentRound := LastCompletedRound + 1
    if (set RoundInfo[GetSession()] = CurrentRoundInfo):
    else:
        Test.ProjectLog("record round info in session weak map.")
```

First call in a session derives the current round from players' *persistent* `LastCompletedRound` (so if you're resuming after a crash mid-circuit, you resume near where you left off, clamped by `TotalRounds`). After that, the session-scoped map just tracks it directly.

## Don't

- Don't persist per-round bookkeeping (start positions, current round pointer) — it doesn't need to survive process restart and persisting it risks stale-data bugs.
- Don't key round-scoped data by `player` if you want it to vanish naturally at session end — key by `session`.
- Don't skip resetting a persistent per-player round record on `PlayerRemovedEvent` — leaving stale `LastRoundFinishOrder` around corrupts next round's start-order sort for a player who left mid-race.
