---
id: technique-076
kind: technique
---

# Overlay theme galleries on a leftover BR POI

**Sources:** social_game_sample
**Tutorial:** Working With social-elimination Islands — **External** (one-paragraph “twist on Greasy Grove”; gallery names). Live folders **Observed**.

Prison Starter assembled rooms as `PlaysetRoot`s over a mounted meadow kit. This island does the other assembly: **keep the BR POI**, drop theme galleries next to it as Outliner folders, zero playset roots.

## Problem

You want a branded 16-player sandbox that already has streets, interiors, and loot — not an empty landscape you kitbash from galleries alone, and not a mechanic syllabus (voting / giftbox / disguise).

## Pattern

Two top-level Outliner trees, almost equal size:

| Folder | Recursive actors | Job |
|---|---|---|
| `FN_Existing/` | **2757** | leftover Greasy Grove (`FN_Assets` 2733, PowerStation 12, Background 16, Landscape 1, Special 7) |
| `GreasyGames/` | **2634** | Social galleries whose names match the docs table |

`GreasyGames/` nested (**Observed** `get_folders` + folder counts):

```
Colorful/     824  — Colorful Building 1–5 + Colorful Entrance 1–2  (Beacon stairway kit)
Cornfield/    691
Dorm/         938  — Matrix beds ×182, coffins ×54 (props, not Hiding Props)
Playground/   180  — Quill jungle gym / slide / jacks
```

The theme geo is **mounted**, not authored: `/SOCIAL_WAVE_Beacon_*`, `/SOCIAL_WAVE_Matrix_*`, `/SOCIAL_WAVE_Core_*`, `/SOCIAL_WAVE_Quill_*`. Project `Content/` only ships `GameFeatureData` + DefaultHLODLayer.

Empty leftover folders after the nested move: root `Background/`, root `Colorful Entrance 2/`, `GreasyGames/Colorful/Building 1–4/`, `Entrance 1/` — same leftover-root class as Prison.

**Contrast:** prison_sample uses **12** `PlaysetRoot` labels (`EnvironmentPack Quarters *`, …) and **0** leftover POI tree. alternate dimension / City Starter mount a theme kit onto authored atmosphere, not a BR neighborhood.

## Don't

- Don’t look for `PlaysetRoot` here. Live count **0**.
- Don’t treat `GreasyGames/Dorm` coffins as Giftboxes. Class is `BP_SOCIAL_WAVE_Matrix_Coffin_*`, not `Device_HidingProp_SwashTide`. Occupancy inspect lives on hiding-prop-occupancy-inspect.md.
- Don’t expect a twin `_Verse` map. One umap. Twin-level is twin-level-device-vs-verse.md.
- Don’t copy Prison’s interior Day Sequence volume. This island places **0** `BP_DaySequenceModifier_Device` and **0** PPV; lighting is default TOD + `LightingScalabilityManager C` + a few Point/Rect lights.
