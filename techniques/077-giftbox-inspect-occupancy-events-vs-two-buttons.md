---
id: technique-077
kind: technique
---

# Giftbox inspect: occupancy events vs two buttons

**Sources:** social_deduction_sample (`coffin_controller.verse` + Verse map)
**Tutorial:** Giftbox Device for Hiding and Teleportation

## Problem

Seekers tap a giftbox. If someone is inside, eject them. If empty, freeze the seeker. The device-only room does this with **two** buttons per prop (valid / invalid) plus a Mutator Zone. That binding mesh does not scale.

## Pattern

`hiding_prop_device` already emits hide enter/exit. One Verse device per prop:

1. `BeginPlayerHideEvent` / `EndPlayerHideEvent` keep `NumberOfPlayers`.
2. One Inspect `button_device`. Occupied → `EjectAllHiddenPlayers()`. Empty → audio + HUD + `spawn{FreezeSeeker()}`.
3. Freeze is a Mutator Zone `Enable` → `Sleep(FreezeTime)` → `Disable`. Default **5** s (**Observed** live `GetDeviceProperties`).

On `OnBegin`, the template **Enable then immediately Disable** the freeze zone — comment: the zone needs a warm-up. Copy that if the first freeze is a no-op.

Hidden travel: `ShouldRandomizeTravelTarget` (live sample **false**) can set `HiddenTravelGroup` / `HiddenTravelTargetGroup` so a hide teleports to another coffin. Device-only uses Timer `EnableHiddenTravel` / `DisableHiddenTravel` bindings instead.

`numberOfOccupants` on the SwashTide giftbox is **3** (**Observed**). Tutorial: max 3 hiders per box.

## Don't

- Don't keep both Inspect Valid and Inspect Invalid once `coffin_controller` is placed. The Verse map dropped the empty-box buttons.
- Don't treat `Device_HidingProp_SwashTide` as a unique API. Docs: cosmetic skin of Hiding Prop Gallery; Verse type is `hiding_prop_device`.
