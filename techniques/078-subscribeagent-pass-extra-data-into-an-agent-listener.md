---
id: technique-078
kind: technique
---

# `SubscribeAgent`: pass extra data into an `agent` listener

**Sources:** social_deduction_sample (`utils.verse` + `voting_device.verse`)
**External origin (file header):** Wrapping Subscribe to pass additional data

## Problem

`button_device.InteractedWithEvent` is `listenable(agent)`. Four answer buttons must call the same handler with *which* answer was pressed. Native `Subscribe` will not carry that extra value.

## Pattern

A tiny wrapper class closes over `ExtraData : t` and forwards `(Agent, ExtraData)`:

```
(Listenable : listenable(agent)).SubscribeAgent(
    OutputFunc : tuple(agent, t)->void,
    ExtraData : t where t:type) : cancelable
```

`voting_device` then:

```
SubscribeVote(Answer : answer):void=
    if (Button := AnswerButtons[AnswerToIndex(Answer)]):
        Button.InteractedWithEvent.SubscribeAgent(OnVote, Answer)
```

`OnVote(Agent, Answer)` appends the player to `Winners` or `Losers`. No four almost-identical lambdas.

There is a `listenable(tuple())` twin (`SubscribeEmpty`) in the same file; this island does not call it.

## Don't

- Don't fork a new wrapper per device. One `utils.verse` extension on `listenable(agent)` is the reusable bit.
- Don't index buttons by label string at runtime. Parallel arrays (`AnswerButtons[AnswerToIndex(Answer)]`) plus this subscribe is the shipped pairing.
