---
id: technique-079
kind: technique
---

# Twin levels: device-only minigame vs a `_Verse` sibling

**Sources:** social_deduction_sample, minigame_mastery_template
**Tutorial:** Social Deduction Create Voting Opportunities / Giftbox Device; Mastery Multiplayer Skill Checks — **External**. Device counts **Observed** live.

## Problem

The same minigame can be built with Creative bindings or with a small Verse device that owns the round. Shipping both in one Outliner folder makes the Verse version unreadable. You need two maps that share art, not two implementations stacked on top of each other.

## Pattern

**Social Deduction** ships **two World Partition maps**, 1500 OFPA each:

| Map | Job |
|---|---|
| `social_deduction_sample` | Device-only. Voting Group + Options, dual inspect buttons per giftbox. Default open. |
| `social_deduction_sample_Verse` | Same rooms. Adds `voting_device` (1) + `coffin_controller` (6). Drops the empty-inspect buttons. |

`social_deduction_sample_Verse_v2.umap` is an **ObjectRedirector** to the Verse map (**Observed** — `get_asset_tags` `DestinationObject`). Ignore it.

Open the `_Verse` asset when you want the Verse devices. `ListDeviceProperties` on the default map returns **0** ScriptDevices — they are not on that umap.

### What Verse actually buys

**Quiz.** Device-only: one hardcoded question, answer C is the only winner, each extra question would need more billboards + bindings. `voting_device` holds `[]question`, `SetText`s the boards, and treats any `CorrectAnswer` as the floor that stays. Cycle with `QuestionID` wrap.

**Giftbox inspect.** Device-only: two buttons per box (occupied eject / empty freeze). Verse map: **one** Inspect button per box; `BeginPlayerHideEvent` occupancy decides eject vs freeze. Live button labels: `Button - Inspect Valid` ×6, no `Inspect Invalid` (**Observed**). `FreezeTime` default **5**.

Do not merge the two setups in one level. The tutorial's point is the *diff*.

**Minigame Mastery** repeats the split for multiplayer QTE (`minigame_mastery_template` / `_Verse`, 2371 / 2341 OFPA). Default: **11** Skilled Interaction Devices (per-floor). Verse map: `qte_game_device` + two `qte_tower_device` — **one SID per tower** (plus the corridor sample). Floor hide is `[]cinematic_sequence_device` on the Verse device (`towerFloors` length **5**). Detail: skilled-interaction-queue.md.

social-elimination Starter Island does **not** ship a `_Verse` map.

## Don't

- Don't assume the Project Browser tile opened the Verse map. Confirm `get_current_level`.
- Don't train on the tutorial's pasted Verse length (277 / 137 lines). Shipped files are shorter — `snippets/verse/` wins.
