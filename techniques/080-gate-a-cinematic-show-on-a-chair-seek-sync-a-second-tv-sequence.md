---
id: technique-080
kind: technique
---

# Gate a cinematic show on a chair; seek-sync a second “TV” sequence

**Sources:** stand_up_template
**Snippet:** `../snippets/verse/show_template_device.verse`

## Problem

Linear show: lock the player, play a performance cinematic, let them pick camera angles, keep a “directed TV” cut in time with the act, restore normal camera and input when they stand up.

## Pattern

**Start**

```verse
TheChair.SeatedEvent.Subscribe(OnSeated)

OnSeated(Agent:agent):void=
    ReturnToFreeLook.Register(Agent)
    NextCamera.Register(Agent)
    PreviousCamera.Register(Agent)
    set ReturnToFreeLookSubscription = option{ReturnToFreeLook.PressedEvent.Subscribe(DoReturnToFreeLook)}
    ...
    set ChairExitSubscription = option{TheChair.ExitedEvent.Subscribe(OnChairExited)}
    if (MainSequencePlaying = false):
        spawn{RunSequence(Agent)}
    else:
        DoReturnToTVMode(Agent)
```

Sit-anim: `Sleep(1.0)` then `MainSequence.Play()` and `TVModeSequence.Play(Agent)`.

**End**

```verse
AwaitMainSequencingEnding(Agent:agent)<suspends>:void=
    MainSequence.StoppedEvent.Await()
    set MainSequencePlaying = false
    TheChair.Eject(Agent)
```

**TV rejoin** (don’t restart from 0):

```verse
CurrentSeekTime := MainSequence.GetPlaybackTime()
TVModeSequence.SetPlaybackTime(CurrentSeekTime)
TVModeSequence.Play(Agent)
```

**Cameras:** keep `var CurrentSequence:?cinematic_sequence_device = false`. Stop it before `CameraSwitches[i].Play`. Wrap with `Mod[CurrentCameraIndex ± 1, CameraSwitches.Length]`. TV-mode next starts at index 0 (`CurrentCameraIndex = -1` then +1).

**Teardown:** `Cancel()` every stored `?cancelable`, `Unregister` every trigger, Stop TV/alt cams. Input rebinds (Sprint → Previous Camera) must not outlive the chair.

## Don't

- Don’t Subscribe Input Triggers in `OnBegin` for the whole match — Register only while seated.
- Don’t Play TV from the start when returning mid-show — seek first.
- Don’t drop subscription handles; you need `Cancel` on exit.
- Don’t copy the docs typo `AwaitMainSequecingEnding`.
