---
id: technique-081
kind: technique
---

# AI leash is a Verse origin + two radii, not a Creative volume

**Sources:** stronghold_template
**Snippet:** `../snippets/verse/stronghold_leash_position.verse`
**Tutorial:** ch.4 Set Up Leash Devices

## Problem

Guards must stay in a camp unless patrolling a path, then return. Fallback = everyone to the keep. Creative has no "leash volume" device — Verse sets it on the character.

## Pattern

Place a **Verse device** as the origin (hide in-game). `@editable` spawners that get the leash on spawn; patrol paths that **clear** it while walking.

```verse
Leashable := Guard.GetFortCharacter[].GetFortLeashable[]
Leashable.SetLeashPosition(GetTransform().Translation, LeashInnerRadius, LeashOuterRadius)
# defaults Observed: inner 2300, outer 2400 (cm). Inner must be < outer.
```

`PatrolPathStartedEvent` → `ClearLeash()`. `PatrolPathStoppedEvent` → `SetLeashPosition` again.

Match **Patrol Path Group** on Guard Spawner and path nodes for the *initial* patrol; Verse APIs assign/disable at runtime.

`DisableLeashAndPatrolPaths`: ClearLeash everyone, `PatrolPath.Disable()`, empty the guard list — used at fallback so they aren't glued to outer camps.

Debug: Island Settings Debug + Navigation for navmesh; this file also `debug_draw` cylinders (green inner / orange outer).

## Don't

- Don't expect a stock "leash device" in the Creative palette — it's this class.
- Don't leave leash on during a scripted patrol or they'll rubber-band off the path.
- Don't set inner ≥ outer.
