---
id: technique-082
kind: technique
---

# Guard spawners as a perception bus; first alert calls reinforcements

**Sources:** stronghold_template
**Snippet:** `../snippets/verse/stronghold_game_manager.verse`

## Problem

Stealth island: many Guard Spawners. First time *anyone* goes Alerted, spawn reinforcements at the player who tripped it. If the squad wipes the camp never detected, a different End Game than if they were spotted. Last handful of alerted guards fall back to the keep.

## Pattern

Subscribe every relevant spawner once:

```verse
SpawnerDevice.SpawnedEvent.Subscribe(OnGuardSpawned)
SpawnerDevice.EliminatedEvent.Subscribe(OnGuardEliminated)
SpawnerDevice.SuspiciousEvent.Subscribe(OnGuardSuspicious)
SpawnerDevice.AlertedEvent.Subscribe(OnGuardAlerted)
SpawnerDevice.TargetLostEvent.Subscribe(OnGuardLostTarget)
SpawnerDevice.UnawareEvent.Subscribe(OnGuardUnaware)
```

Payload is `device_ai_interaction_result`: `Source?` = guard, `Target?` = player.

Keep `[]agent` buckets (Suspicious / Alerted / Investigating). Broadcast on **length 0↔1**, not every event — `PlayerDetectedEvent` fires when `AlertedGuards.Length = 1`.

**First alert only:**

```verse
if (not ReinforcementTriggered?, set DetectedPlayer = option{player[InteractionResult.Target?]}):
    for (S : GuardsReinforcementSpawnersInternal):
        S.Enable()
    MessageDeviceReinforcement.Show()
    ReinforcementsCalledEvent.Signal(Guard)
```

On reinforcement spawn: `ApplyLeashOnGuard` then `ForceAttackTarget(DetectedPlayer, ?ForgetTime:=30.0)`.

**Victory:** all eliminated + `ReinforcementTriggered?` → Detected End Game, else Undetected. After detection, when `NumGuardsSpawned - GuardsEliminated = 3` → `StartFallback()` (disable outer leashes, apply keep leash).

**Scale:** `if (NumberOfActivePlayers > 2)` append Additional spawner arrays. Tracker target = sum of `GetSpawnLimit()`.

Other devices Subscribe the manager’s events (alert mood, barks) — don’t duplicate perception.

## Don't

- Don’t Enable reinforcement spawners at match start.
- Don’t `ForceAttackTarget` before `DetectedPlayer` is set.
- Don’t copy docs `ReinforcementLeashReferernce`.
