---
id: technique-083
kind: technique
---

# Orbit camera stack: `AddTo`/`RemoveFrom` + delayed Hide/Show — not a camera-mode flag

**Sources:** character_environment_environment_template
**Tutorial:** character_environment Environment Template — Using the Orbit Camera Device — pastes this script.

## Problem

Players cycle 3P orbit → simulated 1P (hidden pawn) → default Fortnite cam on right-click / LT. Instant `Hide()` pops through the blend. A camera-mode enum doesn't exist — cameras are a **stack per agent**.

## Pattern

Wrap each `gameplay_camera_orbit_device` with a hide flag:

```
gameplay_camera := class<concrete>():
    Camera: gameplay_camera_orbit_device
    ShouldHideCharacter: logic = false
```

`camera_switch_mode_device` (creative_device):

1. Per player `spawn{ CameraLoop(Player) }`.
2. `loop: SwapCameraInputTrigger.PressedEvent.Await()`.
3. If a cam is on the stack: `CurrentCamera.RemoveFrom(Agent)`.
4. Next in `Cameras[]`: `Next.Camera.AddTo(Agent)`. Past end: index 0, `MaybeCurrentCamera = false` (default cam), `ShowCharacter`.
5. Hide/show **after a blend delay**, not same frame:

```
HideCharacter: Sleep(HideDelay=0.20) then FortCharacter.Hide()
ShowCharacter: Sleep(ShowDelay=0.30) then FortCharacter.Show()
```

6. `Sleep(1.0)` after each swap so the trigger can't spam.

Empty stack = Fortnite default. That's why the last press isn't another Orbit device.

Companion `starting_sequence_device`: don't enable minimap until `CrewQuartersButton.InteractedWithEvent.Await()` — welcome VO `Stop()`, then MapController + HUDController. Spreadsheet door-lock is **not** this script.

## Don't

- Don't `Hide()` the pawn the same tick as `AddTo` — 0.20s delay exists because of the blend.
- Don't treat Camera Types (Scene Graph components) as this island — this is **Orbit devices**.
- Don't forget `RemoveFrom` before `AddTo` or cams stack and never return to default.
- Don't document “3 cameras” as 3 devices if one of the three states is **no device** (default).
