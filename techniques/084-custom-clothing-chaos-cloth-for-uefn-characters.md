---
id: technique-084
kind: technique
---

# Custom Clothing & Chaos Cloth for UEFN Characters

**Sources:** character_environment_sample template + Epic's character environment sample doc

## Clothing is authored separately from the MetaHuman

In character_environment the garments do **not** live under `MetaHumans/` — they are their own content
tree (**Observed**):

```
Content/Characters/Cloth/Captain/
  Captain_Merged_CA               <- the merged clothing asset (Chaos Cloth)
  Combined/Captain_Head_PA        <- physics asset
  Materials/  M_Cloth_Base                       (one master material)
              MI_Cloth_Base_Jacket_LOD1
              MI_Cloth_Base_Jacket_LOD2
              MI_Cloth_Base_Shirt_LOD1
              MI_Cloth_Base_Shirt_Pant_LOD2      <- LOD2 merges shirt+pants+belt
              MI_Cloth_Base_Pants_LOD1
              MI_Cloth_Base_Belt_LOD1
              MI_Cloth_Boot_LOD1
  Belt/   Textures/Lod1/T_Cap_Belt_Lod1_{BaseColor,Normal,ORM}
  Boot/   Cap_Boots_SK  + Textures/Lod1/…
  Jacket/ Textures/Lod1/…  Textures/Lod2/…
  Pants/  Textures/Lod1/…
  Shirt/  Textures/Lod1/…  Textures/Lod2/T_Cap_ShirtPantsBelt_Lod2_{BaseColor,Normal,ORM}
```

## The authoring pipeline (External — Epic doc)

```
DCC package  →  Unreal Engine  →  convert to clothing asset  →  export to UEFN
(Marvelous Designer for the character_environment jacket/sweater)              ↓
                                                    used via the Chaos Cloth component
```

You cannot author cloth in UEFN. It round-trips through full Unreal Engine, where the mesh
becomes a *clothing asset*. Epic's companion pages: *Creating Clothing Assets for Unreal
Editor for Fortnite Using Unreal Engine*, and *Making a Sweater with Marvelous Designer in UEFN*.

**Crossing into UEFN uses the migrate tool, and the essential assets are just two: the cloth
asset and the physics asset** (if one was used) (**External**). character_environment matches exactly —
`Captain_Merged_CA` (cloth) and `Captain_Head_PA` (physics) are the only two assets in the
garment tree that aren't textures or material instances (**Observed**). Useful check when
migrating: if you're moving more than those two plus your textures/materials, you're probably
dragging dependencies you don't need.

**Unlike the MetaHuman itself, none of this is automated.** The MetaHuman importer optimises
the character for you; clothing is your own DCC + full-UE work. This is the part of the
character_environment workflow a user genuinely has to reproduce by hand — which is why Epic publishes the
Captain Roux / Captain Elli learning assets as worked examples (see below).

## The gotcha that matters most (External — Epic doc)

> Cloth simulation on the captain is **turned off by default** for cross-platform performance,
> and **only runs while playing the cinematic sequence**.

This is the single most useful line in the documentation. Simulated cloth on a MetaHuman was
not affordable as an always-on feature even in Epic's own showcase. The pattern to copy:

**Treat cloth simulation as a cinematic-only effect.** Enable it when a sequence starts,
disable it when the sequence ends. Design the character to look correct with simulation off,
because that is its normal state. This also explains the template's
`BP_Captain_Roux_Sim_Use` naming (**Inferred**).

### How it is actually done (Observed — read from the live editor)

The doc describes the behaviour but not the mechanism. The mechanism is **a Sequencer
property track**, not Blueprint and not Verse.

1. On the placed `BP_Captain_Roux_Sim_Use` actor, the `Cloth_CA` (`ChaosClothComponent`) has
   **`bEnableSimulation` = false**. That is the saved, normal state.
2. The cinematic's sub-sequence **`LS_Cap_Helm_Anim`** possesses that exact actor
   (`PersistentLevel.BP_Captain_Roux_Sim_Use_C_UAID_047BCBA963E890EF01_1235794250`), binds its
   `Cloth_CA` component, and carries `MovieSceneBoolTrack`s on:
   - `ChaosClothComponent:bEnableSimulation` — shown as **"Enable Simulation"**
   - `ChaosClothComponent:bSuspendSimulation` — **"Suspend Simulation"**
   - `ChaosClothComponent:bReset`
3. Nothing disables it afterwards, because nothing has to: **Sequencer restores pre-animated
   state when a sequence finishes.** The property returns to the actor's saved `false`.

The parent sequence `LS_Cap_Helm` has no cloth track at all — it is camera cuts, transforms,
audio, fade, and a sub-track. All character work lives in the sub-sequence.

**The pattern to copy:** don't write toggle logic. Put the expensive component's enable flag
on a bool track inside the sequence that needs it, leave the actor's saved value off, and let
Sequencer's restore do the disabling. Same trick works for any costly component flag.

Two details worth having:

- `bSuspendSimulation` and `bReset` are keyed alongside `bEnableSimulation`. `bReset` snaps
  the cloth back to its bind pose — you want that on the frame simulation starts, or the
  garment settles visibly from wherever it was left.
- The same sub-sequence also drives `GroomComponent:SimulationSettings` /
  `HairSimulationSettings:bOverrideSettings` — **hair simulation is gated the same way**. Cloth
  is not the only thing Epic could not afford always-on.

### The cloth component is driven by the body, not animated

`Cloth_CA.leaderPoseComponent` = the **`Body`** skeletal mesh component (**Observed**). The
garment has no animation of its own; it follows the body's pose and simulates on top. The
actor's full component list is `Body`, `Face`, `Mustache`, `Hair`, `Eyebrows`, `Eyelashes`,
`Beard`, `Feet`, `Cloth_CA`, `MetaHuman`, `LODSync` — note there is **no `Torso` or `Legs`
component**: the merged clothing asset replaces them outright. Boots stay separate
(`Cap_Boots_SK`), and the cloth runs through the deformer graph
`DG_LinearBlendSkin_Morph_Cloth_RecomputeNormals` (**Observed**), which is why `OptimusCore`
appears in the module list.

## LOD strategy for garments (Observed)

Three techniques, all worth copying:

1. **Per-LOD texture sets.** `Textures/Lod1/` and `Textures/Lod2/` are distinct texture sets,
   not just mip levels of one set. Only garments that stay visually significant (jacket,
   shirt) get a Lod2 set; belt, boots, pants stop at Lod1.
2. **Merge garments at lower LODs.** `T_Cap_ShirtPantsBelt_Lod2_*` and
   `MI_Cloth_Base_Shirt_Pant_LOD2` collapse three garments into one texture set and one
   material at LOD2. Draw calls and texture count both drop.
3. **ORM packing.** Every texture set is exactly three maps: `_BaseColor`, `_Normal`, `_ORM`
   — Occlusion / Roughness / Metallic packed into one RGB texture. Three textures per
   garment per LOD, never more.

One master material (`M_Cloth_Base`) with per-garment, per-LOD instances. Same discipline as
the MetaHuman material stack.

## Naming convention (Observed)

`T_<Character>_<Garment>_<Lod>_<MapType>` → `T_Cap_Jacket_Lod1_ORM`
`MI_Cloth_<Master>_<Garment>_<LOD>` → `MI_Cloth_Base_Jacket_LOD2`

Note Epic's own inconsistency: `Lod1` in texture names, `LOD1` in material instance names.
Pick one case and hold it.

## Downloadable learning assets (External — Epic doc)

Epic publishes the source-of-truth assets for this workflow: Captain Roux (MetaHuman file,
Marvelous Designer jacket, Photoshop texture file, and a UE project with a jacket animation)
and Captain Elli (MetaHuman file, Marvelous Designer sweater, UE project). The download link
is in a comment at the top of `camera_switch_mode_device.verse` — reachable via **Verse
Explorer** in the UEFN main menu. The link as it appears in the template source:
`https://fn.gg/character_environment-Learning-Assets`. Use falls under the UEFN Supplemental Terms.

## The cloth asset itself (Observed — `Captain_Merged_CA`)

Class `ChaosClothAsset`. Its seven dependencies are the physics asset plus **six** material
instances — `MI_Cloth_Base_{Belt_LOD1, Jacket_LOD1, Jacket_LOD2, Pants_LOD1, Shirt_LOD1,
Shirt_Pant_LOD2}`. `MI_Cloth_Boot_LOD1` is *not* among them: the boots are a separate
skeletal mesh (`Cap_Boots_SK`), not part of the merged garment.

| Property | Value |
|---|---|
| `physicsAsset` | `Combined/Captain_Head_PA` |
| `minLod.perPlatform` | `desktop: 0`, `console: 1`, **`switch: 2`, `pS4: 2`, `xB1: 2`, `android: 2`** |
| `bSupportRayTracing` | true |
| `neverStream` | false |
| `shadowPhysicsAsset` | null |
| `dataflowInstance.dataflowTerminal` | `"ClothAssetTerminal"` |
| `dataflowInstance.dataflowAsset` | **null** |

### The finding that closes this backlog item, unhappily

**The Dataflow graph is not shipped.** Chaos Cloth Assets are built by a Dataflow node graph —
that graph is where solver settings (stiffness, damping, iterations, collision thickness,
per-LOD sim mesh) are authored. `Captain_Merged_CA` retains the terminal node's *name* but its
`dataflowAsset` is null, and no Dataflow asset appears among its dependencies.

So the asset ships as **baked simulation data with its construction graph stripped**. The
solver settings cannot be read, edited, or re-derived from this sample — not by the property
bridge, and not by opening the asset in UEFN either. There is nothing behind the door.

Two consequences worth carrying:

- **this sample cannot teach you Chaos Cloth solver tuning.** Use Epic's Chaos Cloth
  documentation and the downloadable character_environment learning assets (which include the source
  Marvelous Designer garments and a full UE project) instead. That is where the authoring
  graph lives.
- **When you migrate a cloth asset into UEFN, expect to lose the graph.** Keep the authoring
  project in full Unreal Engine as your source of truth; the UEFN copy is an output, not a
  master. This is the same relationship as the MetaHuman importer's baked output.

### Two different LOD ladders — do not confuse them

The cloth asset's own `minLod` is **stricter** than the MetaHuman clothing LOD settings asset
described below:

| | desktop | console | switch | mobile / android |
|---|---|---|---|---|
| `Captain_Merged_CA.minLod` | 0 | **1** | **2** | **2** |
| `MetaHuman_Clothing_LODSettings.minLod` | 0 | 0 | 1 | 1 |

They govern different things — the merged Chaos Cloth garment versus the MetaHuman's clothing
skeletal meshes — and the simulated garment is held to the tighter budget on every non-desktop
platform. On Switch and mobile the simulated cloth never renders above LOD 2.

## The physics asset is arms-only (Observed — `Captain_Head_PA`)

Despite the name, it contains **four bodies and nothing else**:

```
hand_l   hand_r   lowerarm_l   lowerarm_r
```

No head, no torso, no legs. These are exactly the colliders a jacket's sleeves need — the
forearms and hands that swing through the cloth. **Cloth collision is scoped to the bones that
actually intersect the garment**, not to the whole skeleton. That is most of why it is
affordable.

Other observations:

- **Collider type is `KTaperedCapsuleElem`** — tapered capsules (`Radius0`, `Radius1`,
  `Length`, plus `RestOffset`), not the sphere/sphyl/box primitives a ragdoll physics asset
  uses. Tapered capsules are the Chaos Cloth collider shape; a forearm is genuinely a cone.
  *(Practical note: `PhysicsAssetToolset.GetBodyShapes` returns an empty list for these
  bodies, because that API only enumerates Sphere/Capsule/Box. Empty does not mean no
  collision.)*
- **No constraints.** The bodies are collision-only; nothing is jointed.
- `CollisionEnabled: QueryAndPhysics`, `CollisionTraceFlag: CTF_UseSimpleAsComplex`.
- Solver settings (RBAN): `positionIterations` **6**, `velocityIterations` 1,
  `projectionIterations` 1, `cullDistance` **3**, `maxDepenetrationVelocity` 0,
  `fixedTimeStep` 0 (variable), `bUseLinearJointSolver` **true**, `bUseManifolds` false.
  Legacy `solverIterations` 3 / joint 2 / collision 2. `solverType: RBAN`.
- `PreviewSkeletalMesh` resolves to
  `MetaHumans/Cpt_Roux/Female/Tall/NormalWeight/Body/f_tal_nrw_body` — confirming Captain Roux
  is the **Female / Tall / NormalWeight** archetype, and corroborating that the *Medium* body
  in `MetaHumans/Common/` is importer boilerplate she does not use.

`positionIterations: 6` against the engine default of 1 is the one number here worth copying:
Epic bought stability on the sleeves with iterations, then made that affordable by keeping the
body count to four.

## The master cloth material (Observed — `M_Cloth_Base`)

**Shader cost: 236 pixel-shader instructions, 289 vertex-shader instructions, 3 samplers,
3 pixel texture samples, 2 UV scalars.** That is the "reduced-instruction material" claim in
metahuman-in-uefn.md as an actual number — and 3 samplers for 3
textures means not one texture is sampled twice.

Settings: `MD_Surface`, `BLEND_Opaque`, `MSM_DefaultLit`, **`twoSided: true`**,
`bUsedWithSkeletalMesh: true`, `bUsedWithClothing: true`, `numCustomizedUVs: 0`,
`bUseMaterialAttributes: false`. Two-sided is non-negotiable for cloth — you see the inside of
a jacket when it moves.

Twenty expression nodes. The complete parameter surface, which is what the seven `MI_Cloth_*`
instances actually drive:

| Parameter | Type | Default |
|---|---|---|
| `BaseColor` | Texture2D, `SAMPLERTYPE_Color` | `/EpicBaseTextures/Blanks/T_Default_BaseColor_White` |
| `ORM` | Texture2D, **`SAMPLERTYPE_Masks`** | `/EpicBaseTextures/Blanks/T_Default_Masks_SpecRoughMetal` |
| `Normals` | Texture2D, `SAMPLERTYPE_Normal` | `/EpicBaseTextures/Blanks/T_Default_NormalMap` |
| `Metalic` *(sic)* | Scalar | 1 |
| `Spec` | Scalar | 0.5 |
| `Rough` | Scalar | 1 |
| `AO` | Scalar | 1 |
| `FlattenNormal` | Scalar | 0 |
| `ColorMult` | Scalar | 1 |
| `Contrast` | Scalar | 0 |
| `Saturation` | Scalar | 0 |
| `Color` | Vector | (0.110, 0.137, 0.207) — dark desaturated blue |
| `PickColor` | **Static switch** | false |

Two engine material functions do the heavy lifting:
`Engine_MaterialFunctions01/Texturing/FlattenNormal` and
`Engine_MaterialFunctions01/ImageAdjustment/CheapContrast_RGB`, plus a `Desaturation` node and
four `Multiply` nodes. `MP_AmbientOcclusion` is driven from one of those multiplies (the `AO`
scalar against the ORM's occlusion channel).

Four things to copy:

1. **`ORM` is sampled as `SAMPLERTYPE_Masks`, not `Color`.** Packed data must be sampled
   linear. Sampling an ORM as sRGB is the single most common way to get subtly wrong roughness.
2. **Texture defaults point at `/EpicBaseTextures/Blanks/`** — neutral white / flat normal /
   default mask. A material instance that forgets to set a texture renders plausibly instead of
   rendering black. Cheap insurance on a master material with seven instances.
3. **Scalar multipliers alongside every packed channel** (`Metalic`, `Spec`, `Rough`, `AO`).
   The ORM texture provides the variation; the scalar lets an instance dial the whole channel
   without a new texture. This is why seven garment variants need one master.
4. **`PickColor` is a static switch, not a scalar blend.** Flat-colour mode compiles the
   texture path *out* rather than branching at runtime. Static switches are free; `if` nodes
   are not.

`list_parameter_groups` returns empty — every parameter is ungrouped, so the instance editor
shows one flat list. Fine at 13 parameters; group them before you reach 30.

## The cross-platform LOD lever (Observed — `MetaHuman_Clothing_LODSettings`)

| Setting | Value |
|---|---|
| `minLod.default` | 0 |
| `minLod.perPlatform` | **`switch: 1`, `mobile: 1`**, `console: 0`, `desktop: 0` |
| `minQualityLevelLod.default` | 0 |
| `bOverrideLODStreamingSettings` | false |
| `bSupportLODStreaming.default` | false |
| `maxNumStreamedLODs.default` | 0 |

**This is how the MetaHuman's clothing meshes are made affordable on weak hardware, and it is
one line.** Switch and mobile never render clothing LOD 0 — they start at LOD 1, so the
`Textures/Lod1/` sets above are the *best* those platforms will ever see. Console and desktop
get LOD 0.

That reframes the per-LOD texture work: `Lod1` is not a distant-viewing fallback, it is the
shipping quality bar for two whole platforms. Budget it accordingly. (The simulated Chaos
Cloth garment is held to an even tighter ladder — see the comparison table above.)

`bSupportLODStreaming: false` — clothing LODs are resident, not streamed. For a character
who is always on screen in a small map, streaming LODs would only add pop.

## Capture backlog

**Cleared.** All five items were resolved from a live UEFN session and are written up above —
including the negative result on solver settings, which is itself the answer.

One item closed as **unanswerable rather than unread**: `Captain_Merged_CA`'s Chaos Cloth
solver settings. The asset's Dataflow construction graph is not shipped (`dataflowAsset:
null`, absent from its dependency list), so those values do not exist anywhere in this
template. Opening the asset in UEFN will not reveal them. Do not re-add this to the backlog —
go to Epic's Chaos Cloth docs and the character_environment learning assets instead.

Note for anyone comparing physics assets: the MetaHuman Component's `torso` slot points at a
**hoodie** physics asset (`f_tal_nrw_top_hoodie_strings_Physics`) left over from MetaHuman
Creator's default outfit — see metahuman-in-uefn.md. `Captain_Head_PA`
is the one that actually drives cloth collision; the hoodie asset is dead boilerplate.
