---
id: technique-085
kind: technique
---

# Cinematic Sequences & Diegetic Interaction

**Sources:** character_environment_sample, read from a live UEFN session. Everything below is
**Observed** unless marked otherwise.

character_environment's entire interaction design is one button and one sequence. It is also the trigger
for the character's cloth and hair simulation, which makes this note load-bearing for
character-clothing-and-lods.md.

## The whole interaction, end to end

```
Device_Button_V2 "Button"          (invisible, 0.7 m radius, unlimited triggers)
    │  On Interact
    ▼
Device_CinematicSequence           PlayFunction
    │  Sequence
    ▼
LS_Cap_Helm                        camera cuts, transforms, audio, fade, sub-track
    │  MovieSceneSubTrack
    ▼
LS_Cap_Helm_Anim                   possesses Captain Roux — body/face anim, control rig,
                                   cloth sim, groom sim, visibility, materials
```

**One event binding exists in the entire map**: `Button → On Interact → Cinematic Sequence
Device → PlayFunction`. That is the complete device-to-device wiring of a first-party
showcase island.

## The button is invisible

| `Device_Button_V2` setting | Value |
|---|---|
| `VisibleDuringGame` | **False** |
| `InteractionRadius` | 0.7 |
| `Interaction Text` | *(empty)* |
| `InteractTime` | 0 |
| `Delay` / `ResetDelay` | 0 / 0 |
| `TimesCanTrigger` | **0** (unlimited) |
| `TriggerSound` | True |
| `EnabledAtGameStart` | True |

**The pattern: hide the device, let the world be the affordance.** The player does not
interact with a floating Creative button — they walk up to the Captain and the prompt appears.
An invisible `Device_Button_V2` parked on top of a character or prop is how you get a diegetic
interaction without writing any Verse.

Two details that make it work: `InteractTime` 0 (instant, no hold-to-interact bar) and
`TimesCanTrigger` 0 (replayable — the showcase is meant to be watched more than once).
`Interaction Text` is left empty so the default prompt is used.

## The sequence device does not take over the screen

| `Device_CinematicSequence` setting | Value |
|---|---|
| `Sequence` | `Cinematics/Cap_Reveal/LS_Cap_Helm` |
| `bAutoPlay` | **False** |
| `bLoopPlayback` | False |
| `PlayRate` | 1.0 |
| `EnabledOnPhase` | GameplayOnly |
| `Visibility` | Everyone |
| `ShowHUDDuringCinematic` | **Show** |
| `DisablePlayerInput` | **No** |
| `bLevelSequenceActorAlwaysRelevant` | True |

**`DisablePlayerInput: No` and `ShowHUDDuringCinematic: Show` are the interesting choices.**
This is not a cutscene that seizes control; it is a performance that happens in front of you
while you keep walking. It also means the player can still press the camera-swap input
mid-cinematic — the two systems are deliberately independent, which is why the Verse camera
device knows nothing about the sequence.

`bAutoPlay: False` is what makes the button meaningful. `bLevelSequenceActorAlwaysRelevant:
True` guards against the sequence actor being culled.

## Parent sequence vs. sub-sequence: a clean split

The division of labour is worth copying.

**`LS_Cap_Helm`** (the parent, ~0.9 MB) holds only *staging*:
`MovieSceneCameraCutTrack`, `MovieScene3DTransformTrack`, `MovieSceneAudioTrack`,
`MovieSceneFadeTrack`, `MovieSceneFloatTrack`, `MovieSceneSubTrack`. It also carries
`MovieSceneVisibilityTrack`s for set dressing — notably a group labelled **"Old Doors"** with
four visibility tracks.

**`LS_Cap_Helm_Anim`** (the sub-sequence, ~0.8 MB) holds all *character* work:
`MovieSceneSkeletalAnimationTrack`, `MovieSceneControlRigParameterTrack`,
`MovieSceneBoolTrack`, `MovieSceneComponentMaterialTrack`, `MovieSceneVisibilityTrack`,
`MovieSceneIntegerTrack`, `MovieSceneBindingLifetimeTrack`.

**Why split:** the character performance can be re-timed, re-recorded, or reused without
touching camera cuts and audio, and vice versa. If you have an animator and a cinematographer
working at once, this is the seam that stops them colliding.

Animation assets driven from the sub-sequence:
`LAS_Cap_Walkup_Anim_v3`, `LAS_NEWCap_Walkup_Anim_v2_Body`,
`LAS_NEWCap_Walkup_Anim_v2_Face`, `AS_MH-Cap-Greeting-B_T13_AnimPolish21-02` — separate body
and face tracks, as MetaHuman work always is. The `NEW`/`v2`/`v3` suffixes and the retained
older take are Epic's iteration debris, not a naming convention to copy.

## Expensive simulation is a keyed property, not logic

Covered in full in
character-clothing-and-lods.md.
In short: the sub-sequence binds the Captain's `Cloth_CA` component and keys
`ChaosClothComponent:bEnableSimulation` (plus `bSuspendSimulation` and `bReset`) on a bool
track; the groom's `SimulationSettings` are driven the same way. No Blueprint, no Verse.

### The three tracks and their sections (Observed)

Read directly off the sub-sequence's MovieScene:

| Track | `displayName` | Section range | Active | `completionMode` |
|---|---|---|---|---|
| `MovieSceneBoolTrack_1` | **Enable Simulation** | `Open:0:Open:0` | true | `ProjectDefault` |
| `MovieSceneBoolTrack_5` | **Suspend Simulation** | `Open:0:Open:0` | true | `ProjectDefault` |
| `MovieSceneBoolTrack_4` | **Reset** | `Open:0:Open:0` | true | `ProjectDefault` |

All three sections are **infinite** — both bounds open — with zero pre-roll and post-roll. The
sections therefore cover the entire sequence rather than a window inside it, so what turns
simulation on and off within the cinematic is the bool *channel's* keys, not the section
extent.

**Correcting an earlier inference:** a previous revision of this note stated the sections were
set to `RestoreState`. They are `ProjectDefault`, which resolves to the project's
`DefaultCompletionMode` — conventionally `RestoreState`, but that is the project setting's
behaviour, not something this sample asserts. The observable facts are: the actor's saved
`bEnableSimulation` is `false`, the tracks exist and are bound to the right component, and
Epic's documentation states simulation runs only during the cinematic. How the value is
returned afterwards is **Inferred**, not read.

**Generalise it:** any component flag you want on *only* during a cinematic belongs on a bool
track in that cinematic, with the actor's saved value set to off. Set `completionMode`
explicitly to `RestoreState` rather than relying on `ProjectDefault` if the restore matters —
character_environment gets away with the default, but it is a project-wide setting someone else can change.

## Audio lives in the Fab pack

`LS_Cap_Helm` depends on three cues and an attenuation from mount
`/04beff84-4274-b155-3774-1fabdaa7bf9d/` (the **character_environment_Audio** pack):

| Asset | Role |
|---|---|
| `vo_cine_reveal_Captain_stereo_Cue1` | the Captain's dialogue |
| `mx_cine_reveal_stereo_Cue` | music |
| `fx_char_cine_reveal_stereo_Cue1` | character SFX |
| `Att_Cine_CaptReveal` | attenuation |

Split by `vo_` / `mx_` / `fx_` on separate audio tracks, with a cinematic-specific attenuation
asset rather than a general-purpose one. See
project-config-and-dependencies.md — the sequence is
template content, the audio it plays is referenced Fab content.

## Character lighting is a dedicated rig

Two spot lights exist purely for the MetaHuman (**Observed**): `MH_Light_00` and `MH_Light_01`,
placed on her, separate from the room's `SpotLight_Helm_Key` and `SpotLight_Helm_RimLight` and
from `DirectionalLight_Ship`. Sample values from `MH_Light_00`:

| Property | Value |
|---|---|
| `intensity` | 1 (`intensityUnits: Unitless`) |
| `lightColor` | (0.941, 1.0, 0.984) — a faint cool-green tint |
| `temperature` | 5000 K, `bUseTemperature: true` |
| `attenuationRadius` | 278.8 |
| `innerConeAngle` / `outerConeAngle` | 0 / 40 |
| `castShadows` | true |

**The lesson: light the character separately from the room.** A tight (40°, ~2.8 m radius)
shadow-casting spot with a deliberate colour temperature, that only affects her. Environment
lighting is tuned for the space; face lighting is tuned for the face. There is also a
dedicated `MH_Heml_CineCameraActor` and `Camera_Helm_00` among the eight `CineCameraActor`s in
the map.

## Capture backlog

Track structure, bindings, section ranges, and section settings are all captured above. What
remains needs the Sequencer timeline itself.

- **Keyframe times and values** — for `bEnableSimulation` / `bSuspendSimulation` / `bReset`,
  and for the shot cuts. `FMovieSceneBoolChannel` is not exposed as a readable property:
  `get_properties` on a `MovieSceneBoolSection` rejects `boolCurve`, and there is no Sequencer
  toolset on the MCP bridge. The `.uasset` stores channels as an opaque custom-serialised blob,
  so string-scanning does not recover them either. **This one genuinely requires opening
  Sequencer.**
- **Shot timing and cut structure** — how many shots, their durations, and where the dialogue
  beat sits relative to the cloth-sim enable key. Same blocker as above. This is the main
  remaining gap in the note.
- `BP_LumenExposureManager` — the map's one Lumen/exposure tuning actor. Not opened. Relevant
  to cinematics because auto-exposure is what makes or breaks a camera cut into a dark
  interior; the camera devices ship `autoExposureBias: 1`, `AEM_Histogram`,
  `bMegaLights: true` in their post-process settings (**Observed**).
  **Partly answered by a second template.** lighting_template's `BP_Environment_Light_Rig`
  carries the same job as three *named* post-process components — `LumenExposure` (histogram,
  min 0.075 / max 1.0, bias 0.5, GI+reflections Lumen), `BasicExposure` (disabled no-Lumen
  fallback), `ColorGrading` — with the values read out in
  environment-lighting-rig.md.
  character_environment's actor is still unopened, but the *pattern* it is an instance of is now captured,
  and it matches character_environment's camera-device settings (`AEM_Histogram`, a small positive bias).
- The `"Old Doors"` visibility group in `LS_Cap_Helm` — four visibility tracks on set dressing,
  purpose unread. Low value.

## Verify before reuse

`compatibilityVersion 41.30`. `Device_CinematicSequence` and `Device_Button_V2` setting names
have been stable, but `ShowHUDDuringCinematic` and `DisablePlayerInput` are comparatively
recent additions — confirm they exist in the user's UEFN version before recommending them.
