---
id: technique-086
kind: technique
---

# Gameplay Camera Switching in Verse

**Sources:** character_environment_sample — `../snippets/verse/camera_switch_mode_device.verse`
(the template's only Verse file, retained verbatim)

Cycle a player through a list of cameras on a button press, hiding the player character on
cameras where it shouldn't be visible.

## The reusable patterns

### 1. `class<concrete>` as a designer-editable struct

```verse
gameplay_camera := class<concrete>():
    @editable Camera:gameplay_camera_orbit_device = gameplay_camera_orbit_device{}
    @editable ShouldHideCharacter:logic = false
```

Then on the device:

```verse
@editable Cameras:[]gameplay_camera = array{}
```

**This is the pattern to remember.** An `@editable` array of a `concrete` class gives the
designer an expandable list in the Details panel where each entry carries *both* a device
reference and its own settings. It replaces parallel arrays (`Cameras:[]device` +
`HideFlags:[]logic`) that silently desync when someone adds one and forgets the other.

`<concrete>` is required: every field needs a default so UEFN can instantiate a fresh entry
when the designer clicks `+`.

### 2. Per-player spawned loop

```verse
OnBegin<override>()<suspends>:void=
    Players := GetPlayspace().GetPlayers()
    for (Player : Players):
        spawn{CameraLoop(Player)}
```

Each player gets an independent state machine. Camera state is per-player, so it must not
live on the device.

**Caveat this sample does not handle:** `GetPlayers()` at `OnBegin` only catches players
already present. character_environment is configured `maxPlayers: 1`, so it never mattered. For anything
multiplayer, also subscribe to `GetPlayspace().PlayerAddedEvent()` and spawn a loop for
late joiners.

### 3. Explicit state in the loop, not on the class

```verse
CameraLoop(Agent:agent)<suspends>:void=
    var CameraIndex:int = 0
    var MaybeCurrentCamera:?gameplay_camera_device = false
    loop:
        SwapCameraInputTrigger.PressedEvent.Await()
        ...
```

State lives as locals in the suspending function. `?gameplay_camera_device` initialised to
`false` is idiomatic Verse for "no camera yet" — cleaner than a sentinel index.

### 4. Cameras are a stack: `AddTo` / `RemoveFrom`

```verse
if (CurrentCamera := MaybeCurrentCamera?):
    CurrentCamera.RemoveFrom(Agent)
...
NextCamera.Camera.AddTo(Agent)
set MaybeCurrentCamera = option{NextCamera.Camera}
```

**Always `RemoveFrom` the previous camera before `AddTo` the next.** Gameplay cameras push
onto a per-player stack; skipping the remove leaves cameras piled up and the pop order wrong.

### 5. Cycling with automatic reset via failure

```verse
if (NextCamera := Cameras[CameraIndex]):
    ... use it, set CameraIndex += 1
else:
    set CameraIndex = 0
    set MaybeCurrentCamera = false
    ShowCharacter(Agent)
```

The out-of-bounds index *fails* rather than errors, and the `else` branch becomes the
"wrapped around, back to default camera" case. Idiomatic Verse — no explicit length check.
Note it costs one extra press to return to default, which is deliberate: the default camera
is a real stop in the cycle.

### 6. Input debounce

```verse
Sleep(1.0)
```

A flat 1-second sleep at the end of each iteration. Blunt but effective against input spam,
and cameras are a transition-heavy operation. Tune down for responsive UI; keep something.

## The actual gotcha: character visibility timing

```verse
@editable HideDelay:float = 0.20
@editable ShowDelay:float = 0.3

HideCharacter(Agent:agent)<suspends>:void=
    Sleep(HideDelay)
    if (FortCharacter := Agent.GetFortCharacter[]):
        FortCharacter.Hide()
```

`Hide()`/`Show()` on the same frame as the camera change **pops** — you see the character
vanish before the camera has moved, or the empty space before it appears. Both delays are
`@editable` because the right value depends on the camera's blend time.

Note the asymmetry: hide is faster (0.20) than show (0.3). Showing late is less noticeable
than showing early. Hiding is `spawn{}`ed (fire-and-forget, the loop continues); showing is
called directly.

`Agent.GetFortCharacter[]` is a failable call — the character may not exist. Always guard it.

## Logging

```verse
camera_switch_mode_log := class(log_channel){}
Logger:log = log{Channel := camera_switch_mode_log}
Logger.Print("Added a camera at {CameraIndex}.")
```

Declare a `log_channel` subclass per device so output is filterable, rather than using bare
`Print`. Requires `using { /UnrealEngine.com/Temporary/Diagnostics }`.

## Level wiring (Observed — read from the live editor)

`Device_CameraModes_Orbit` ×2, `Device_InputTrigger` ×1, `VerseDevice` ×1 — the Verse device
is the only glue. `Device_CinematicSequence` + `Device_Button_V2` handle the separate
cinematic; see cinematic-sequences.md.

The `camera switch mode device` instance, as authored:

| `@editable` | Value |
|---|---|
| `HideDelay` | **0.20** — unchanged from the source default |
| `ShowDelay` | **0.3** — unchanged from the source default |
| `SwapCameraInputTrigger` | the one `Device_InputTrigger` |
| `Cameras[0]` | `Camera: Orbit`, `ShouldHideCharacter` = **false** |
| `Cameras[1]` | `Camera: FPS`, `ShouldHideCharacter` = **true** |

So the cycle is **default → Orbit (character visible) → FPS (character hidden) → default**.
Only one of the two cameras ever hides the character, and the delays were never retuned —
Epic shipped the values that were written in the source.

### The two camera devices are the same device class, configured differently

Both are `Device_CameraModes_Orbit`. The FPS one is an orbit camera with the boom collapsed:

| Setting | `Camera: Orbit` | `Camera: FPS` |
|---|---|---|
| `Distance` | 130 | **0** |
| `FieldOfView` | 70 | **95** |
| `OffsetX / Y / Z` | 0 / 27 / 62 | **45 / 0 / 76** |
| `HorizontalSpeed` / `VerticalSpeed` | 7.5 / 7.5 | **0 / 0** |
| `HidePlayerCharacter` | False | **False** |
| `AddToPlayersOnStart` | False | False |
| `TransitionInTime` / `Type` | 0.5 / Ease-In-Out | 0.5 / Ease-In-Out |
| `TransitionOutTime` / `Type` | 0.5 / Ease-In-Out | 0.5 / Ease-In-Out |

**Recipe worth stealing: a first-person camera made from an orbit camera.** Set `Distance` to
0, both orbit speeds to 0, widen FOV to ~95, and push the pivot forward and up
(`OffsetX` 45, `OffsetZ` 76) so it sits at eye position instead of inside the chest. You do
not need a different device class.

`AddToPlayersOnStart` is False on both — activation belongs entirely to the Verse device.

### The finding that changes how you read this device

**`HidePlayerCharacter` is False on both cameras.** Epic did *not* use the camera device's
own built-in hide option. They left it off and hand-rolled hiding in Verse, with a delay.

That is the justification for the whole `ShouldHideCharacter` / `HideDelay` design: the
built-in toggle hides on the frame the camera is added, which pops. If you only remember one
thing from this note, it is that Epic's own showcase found the device's built-in setting
unusable and replaced it with delayed `FortCharacter.Hide()`.

### The number `HideDelay` compensates for

**Camera blend time is 0.5s, Ease-In-Out, both directions, on both cameras.**

Against that: `HideDelay` 0.20 and `ShowDelay` 0.3 — both *shorter* than the blend. The
character is hidden 40% of the way into the blend-in and shown 60% of the way into the
blend-out, i.e. while the camera is still moving in both cases. The delays are not "wait for
the blend to finish"; they are "wait until the camera has moved far enough that the swap is
not the thing your eye is on". Start at ~40–60% of your blend time rather than at 100%.

### Which input is bound (Observed)

The `Device_InputTrigger`:

| Setting | Value |
|---|---|
| `Input Type` | Creative Input Action |
| `CreativeInput` | **Custom 2 (Target)** |
| `Consume Input` | True |
| `Enabled at Game Start` | True |
| `Show On HUD` | True |
| `HUD Description` | "Change Camera" |
| `RegisteredPlayerBehavior` | Add Registered |
| Mobile icon | hand (`T_UI_IconLibrary_hand_*`) |

A **Creative Input Action** slot (`Custom 2`), not a raw key — that is what makes it work
across keyboard, gamepad, and touch without per-platform binding. `Consume Input` True stops
the press also reaching gameplay. Note the HUD prompt text lives on the *input trigger*, not
the Verse device.

## Verify before reuse

Written against `VerseVersion 1` / `VerseAPI 1`, `compatibilityVersion 41.30`.
`gameplay_camera_orbit_device` and the `AddTo`/`RemoveFrom` API should be re-checked against
the user's UEFN version.
