---
id: technique-087
kind: technique
---

# MetaHumans in UEFN

**Sources:** character_environment_sample template + Epic's character environment sample doc

## The budget problem this all exists to solve (External — Epic doc)

| | |
|---|---|
| UEFN project **upload** cap | 2 GB disk |
| UEFN project **download** cap | **400 MB** |
| A typical UE MetaHuman | ~800 MB |
| An optimised UEFN MetaHuman | **~60 MB** |

A stock UE MetaHuman is **twice the entire island download budget**. Everything below —
baked textures, per-LOD material instances, groom card fallbacks — is the ~13x reduction
that makes one fit. Epic names three constraints: memory limits, per-platform FPS targets,
and holding visual quality across all platforms.

Practical consequence: **plan for a small number of MetaHumans.** At ~60 MB each against a
400 MB download cap, and with half of character_environment's authored assets going to one character,
a handful is the ceiling — and that is *after* Epic's optimisation pass.

## Can you do this with your own MetaHuman? **Yes — the optimisation is in UEFN, not the template.**

This is the question that decides whether the character_environment template is needed at all. The answer
is that **the ~13× reduction is performed by UEFN's MetaHuman export/import pipeline**, not by
anything hand-authored in character_environment. Epic fine-tuned *the export* — MetaHumans exported for
UEFN are the same MetaHumans used in Unreal Engine, run through a tuned conversion.

The user-facing workflow (**External**):

1. Create the MetaHuman — **MetaHuman Creator** (`metahuman.unrealengine.com`), or
   **Mesh to MetaHuman** to convert a scanned/sculpted/modelled head into a rigged MetaHuman.
2. In UEFN: **Window → MetaHuman Importer** (also available by right-clicking in the Content
   Browser). Newer MetaHuman Characters use the **UEFN Export pipeline**, which assembles the
   UEFN-ready assets directly.
3. Place it on the island — add the character Blueprint to the level, or spawn it as an NPC
   with the **NPC Spawner device**.

Baked textures, per-LOD material instances, the reduced `_UEFN` material stack, 4 LODs per
body/face, LODSync, and the MetaHuman Component all arrive **with the import**. Epic's phrase
for the LOD tuning is that performance is "optimised **by default**."

### Evidence from the template itself (Observed)

`MetaHumans/Common/` in character_environment is **stock importer output, not character_environment content**:

- `Face_Archetype_Skeleton`, `Face_Archetype_Physics`, `MetaHuman_ControlRig`,
  `MHGizmoLibrary`, `faceboard_2x`, `convergence_3l`, `M_Cartilage`,
  `MI_lacrimal_fluid_Inst` — generic rig and archetype assets.
- A **Medium** body archetype that Captain Roux (Tall) never uses.
- **Hoodie** garment meshes (`f_tal_nrw_top_hoodie`, `f_med_nrw_top_hoodie_CtrlRig`) — the
  MetaHuman Creator default top. Roux wears a bespoke jacket, shirt, and pants.

Unused archetypes and an unworn default hoodie shipped in the template because the importer
put them there. `Common/UEFN/` — the master materials `M_MetaHumanSkin_UEFN`,
`M_MetaHumanEye_UEFN`, `M_MetaHumanFabric_UEFN` and the `MF_High_*` mask functions — sits
inside that same boilerplate folder, so it is near-certainly part of the standard UEFN
MetaHuman set rather than authored for character_environment (**Inferred**, but strongly).

**What this means for a user:** import your own MetaHuman and you get the same optimised
stack. Everything below is *how to read and reason about* what the importer gives you — and
where your own work is still required (clothing, cinematic-gated cloth sim, LOD budgeting).

### Version caveat

character_environment is pinned to `compatibilityVersion 41.30`. Current UEFN ships a preloaded locomotion
animation set at `MetaHumans/Common/Common/Locomotion` (**External**); character_environment's
`Common/Common/` has only three assets and no Locomotion folder (**Observed**). The importer
has moved on since this sample. Prefer current Epic docs over template structure for
anything importer-produced.

## The core pattern: a parallel `UEFN/` material stack

A MetaHuman imported into UEFN does **not** use the standard MetaHuman material set. The
template keeps the original MetaHuman assets and adds a sibling `UEFN/` folder holding a
reduced, UEFN-specific replacement stack (**Observed**):

```
MetaHumans/
  Common/                          shared across all characters
    Common/                        Face_Archetype_LODSettings_High
                                   MetaHuman_Clothing_LODSettings
                                   MetaHuman_ControlRig
    Face/ FHair/ Female/ Materials/ Textures/ Controls/
    UEFN/                          <-- the UEFN stack, shared
      Body/Materials/              M_MetaHumanFabric_UEFN, MI_Fabric_Torso
      Face/Materials/              M_MetaHumanSkin_UEFN, M_MetaHumanEye_UEFN
                                   MI_Head_LOD1, MI_Body, MI_Eye_Simple, MI_Teeth
      Face/MaterialFunctions/      MF_Eye_Refraction, MF_High_AnimatedMaps,
                                   MF_High_HeadMask01..03, MF_High_MaskDelta01..03
      Face/Textures/               T_MH_BentNormal_AO_UEFN, T_Shared_Eye_*, T_Teeth_*_Baked
  Cpt_Roux/                        per-character
    BP_Captain_Roux_Sim_Use        blueprint wrapper for the character
    Face/Captain_Roux_FaceMesh
    Female/Tall/NormalWeight/Body/f_tal_nrw_body
    FHair/  MaleHair/  Materials/
    UEFN/                          <-- per-character UEFN overrides
      Face/Materials/              MI_Head_LOD1, MI_Head_LOD3, MI_Head_LOD5
      Face/Textures/               T_HeadLOD{1,3,5}_{BaseColor,Normal,Specular}_Baked
      Face/Textures/TextureUpdates/ FaceColor_CM1..3, FaceNormal_WM1..3, FaceNormal_MAIN_LOD1
      Body/Textures/               T_{Body,Torso}_{BaseColor,Normal,Specular}_Baked
```

**Takeaways:**

- **Split shared vs. per-character.** `Common/UEFN/` holds the master materials and mask
  functions; `<Character>/UEFN/` holds only that character's material *instances* and baked
  textures. One master material, N characters.
- **Everything under `UEFN/` ends in `_Baked`.** The layered MetaHuman face shader (animated
  wrinkle maps, mask deltas, colour maps) is flattened to plain BaseColor / Normal / Specular
  textures per LOD band. The `TextureUpdates/` folder holds the *inputs* to that bake
  (`FaceColor_CM1..3` colour maps, `FaceNormal_WM1..3` wrinkle maps) — **Inferred**, but the
  naming and the presence of matching `MF_High_HeadMask*` / `MF_High_MaskDelta*` functions
  make it the obvious reading. Epic's doc confirms the intent: *"materials were rewritten with
  fewer instructions, and multiple textures were baked into materials for additional savings,"*
  with texture compression alone halving memory (**External**).
- **Bake per LOD band, not once.** Separate `MI_Head_LOD1 / LOD3 / LOD5` with their own
  texture sets. Do not carry LOD0-resolution face textures down the chain.
- **The character is wrapped in a Blueprint** (`BP_Captain_Roux_Sim_Use`) rather than placed
  as a bare skeletal mesh. Given that Epic's doc says cloth sim is disabled by default and
  runs only during the cinematic, the `Sim_Use` suffix is very likely "the variant with cloth
  simulation enabled, for cinematic use" (**Inferred**).

## Body archetype naming

MetaHuman bodies are named `<sex>_<height>_<weight>_body`, e.g. `f_tal_nrw_body` =
female / tall / normal-weight. character_environment also carries `Female/Medium/NormalWeight` and
`Female/Tall/NormalWeight` variants under `Common/`. Expect these exact tokens
(`f`/`m`, `med`/`tal`/`shr`, `nrw`/`ovw`/`unw`) when locating a body mesh. (**Observed**
for the two used; the full token set is **External**.)

## LODs and the LODSync component (External — Epic doc)

- Body and face each ship **4 LODs**, tuned per platform so performance is optimised by
  default. Don't hand-roll LOD distances unless you have a measured reason.
- A **LODSync component** drives LOD transitions across *all* MetaHuman components together
  (face, body, hair, cloth). Without it the parts pop independently — this is the component
  to look for when a MetaHuman's head and hair change detail at different distances.

## Animation gating: the MetaHuman Component (External — Epic doc)

The animation Blueprint (a state machine) and the Control Rig both **depend on the MetaHuman
Component**, which toggles individual features:

- neck correctives
- procedural Control Rig adjustments
- facial animation

Face and body animation were optimised for "notable performance gain versus the original
fully-enabled setup" — i.e. **the default UE MetaHuman has all of these on, and shipping to
UEFN means turning off the ones a given shot doesn't need.** If a MetaHuman is expensive and
only ever seen at distance, this component is the first place to cut.

## Hair (groom)

```
FHair/Hair/            Hair_S_BobLayered, Eyelashes_L_SlightCurl   (groom assets)
FHair/Hair/            Hair_S_BobLayered_CardsMesh_Group0_LOD{0,1} (card meshes)
FHair/GroomBinding/    <Groom>_<HeadMesh>_lod0_Face_Archetype_Binding
FHair/Helmets/         Hair_S_BobLayered_Helmet_LOD{5,7}
MaleHair/              Eyebrows_M_FlatThick + its own binding + card LODs
```

Three things worth carrying forward:

1. **A groom needs an explicit `GroomBinding` per head mesh.** Change the head mesh and the
   binding must be regenerated — the binding name encodes both sides.
2. **Strands → cards → "helmet".** Epic's doc states MetaHumans use strand-based grooms at
   **LOD 0 only**, switching to conventional hair cards below that (**External**). The
   template goes one step further with a solid `_Helmet_LOD5/LOD7` mesh at far distance
   (**Observed**). That third stage is the one people forget and it is where the far-LOD cost
   goes. Epic reports the groom rework roughly **halved cook size** and ran faster.
3. **Eyebrows and eyelashes are separate grooms with separate bindings**, and here they came
   from a different source folder (`MaleHair/`) than the scalp hair. Mixing is normal.

## Engine modules a UEFN MetaHuman drags in (Observed)

`MetaHumanRuntime`, `MetaHumanSDKRuntime`, `MetaHumanUEFNRuntime`, `HairStrandsCore`,
`ChaosClothAssetEngine`, `OptimusCore` (deformer graph), plus `CinematicCamera` /
`LevelSequence` for the cinematic. If a project needs none of these, it does not need a
MetaHuman — the cost is not just the mesh.

## Scale check

155 of the template's 307 authored assets — **half the project** — are the single MetaHuman
and its shared support assets. Budget accordingly.

## Related

- character-clothing-and-lods.md — the garments are *not*
  part of the MetaHuman.
- gameplay-camera-switching.md — how the character is shown
  and hidden for cinematic cameras.

## Authoring path

MetaHumans are authored in **MetaHuman Creator** (`metahuman.unrealengine.com`), then brought
into the project (**External**). The template itself is found under **Feature Examples** in the
UEFN Project Browser, and the shipped GDC 2024 experience is playable at island code
**7100-3544-3074**.

Epic's follow-on reference: *Setting Up MetaHumans in UEFN* —
`dev.epicgames.com/documentation/metahuman/downloading-and-exporting-metahumans/setting-up-metahumans-in-uefn`

## MetaHuman Component and LODSync, as authored (Observed — live editor)

The backlog asked *which features character_environment disabled*. The answer is **none of them** — every
feature boolean is on. All of the saving is done with **LOD thresholds** instead.

### MetaHuman Component (`BP_Captain_Roux_Sim_Use.MetaHuman`)

| Setting | Value | Effect |
|---|---|---|
| `bEnableBodyCorrectives` | **true** | on |
| `bodyLODThreshold` | **0** | body correctives, head-movement IK and pose drivers run at **LOD 0 only** |
| `rigLogicLODThreshold` | **2** | facial RigLogic runs LOD 0–2, off at LOD 3 |
| `bEnableNeckCorrectives` | **true** | on |
| `neckCorrectivesLODThreshold` | **0** | LOD 0 only |
| `bEnableNeckProcControlRig` | **true** | on |
| `neckProcControlRigLODThreshold` | **0** | LOD 0 only |
| `bodyComponentName` / `faceComponentName` | `Body` / `Face` | |

**The reusable decision: turn nothing off, threshold everything to LOD 0 — except the face.**
Body deformation quality only matters when the character fills the screen, so it is bought at
LOD 0 and dropped immediately. Facial animation is the one thing carried further out
(LOD 0–2), because a talking head reads as broken far sooner than a slightly stiff elbow.

This is strictly better than disabling features: at LOD 0 you get full MetaHuman quality, and
the cost vanishes the moment the character is not the subject.

### LODSync (`BP_Captain_Roux_Sim_Use.LODSync`)

| Setting | Value |
|---|---|
| `numLODs` | 4 |
| `forcedLOD` | -1 (automatic) |
| `minLOD` | 0 |
| Drive | **`Body`, `Face`** |
| Passive | `Torso`, `Legs`, `Feet`, `Hair`, `Eyebrows`, `Eyelashes`, `Mustache`, `Beard` |

Two drivers, everything else follows. Whichever of Body/Face is visible and highest-priority
sets the LOD for all eleven components, so the face can never be at a different LOD than the
hair sitting on it.

**The groom LOD mapping is the non-obvious part:**

```
customLODMapping:  hair, beard, mustache, eyebrows, eyelashes  →  [1, 3, 5, 7]
```

Character LOD 0→groom LOD 1, 1→3, 2→5, 3→7. Two consequences worth carrying:

- Grooms have **8 LODs against the character's 4**, and are stepped two at a time.
- **Groom LOD 0 is never used.** Even at the character's highest LOD the hair renders at
  groom LOD 1. If you are authoring grooms for UEFN, LOD 0 is not your hero LOD — LOD 1 is.

### Two dead references to know about

- `LODSync.componentsToSync` lists **`Torso` and `Legs`**, but this actor has no such
  components — the merged Chaos Cloth garment replaced them. Harmless leftovers from the
  MetaHuman archetype.
- The MetaHuman Component's `torso` slot still points at MetaHuman Creator's **hoodie**:
  `f_med_nrw_top_hoodie_CtrlRig` and `f_tal_nrw_top_hoodie_strings_Physics` — a control rig
  and physics asset for an outfit Roux does not wear. (`legs` and `feet` slots are null.)
  This is the same importer-boilerplate leakage that put a hoodie in `MetaHumans/Common/`.

Both are evidence for the reading in Does deprecating this sample cost users the MetaHuman
optimisations? — much of what looks authored is importer
output that nobody cleaned up. When you swap a MetaHuman's body parts for custom clothing,
**sweep the MetaHuman Component's body-part slots and the LODSync list**, or you ship rigs for
garments that aren't there.

## Capture backlog

**Cleared.** The bake, the `_UEFN` material stack, and the LOD setup were closed earlier as
importer output. The one genuinely template-specific item — MetaHuman Component features and
LODSync configuration — is recorded above from a live UEFN session.
- ~~Whether the texture bake is a UEFN action or hand-authored~~ — **closed:** performed by
  the MetaHuman export/import pipeline.
- ~~Whether `M_MetaHumanSkin_UEFN` etc. are bespoke~~ — **closed:** stock importer output,
  found in the boilerplate `Common/` folder.
- ~~`BP_Captain_Roux_Sim_Use` MetaHuman Component features and LODSync config~~ —
  **closed:** recorded above.
- `Face_Archetype_LODSettings_High` — whether `MI_Head_LOD1/3/5` are auto-assigned per LOD.
  *Waived:* importer-produced, and current UEFN docs have a dedicated LOD page that will be
  more current than a 41.30 template.
