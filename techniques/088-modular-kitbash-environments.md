---
id: technique-088
kind: technique
---

# Modular Kitbash Environments

**Sources:** character_environment_sample, playset_sample

## The ratio

**64 modular Blueprints → 1136 placed `FortStaticMeshActor`s → 1 map** (**Observed**).
Plus 322 references into an external Fab prop pack. The environment is not modelled; it is
assembled from a small kit.

## Kit taxonomy

```
Content/Geometry/
  ModularBP/
    StructuralComponents/
      Walls/           BP_Wall128_01_N1, BP_MaintenanceWall_01_N1,
                       BP_WallPanels_02_Pipes_01_N1
      Doors/           BP_Doorway_SingleV2_01_N1, BP_AirlockDoor_L_01_N1,
                       BP_AirlockDoor_R_01_N1, BP_Doorway_SingleV2_AirlockClosed_01_N1
      Floor/           BP_ElevatedFloor_CenterConnector_02_NoLights_N1,
                       BP_ElevatedFloor_HalfStraight_01_Lights_N1,
                       BP_GrateCoverHatch_Lid_01_N1
      Ceiling/         BP_DropCeilingPanel_01_N1
      Strut/           BP_ModularStrut_N1, BP_Strut_01_N1
      CircularElements/BP_CircularElement_04_N1, BP_CircularElement_06_N1
      Elevator/        BP_ElevatorMover_N1, BP_ElevatorLever_N1, BP_CargoShutter_01_N2
    Props/             ~40 BPs: crates, consoles, monitors, food containers, vents,
                       server racks, cups, keyboards, circuit breakers, shields
  Helm/Modular/        BP_HelmConsole_01
  Keyboard/            BP_Keyboard_01_N1, BP_Keyboard_Small_01_N1
  CharacterEnvironmentExt/         BP_pipeSet_C
```

**Split the kit by architectural role first** (walls / doors / floor / ceiling / struts),
props second. Every structural category is a folder even when it holds one asset — the empty
slots tell the next artist what's missing.

## Conventions worth stealing

- **`_N1` / `_N2` suffix** = variant or revision number, carried on nearly every geometry BP
  (**Inferred** — consistent enough to be a rule, meaning unconfirmed). `BP_CargoShutter_01_N2`
  is the only `_N2`, i.e. a second revision of one asset. Pick *some* explicit revision token;
  it beats `_new`, `_final`, `_v2b`.
- **Encode variants in the name, not in options.** `BP_ElevatedFloor_HalfStraight_01_Lights_N1`
  vs `BP_ElevatedFloor_CenterConnector_02_NoLights_N1` — lit and unlit are separate assets.
  Cheaper at runtime than one Blueprint with a toggle, and searchable in the content browser.
- **Left/right as separate assets.** `BP_AirlockDoor_L_01_N1` / `_R_01_N1` rather than a
  mirrored instance.
- **State baked into the asset.** `BP_Doorway_SingleV2_AirlockClosed_01_N1` — a closed variant
  exists as its own asset because most doors never need to open.
- **`_MeshDecal` in the name** flags assets that carry mesh decals
  (`BP_Prop_Crate_HelmPanels_MeshDecal_01_N1`) — a rendering-cost signal readable from the
  content browser.

## "Greeble soup"

`BP_Prop_GreebleSoup_00_N1` through `_03_N1` (**Observed**). Four pre-assembled clusters of
small unidentifiable detail geometry, placed wherever a surface reads as too clean.

This is the highest-leverage single idea in the kit. **Build 3–5 detail clusters once, then
scatter them.** It buys the "dense sci-fi" look for four assets instead of hundreds of
individual placements, and it keeps detail density consistent across the map. Same idea
appears as `BP_AirVentDetail_01_N1` and `BP_pipeSet_C`.

## Hero vs. filler props

`BP_Prop_HeroCrate_02_N1`, `BP_Prop_HeroCrate_SmallProp_N1` are explicitly named "Hero" —
high-detail assets for camera-adjacent placement — against `BP_Prop_Crate_00/01/02_N1` for
background. **Encoding the intended viewing distance in the asset name** stops artists from
spending a hero asset on background dressing.

## Blueprint vs. static mesh actor

Only ~5% of placed actors are Blueprints. Blueprints are used for *clusters and mechanisms*
(greeble soup, elevator mover, pipe sets, consoles that combine several meshes); everything
placed individually is a plain `FortStaticMeshActor`. Don't wrap a single mesh in a Blueprint. Don't BrickAssembly BRICK gallery bricks into `SM_*` wall/pillar/doorway modules and call it this kit — that is modular-BrickAssembly-building-modules.md (**0** `Parent_Building*` there).

## Mechanisms

`BP_ElevatorMover_N1` (3 instances) + `BP_ElevatorLever_N1` + `BP_CargoShutter_01_N2` — moving
parts are Blueprints, not Verse devices. For simple, non-gameplay-critical motion, a Blueprint
is the lighter tool. (**Inferred** from usage — the Blueprint logic itself is a backlog item.)

## Lighting fixtures as assets

`Content/Lighting/BP_LightSources/Light_WallLightFixture_{SmallAmberLight,SmallGreenLight,
SmallRedLight,FloorObservationRoom_02}` (**Observed**) — the light and its physical fixture
are one asset, one per colour. Artists place a *fixture*, not a light, and colour consistency
comes for free.

## The kit Blueprints are project content; their meshes are not (Observed)

Resolved from a live session. A placed `BP_CeilingCamera_01_N1` actor has components
`Pivot`, `SM_CeilingCamera_01_N1`, `SM_UI_Element_33`, and its mesh component points at:

```
staticMesh        /04f0ce9d-4b1a-a806-e1b8-6e8582031a95/Geometry/Camera/SM_CeilingCamera_01_N1
overrideMaterials [null ×8, /04f0ce9d-…/Materials/MasterMaterials/SpecialtyMasters/
                            MI_RandomBlinkyGrid_Cameras]
```

Mount `/04f0ce9d-4b1a-a806-e1b8-6e8582031a95/` is the **character_environment_Props Fab pack** — this
upgrades the KB's earlier **Inferred** guess about that mount to **Observed**, and shows its
internal layout: `Geometry/<Category>/` for meshes, `Materials/MasterMaterials/` with a
`SpecialtyMasters/` subfolder for the interesting ones.

**The split to copy:** the *Blueprint* (`BP_CeilingCamera_01_N1`) lives in the project; every
*mesh and material* it references lives in the Fab pack. The kit is composition logic over
external art. That is what keeps a 1136-static-mesh-actor map inside the 400 MB download cap,
and it is why retiring this sample would break anyone building on `ModularBP/` unless the
pack is available to them.

Two incidental observations:

- A `Pivot` component as the root, with meshes as children — so the Blueprint's origin is a
  deliberate snapping point, not the mesh's own pivot. Standard kitbash hygiene, and it is what
  makes grid placement predictable.
- `overrideMaterials` is a mostly-null array with one entry set at index 8. Overriding a single
  slot on a multi-material mesh (here, to make the camera lens blink) is cheaper than authoring
  a variant mesh — and `MI_RandomBlinkyGrid_Cameras` is one instance reused across all seven
  placed ceiling cameras.

## How a greeble cluster is actually assembled (Observed)

The highest-leverage idea in the kit, and the answer is more boring — and more useful — than
expected.

`BP_Prop_GreebleSoup_01_N1`, component list in full: a `Pivot` scene root plus **38 static
mesh components**, one stray. `BP_Prop_GreebleSoup_00_N1` runs to roughly **90 components**.
The mix in `_01`:

```
structure   SM_GrateCoverSquare_Rim_01, SM_GrateCoverSquare_Plate_01,
            SM_MD_GrateCoverSquare_Plate_01, SM_HelmStrut_Cover_02 ×2,
            SM_Stairs_StepConnector_01 ×2, SM_Stairs_RailEnd_01 ×3,
            SM_ElevatorFloor_GuideRails_01
clutter     SM_ModularCrate_05 ×7, SM_ModularCrate_04, SM_CargoShaft_Handle_01 ×3,
            SM_CircuitBoard_01 ×3, SM_StrutCable_03, SM_StrutCable_04
decals      D_Warning_Shape03 ×6, D_Warning_Caution, D_Triangleorange, D_Shape01,
            SM_MDecal_Set01_PBR_016 ×3
```

### Three things this settles

**1. There is no construction script. The clusters are hand-placed and frozen.**

Two separate instances of `BP_Prop_GreebleSoup_01_N1` return **identical component lists in
identical order**. And neither `BP_Prop_GreebleSoup_01_N1` nor `BP_ElevatorMover_N1` contains a
single `K2Node_*` entry — their `EventGraph` / `UbergraphPages` are empty and only
`SimpleConstructionScript` carries content.

So the answer to "does the construction script randomise?" is **there is no construction
script logic at all**. Every greeble Blueprint is a pure static container. Variation between
the ~25 placements comes from **actor transform, rotation and scale only** — which is why
instance bounds range from 95 to 345 units on the same class.

**This is almost certainly a platform constraint, not Epic's restraint.** A follow-up scan of
every authored Blueprint in the project found **75 Blueprints and zero K2Node graph nodes** —
see
project-config-and-dependencies.md.
In UEFN, Blueprints are component assemblies; logic lives in Verse. So "why didn't Epic write a
randomiser?" has the mundane answer: **they couldn't.**

The design lesson survives the correction, but restated honestly: on UEFN you compose kitbash
clusters as fixed Blueprints and get variety from placement transforms, because scripted
assembly is not on the menu. If you want procedural placement, that is a Verse job, not a
construction script.

**2. Mesh decals ship inside the cluster.** The `D_*` and `SM_MDecal_*` components — warning
triangles, caution stripes, grime — are part of the prefab, not placed separately afterwards.
Bundling signage with the geometry it belongs on is what makes a cluster look authored rather
than scattered, and it means one placement carries its own storytelling.

**3. `SM_MD_` means mesh decal** (**Observed**). `SM_ModularCrate_Drawer_01` and
`SM_MD_ModularCrate_Drawer_01` both appear in the same cluster, in quantity (18 and 13
respectively in `BP_Prop_GreebleSoup_00`). The character_environment_Props pack contains
`Geometry/MeshDecal_Sets/MeshDecal_Set{01,02,03}/` holding ~80
`SM_MDecal_Set0N_{PBR,Normal}_NNN_N1` meshes, which confirms the prefix: `MD`/`MDecal` is a
thin overlay mesh laid onto a base mesh to add wear, panel lines, or signage without a new
texture on the host.

Note both `PBR` and `Normal` decal variants exist — full-material decals and normal-only
decals. Normal-only is the cheap one; use it for surface detail that needs no albedo change.

### The four cluster sizes

| Class | Typical bounds | Role |
|---|---|---|
| `BP_Prop_GreebleSoup_00_N1` | ~640–790 × ~100 × ~90–120 | long wall/ceiling runs |
| `BP_Prop_GreebleSoup_01_N1` | ~95–345 × ~77–228 × ~111–316 | medium wall clusters |
| `BP_Prop_GreebleSoup_02_N1` | ~20–100 | small accents |
| `BP_Prop_GreebleSoup_03_N1` | ~145 × 83 × 86 | labelled `BP_Prop_GreebleSoup_00_EightLength` |

**A size ladder, not a random generator.** Long runs, medium clusters, small accents — pick by
the gap you need to fill.

## The elevator does not move (Observed)

`BP_ElevatorMover_N1` has exactly **three components**: `StaticMeshComponent0`,
`BoundingBoxComponent`, `EditorOnlyStaticMeshComponent`. No Timeline, no curve, no child
meshes — and, as above, an empty event graph with no `K2Node_*` of any kind.

The backlog asked for "mechanism logic in Blueprint rather than Verse". **There is none.** The
name describes an intent that was never implemented, or was implemented in the larger GDC build
and stripped — consistent with the cut Cargo Hold wing described in
world-partition-data-layers.md. Three
`BP_ElevatorMover` instances and around twenty-five `BP_ElevatorLever_N1` instances are placed;
all of them are set dressing.

Do not cite this sample as an example of Blueprint mechanisms. It has no working ones.

### Stale names are everywhere in this kit

Worth knowing before you trust a label:

- Every placed `BP_ElevatorMover_N1` is *internally* named `BP_Monitor_05_C_UAID_…`.
- So are the `BP_Prop_GreebleSoup_02_N1` instances.
- `BP_Prop_GreebleSoup_03_N1` instances carry the label `BP_Prop_GreebleSoup_00_EightLength`.
- One greeble component is named `GrateCover_01_Square_Hatch_group4_pCube18_009` — a raw Maya
  export name that survived into a shipped first-party template.

Epic duplicated existing Blueprints to make new ones, and the placed actors kept the *original*
class's name. **The outliner label, the actor's internal name, and the actual class are three
different strings here.** Always read the class path, not the label.

## A kit Blueprint should inherit a building system, not wrap a mesh

**Source: playset_sample** (**Observed** — live editor, `AssetTools.get_dependencies`
and `ObjectTools.list_properties`). character_environment's kit above is *set dressing* — its Blueprints are
mesh clusters with empty event graphs. A kit meant to be **built with** looks different, and the
difference is entirely in what the Blueprint inherits.

`BP_GrayBox_Solid_Wall` has five dependencies. **Four are Fortnite gameplay systems; the mesh is
last:**

```
/Game/Building/ActorBlueprints/Wall/Parent_BuildingWall.Parent_BuildingWall_C   parent class
/Game/Building/EditModePatterns/Wall/EMP_Wall_Solid                             edit-mode pattern
/Game/Building/NavigationAreas/FortMetaNavWallAreaDef.FortMetaNavWallAreaDef_C  AI nav area
/Game/Balance/DataTables/ResourceRates                                          harvest economy
/playset_sample/SetupAssets/Meshes/Walls/SM_GrayBox_Wall_01              the mesh
```

**The four things to wire, per piece:**

| Wire | Why | Skip it and… |
|---|---|---|
| `Parent_Building*` class | inherits placement, health, repair, damage state | you get a prop, not a build piece |
| Edit-mode pattern (`EMP_*`) | players can edit the piece in-game | the piece is uneditable |
| Nav area def (`FortMetaNav*AreaDef`) | AI paths through/around correctly | AI ignores or gets stuck on it |
| `resourceType` / `ResourceRates` | harvests into the right material | wrong or no resource on break |

**Pick the parent by architectural role.** All nine ship in this kit and map 1:1 onto the folder
taxonomy above: `BuildingWall`, `BuildingFloor`, `BuildingStairs`, `BuildingRoof`,
`BuildingPillar`, `BuildingProp`, `BuildingPropWall`, `BuildingCorner`, `BuildingPropCorner`.

### Variants are configurations, not new meshes

`Parent_BuildingWall_C` exposes **302 properties, 101 building-related** — including complete
door rigging: `bSwingingDoor`, `doorScale`, `doorOffset`, `doorTargetRotationOffset`,
`slidingDoorTargetRotationOffset`, `automaticSlidingDoorBoxExtent`, `slamOpenDoorBoxExtent`,
`doorPropClass`, `doorAnimatingMaterial`, `doorSmartLinkComp`.

So `BP_GrayBox_Door_S`, `BP_GrayBox_DoubleDoor_Sliding_S` and `BP_GrayBox_Half_Wall_Door` are
**one inherited system configured three ways** — not three authored mechanisms. If you are about
to build door logic by hand in a kit Blueprint, check the parent class first; it is probably
already there.

Props inherit less but still get a lot. `BP_Graybox_Prop_Large` (`ResourceRates` + mesh only)
still exposes **273 properties, 73 building-related**: `bIsBuildingVolume`, `resourceType`,
`buildingAttachmentSlot`, `buildingAttachmentType`, `buildingPlacementType` (trap/deco placement),
`buildingAttachmentPointOffset`, `bAutoAssignNavProperties`, `bUseDamageSet`, `bShowDamageCracks`,
`savedHealthPct`, `currentBuildingLevel`, `onRepairBuildingStarted`.

### Pair Blueprints and meshes 1:1 by category

playset_sample keeps `SetupAssets/Blueprints/<Category>/` and
`SetupAssets/Meshes/<Category>/` as mirrored trees — 62 BPs against 63 meshes across Walls (25),
Balconies (8), Stairs (8), Roofs (5), Props (4), Railings (4), Pillars (3), Trims (3), Floors (2),
over a single shared `TextureData/GrayBox`. Same "category folder even when nearly empty"
discipline as character_environment, but with the mesh tree mirrored so a missing wrapper is visible at a
glance.

Shape suffixes are consistent and worth copying: `_C` corner, `_D` diagonal, `_I` inner,
`_O` outer, `_S` small, `_L` large, `_T`, `_W`, `_Half`, `_Quarter`, `_Rail`.

> **One graybox kit, two lessons.** character_environment shows how to *assemble* an environment from a kit;
> playset_sample shows how to *author* a kit that Fortnite's building system understands.
> They are different jobs — do not copy character_environment's empty-graph prop Blueprints when you need a
> buildable piece.

## Capture backlog

**Cleared.** All five items resolved from a live UEFN session. Items 9 and 10 (wall grid size,
light-fixture bundling) were folded in as low-value; for comparable light values that *were*
read, see the character lighting rig in
cinematic-sequences.md.

Not captured, and deliberately not carried: internal component *transforms* within each
cluster. There are ~90 of them in one Blueprint and they describe Epic's spaceship, not a
transferable pattern. The pattern — fixed hand-composed clusters, decals included, varied by
placement transform — is above.
