---
id: technique-089
kind: technique
---

# Project Config, Budgets & External Dependencies

**Sources:** character_environment_sample + Epic's character environment sample doc; intro_to_uefnfk1

## The hard limits (External — Epic doc)

| Limit | Value |
|---|---|
| Project **upload** | 2 GB disk |
| Project **download** | **400 MB** |

The 400 MB download cap is the one that shapes design. Reference points from character_environment:
a stock UE MetaHuman is ~800 MB (2× the entire budget); optimised for UEFN it is ~60 MB (15%
of budget) — and one of them accounted for half the template's authored assets.

**When a user asks "can I put N MetaHumans / this much content in my island", this table is
the answer.** Everything in metahuman-in-uefn.md and
character-clothing-and-lods.md is downstream of it.

## Ship the map, reference the art

character_environment's props, audio, and VFX are **not in the project**. They are Fab content packs the
project references (**Observed** — `References/*.uref`):

```json
{ "fileVersion": 2,
  "id": "/EpicGames@Fab.com/character_environment_Props",
  "projectId": "4aa8e375-4294-1ad0-74fe-3e972bbf4aa6",
  "bEnabled": true }
```

| Pack | Project id |
|---|---|
| `/EpicGames@Fab.com/character_environment_Props` | `4aa8e375-4294-1ad0-74fe-3e972bbf4aa6` |
| `/EpicGames@Fab.com/character_environment_Audio` | `77934947-4208-5d20-d5f2-a58568e15aa2` |
| `/EpicGames@Fab.com/character_environment_VFX`   | `31850f48-4544-4899-adba-558244a7057f` |

### The mount GUID is not the project id (Observed — live editor)

A `.uref` declares a **project id**, but the pack's content mounts under a *different* GUID,
and that mount is what actually appears in asset paths. All three resolved from a live session:

| Mount seen in asset paths | Is | Declared project id in the `.uref` |
|---|---|---|
| `/04f0ce9d-4b1a-a806-e1b8-6e8582031a95/` | character_environment_**Props** | `4aa8e375-4294-1ad0-74fe-3e972bbf4aa6` |
| `/04beff84-4274-b155-3774-1fabdaa7bf9d/` | character_environment_**Audio** | `77934947-4208-5d20-d5f2-a58568e15aa2` |
| `/8716e818-4e40-0b9d-eb87-09b5bf75e877/` | character_environment_**VFX** | `31850f48-4544-4899-adba-558244a7057f` |

**You cannot match a mount to a pack by comparing GUIDs to the `.uref`** — they are different
identifiers for the same dependency, and neither `projectId` appears anywhere on disk.

**The reliable way to map them is the editor log.** On project open, `LogValkyrie` prints one
`Exposing '<pack id>' content` line per pack, and the Content Browser alias warnings name both
the pack and its mount:

```
LogValkyrie: OpenProject_GatherReferencedContent - Done (0 source projects, 3 external content)
LogValkyrie: Exposing '/epicgames@fab.com/character_environment_Props' content took 0.02 sec
LogContentBrowserAliasDataSource: … alias /Referenced Content/character_environment VFX/Steam/NS_Steam_…
                                  for /8716e818-4e40-0b9d-eb87-09b5bf75e877/Steam/NS_Steam_…
```

Failing that, match by content: load an asset from the mount and look at what it is
(`…/Geometry/Camera/SM_…` → props; `…/Cine/Reveal_Cine/vo_…` → audio; `…/Steam/NS_…` → VFX).

### The `.uref` file format, fully decoded (Observed)

A reference is one small JSON file per pack in `References/`. Nothing about it is opaque.

**Filename:** `<PackName>_<md5 of the id field>.uref`

Verified against all three of character_environment's references — the 32-hex suffix is exactly
`md5("/EpicGames@Fab.com/<PackName>")`:

| File | `md5(id)` |
|---|---|
| `character_environment_Props_96f067214fa6497e4e87949ec6118904.uref` | `96f067214fa6497e4e87949ec6118904` ✓ |
| `character_environment_Audio_42a05817e5a84a7c959b279b5223bd73.uref` | `42a05817e5a84a7c959b279b5223bd73` ✓ |
| `character_environment_VFX_70ef854461f029eb025b7f759c623cf5.uref`   | `70ef854461f029eb025b7f759c623cf5` ✓ |

Note the hash is over the **canonical-cased** id (`/EpicGames@Fab.com/…`), not the lowercase
form the log prints (`/epicgames@fab.com/…`).

**Contents:** four fields, no tokens, no entitlement blob, no version pin.

```json
{
	"fileVersion": 2,
	"id": "/EpicGames@Fab.com/character_environment_Props",
	"projectId": "4aa8e375-4294-1ad0-74fe-3e972bbf4aa6",
	"bEnabled": true
}
```

**The references are not listed in `.uefnproject`.** That file has a `plugins` array, but it
contains only the project's own module. `References/` is the entire mechanism — which is why a
`.uref` is so easy to lose when copying a project by hand, and why extraction records them
verbatim.

`bEnabled` implies references can be switched off without deleting them.

### What a referenced pack actually contains — and what it doesn't (Observed)

Enumerated across all three character_environment packs in a live session:

| Pack | Top-level content |
|---|---|
| character_environment_**Props** | `Geometry/` (34 subfolders: `Corridor_Modules`, `DoorModule`, `GrateCovers`, `HeroCrates`, `LightFixtures`, `MeshDecal_Sets`, `ModularCrate`, `ModularStruts`, `Stairs`, `WallConsole`, …), `Materials/`, `Textures/` |
| character_environment_**Audio** | `Cine/Reveal_Cine/` sound cues, `Attenuation/` |
| character_environment_**VFX** | `Common/`, `Dustmotes/`, `Sparks/`, `Steam/` — Niagara systems |

**All three contain zero Blueprint assets.** (`find_assets` filtered to `/Script/Engine.Blueprint`
returns an empty list for each mount.) What they hold is static meshes, materials and material
instances, textures, Niagara systems, and audio — raw art, no assembly and no logic.

`Geometry/MeshDecal_Sets/` also settles the `SM_MD_` naming question raised in
modular-kitbash-blueprints.md: it contains ~80
`SM_MDecal_Set{01,02,03}_{PBR,Normal}_NNN_N1` meshes, so **`MDecal` / `MD` means mesh decal**.

### UEFN project Blueprints contain no logic at all (Observed)

Scanned every authored `.uasset` in character_environment_sample that declares a
`BlueprintGeneratedClass`, excluding `__ExternalActors__`:

```
Authored Blueprint assets:        75   (59 of them under Geometry/ModularBP/)
...containing any K2Node graph node: 0
```

**Not one graph node across 75 Blueprints.** (The 13 apparent hits in external actor packages
are false positives — placed device actors embed delegate names like
`BndEvt__…_K2Node_ComponentBoundEvent_0_…`, which are names, not nodes.)

> ### Corrected by a second template — this generalises too far
>
> The scan above is accurate **for character_environment_sample**, but the conclusion drawn from it is not
> a platform rule. **intro_to_uefnfk1's `BP_DemoDisplay` is construction-script driven**
> (**Observed**, live editor): `ActorTools.get_components` on a placed instance returns
> `NODE_AddInstancedStaticMeshComponent-0…9` and `NODE_AddTextRenderComponent-0…2` — the
> `NODE_Add*` prefix is what a *graph* names the components it spawns, as opposed to the plain
> names a static component list produces. It exposes ~20 parameters (`width`, `height`,
> `title Text`, `description Text`, paragraph spacing, colours, a `DataTableRowHandle`) and
> builds 55 differently-sized signage boards from one class.
>
> So the honest statement is narrower: **UEFN Blueprints frequently carry no runtime logic, and
> gameplay logic belongs in Verse — but construction-script (edit-time) logic is alive and well,
> and is the idiomatic way to build parameterised set dressing.**
>
> Diagnostic caveat that matters here: the `K2Node` string-scan finds **event-graph** nodes.
> `BP_DemoDisplay` lives in `/ContentHall`, outside the project, so it was never in the scanned
> set — the count above is not evidence about it either way. When judging "does this Blueprint
> do anything", check a placed instance's component names for the `NODE_Add*` prefix as well as
> scanning for `K2Node`.

Within character_environment itself the reading holds: its Blueprints are **component assemblies**, and
logic lives in Verse. Bear it in mind whenever a UEFN Blueprint's name implies behaviour —
`BP_ElevatorMover_N1` moves nothing.

> **Open question — not yet answered.** Whether Fab/UEFN packs are *forbidden* from containing
> Blueprints, or Epic simply chose to ship raw art only. Three packs from one publisher in one
> batch is a small sample. This matters because it decides where reusable assemblies can live.
> See the resume note below.

### Open thread: where should assemblies live?

If assemblies cannot go in a Fab pack, the natural home may be UEFN's **Scene Graph entity
prefab** system rather than Blueprints. Evidence that it exists in this build (**Observed**,
but *not* yet checked for actual assets):

- Classes: `/Script/Entity.EntityPrefab`, `/Script/EntityPrefabEditor.EntityPrefabFactory`,
  `AssetDefinition_EntityPrefab`, `EntityPrefabThumbnailRenderer`,
  `/Script/FortniteGame.FortCreativeUserPrefabItemDefinition`,
  `/Script/FortniteGame.FortLevelInstancePrefab`.
- Log categories: `LogEntityPrefabEditorMode`, `LogEntityPrefabThumbnailRenderer`,
  `LogPrefabSpawner`.
- character_environment_sample has `bIsSceneGraphSystemAllowed: true` but
  `bEnableOneFilePerEntity: false`, and its assemblies are all Blueprints — so this sample
  predates or declines the prefab route.

**Unverified:** whether any `EntityPrefab` assets exist in this project or the packs, whether
prefabs can be published in a pack, and whether they can carry Verse components. Answering this
would decide the assembly-preservation question outright.

### Referenced packs are *exposed*, not downloaded (Observed)

This is the part that surprises people, and it corrects a natural assumption.

Opening character_environment_sample required **no pack download and no user action**. The three packs
were exposed in **0.00–0.02 seconds each**, and the project reported `0 dependency project(s)`.
Zero seconds is not a fast download; it is no download.

What actually happens:

- Epic's first-party packs already ship **inside the Fortnite/UEFN build** — ~84 GB of
  IoStore containers under `<Build>/FortniteGame/Content/Paks/`.
- On open, UEFN creates an **empty stub directory** per pack at
  `<Build>/FortniteGame/Plugins/GameFeatures/<mount-guid>/Content/` — a mount point with no
  files in it — and mounts the pack's virtual path there. Asset bytes stream from the paks.
- The Content Browser then aliases the mount under `/Referenced Content/<Pack Name>/`, which
  is the friendly path a user sees.

**How to tell "present" from "missing" properly.** An empty stub folder proves nothing, and a
`refPath` proves nothing either — soft object paths serialise fine whether or not the target
exists. Three checks that do work:

1. `AssetTools.get_asset_class` returns the real class (not empty).
2. `ObjectTools.get_properties` returns **data-dependent** values — for a static mesh,
   `staticMaterials` with `uVChannelData.bInitialized: true` and real UV densities. That data
   only exists if built render data is loaded.
3. The log has **no missing-reference warnings naming the mount GUID**. Search it; Fortnite's
   own startup noise is full of unrelated "failed to load" lines, so filter by GUID.

**The practical consequence:** for Epic first-party packs, a teammate opening a project that
`.uref`s them needs no download either, provided their build matches. The `.uref` is closer to
"switch this on" than to "go and fetch this". Third-party Fab packs are a different matter and
this finding should not be generalised to them.

This matters when auditing: a reference into an unrecognised mount is not a broken reference,
and counting mounts is not counting packs.

322 of the map's actor references resolve into an external mount, i.e. a large share of what
you see on screen is pack content. Consequences:

- **A `.uref` is a declaration, not a copy.** For Epic first-party packs the content is already
  in your build, so opening the project just exposes it (see above) — no download, no prompt.
  For anything not shipped with the build, the pack must be acquirable, or the map opens full
  of missing references.
- Referenced pack content does not consume your project's own asset budget the way copied
  content does, and updates to the pack flow through.
- Each `.uref` is a tiny JSON file — **this is exactly the kind of dependency that vanishes
  silently when a template is deleted**, which is why extraction records them verbatim.

## `.uefnproject` — settings that change what's possible

```json
"compatibilityVersion": "41.30",
"requiredRedirectorStartingVersion": "41.30.56430492",
"apiAccess": "Private",
"dataSets": {
  "matchmaking": { "maxPlayers": 1, "maxTeamCount": 1, "maxTeamSize": 1,
                   "allowJoinInProgress": true, "allowSquadFillOption": false,
                   "islandQueuePrivacy": "Unrestricted" },
  "experimental": { "sceneGraph": { "bIsSceneGraphSystemAllowed": true,
                                    "bEnableOneFilePerEntity": false } }
}
```

- **`compatibilityVersion` pins the project to a UEFN release.** Any API claim carried out of
  a template is only known-good at its pinned version. Always re-verify.
- **`maxPlayers` is a design decision recorded in config.** character_environment is 1. A technique proven
  in a solo showcase is not thereby proven multiplayer-safe — cloth sim, per-player camera
  stacks, and MetaHuman cost all scale with player count.
> ⚠ **Do not grade `experimental.*` as Observed from a working copy.** Every flag in this block
> is writable from UEFN's Project Settings UI, and UEFN rewrites the descriptor on disk when one
> is flipped — leaving no trace that distinguishes a shipped value from an operator toggle. They
> are **off by default**. Read them from a pristine template source (fresh "new project from
> template", or Perforce) before drawing any conclusion about what a template ships with. This
> is not hypothetical: it invalidated a headline finding in the intro_to_uefnfk1 digest.

- **Scene Graph is an experimental opt-in** (`bIsSceneGraphSystemAllowed`), separate from
  `bEnableOneFilePerEntity`. character_environment enables the system but not one-file-per-entity.

## `.uplugin` — the module manifest

```json
"VersePath": "/invaliddomain/character_environment_sample",
"VerseVersion": 1, "VerseAPI": 1,
"EnableVerseAssetReflection": true,
"CanContainContent": true, "CanContainVerse": true,
"EnableSceneGraph": true,
"UGCLocalizationTargetName": "character_environment_sample"
```

`"/invaliddomain/..."` is the placeholder Verse path for a project never published under a
real creator domain — expect it in templates, replace it in real projects.
`EnableVerseAssetReflection` is what lets Verse see content assets.

## `.loreignore`

Templates ship a `.loreignore` (same glob syntax as `.gitignore`) excluding `Binaries`,
`DerivedDataCache`, `Intermediate`, `Saved`, IDE folders, **and all source-art extensions**
(`.psd`, `.fbx`, `.blend`, `.wav`, `.exr`, `.ma`, …). Its own comment notes: *"Remove these
if you want to have your source assets tracked as well."*

Default UEFN behaviour is **source art is not versioned**. If a team wants `.psd` and `.fbx`
alongside the project, that is a deliberate `.loreignore` edit. Note the one careful
exception Epic wrote in:

```
# Only ignore png files in Content directory, pngs in other locations
# like Resources should be included (eg project icon)
**/Content/**/*.png
```

## Housekeeping observed in a first-party template

`Content/Developers/jojitsuruga/Collections/` and an empty `Content/Collections/` shipped in
Epic's own template. Sweep `Developers/` before publishing — everyone forgets.

## Capture backlog — blocks retirement

**Empty — this note is complete.** Everything here came from plain-text manifests
(`.uefnproject`, `.uplugin`, `.uref`, `.loreignore`) or Epic's documentation, so nothing is
locked in binary. This note does not block any template's retirement.

## Where templates come from

UEFN **Project Browser → Feature Examples**. The shipped character_environment experience is playable in
Fortnite at island code **7100-3544-3074** (**External**) — worth loading to see the intended
result before rebuilding a technique.
