---
id: technique-090
kind: technique
---

# World Partition & Data Layers

**Sources:** character_environment_sample

character_environment ships **39 data layer assets** for a single interior map. (**Observed** —
`_raw/*/datalayers.txt`, `assets.txt`.)

> ## Read this first — the shipped map does not stream
>
> **`LevelHasStreamingDisabled = 1` on `character_environment_sample.umap`** (**Observed**, asset registry
> tag read from the live editor). World Partition runtime streaming is **off**. Corroborating,
> from the same session:
>
> - Only **18** of the 39 data layer assets have an instance in this world.
> - **17 of those 18 start `Activated`.** Exactly one — `DL_Corridor_LowerRing` — starts
>   `Unloaded`. Every instance has `streamingPriority` 0 and `overrideBlockOnSlowStreaming`
>   `NoOverride`.
> - The HLOD layer is `bIsSpatiallyLoaded: false`.
>
> An earlier revision of this note said "streaming was designed up front". That was inferred
> from asset counts and is **wrong for the map as shipped**. The layering is real and worth
> copying as an *organisational* scheme; it is not a worked example of runtime streaming, and
> the initial-state column below is not a streaming choreography you can learn pacing from.
>
> Why it probably looks like this: the map is a small single-player interior that fits in
> memory, and the layer set is inherited from the larger GDC character_environment experience (see
> Loading sequences — a whole Cargo Hold wing exists in the layer assets
> but not in this level). Turning streaming off avoids hitches on a map that never needed it.
>
> **If you came here for streaming:** take the naming scheme and the structure/props split,
> and get the runtime-state values from Epic's World Partition documentation instead.

## The naming scheme

Two levels, always `DL_` prefixed:

```
DL_<Area>                       DL_CommonRoom, DL_Helm, DL_Workshop, DL_CargoHold,
                                DL_CrewQuarters, DL_EngineRoom, DL_NavRoom
DL_<Area>_<SubGroup>            DL_CommonRoom_KitchenProps
                                DL_CommonRoom_LoungeProps
                                DL_CommonRoom_CosyCornerProps
                                DL_CommonRoom_HydroponicsProps
                                DL_CommonRoom_VRCornerProps
                                DL_CrewQuarters_Props
                                DL_Helm_Props
DL_<Area>_<Section>             DL_CargoHold_Section1A … Section5A
DL_<Area>_<Feature>             DL_CargoHold_{CargoCrane,Cargobot,Gantry,LowerDeck,MH_Corridor}
```

**The split that pays off: structure vs. props.** `DL_Helm` (2154 actor references) is
separate from `DL_Helm_Props` (1156). The room's shell can stay resident while its clutter
streams in and out. Do this for every room.

## Connective space gets its own layers

`DL_Corridor_East`, `DL_Corridor_West`, `DL_Corridor_West_Exterior`,
`DL_Corridor_West_Hidden`, `DL_Corridor_LowerRing`, `DL_Workshop_Corridor`,
`DL_CargoCorridor`, `DL_Shoulder_East`, `DL_Shoulder_West`.

Corridors are the transition budget — they are what is loaded while the next room streams in.
Note `_Hidden` and `_Exterior` variants of the same corridor: the same physical space has
different content sets depending on where the player can see from.

## Lighting is layered independently of geometry

```
DL_Global_Lighting        DL_MapLights        DL_ExteriorLighting
DL_LightBlockers          DL_Lighting_WildWest
```

Keeping lights on their own layers lets lighting load/unload on a different schedule than the
geometry it lights, and lets a whole lighting scheme swap out (`DL_Lighting_WildWest` is a
themed alternate set — **Inferred** from the name, which is off-theme for a spaceship and
reads as a variant or event dressing).

`DL_LightBlockers` as a dedicated layer is a nice trick: the invisible geometry that occludes
or blocks light is grouped so it can be reasoned about separately from visible geometry.

## Loading sequences

`Content/DataLayers/LoadingSequences/LoadCQCommonRoomCargoHold` is a **LevelSequence**
(**Observed**) that references **21 data layer assets** — the Cargo Hold wing
(`DL_CargoHold`, its five `Section1A…5A`, `CargoCrane`, `Cargobot`, `Gantry`, `MH_Corridor`,
`LowerDeck`), `DL_CargoCorridor`, `DL_CargoUtilityBay`, `DL_CommonRoom` plus its five prop
groups, `DL_CrewQuarters`(`_Props`), and `DL_Workshop_Corridor`.

So the *technique* is real and worth knowing: **drive data layer states from a Level Sequence
data-layer track**, so a load transition is authored on a timeline next to the camera and
audio that accompany it, rather than wired through device logic.

But in this sample it is dead content (**Observed**):

- `get_referencers` on the asset returns **nothing** — no device, level, or sequence plays it.
- Most layers it names — every `DL_CargoHold*`, `DL_CargoCorridor`, `DL_CargoUtilityBay` —
  have **no instance in this world at all**. They are part of the 21 uninstanced assets.
- Runtime streaming is disabled map-wide anyway.

This is the clearest evidence that the shipped template is a **cut-down of the larger GDC
character_environment experience**: the ship had a Cargo Hold wing, the layer assets and its load sequence
survived the cut, the geometry did not. Treat the asset as an illustration of the pattern, not
as something that runs.

## Actor distribution (Observed)

| Data layer | Actor refs |
|---|---|
| `DL_Helm` | 2154 |
| `DL_Helm_Props` | 1156 |
| `DL_NavRoom` | 1126 |
| `DL_Shoulder_East` | 158 |
| `DL_Workshop_Corridor` | 108 |
| `DL_CrewQuarters` | 46 |
| everything else | ≤ 24 |

Three layers hold nearly everything. The other 36 are small. Read this as: **the hero spaces
get exhaustive layering, background spaces get a single layer each.** Don't spend effort
subdividing rooms the player passes through.

(Counts are reference counts from actor packages, so they over-count actors with multiple
components — use them as relative weight, not exact actor totals.)

## The 18 instanced layers and their initial runtime state (Observed)

The complete set present in the shipped world. Everything else in `DataLayers/` is an
uninstanced asset.

| Data layer instance | Initial runtime state |
|---|---|
| `DL_CommonRoom` | Activated |
| `DL_CommonRoom_CosyCornerProps` | Activated |
| `DL_CommonRoom_HydroponicsProps` | Activated |
| `DL_CommonRoom_KitchenProps` | Activated |
| `DL_CommonRoom_LoungeProps` | Activated |
| `DL_CommonRoom_VRCornerProps` | Activated |
| `DL_Corridor_East` | Activated |
| **`DL_Corridor_LowerRing`** | **Unloaded** |
| `DL_CrewQuarters` | Activated |
| `DL_CrewQuarters_Props` | Activated |
| `DL_Helm` | Activated |
| `DL_Helm_Props` | Activated |
| `DL_NavRoom` | Activated |
| `DL_Shoulder_East` | Activated |
| `DL_Workshop_Corridor` | Activated |
| `DL_ExteriorLighting` | Activated |
| `DL_Global_Lighting` | Activated |
| `DL_LightBlockers` | Activated |

All 18 also have `bIsInitiallyVisible: true` and `bIsInitiallyLoadedInEditor: true`.

`DL_Corridor_LowerRing` starting `Unloaded` is the one real use of the mechanism in the
shipped map: a corridor that exists in the level but is switched off — cut content held in
place rather than deleted. That is a legitimate technique in its own right (**ship geometry
disabled via an Unloaded data layer instead of deleting it**), just not a streaming one.

## HLOD settings (Observed — `character_environment_sample_DefaultHLODLayer`)

| Setting | Value |
|---|---|
| `layerType` | **MeshMerge** |
| `bIsSpatiallyLoaded` | **false** |
| `cellSize` | 25600 |
| `loadingRange` | 51200 |
| `hLODBuilderClass` | null (engine default for MeshMerge) |
| `parentLayer` | null (single level, no HLOD chain) |
| `editorLoadingBehavior` | Default |

`MeshMerge` + non-spatial means: build **one merged proxy mesh** for the map that is always
resident, rather than a spatial grid of proxies that stream. Consistent with streaming being
off. `cellSize`/`loadingRange` are carried but inert while the layer is non-spatial. The
HLOD material is Fortnite's stock `BaseFlattenMaterialAthena_MedHLOD`, and LOD selection is
`LowestDetailLOD` (**Observed** in the map package).

For a small interior this is the right shape. For a large map you would want a spatially
loaded HLOD layer, and probably a parent chain.

## Supporting facts

- 588 `ActorFolder` entries: editor folders are used heavily *in addition to* data layers.
  Folders organise the outliner; data layers control runtime loading. They are not the same
  tool and character_environment uses both.
- Map asset-registry tags (**Observed**): `LevelIsPartitioned: 1`,
  `LevelIsUsingActorFolders: 1`, `LevelIsUsingExternalActors: 1`,
  `LevelHasStreamingDisabled: 1`. The first three are what make One File Per Actor and the
  2474 external actor packages work; the fourth is the one that matters most and is the
  easiest to miss.
- **space_station_sample** also has streaming **off** and one layer `DL_Hangar` (1554 actors, `Activated`). A cinematic unloads it during the hyperspace shot — hide hose, not streaming choreography. Write-up: hyperspace-attach-datalayer.md.

## Capture backlog

**Cleared.** Initial runtime states, the loading sequence, HLOD settings, the full instanced
layer list, and the map's streaming flags were all read from a live UEFN session and are
recorded above.

Still unread, but low value and not blocking:

- Per-actor membership of each layer (which of the 2474 external actors sits in which layer).
  The relative weights in Actor distribution already carry the
  lesson; the exhaustive mapping is Epic's level, not a transferable pattern.
