---
id: technique-091
kind: technique
---

# Conversation Device as a Field Operator respawn picker (banks + teleporters)

**Sources:** team_conquest_sample
**Tutorial:** space-opera™ Team Conquest — Spawning System / Device Setup. Quest-tree Conversation+Verse stays on conversation-device-verse-quest.md. Persona stays on persona-voice-structured-output.md. Widgets-folder chrome: custom-ui-widgets-folder.md (this island is a second source).

## Problem

After the first capture, eliminated players should **choose a forward spawn** without placing a full pad cluster at every landmark. Teleporter listen slots cap how many destinations you can wire.

## Pattern (shipped)

Four `Device_Conversation` (`CD_{Rebels,Empire}_Z{B,C}_Respawn`):

| Device | Bank (folder **`DIalogue/`**, typo) | `customWidget` |
|---|---|---|
| Rebels ZB / ZC | `CD_Rebels_Z_Spawn` / `CD_Rebels_ZC_Spawn` | `UW_UI_Conversation_Lightside` |
| Empire ZB / ZC | `CD_Empire_ZB_Spawn` / `CD_Empire_ZC_Spawn` | `UW_UI_Conversation_Darkside` |

All `conversationType=Custom`, `display Name=Field Operator`, `show Indicator Bubble=false`.

Wiring:

1. Capture `OnControlChange` → Conversation **Enable** (ZB after A, ZC after B).
2. Switch `S_Dialogue_GameStart` starts **On**; Zone A capture **Disable**s it. Conversations also `On Check Result On → Disable` from that switch so the prompt stays off until A falls.
3. Player Spawner `On Player Spawned` → `InitiateConversation`.
4. Conversation choice events → Teleporter `Teleport` at the named landmark (`TD_Rebels_ZB_Frontlines`, `TD_Rebels_ZC_STARFIGHTER_*`, `TD_Empire_ZC_TieFighter_*`, …). **12** teleporters.

Initial pads (`PS_*_ZoneA_*`) stay `enabledDuringPhase=Always`, `useAsIslandStart=false`, `visibleInGame=false`, team-filtered. Later pads Enable/Disable with the same capture events — Conversation only *routes* a player who already landed on a pad.

HUD Messages (`HUDM_*`) fire on `OnControlChangeStarts` (contested, dedicated Empire/Rebels widgets) and `OnControlChange` (capture, Lightside/Darkside widgets). Live `message` on all six is the same leftover string **“Resistance now controls the First Town”** — train on the widget split, not the copy.

## Don't

- Don't look for `Content/Dialogue/` — shipped folder is **`DIalogue/`**. Docs say Dialogue.
- Don't treat this as Narrative’s quest enum. No Verse Enable/Disable; classic bindings only.
- Don't expect a pad per destination. Docs: Teleporter listen count (**10** in prose) is why destinations share Conversation choices.
- Don't place `Swap_assets_on_capture.verse` / `Vehicle_Speed_Mutator.verse` as the picker — those files exist and are **unplaced**. The one placed Verse device is `Hide_on_Start` (hides **24** props).
