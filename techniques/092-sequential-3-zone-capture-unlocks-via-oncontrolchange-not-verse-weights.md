---
id: technique-092
kind: technique
---

# Sequential 3-zone capture unlocks via `OnControlChange` — not Verse weights

**Sources:** team_conquest_sample
**Tutorial:** space-opera™ Team Conquest — Capture Areas and Event Triggers. Contrast: weighted-capture-spawns.md (Deserted tags + ally density, all three points live). Asymmetric fill stays on asymmetric-cap-balancer.md. Arena rotation stays on persistable-arena-rotation.md.

## Problem

A 12-player 2-team attack/defend map needs **one live objective at a time**. Capturing A must unlock B (and vehicles / a hero / new pads) without a Verse spawn manager.

## Pattern (shipped)

Three `Device_CaptureArea_V2` (`CA_ZoneA/B/C`). All `startingTeam` **TeamIndex 2** (Empire), `canBeCapturedByTeam` **TeamIndex 1** (Rebels), `enableDuringPhase=Gameplay Only`, `neutralizeTimeOverride=No Neutralize`, `useSplineShapeForBoundary=true`, `controlTime` **15 / 15 / 25**.

| Zone | `captureAllowedOnGameStart` | What `OnControlChange` enables |
|---|---|---|
| A | **true** | HeroA `ApplyTransformationToInstigator`; Zone-B pads / guards / AAT Enable; Switch Disable (opens Conversation path); Beacon B |
| B | **false** (A `AllowCapture`s B) | HeroB transform; Zone-C pads / Fighter / starfighter / next AAT; Beacon C; Conversation ZC |
| C | **false** (B `AllowCapture`s C) | `EGD_RebelWin` Activate |

Timer `T_GameTimer` duration **900**, `startAtGameStart=true`, `canInteract=No`, `OnSuccess` → `EGD_EmpireWin`. Rebels do not need a score race — capturing C **is** the win.

Beacons: A `enabledOnPhase=Always`; B and C `None` until the previous capture Enables them, then the next capture Disables the old one.

Vehicles start `enabled During Phase=None` except Turbolaser **Always**. AAT Enable on A, Disable on B; starfighter / Fighter Enable on B.

## Don't

- Don't copy Deserted’s weighted Verse pads. This island never scores periodically (`periodicScoring=Off`) and never tags spawners.
- Don't trust Outliner labels for hero zone. `HD_ZoneB_Chewwy` is wired to **A**; `HD_ZoneC_HeroB` is wired to **B**. VillainA / VillainC have **no** `ApplyTransformation` bindings.
- Don't wait for `progressDecayType=No Decay` — live is **Default**. Accent is **Team Relationship Color**, not docs Blue.
- Don't count on turret turrets. Live spawners are Turbolaser (`CopperPuck`/`IronPuck`) + AAT (`CloverDog`) + THEME_PACK starfighter/Fighter.
