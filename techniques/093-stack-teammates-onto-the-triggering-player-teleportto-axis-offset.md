---
id: technique-093
kind: technique
---

# Stack teammates onto the triggering player (`TeleportTo` + axis offset)

**Sources:** turtle_arcade_sample (`multiplayer_teleporter.verse`)
**Tutorial:** TURTLE Travel Between Areas — **External**. Snippet **Observed**.

## Problem

A 4-player sidescroller teleports one pawn (lobby timer, encounter trigger, sewer exit). Barriers and Fixed Point cameras then lock the arena around that pawn. Anyone left behind is stuck outside the fight.

## Pattern

One `creative_device` with two `@editable`s:

- `TriggerDevice : trigger_device`
- `TeleportOffsetAxis : vector3` (shipped `{X := 0.0, Y := 1.0, Z := 0.0}`)

`OnBegin` subscribes `TriggerDevice.TriggeredEvent`. The handler:

1. Resolves the optional `?agent` to a `fort_character`.
2. Iterates `GetPlayspace().GetPlayers()`, skipping the triggering character.
3. `TeleportTo[origin + (TeleportOffsetAxis * TeleportOffset), rotation]` with `TeleportOffset` stepping **100** uu per teammate.
4. Failed `TeleportTo` is a no-op (`if` / decides) — players stay put rather than falling through geo.

Five placed copies, same class, different triggers (**Observed** `GetDeviceProperties` → Verse subobject → `savedActor`):

| Device | Trigger |
|---|---|
| Intro: MultiplayerTeleporter | Intro: TeleportingTrigger (after lobby teleporter) |
| FirstEncounter: MultiplayerTeleporter | FirstEncounter: Trigger |
| Sewers: MouserMultiplayerTeleporter | Sewers: MouserTrigger |
| Sewers: ExitMultiplayerTeleporter | Sewers: ExitTeleportTrigger |
| BossFight: MultiplayerTeleporter | BossFight: IntroTrigger |

`ListEventBindings` on the Verse actors is **empty**. The Creative graph never binds them; the `@editable` trigger *is* the wiring. `visibleInGame=false`.

Live `teleportOffsetAxis` on all five is `{x:0, y:1, z:0}` — **Y**, matching the snippet default.

## Don't

- Don't emit the tutorial's **"offset in the X axis"**. Travel chapter prose says X; shipped vector and comment example are **Y**.
- Don't train on the docs expand widget if it truncates. Ground truth is `snippets/verse/multiplayer_teleporter.verse`.
- Don't expect `GetDeviceProperties.triggerDevice` to be the placed trigger path. It returns `…VerseDevice_C_….multiplayer_teleporter_0.__verse_0x…_TriggerDevice`; read `savedActor` on that subobject.
