---
id: technique-094
kind: technique
---

# TURTLE Sewer Tunnel: hidden-travel groups, not a Creative Teleporter

**Sources:** turtle_arcade_sample, city_starter_template
**Tutorial:** TURTLE Travel Between Areas — **External**. Live values **Observed**.

Giftbox occupancy-inspect (Social) also uses hiding-prop travel fields; do not duplicate — that packet is inspect/eject: hiding-prop-occupancy-inspect.md.

## Problem

Players should *choose* when to drop from streets into the sewers, with a tunnel animation, while a 4-player squad still fits. A Teleporter + Trigger is the *forced* gather path (lobby start, sewer *exit*), not this grate.

## Pattern

`Device_HidingProp_Sewer_Tunnel` is a hiding-prop skin with **Hidden Travel** filled in:

| Actor | `hiddenTravelGroup` | `hiddenTravelTargetGroup` | `numberOfOccupants` |
|---|---|---|---|
| Sewers: TURTLETunnelEntrance | **0** | **2** | 4 |
| Sewers: TURTLETunnelExit | **2** | **1** | 4 |

Walk into the grate → hide + travel to the paired group. No Verse.

Exit grate `OnStopHiding` → `Disable` itself, `AddToPlayer` on the sewer Fixed Angle camera, `AddTo` on the sewer Side Scrolling Controls (the second controls device from the cameras chapter).

The **return to streets** is *not* another grate. Docs + live: Trigger → Fade HUD Message → 1 s-class timer → Verse multiplayer teleporter, so the street tiles can stream in. Construction Perspective `AddToAll` is on that same exit trigger.

## Don't

- Don't wire a Creative Teleporter into the grate. Teleporters are the lobby start and the sewer *leave* gather.
- Don't copy Social's Valid/Invalid inspect buttons. This prop is travel, not seeker inspect.
- Don't set `numberOfOccupants` below the **spawn-pad count you actually shipped**. Arcade matches `maxPlayers=4`. City Starter ships both grates at group **1→1**, occupants **2**, with only two spawn pads — while island `maxPlayers=16`.

City Starter uses the same `Device_HidingProp_Sewer_Tunnel` as the lair↔streets hop (HUD `OnStopHiding` starts the pizza tour). Rooftops then use Creative Teleporters, not a second grate. Digest: city_starter_template.md.
