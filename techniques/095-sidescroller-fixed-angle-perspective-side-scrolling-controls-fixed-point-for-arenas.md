---
id: technique-095
kind: technique
---

# Sidescroller: Fixed Angle Perspective + Side Scrolling Controls, Fixed Point for arenas

**Sources:** turtle_arcade_sample
**Tutorial:** TURTLE Cameras and Controls — **External**. Live values **Observed**.

Contrast: first-person toggle via Switch is switch-gated-first-person.md. Deeper camera stacks: gameplay-camera-switching.md.

## Problem

A beat-em-up in a 3D Fortnite island must lock movement to two axes and swap to a static fight camera, then restore traversal without fighting the previous camera.

## Pattern

Two device classes, **not** Verse:

1. **Fixed Angle Camera** = `Device_CameraModes_Perspective` (`projection Mode=Perspective`).
2. **Side Scrolling Controls** = `Device_CustomControls_Sidescroller` (`constrain Movement=true`; Jump/Crouch `Dedicated`).

Neither `addToPlayersOnStart`. Lobby timer `OnSuccess` → Intro camera `AddToAll` + Intro sidescroller `AddToAll`. Encounters `AddToAll` / `AddToPlayer` a **Fixed Point** camera (`Device_CameraModes_FixedPoint`). Tracker `When Complete` (or sewer tunnel `OnStopHiding`) restores the next Perspective device.

Live outdoor vs sewer angle (**Observed**):

| Camera | distance | pitch | FOV | transition |
|---|---|---|---|---|
| Intro / Construction Perspective | 800 | **-15** | 100 | Ease-In-Out 0.5 / 0.5 |
| Sewers Perspective | 800 | **-20** | 90 | — |
| First Encounter Fixed Point | — | — | 90 | Ease-In-Out **0.2** in |

A **second** sidescroller (`Sewers: SideScrollingControls`) is `AddTo` on TURTLE Tunnel Exit `OnStopHiding`, matching the docs “faulty behavior” note when leaving the sewers.

Live counts: **4** Perspective, **3** Fixed Point, **2** Sidescroller.

## Don't

- Don't train **5** Fixed Angle cameras. Cameras chapter lists 5; OFPA + Outliner show **4**.
- Don't leave `addToPlayersOnStart` on. The lobby cine would inherit the sidescroller camera.
- Don't reuse one Perspective for streets and sewers. Pitch/FOV change is the whole sewer-camera lesson.
