---
id: technique-096
kind: technique
---

# theme enemies: TURTLE Character Spawner enum vs NPC Character Definition

**Sources:** turtle_arcade_sample
**Tutorial:** TURTLE Enemy Encounters and Obstacles — **External**. Live values **Observed**.

Contrast: Walker Type lives on an NPC Character Definition (walker-character-definition.md). Verse `npc_behavior` wander is npc-behavior-navigatable-focus.md.

## Problem

EnemyFaction / BossEnemy are Melted Iron theme characters. Mousers are a custom NPC. One spawner class does not cover both.

## Pattern

**Foot Elite and BossEnemy** use `Device_TURTLE_CharacterSpawner` (`TURTLE Character` enum), not an NPC Character Definition.

| Label | `TURTLE Character` | HP | Enable |
|---|---|---|---|
| FirstEncounter: Spawner1–3 | `EnemyFaction` (Sword / Shadow) | 75 / 75 | Encounter trigger `Enable` |
| BossFight: SupportSpawner1 | `BossEnemy` | 300 / 300 | Boss cine `OnStopped` → `Enable` |
| BossFight: SupportSpawner2–3 | `EnemyFaction` | 100 start / 300 max | Support1 `On Spawned` → `Spawn` |

`nPCBehaviorScript` is not the API. Count / timer still sit on the spawner (`spawnCount=1`, First Encounter `allowInfiniteSpawn=true` with `totalSpawnLimit=1`).

**Mousers** use the generic `Device_CharacterSpawner` + authored definition:

`/4e87564d-…/NPCCharacterDefinition_RobotEnemy`

| Slot | Live |
|---|---|
| Type | `BP_Chompy_CharacterType_C_0` (nested in the RobotEnemy asset) |
| Behavior | `CharacterModifier_ChompyBehavior_C_0` |
| modifiers[0] | Health — 100 / 100, 0 shield, vulnerable |
| modifiers[1] | Team — `teamIndex` **50** |
| `nPCBehaviorScript` | **null** |

Three sewer spawners: `enabledAtGameStart=false`, infinite, `spawnTimer=1`, `spawnOnEnabled=true`. Tracker target **9** = three spawners × three waves (docs).

The cinematic BossEnemy dummy is a **Character** device (`Device_Character_V2`) with cosmetic `/COSMETIC_ASSETS/…/Character_PeacefulPoem` — visibility toggles around the sequence. The fight BossEnemy is the TURTLE spawner enum.

## Don't

- Don't put Mousers on a TURTLE Character Spawner, or EnemyFaction on a generic NPC Spawner. The island splits the two classes on purpose.
- Don't cite `NPCCharacterDefinition_Chompy` as a second definition. It is an **ObjectRedirector** onto `NPCCharacterDefinition_RobotEnemy` (same as `NewNPCCharacterDefinition_Chompy`).
- Don't subclass `npc_behavior` for either enemy. Foot/BossEnemy = theme enum; RobotEnemy = Character Definition Type `BP_Chompy_CharacterType`.
