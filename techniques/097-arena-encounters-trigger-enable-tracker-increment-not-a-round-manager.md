---
id: technique-097
kind: technique
---

# Arena encounters: Trigger Enable + Tracker Increment, not a round manager

**Sources:** turtle_arcade_sample
**Tutorial:** TURTLE Enemy Encounters and Obstacles — **External**. Bindings **Observed**.

Walker horde gating (volume/checkpoint vs infinite hoses) is volume-checkpoint-gated-horde.md — same Enable/Disable idea, different bus (volume vs tracker).

## Problem

A sidescroller needs a locked fight, a completion signal, then a restore of traversal camera / barriers / a “proceed” arrow. There is no Verse round loop.

## Pattern

Creative bindings only:

1. **Trigger `OnTriggered`** → Fixed Point `AddToAll`/`AddToPlayer`, Barriers `Enable`, Checkpoint `Register`, spawners `Enable`, Tracker `Assign to All`, Verse gather device (via `@editable` trigger).
2. Each spawner **`On Eliminated` → Tracker `Increment Progress`**. `statToTrack=None`; the tracker is a counter, not a native elim stat.
3. **`When Complete`** → disable the exit barrier, `Show Props` on the LED arrow Prop Manipulator, restore Perspective (`AddToPlayer` / `AddToAll`), pizza `Spawn Supply Drop`, disable the infinite spawners.

Live targets (**Observed**):

| Room | Tracker `targetValue` | Spawners |
|---|---|---|
| First Encounter | **3** | 3× EnemyFaction TURTLE spawners |
| Sewer Mousers | **9** | 3× NPC spawners, 1 s infinite respawn (3 waves) |
| Boss | **10**, `assignOnGameStart=true` | **not** the win condition |

Boss win is `SupportSpawner1` (BossEnemy) `On Eliminated` → End Game `Activate`. Cine `OnStopped` Enables BossEnemy; BossEnemy `On Spawned` `Spawn`s the two Foot support spawners. Extra cine Barrier Enables on intro trigger and Disables when the sequence stops (player input stays on during the shot).

## Don't

- Don't read `statToTrack` as Eliminations. It is `None`; progress is explicit Increment from spawner events.
- Don't treat the Boss Tracker as the win screen. End Game listens to BossEnemy elim. The tracker (`targetValue=10`, assigned at game start) looks leftover.
- Don't skip the Verse gather on the same trigger. Barriers will trap whoever wasn't standing on the volume.
