---
id: technique-098
kind: technique
---

# Fill the team with the largest remaining cap, not an even split

**Sources:** triad_infiltration_template
**Snippet:** `../snippets/verse/triad_infiltration.verse`
**Tutorial:** ch.3 Balancing Teams Asymmetrically — same algorithm; class name drifted (`triad_infiltration_game` vs `triad_infiltration`).

## Problem

Three roles (2 infiltrators / 4 attackers / 4 defenders). An even 10/3 split would dump 3-4 onto the stealth team and break the mode. New joiners must preserve the **ratio**.

## Pattern

```verse
# Island Settings team order IS the index contract:
MaybeInfiltrators = option{AllTeams[0]}  # cap 2
MaybeAttackers    = option{AllTeams[1]}  # cap 4
MaybeDefenders    = option{AllTeams[2]}  # cap 4
TeamsAndTotals[Team] = Maximum…

FindTeamWithLargestDifference():?team =
    # argmax (MaximumTeamSize - GetAgents[Team].Length)
    # returns false if every team is full / player already on that max-diff team

BalancePlayer: AddToTeam[InPlayer, AssignedTeam]  # fails = already there
BalanceTeams: Shuffle(GetPlayers()) then BalancePlayer each
```

Set island **max players = sum of caps** so nobody is unteamed.

JIP: `PlayerAddedEvent` → `BalancePlayer` → `Respawn` at `Teleporters[TeamIndex]` (not the round-start `Teleport`). If team is Infiltrators, notify `invisibility_manager.OnInfiltratorJoined`.

Weapons: `WeaponGranters[TeamIndex].GrantItem` on `SpawnedEvent`. Swap Island Settings order and you grant the wrong kit with no compile error.

## Don't

- Don't even-split (`Length % 3`).
- Don't copy docs class `triad_infiltration_game`.
- Don't assume `Teleport` == `Respawn` — shipped uses Teleport at round start, Respawn on JIP.
