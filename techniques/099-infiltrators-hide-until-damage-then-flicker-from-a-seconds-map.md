---
id: technique-099
kind: technique
---

# Infiltrators: `Hide` until damage, then flicker from a seconds map

**Sources:** triad_infiltration_template
**Snippet:** `../snippets/verse/invisibility.verse`

## Problem

Stealth team should be invisible at spawn, reveal briefly when shot, then vanish again. Optionally the whole stealth team flickers when one is hit. Don’t stack a new flicker coroutine per pellet.

## Pattern

After `BalanceTeams`, owner calls `StartInvisibilityManager(AllTeams, AllPlayers, Infiltrators)`.

```verse
FortCharacter.Hide()
spawn{OnInfiltratorDamaged(Player)}   # loop: Await DamagedEvent

# per pellet:
if (not IsFlickering[Agent]):         # PlayerVisibilitySeconds[Agent] > 0
    set PlayerVisibilitySeconds[Agent] = VulnerableSeconds   # 3
    spawn{FlickerCharacter}           # Hide/Show every FlickerRateSeconds 0.4
else:
    set PlayerVisibilitySeconds[Agent] = VulnerableSeconds   # reset timer only
```

`FlickerCharacter` decrements the map by `FlickerRateSeconds * 2` per cycle; `<= 0` → `Hide()` and break.

`IsVisibilityShared`: on damage, iterate `GetAgents[CurrentTeam]` and StartOrReset each teammate.

Respawn from infiltrator pads: `OnPlayerSpawn` Hide again if `AllTeams[0]`.

## Don't

- Don’t use a mutator volume for this — it’s `fort_character.Hide/Show`.
- Don’t spawn FlickerCharacter while `IsFlickering` or you’ll overlap loops.
- Don’t call StartInvisibilityManager before teams exist.
