---
id: technique-100
kind: technique
---

# Bind UMG to a Creative device through a Viewmodel

**Sources:** user_interfaces_template
**Tutorial (outline only):** UMG Widgets — **External**. Live widget not opened this pass; treat `Progress` names as External until backlog item 2 is closed.

## Problem

A HUD Controller (or Timer, Stat Creator, Skilled Interaction…) already knows player health/shield/item. You want a custom bar, not the default HUD, without polling in Verse.

## Stack

1. **User Widget** (or **Modal Dialog Variant** for Pop-Up Dialog / Conversation) authored in UMG.
2. Device has a widget slot (HUD Controller: Health/Shield HUD, Equipped Item Info Widget Override, …).
3. Widget picks a **device-specific Viewmodel** (e.g. *Device - HUD Controller Team/Squad Player Info List* → Controlling Player Info → Health / Shield).
4. **View Bindings:** Image.Brush → conversion **Set Scalar Parameter** → Parameter Name `Progress` → Value = viewmodel Health or Shield.
5. Material instance on that Image is UI-domain (`M_ProgressBar_Basic` family). `Progress` is 0–1 (`Progress Start` / `ProgressEnd`). Damage slides the fill; restore slides it back.

Modal Dialog Variant is **not** interchangeable with User Widget. Conversation can use a Conversation Bank *or* a Modal Dialog Variant for look.

## Equipped item

Second viewmodel on HUD Controller: equipped item icon, name, ammo icon/count/surplus, magazine-type flag. Template samples use assault rifle `WID_Assault_AutoHigh_Athena_C_Ore_T03` (**External**, tutorial).

## When Verse instead

If there is **no** device viewmodel for the data, drive the widget from Verse fields (`set Widget.IntField`) or build the HUD in Verse (`material_block` + parameter writes). Don't fight the viewmodel path for health/shield — HUD Controller already has it.

## Don't

- Don't put gameplay logic in the UMG event graph for these samples — the tutorial's widgets are layout + bindings.
- Don't expect Fortnite Custom Button OnClicked to reach Verse. Clicks that must reach Verse go through `button`+`button_slot` or a BP-writable Button + Event Dispatcher (`event(tuple())`).
