---
id: technique-101
kind: technique
---

# Clickable custom UI in Verse = `button` + `button_slot`

**Sources:** user_interfaces_template, energy_ability_sample
**Snippets:**
- `../snippets/verse/ui_simple_button_device.verse`
- `../snippets/verse/ui_custom_button_device.verse`
- `../../energy_ability_sample/snippets/verse/menu_EnergyBlade_customization_device.verse` — `ui_helper.Buttons.BaseButton` wraps `UW_Item` / tabs in `button` + `button_slot`

## Problem

You want a HUD control that looks like your art (text, material, UMG visual) and fires Verse on click, without Fortnite's default button chrome.

## The pattern (shipped)

Wrap **any** widget in the base `button` via `button_slot`. `OnClick()` is a listenable. Highlight/Unhighlight let you restyle.

```verse
using { /UnrealEngine.com/Temporary/UI }

TextWidget : text_block = text_block{}
VerseButton : button = button{ Slot := button_slot{ Widget := TextWidget } }

VerseButton.OnClick().Subscribe(OnClicked)
VerseButton.HighlightEvent().Subscribe(OnHighlighted)
VerseButton.UnhighlightEvent().Subscribe(OnUnhighlighted)

PlayerUI.AddWidget(RootOverlay, player_ui_slot{ InputMode := ui_input_mode.All })
PlayerUI.SetFocus(VerseButton)
```

`ui_simple_button_device` does this twice in a vertical `stack_box`, plus a `button_quiet` Close (default chrome, fine for dismiss). Show on `volume_device.AgentEntersEvent`; `RemoveWidget` on close.

## Custom visuals

`CustomizableButton` in `ui_custom_button_device.verse` puts a **UMG widget** (`UW_VerseButton01`) *inside* the `button_slot`, then drives Verse fields on that UMG (`Background`, `Icon`, `Label`, `TextColor`, `WasHighlighted`) from Highlight/Unhighlight. Clicks still come from `VerseButton.OnClick()`, not from a Fortnite Custom Button dispatcher.

```verse
var VerseButton:button = button{ Slot := button_slot{ Widget := overlay{} } }
# later:
VerseButton.SetWidget(button_slot{
    Widget := ButtonVisual
    HorizontalAlignment := horizontal_alignment.Fill
    VerticalAlignment := vertical_alignment.Fill
})
GetClickListenable():listenable(widget_message) = VerseButton.OnClick()
```

Four instances in a left-bottom `stack_box` over a dim `color_block`.

## Don't

- Don't put `button_quiet` over a texture and expect restyling — no Verse style API; faint hover tint.
- Don't use Fortnite palette **Custom Button** (`UIFrameworkCustomButtonWidget`) as the click source — `OnButtonClicked` is often not BP-writable, so a UMG dispatcher never fires (animations still play and fool you). this sample's UMG-input demo listens to **Enhanced Input** (`UI.PreviousTab` / `NextTab` / `Use`) and Verse fields (`TriggerIntro`), not that button class.
- Don't `AddWidget` two top-level `InputMode.All` canvases — they hide each other. One root overlay; Z-order inside it.
- Free cursor = `player_ui_slot{ InputMode := ui_input_mode.All }`. Without it, clicks don't reach the widget.
- Don't train Action Adventure quest chrome from this wrap — that dialog uses `button_loud` / `button_quiet`, not `button`+`button_slot` (persistable-quest-line-verse-dialog.md).

## Related

UMG-only close: `ui_custom_button_umg_device.verse` adds the User Widget itself and `SubscribeToEvent(Widget.CloseEvent, …)` from `verse_event_helpers` because UMG dispatchers are `event(tuple())`.
