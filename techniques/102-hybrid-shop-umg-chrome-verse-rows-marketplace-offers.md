---
id: technique-102
kind: technique
---

# Hybrid shop: UMG chrome + Verse rows + Marketplace offers

**Sources:** user_interfaces_template
**Snippets:** `ui_shop.verse`, `ui_shop_offer_item.verse`, `ui_link_volume_to_shop_device.verse`, `ui_shop_setup_device_template.verse`
**Tutorial:** Verse Fields Examples — outline OK; class names in docs dropped the `ui_` prefix.

## Problem

In-island offers need Epic checkout (`BuyOffer`) *and* a designed storefront. Neither a pure User Widget nor a pure `overlay` does both well.

## Shipped split

| Concern | Where |
|---|---|
| Background, header, item-detail panel | UMG widgets (`UW_Shop01_*`) referenced as generated classes |
| Offer rows, Close, Buy | Verse `stack_box` + `button_quiet` / `button_loud` |
| Offer data | Marketplace `[]offer`; `price_vbucks` → float for the details widget |
| Open trigger | `volume_device.AgentEntersEvent` → `ShowArrayOfOffers` / `ui_shop02.ShowUI` |

`PopulateOffers` instantiates `ui_shop_offer_item` per offer (repeatable row). Buy:

```verse
BuyButton.OnClick().Await()
if (PlayerUI := GetPlayerUI[Player]):
    HideUI(PlayerUI)
BuyOffer(Player, SelectedOffer)
```

Cursor: `player_ui_slot{ InputMode := ui_input_mode.All }` on `AddWidget`.

Demo02 (`ui_shop02*`) is a second implementation **not in the tutorial chapter**. Same idea, more files (`buy_button`, `offers_data`). Link device has **two** volumes, one per shop.

## Module stubs

`ui_helper.verse` declares empty `UserInterfaces.…` modules so `UserInterfaces.UIFeatureExamples.InIslandTransactions.Demo01.Widgets.UW_Shop01_Background` resolves after Verse asset reflection. Copy that stub tree (or the folders) when migrating.

## Don't

- Don't train on the tutorial name `shop_setup_device_template` — shipped is `ui_shop_setup_device_template`.
- Don't assume chapter 3's CloseButton margin snippet matches `CreateUI` in the file. Diff before paste.
- Experimental marketplace flags in the descriptor are **Unverified**; the Verse imports `/Fortnite.com/Marketplace` are Observed.
