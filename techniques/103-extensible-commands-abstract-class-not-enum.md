---
id: technique-103
kind: technique
---

# Extensible commands: abstract class, not enum

**Sources:** verse_device_template
**Snippet:** `../snippets/verse/commands.verse`
**Tutorial:** ch.04 — outline good; extra `<unique>` in docs is **not** in the file.

## Problem

UI must enqueue “Forward / TurnLeft / TurnRight” and the NPC must switch on them. After publish you may want more verbs. Verse `enum` cannot gain cases post-publish.

## Shipped

```verse
command := class<computes><abstract>:
    DebugString():string

Commands := module:
    forward_command<public> := class<computes>(command):
        DebugString<override>():string = "Forward"
    turn_command<public> := class<computes><abstract>(command):
    turn_right_command<public> := class<computes>(turn_command):
        DebugString<override>():string = "TurnRight"
    ...
    Forward<public>:forward_command = forward_command{}
    TurnRight<public>:turn_right_command = turn_right_command{}
    TurnLeft<public>:turn_left_command = turn_left_command{}

ToString(Command:command):string = Command.DebugString()
```

`<computes>` = construct at module scope. `<abstract>` = no raw `command{}`. Dispatch:

```verse
if (Commands.forward_command[Command]):
    CurrentTransform.Rotation.GetLocalForward()
else if (Commands.turn_right_command[Command]):
    CurrentTransform.Rotation.GetLocalRight()
```

That is a **failable type query**, not `Command = Commands.Forward` (what ch.01 prints).

## Don't

- Don't add `<unique>` because ch.04 did — shipped `command` is only `<computes><abstract>`.
- Don't name `turnright_command`. Shipped is `turn_right_command`.
- Don't use enum for an open verb list.
