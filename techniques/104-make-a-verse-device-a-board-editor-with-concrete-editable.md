---
id: technique-104
kind: technique
---

# Make a Verse device a board editor with `<concrete>` + `@editable`

**Sources:** verse_device_template
**Snippet:** `../snippets/verse/gameboard.verse`

## Problem

Five puzzle boards, each with camera, start prop, end trigger, N obstacles (barriers + dissolve/appear cinematics + pad). Designers reorder/add boards without touching Verse.

## Pattern

`class<concrete>` so the type can be an `@editable` field (and array element) on a `creative_device`.

```verse
gameboard<public> := class<concrete>:
    @editable
    TileSize<public>:vector3 = vector3{X:=512.0, Y:=512.0, Z:=384.0}
    @editable
    Camera:gameplay_camera_fixed_point_device = gameplay_camera_fixed_point_device{}
    @editable
    EndGoal:trigger_device = trigger_device{}
    @editable
    Obstacles:[]obstacle = array{}
    @editable
    var StartingCharacterPosition<public>:creative_prop = creative_prop{}
    @editable
    OpeningCinematic:cinematic_sequence_device = cinematic_sequence_device{}
    EndGoalReached<public>:event() = event(){}
```

On the minigame device: `@editable var Gameboards:[]gameboard = array{}`. Details panel: add/remove/reorder elements (docs ch.02). Start pose is a **prop's transform** + ZOffset so spawn doesn't clip.

`obstacle` is also `<concrete>` with `@editable` Barriers[], BarrierDissolves[], BarrierAppears[], Trigger. Passed: Disable barriers, Play dissolve SEQ. Reset: reverse.

`Start()` = `Camera.AddToAll()` + opening cinematic. `Stop()` = `RemoveFromAll()`.

Tile math uses **Temporary/SpatialMath XYZ**, matching Fortnite tiles. Don't convert to Left/Up/Forward here.

## Don't

- Don't omit `<concrete>` and wonder why the struct isn't in the device Details.
- Don't hardcode five boards in Verse when the array already is the level data.
