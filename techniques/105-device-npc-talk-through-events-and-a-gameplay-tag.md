---
id: technique-105
kind: technique
---

# Device ↔ NPC talk through events and a gameplay tag

**Sources:** verse_device_template
**Snippets:** `verse_commander_minigame.verse`, `verse_commander_character.verse`, `verse_starter_tags.verse`

## Problem

A `creative_device` owns the match (boards, UI, spawn). An `npc_behavior` owns walking. The NPC is spawned later, so it cannot be wired with `@editable` in the other direction at edit time.

## Pattern

1. Declare a tag: `verse_commander_minigame_tag := class(tag){}`.
2. **Human must put that tag on the minigame device** (VerseTagMarkup). Hub Getting Started. Without it, `OnBegin` on the NPC logs and returns.
3. NPC looks the device up at runtime:

```verse
CreativeObjects := CharacterTeleportOutVFX.FindCreativeObjectsWithTag(verse_commander_minigame_tag{})
for (CreativeObject : CreativeObjects):
    if (MinigameManager := verse_commander_minigame[CreativeObject]):
        set VerseCommanderMinigame = MinigameManager
```

Shipped query is **on a device the behavior already `@editable`s** (teleport VFX), not a global `GetCreativeObjectsWithTag`. Docs show the global + `[0]`. Use the snippet.

4. Typed events on the device:

| Event | Payload | Direction |
|---|---|---|
| `ExecuteCommandsEvent` | `tuple([]command, vector3)` queue + tile size | device → NPC |
| `BoardResetEvent` | `transform` start pose | device → NPC |
| `CharacterFinishedMovingEvent` | none | NPC → device |
| `GameEndedEvent` | none | device → NPC (cancel loops) |

NPC `race`s `loop{ AwaitCommands() }` vs `AwaitReset()` vs `GameEndedEvent.Await()`.

## Don't

- Don't `@editable` the minigame on the NPC asset and expect a spawned instance to fill it — tag lookup is the shipped seam.
- Don't assume the template tagged itself. PIE without the tag = dead NPC.
