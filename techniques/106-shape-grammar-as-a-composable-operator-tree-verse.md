---
id: technique-106
kind: technique
---

# Shape grammar as a composable operator tree (Verse)

**Problem.** You have a voxel volume the player edited and you want to turn each region into
buildings/walls/roofs — with rules like "split into floors; each floor into corners, walls
and a centre; a wall voxel with something in front of it becomes a solid box instead."
Hardcoding that as nested if/else per shape does not scale and cannot be retuned by a
designer.

**How UEFN's Procedural Building template does it** (`Lib/SpaceGrammar/`).

Every rule is a class extending one base:

```verse
vo_base<public> := class:
    Operate<public>(Box:voxel_box, Rot:prop_yaw, BuildSystem:build_system):void={}
```

An operator receives an axis-aligned `voxel_box` (origin+size in voxel coords), a yaw, and
the build system. It either **splits** the box and passes sub-boxes to child operators, or
**terminates** by spawning a prop (`vo_prop`).

A building style is then *assembled*, not coded — `rs_building1.SetupRules`:

```verse
set VO_Root = vo_floorsplit:            # cut box into 1-voxel floors
        VO_FloorHeight1 := vo_cornerwallsplit:   # each floor → 4 corners + walls + centre
            VO_CornerLength1 := vo_corner:       # corner voxel: corner/face/box depending on neighbour occlusion
                VO_Corner  := vo_prop: Prop := PT.Building1_corner
                VO_Face    := vo_prop: Prop := PT.Building1_face
                VO_Covered := vo_prop: Prop := PT.Building1_box
            VO_Wall := vo_wall:                  # wall run: N-wide pieces where they fit
                VO_Width1 := vo_facecoveragecheck_1voxel:  # wall→box if something outside
                    VO_Clear   := vo_prop: Prop := PT.Building1_face
                    VO_Covered := vo_prop: Prop := PT.Building1_box
            VO_Centre  := RoofTileCheck          # top floor centre = roof
            VO_TooSmall := vo_tilefill:          # tiny boxes: per-voxel 4-way neighbour check
                VO_Tile := FourWay
```

## The operator zoo (17 shipped operators)

| Category | Operators |
|---|---|
| **Split** | `vo_floorsplit` (N-high + 1-high floors), `vo_floorsplit_AB` (alternating floors), `vo_wall` (N-wide runs), `vo_wallalternate` (A/B strips), `vo_slice` (cut a face, absolute or proportional), `vo_cornerwallsplit` (corners+walls+centre), `vo_volumefitandremainder` (tile a fixed volume, pass leftovers on), `vo_tilefill` (1×1 tiles over XY) |
| **Test/branch** | `vo_fourwaycheck` (which of 4 neighbours are solid → 16-way piece pick), `vo_corner` (two-face occlusion → corner/face/box), `vo_facecoveragecheck` (+ fast `_1voxel` variant), `vo_positioncheck` (top/bottom of original box), `vo_topbottomcheck` (top/middle/bottom), `vo_sizecheck` (min dimensions) |
| **Leaf** | `vo_prop` (spawn, with optional backfill + top alignment), `vo_random` (pick child by position-hashed RNG, optional random yaw), `vo_wfc` (hand off to WFC results) |

## Patterns worth copying

- **Rotation is threaded, not global.** `Operate(Box, Rot, …)` carries a `prop_yaw`
  (`deg_0/90/180/270`); face tests apply `ApplyRotToFace`, and `vo_prop` aligns the prop's
  +Y face to the box face implied by `Rot`. One grammar serves all four orientations.
- **Sub-boxes are value types.** `voxel_box` is a `struct<computes>`; operators mutate local
  copies (`set SubBox.Origin = …`) and pass them down — no shared mutable geometry state.
- **Leaf spawning is idempotent.** `vo_prop.SyncMeshGrid` compares the desired
  `(piece_type, yaw)` against what's already spawned in `mesh_grid` and only
  clears+respawns on mismatch — so re-running the grammar after a one-voxel edit only pays
  for changed props (`vo_wfc.SpawnPropIfNeeded` does the same).
- **Determinism without seed plumbing.** `vo_random` and WFC prop pick use
  `RandomIntFromVector(Box.Origin, …)` — prime-tuple hash of voxel coordinates
  (`(37,53,61,191)`), so "random" variation is stable across re-meshes of the same box.
- **Grammar ↔ engine seam is one method.** `rs_base.ApplyRulesToBox` calls
  `VO_Root.Operate(...)`; `build_system.SelectRuleSet` maps `build_category` → rule set.
  New style = new `rs_*` subclass + one case.

**Scope:** works at ~10³-voxel granularity (this is a build-site toy, not city generation).
The grammar never touches actors directly — it only calls `mesh_grid.SpawnProp`, keeping the
whole tree unit-testable in principle.

**Sources:** verse_proceduarl_building.
