---
id: technique-107
kind: technique
---

# Interactive voxel editing: flat grid → SDF → convex boxes → incremental remesh

**Problem.** Player adds/removes one voxel at a time in a 10×10×10 grid and the world must
re-mesh into building pieces *every edit*, without respawning everything.

**How UEFN's Procedural Building template does it** (`Lib/`).

## Three parallel grids, one index math

| Grid | Stores | File |
|---|---|---|
| `voxel_grid.Cells` | `[]?voxel_cell` — category per voxel (the "truth") | `voxel_grid.verse` |
| `voxel_grid.SDF` | `[]?sdf_entry` — distance to nearest empty voxel in ±X/±Y/±Z + `SumMinDist` | same |
| `mesh_grid.PropGrid` | `[]?prop_entry` — the spawned `creative_prop` per voxel | `mesh_grid.verse` |

All three (and WFC's `Tiles`) are flat 1D arrays indexed
`(X*(Size.Y*Size.Z)) + (Y*Size.Z) + Z`, all preallocated with a single `for` comprehension
(the tutorial's "Verse Performance" section calls out element-by-element `+=` as O(N²) —
the shipped code is the counter-example).

## SDF → convex decomposition (`build_system.UpdateBoxesFromSDF`)

1. On edit, `AddBoxToSDF`/`RemoveBoxFromSDF` recompute distances only within
   `MaxDistanceSearch` (6) of the changed box, and only in the directions that could have
   changed (`UpdateSDF_2` + directional `FindDistInPosX…`).
2. Voxels in the affected region are sorted **deepest-first** (`SumMinDist`) by an in-place
   quicksort with a 50-deep recursion guard (`QuickSortVoxelsSortedByDepth` — hand-rolled,
   Verse has no sort).
3. From each unclaimed deepest voxel, `ExpandBoxInSDF` greedily grows a box face-by-face
   (XY faces then Z) while every new face cell is same-category, filled, and unclaimed.
4. The result set of convex `box_info`s is **diffed against the previous set** — boxes that
   survived (`PreviousBoxInfo.Box.Equals[Box]` + same category) are kept. Untouched boxes
   skip re-meshing entirely. However, surviving boxes that touch or overlap the modified
   region (`BoxAffectedCheck[ModifiedBox, PreviousBoxInfo.Box]`) are added to
   `BoxInfosToMesh` and re-run their grammar rules; respawning is avoided at the leaf prop
   level rather than the box level.

## Spawn amortization + mesh diffing

- `mesh_grid.SpawnProp` never spawns directly — it enqueues onto
  `prop_spawner_device.CurrentSpawnInfos` and **signals an event**;
  `SpawnPendingProps` is one `loop { SpawnPropsEvent.Await(); for(…){SpawnProp(…)} }`
  coroutine started in `OnBegin`. Spawning is decoupled from the synchronous edit tick,
  though `SpawnPendingProps` drains the queue in a single unyielded pass (no per-prop `Sleep()`
  or frame budget cap), so queued spawns execute in a burst upon resumption.
- Grammar leaves (`vo_prop`, `vo_wfc`) call `SyncMeshGrid`/`SpawnPropIfNeeded`: compare
  desired `(piece_type, yaw)` with the existing `prop_entry` via `MatchPropEntry` and
  do nothing on match — this is where unchanged props are preserved when touching boxes
  re-run their grammar rules.
- Deletion is `Prop.Dispose()` via `CheckIn`; `mesh_grid.ClearPropsForBox` wipes a region.

## The input side (why it's in this note: it decides *which* voxel)

- `ray_caster` is a textbook **DDA voxel traversal**: step to the nearest next cell boundary
  per axis (`SafeDivide` guards a zero ray component with `Inf`), track hit face via
  `voxel_box_face`. `voxel_grid.Trace` loops `Caster.Next()` until a solid voxel
  (`Trace_VoxelIsSolid` — surface-only categories like Park don't block) or the floor plane
  (`Z = -1` special case).
- **Turbo build = trace a virtual plane.** While holding Place, the trace switches to
  `TracePlane` against the plane defined by the drag-start voxel+face, so you can paint
  overhangs that a real raycast could never hit (`player_data.SelectModeTick_Trace`).
- Build legality: surface-only categories need support below; others need any solid
  neighbour or ground; `CheckPlayerOverlapsVoxel` (AABB vs `fort_character.GetBox()`) stops
  you entombing yourself.
- Rotation of the face-cursor ghost is a `case` over the six faces with explicit degree
  triples — the template predates Scene Graph; everything is `creative_prop.TeleportTo`.

**Gotcha preserved from source:** `prop_yaw` rotations are enum quarter-turns converted by
`PropYawToRadians` (`0.5*PiFloat` steps) — don't feed it degrees.

**Sources:** verse_proceduarl_building.
