---
id: technique-108
kind: technique
---

# Wave Function Collapse in pure Verse

**Problem.** Generate locally-coherent 2D tile layouts (here: grass/water park tiles) where
only adjacency rules decide what can neighbour what — no hand-authored maps, and fast enough
to re-run interactively when the player edits one voxel.

**How UEFN's Procedural Building template does it** (`Lib/WFC/` — ~750 lines total).

## Model: edge labels, not tile pairs

Adjacency is defined per **edge label** (`wfc_model.verse`):

```verse
wfc_edge_label := class<unique>:      # <unique> so labels are map keys by identity
    Name:[]char
    Symmetric:logic

wfc_mesh := class:
    Props:[]piece_type                # prop variants, one picked at spawn time
    Edges:[]wfc_edge_label            # 4 labels, clockwise from +Y
    AllowRotation:logic = true        # false = one fixed orientation
    RandomPropRotation:logic = false  # cosmetic yaw jitter at spawn

wfc_mesh_option := class<unique>:     # (mesh, quarter-turn) pair
    Mesh:wfc_mesh
    Rot:int = 0                       # 0..3
```

`wfc_model.AddMesh` expands each mesh into up to 4 `wfc_mesh_option`s and indexes them by
edge label **and by the direction the label ends up on after rotation**
(`wfc_edge_options.OptionSets[Direction]`). Non-symmetric labels come in mirrored pairs
registered with `AddLabelFlipPair(GrassToWater, WaterToGrass)`; matching always looks up
`FlipEdgeLabel(label)` so asymmetric joins (sand→water vs water→sand) work.

The park model is 5 tiles (`wfc_model_factory.GetParkModel`): grass, water,
straight/in-corner/out-corner shoreline. The model configures a `FallbackPropOpt` (grass),
though in the shipped `vo_wfc.verse` implementation over-constrained (0-option) tiles fail
the `NumRemainingOptions = 1` guard and leave the tile unspawned.

## System: observe/propagate with an explicit queue

`wfc_system` keeps `Tiles : []?wfc_tile` over the **same flat 1D indexing as the voxel grid**
(`X*(Size.Y*Size.Z) + (Y*Size.Z) + Z` — the index math is shared across all three grids, see
`voxel-grid-sdf-remesh.md`).

- `wfc_tile.RemainingOptions : [wfc_mesh_option]int` (map-as-set) + `NumRemainingOptions` +
  `ShannonEntropy` (here just the option count — the TODO notes weights would extend it).
- `AllOptions` flag is the big-win optimization: an unconstrained tile stores no map at all
  until first constrained (`Reset()` sets `AllOptions=true` and skips building a
  `NumAllMeshOptions`-sized map for every tile).
- `Observe()` picks a random min-entropy remaining tile (`GetRandomInt` over the min-entropy
  index list), collapses it to one option, enqueues its 4 neighbours. Note that collapse
  selection is stochastic (`GetRandomInt`); only the post-collapse cosmetic prop variant /
  rotation selection at spawn time uses deterministic prime position hashing.
- `Propagate()` drains `PropagationQueue` by index (`loop` + `QueueIndex += 1`, not array
  pops): for each (affected pos, instigator tile, direction) it intersects the affected
  tile's options with everything compatible with the instigator's current options on that
  edge; if the set shrank, its neighbours are re-enqueued. A tile collapsing to 0 options
  logs "Overconstrained!" — the template deliberately **does not backtrack**. (Note: in
  the shipped `vo_wfc.verse`, zero-option tiles do not reach `FallbackPropOpt` due to the
  outer `Tile.NumRemainingOptions = 1` guard, leaving the cell empty).
- `Regenerate()` = `loop { if(not Observe?) break; if(not Propagate?) break }`.

## Incremental edits (the part most WFC write-ups skip)

- `AddRegion(box, model, RecalcRadius)` / `RemoveTile(voxel, RecalcRadius)` reset only the
  touched region (expanded by radius, clamped) and re-run.
- `ResetRegion` reseeds border constraints correctly: region-edge tiles enqueue entries
  pointing at the collapsed tiles **outside** the region; where there is no neighbour tile it
  uses the model's `BorderTile` (a synthetic tile built from border labels) — and when two
  tiles use **different models**, propagation substitutes the other model's border tile
  (`Propagate`: `AffectedTile.Model <> InstigatorTile.Model` → `UseInstigatorTile = OtherBorder`).
- `NotifyChangedVoxel` bridges to the voxel grid: changing a *solid* voxel resets that tile
  plus (if not topmost) the one above — building edits invalidate park tiles vertically.

## Verse-specific idioms worth lifting

- Maps-as-sets: `[wfc_mesh_option]int` with `if(set Map[O] = 0){}` — Verse has no set type.
- `<unique>` classes as map keys (labels, options) — identity equality instead of value
  comparison.
- `GetRemainingTilesWithMinEntropy` compares floats with an epsilon (`±0.0001`) so ties are
  collected and randomly chosen between.
- Randomness at collapse time uses global `GetRandomInt`; randomness at **spawn** time uses
  the position hash (`RandomIntFromVector`) so re-meshing an unchanged region doesn't
  reshuffle prop variants.

## Implementation caveats and limits in shipped code

- **First-constraint entropy bug:** in `wfc_system.verse`, the first time a tile is constrained
  (`Tile.AllOptions?` branch), `Tile.RemainingOptions` is replaced with `OptionsToCheckAgainst`,
  but `Tile.NumRemainingOptions` and `Tile.UpdateShannonEntropy()` are not called in that branch.
- **Incomplete reset in `ResetWFC`:** `ResetWFC()` reallocates the `Tiles` array with `false`,
  but leaves `RemainingTiles` and `PropagationQueue` uncleared.
- **Zero-option fallback bypass:** `vo_wfc.SpawnPropsFromTiles` wraps option extraction and
  fallback spawning inside `if: Tile.NumRemainingOptions = 1`, preventing over-constrained tiles
  from ever reaching `FallbackPropOpt`.

**Sources:** verse_proceduarl_building. Tutorial's WFC prose matches this code; trust the
snippets.
