---
id: technique-109
kind: technique
---

# No-Verse earthquake: Trigger → Cinematic Sequence → shake + one-shot VFX

**Sources:** vfx_template
**Tutorial:** Falling Particles and Camera Effects

## Problem

A hallway should rumble: camera shake, dust from the ceiling, a splash, a sound. There is **zero Verse** in the project. Wiring that in devices without a sequencer dumps Enable calls onto every particle actor; putting it all in Niagara hides it from designers.

## Pattern

One **Cinematic Sequence device** is the bus. Triggers only `PlayFunction`. The sequence owns shake, falling debris, splash, and audio.

```
Trigger / Trigger2  --OnTriggered-->  Cinematic Sequence Device.PlayFunction
                                      sequence = Content/NewLevelSequence
                                      autoplay off, loop off, GameplayOnly
                                           |
                                           +-- camshaker (CameraShakeSource → NewCameraShakeBase)
                                           +-- enable VFX Creator Falling/* (startEffectsWhenEnabled=false)
                                           +-- NS_LavaSplash_01 (bAutoActivate=false)
                                           +-- earthquake SoundWave
```

Keyframe times inside `NewLevelSequence` are **unanswerable** (Sequencer values). The *binding* is Observed: cinematic `sequence` → `/…/NewLevelSequence.NewLevelSequence`; `camshaker` is a referencer of that sequence; splash actors live in Outliner folder `Sequence/` next to `RockChunkLG C 01/02`.

### Camera shake is a source actor, not a device option

`camshaker` (`camshaker_C`) has `CameraShakeSourceComponent`:

| Property | Value |
|---|---|
| `CameraShake` | `/…/NewCameraShakeBase.NewCameraShakeBase_C` |
| `bAutoStart` | **false** |
| Attenuation | Quadratic, inner=outer **20000** (whole cave) |

Billboard 46: *"Camerashaker blueprint calls up the Newcamerashake blueprint which determines the camera shake motion that is triggered with the sequencer."*

`earthquake.uasset` is a **SoundWave** (not a Cue). `Get_referencers` on it is empty — same OFPA blind spot as the sequences; billboard 29 names it as part of the shot.

### Triggers are invisible, one-shot, and chained — still no Verse

All 8 `Device_Trigger_V2`: `visible in Game` **false**, `playTriggerVFX/SFX` **false**, `timesCanTrigger` 1. Player-triggered except `Trigger6-2` (creature-elim bus).

Gameplay chain (**Observed** — `DeviceToolset.ListEventBindings`, inverted to source→target):

| Source | Event | Targets |
|---|---|---|
| Trigger, Trigger2 | OnTriggered | Cinematic `PlayFunction` |
| Trigger2 | OnTriggered | 8 player spawners `Disable`; Audio Player4 `Play` |
| Trigger3 | OnTriggered | Trigger4 `Enable` |
| Trigger4 | OnTriggered | Trigger5+6 `Enable`; Creature Placer (Brute) `Enable`; Audio Player5 `Play`; Barrier Device5 `Enable` |
| Trigger5 | OnTriggered | End Game `Activate`; Radio `Play` |
| Trigger6 | OnTriggered | Creature Placer3+5 (Ranged) `Enable`; Audio Player5 `Play` |
| Trigger6-2 | OnTriggered (1.5 s delay; fired by both Brute `On Eliminated`) | Audio Player6 `Play`; Barrier Device5 `Disable` |
| Trigger7 | OnTriggered | Audio Player6 `Play`; Barrier Device5 `Disable`/`Enable` (door at hall end) |

Trigger4/5/6 start **disabled** (`enabled On Game Start = false`) and are enabled by the previous trigger — a door/combat latch, not the VFX lesson.

HUD Controller has **no** event bindings; billboard 44 says it hides UI because this is a graphics demo. Island Settings / Class Designer / Item Placer (`BP_ItemContainer`, the sword) sit at Outliner root (`folderPath` None).

### Ambient vs event particles use the same Creator class

See vfx-surface-split.md. Quake debris: `startEffectsWhenEnabled = false`, loop Custom 5 s once. Ceiling dust: Forever, 20 s gap, starts when enabled. The sequencer only has to Enable the debris devices.

## Don't

- Don’t subscribe Niagara in Verse for this — the island’s answer is Play on a Cinematic Sequence device.
- Don’t auto-play the sequence (`bAutoPlay` false, `enabledOnPhase` GameplayOnly).
- Don’t route the shot through a Popup theme picker — that is popup-theme-cinematic.md (Button → Choose Theme → Party/Scary).
- Don’t leave the shake source on (`bAutoStart` false) or the cave rumbles in the editor.
- Don’t treat `FallingRock_01` / `FallingRock_02` / `Intro_camera_sequence` as extra worked examples: asset-registry referencers are **empty**. The live cinematic points at `NewLevelSequence` only. Those three may be leftovers or possessed sub-shots; keyframes are unreachable either way.
