---
id: technique-110
kind: technique
---

# Three VFX surfaces, one island — Niagara actor vs VFX Creator vs VFX Spawner

**Sources:** vfx_template
**Tutorial:** VFX - Lava Cave

## Problem

A lava cave needs looping ambient particles, one-shot event particles, a colored fog bank, and a custom lava look. Putting all of that in authored Niagara (or all of it in Creative devices) is the failure mode — the surfaces have different jobs.

## Pattern

Ship **three layers**. Do not mix them unless the exhibit needs both (braziers).

| Surface | What this island uses it for | Authored here? |
|---|---|---|
| **Niagara actor** | Custom lava (burst + splash) and stock fire (`NS_Janus_Campfire`) | Burst/splash yes; campfire is `/Game/Effects/.../NS_Janus_Campfire` |
| **VFX Creator device** | Generic sprites: looping ceiling dust, fire-pit smoke, quake debris | No — device presets (`Dots01` / `Smoke01`) |
| **VFX Spawner device** | Named Fortnite preset (`AreaOfFog`) tinted as lava steam | No |

### 1. Authored Niagara — when the look needs custom materials

`Content/VFX/NS_LavaBurst` + `NS_LavaSplash_01` plus `M_Lava_*` / `MI_Lava_*` (**Observed** — `assets.txt` + live `NiagaraComponent0.Asset`).

`NS_LavaBurst` summary (**Observed** — `GetSystemSummary` on `/…/VFX/NS_LavaBurst.NS_LavaBurst`):

- User params: `MaxSpawnRate` 2, `ScaleMultiplier` 1.2, `SpawnRadius` 300.
- Live emitters: **Bubble** (mesh), **Small_Spray**, **Large_Splashes**, **Large_Splashes_Web**, **Ripple**. Debug emitters (`Debug_Circle`, `Test_Splash_Webs`, `Debug_RandomSubUV`) ship **disabled**.
- Placed as `NiagaraActor`s in `Atmosphere/VFX/` (`NS LavaBurst` / `NS LavaBurst Small`), `bAutoActivate = true`.

`NS_LavaSplash_01` sits in `Sequence/` next to two `RockChunkLG` rocks, `bAutoActivate = false` — the splash is a **shot**, not an ambient loop. Collision modules are in its dependency list (burst does not use `Collision`).

**Lesson:** custom SubUV/lava materials → authored Niagara. Keep a disabled debug emitter in the system if you need a spawn-radius gizmo; do not delete it from the asset.

### 2. VFX Creator — when a sprite loop is enough

`Device_VFX_Creator_V2` ×20, three Outliner jobs (**Observed** — folders + `get_properties`):

| Folder | Count | Sprite | Loop | `startEffectsWhenEnabled` | Job |
|---|---|---|---|---|---|
| `Atmosphere/VFX/Falling Dust` | 4 | `Dots01`, size 8, zone 0.5 | Forever, 20 s gap | **true** | ambient grit |
| `Atmosphere/VFX/FirePitSmoke` | 10 | `Smoke01`, size 7.5, gravity **0** | Forever | **true** | braziers / pits |
| `Atmosphere/VFX/Falling` | 6 | `Dots01`, size 4, zone 3, amount 100 | Custom, 5 s × 1 | **false** | quake debris (enabled by the sequence) |

Dust/smoke are Always-on, Continuous, dark brown / grey. Debris does **not** auto-start — the cinematic enables it. Same device class, opposite enable policy.

**Lesson:** VFX Creator is a designer-facing sprite box (shape, size, gravity, loop). It is not a Niagara editor. Use it for dust/smoke/debris; do not rebuild lava in it.

### 3. VFX Spawner — when a named Fortnite preset already exists

One `Device_VFXSpawner_V2` labeled `VFX Spawner - Fog` (**Observed**):

| Option | Value |
|---|---|
| `visual_Effect` | **AreaOfFog** |
| `effectType` | Continuous |
| `enabled On Phase` | Always |
| `colorize VFX` | true |
| `custom Color` | orange `(0.86, 0.29, 0.13)` |

Billboard 41: *"The smoke coming from the lava is created with the VFX spawner device with the AreaofFog preset."*

**Lesson:** Spawner = pick a catalog effect and tint it. Creator = author a sprite. Niagara = author a system. Fog-over-lava is a tinted preset, not a new Niagara.

### Brazier = Creator + stock Niagara + a Creative light (the one mix that is intentional)

Billboards 18–22 / 24+: torch = `NS_Janus_Campfire` + `Device_PointLight_V2`; brazier = that plus VFX Creator smoke. Live campfire Asset is `/Game/Effects/Fort_Effects/World/Enviroment/Smoke/NS_Janus_Campfire` (stock, path typo `Enviroment` ships). Point-light sample `Customizable Light2`: `lightType` **Spot**, intensity **5%**, orange-red, **no shadows**, Always.

Engine `SpotLight`s in `Atmosphere/Lights/NonLumen` and `…/Silhouette` do the statue wash (sample `SpotLight - Statues 1`: intensity 10, atten 1800, outer cone 44°, Stationary, no shadows). Tutorial calls these “point lights”; the Creative device is a Point Light **class** whose type is Spot.

## Don't

- Don’t put looping dust in authored Niagara — VFX Creator Forever + `startEffectsWhenEnabled` is the island’s ambient path.
- Don’t auto-activate event Niagara (`NS_LavaSplash_01` is off until the sequence).
- Don’t treat `Device_VFX_Creator_V2` and `Device_VFXSpawner_V2` as the same device. Creator = sprite params; Spawner = `visual_Effect` enum.
- Don’t copy Brick Editor Party/Scary dressing as this lava-cave split — popup-theme-cinematic.md uses Creator vs Spawner only as cine props (Fireworks label vs live `Bubbles`).
- Don’t copy LuminousFortress walls / `Rocks_Ref` as VFX. The kit is `Content/VFX/` + the three device jobs above.
