---
id: technique-111
kind: technique
---

# Claim a homestead pad: disable + `[agent]spawner` map + intro cam/stasis with sequence timeout

**Sources:** village_starter_template
**Tutorial:** none (spreadsheet Docs=No).

## Problem

4 players, 4 houses. First spawn should assign a unique homestead, play an intro looking at the player, then free them. Next joiner must not land on an occupied pad. A missed `SpawnedEvent` must still claim.

## Pattern

Bundle per house (`initial_player_spawn_info`):

```
PlayerSpawner      : player_spawner_device
PlayerIntroCamera  : gameplay_camera_fixed_point_device
HomeCounter        : player_counter_device   # failsafe volume on the lot
```

`village_player_spawner`:

1. `PlayerSpawnList: []initial_player_spawn_info` — one entry per homestead.
2. On spawn (`SpawnedEvent.Await` **and** `HomeCounter.CountedEvent.Await` — two listeners, first write wins):
   - if `PlayerToSpawner[Agent]` already set → return
   - `set PlayerToSpawner[Agent] = Spawner`
   - `PutInStasis({AllowFalling := true})`
   - `StartSequence.Play(Agent)` (fade from black)
   - `Spawner.Disable()`
   - camera `RemoveFrom` then `AddTo` (stack reset, same as character_environment orbit)
   - `race { StoppedEvent.Await() / Sleep(6.0); GoToEndAndStop }` — **timeout so a dead cine can't freeze them**
   - `FirstDialog.Show(Agent)`; `ReleaseFromStasis`; strip stasis control mode
3. `GetPlayspace().PlayerRemovedEvent().Subscribe(ReactivateSpawner)` → `Enable()` that pad, rebuild the map without that agent.

OnBegin also `HomeCounter.TransmitForAllCounted()` so anyone already standing in a lot is claimed if the spawn event was missed.

## Don't

- Don't leave pads Always-on — two players share a house.
- Don't `Await` the cine with no timeout.
- Don't trust only `SpawnedEvent` — the counter exists because that event is flaky at round start.
- Don't put intro cameras on the gameplay camera stack without `RemoveFrom` first.
