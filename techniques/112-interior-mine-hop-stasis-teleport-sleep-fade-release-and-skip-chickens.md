---
id: technique-112
kind: technique
---

# Interior / mine hop: stasis → Teleport → Sleep(fade) → release — and skip chickens

**Sources:** village_starter_template
**Tutorial:** none.

## Problem

Village overworld, interiors, and a mine chain (5 teleporter hops) must fade the player without letting NPCs or wandering wildlife ride the pad.

## Pattern

Two binding structs so the Details panel stays typed:

```
zone_transition_buttons  : button  + destination teleporter
zone_transition_triggers : volume  + destination teleporter
```

Per binding, a `zone_transition_subscription` fires `DoTransitionEvent` with a payload (`teleporter`, `fort_character`, `agent`).

`TransitionToNewZone`:

```
if player[Agent]:                         # not NPC / chicken
    FortCharacter.PutInStasis(AllowFalling := true)
    Destination.Teleport(Agent)
    Sleep(2.0)                            # fade-to-black window
    FortCharacter.ReleaseFromStasis()
```

`spawn{ ListenToSubscription }` per binding so many doors can fire in parallel. Button = `InteractedWithEvent`; volume = `AgentEntersEvent`.

Mine chain on this island is **five** of these (Entrance / SideScroller / Raptors / Well / Exit) — same class, more rows in `ZoneTriggerList`.

## Don't

- Don't `Teleport` without stasis — the pawn keeps walking during the fade.
- Don't skip the `player[]` cast — volumes overlap livestock.
- Don't block `OnBegin` on one `Await`; spawn a listener per door.
- Don't confuse with Deserted spawn *closets* (Lock + delayed Trigger, no Verse).
