---
id: technique-113
kind: technique
---

# Voice Persona NPC: channel + `RegisterAction` structured output, mutate gameplay on `StopSay`

**Sources:** voice_conversation_sample
**Snippets:** `persona_behavior.verse`, `ui_manager.verse`
**Official island tutorial:** none found. Do **not** substitute Conversation Device docs.

## Problem

Haggle with a spoken NPC. The model must decide “we agreed on N gold” or “I was told to leave,” then Verse awards gold / penalty and despawns — *after* the line finishes. Personality and budget must vary per spawn without a new Character Definition each time.

## Pattern

```verse
# 1. Channel
TeamGroup := agent_group(team_member_info){}
TeamVoiceChannel := voice_channel(team_member_info){ Name := …, Group := TeamGroup }
TeamGroup.AddMember(Player, team_member_info{})
TeamGroup.AddMember(NPCAgent, team_member_info{})
SimEntity.AddChatChannel(TeamVoiceChannel)
Player.SetConversationTarget(PersonaComponent, TeamVoiceChannel)

# 2. Personality (append, don’t replace)
option{ Persona.SetPersonality[ Join(array{Persona.Personality, CharacterPrompt, BudgetPrompt}, "\n") ] }
# Budget as "starting offer", not "budget" — model treats "budget" as a ceiling.

# 3. Structured output
FinalOfferBinding := prompt_binding_definition:
    Name := MakeMessage("FinalOffer")
    Description := MakeMessage("Signals whenever you have received an agreement…")
Session.RegisterAction(FinalOfferBinding, false, FinalOfferResponse, FinalOfferFunction)
# FinalOfferResponse fields use @ai_description so the LLM fills FinalOffer:logic + FinalPrice:int

# 4. Callbacks only SET FLAGS. StopSayEvent does score/despawn.
Persona.StopSayEvent.Subscribe(StopSayChecks)
Persona.CommitSayEvent.Subscribe(SetCloseCaptions)   # captions
Persona.PromptToTalk[GreetingMessage, Channel]
```

OnEnd: `Cancel()` the RegisterAction cancelables.

HUD sibling: `FindCreativeObjectsWithTag(ManagerTag)` → `ui_manager[Manager]`. Catalog: `AddWidget(Catalog, player_ui_slot{InputMode := ui_input_mode.All})`; close = RemoveWidget.

## Don't

- Don’t fire `DespawnAll` inside `FinalOfferFunction` — the NPC is still talking.
- Don’t train Conversation Device (trees, pins) as this.
- Don’t `@editable` the HUD on the NPC behavior if a tag lookup exists.
- Don’t tell the model “your budget is N” if N should move.
