---
id: technique-114
kind: technique
---

# Toggle first-person with Switch → `AddToPlayer` / `RemoveFromPlayer`

**Sources:** horde_npc_sample
**Tutorial:** zombie-survival Cameras — **External**. Bindings **Observed**.

Deeper camera-device syllabus: characters_cameras_npc_template hall 07. This island only ships **First Person Camera**, not Perspective / Fixed Point.

## Problem

Walkers in first-person feel claustrophobic (horror). BatWeapon’s slam/heavy swing is authored for third-person. You need one camera device the player can flip per room without Verse.

## Pattern

One `Device_CameraModes_FirstPerson` at map center:

| Option | Live |
|---|---|
| priority | **0** |
| addToPlayersOnStart | **true** |
| enabled During Phase | Always |
| FOV | 80 |
| headMotionIntensity | None |
| targetingReticle | Display None |
| cameraShake | true |
| removeOnElimination | false |

Three FPS switches (one per room) bind:

```
On Turned On  → AddToPlayer
On Turned Off → RemoveFromPlayer
```

Same pair on all three switches, same camera. Billboard labels: `FPS Toggle`.

BatWeapon still forces third-person while wielded (docs; not a device setting you can read on the bat). The switch is for walking / shotgun / flashlight, not for melee presentation.

Same wiring shape as the Heavy Linework PPE: `Switch_PPEToggle` `On Turned On/Off` → Post Process `EnableForAll` / `DisableForAll`. PPE starts **off** (`enabledDuringPhase=None`, `postProcessEffect=/Game/Creative/PostProcess/PP_HeavyLinework`).

## Don't

- Don't spawn extra First Person Camera devices per room. One device, many switches.
- Don't set Priority above 0 if you want this to be the default stack. Docs: Priority **0.0** + Add to Players on Start.
- Don't expect the PPE switch to preview in the editor viewport — docs: push to a Fortnite session.
