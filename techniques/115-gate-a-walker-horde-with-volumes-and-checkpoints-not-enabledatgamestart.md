---
id: technique-115
kind: technique
---

# Gate a Walker horde with volumes and checkpoints, not `enabledAtGameStart`

**Sources:** horde_npc_sample
**Tutorial:** zombie-survival Level Design — **External**. Bindings **Observed**.

## Problem

Walkers are slow but cheap to swarm. If every NPC Spawner is Enabled at Game Start, the island pays the **~50 Walker** memory cap immediately and the horror/action beats fire at once. You need spawners that wake when the player reaches a corridor, a checkpoint, or an arena volume — and die when they leave.

## Pattern

All **22** spawners share `WalkerDefinition`. The difference is **when they Enable / Spawn / Disable**.

### Horror (Room 2 — 13 spawners)

`enabledAtGameStart=false`. Incoming bindings (every spawner also `Disable` on `EndRoomTP` `On Teleported`):

| Wake | Typical extra |
|---|---|
| Volume `OnEnter` → `Enable` + `Spawn` | early corridor (Spawner3, 31) |
| Checkpoint `On First Activation` → `Enable` (+ often `Spawn`) | after the player has committed |

Spawn counts vary **1–6** with `totalSpawnLimit` 2–7 and `allowInfiniteSpawn=false`. That is the “irregular groups” the corridor billboards teach.

Docs table: 13 NPC Spawners, 3 Volumes, 1 Checkpoint, 1 Teleporter. Live: **13 spawners, 3 volumes, 2 checkpoints**, 1 teleporter (**Observed**). Two checkpoints, not one.

### Action (Room 3 — 5 spawners)

Still `enabledAtGameStart=false`, but each spawner is a **hose**:

| | Live |
|---|---|
| spawnCount | 8 or **10** |
| allowInfiniteSpawn | true |
| totalSpawnLimit | 30 |
| spawnTimer | 1.5 s |
| spawnRadius | 5 |

Arena bus (**Observed**): `Volume Room 3 Arena Start` `OnEnter` → `Trigger Room 3 Arena Start` → Enable all five spawners + Start four timers. Finish volume / Give-up button → `Trigger Room 3 Arena Finish` → Disable spawners + Reset timers. Escape timer **duration=120**, `showOnHUD=true`, `successOnTimerEnd=true`, `completionBehavior=Reset`.

Fewer devices, more Walkers — docs: “continuous rate of 10.” Two of five spawners use 10; three use 8.

### Demo room (Room 1)

Opposite: `enabledAtGameStart=true`. The labeled demo Walker respawns `On Player Spawned`. `Switch_SpawnWalker` `On State Changes` → `Spawn` on the count-3 spawner (self `On Turned Off` → `Toggle State` so one click is a pulse).

## Don't

- Don't Enable-at-start a 10-count infinite spawner in a corridor. Horror is volume/checkpoint gated for a reason (perf + pacing).
- Don't treat the Level Design “1 checkpoint / 1 item spawner” table as complete. Live Horror has **2** checkpoints and **2** item spawners.
- Don't forget the exit `Disable`. Horror uses the end teleporter; Action uses the leave-volume trigger. Infinite spawners without Disable keep ticking after the beat.
