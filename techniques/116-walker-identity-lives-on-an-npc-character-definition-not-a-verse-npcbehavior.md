---
id: technique-116
kind: technique
---

# Walker identity lives on an NPC Character Definition, not a Verse `npc_behavior`

**Sources:** horde_npc_sample
**Tutorial:** TWDU Walker NPC Template + Configuring Walkers — **External**. Live values **Observed**.

Contrast (do not duplicate): Verse wander on a generic NPC is npc-behavior-navigatable-focus.md. This island has **zero** `.verse` files.

## Problem

You need Walkers (slow horde, headshot model, bite DoT), not a custom nav script. A `creative_device` cannot possess them. Guard / Creature spawners do not take the Walker type.

## Pattern

1. Author an **NPC Character Definition** (`Add > Artificial Intelligence > Character Definition`).
2. Set **Type** to `Walker` or `Walker (Prisoner Uniform)`. Prisoner is appearance only.
3. Assign it as **NPC Character Definition** on an NPC Spawner (`Device_CharacterSpawner`). Dragging the asset into the viewport creates the spawner for you.

this sample ships **one** definition used by all **22** spawners:

`/a6e35c5f-014b-4b20-9f00-60866557bd08/AI/WalkerDefinition`

Live subobjects (**Observed**):

| Slot | Class |
|---|---|
| Type | `BP_MeadowDrosera_CharacterType_Random_Default` |
| Behavior | `CharacterModifier_MeadowDroseraBehavior` |
| modifiers[0] | `CharacterModifier_Health` — 100 HP, 0 shield, vulnerable |
| modifiers[1] | `CharacterModifier_MeadowDroseraGameplay` — Walker Gameplay Modifier |

Gameplay modifier values on the shipped asset (**Observed**; Configuring Walkers defaults):

| Option | Live |
|---|---|
| Headshot Damage Model | true |
| Attack Damage | 12 |
| Bitten DOT damage | 2 |
| Remove Bite DOT on DBNO | true |
| Shield Prevents Bitten Status Effect | false |
| Skip Spawning Animation | false |
| Wait Time Between Wandering Min / Max | **3 / 15** |
| Prisoner Uniform Color | Orange (unused — Type is not Prisoner) |

Horde **count / radius / timer** stay on the **spawner**, not the definition. Same `WalkerDefinition` can be a single demo Walker (`spawnCount=1`) or an arena wave (`spawnCount=8–10`, `allowInfiniteSpawn=true`, `spawnRadius=5`).

`NPCBehaviorScriptComponent.nPCBehaviorScript` is empty on the demo spawners. Do not look for a Verse override here.

## Don't

- Don't subclass `npc_behavior` expecting a Walker. Type=`Walker` is the theme enemy; Verse wander is a different syllabus (characters_cameras_npc_template).
- Don't copy `Content/AI/WalkerCharacterDefinition` — it is an **ObjectRedirector** to `WalkerDefinition` with **0** referencers.
- Don't train the wandering-max as **16**. Weapons / Level Design chapters say 16; live + Configuring Walkers say **15**.
- Don't put Walker spawners on a Guard Spawner. UEFN-only (`Fortnite > Devices > !Beta` in the hub).
